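# Player animation Blueprint for the Rocket Ram item, overriding item animation layers from wind-up through launch, flight and impact.
# Each update caches the pawn's weapon and launch direction, runs RocketRamState, and uses delays to time state flags.
# Full-body and upper-body state machines use those flags, with state aliases sharing impact, slide and exit transitions.
# The final-pose layer adds pelvis pitch/yaw rotation, a leg jump additive, suppression curves and inertialization.
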AnimBlueprint /RocketRamGameplay/Items/RocketRam/Animations/Animation/RocketRam_Player_AnimBP
target_skeleton=&Fortnite_M_Avg_Player_Skeleton
graphs [
 AnimGraph {nodes [Output Pose() {layerGroup=None}]}
 EventGraph {nodes [
   PrintString!(
    execute
    InString="To String (Float)"(InDouble=Get WallImpactYaw() -> (WallImpactYaw:float))
    bPrintToScreen
    bPrintToLog
    TextColor=(R=1,G=0,B=0.9365,A=1)
    Duration=0.1
    Key
   ) -> (then)
   > n1 = Sequence!(execute) -> (
    then_0
    then_1
    then_2:Exec
    then_3:Exec
    then_4:Exec
    then_5:Exec
    then_6:Exec
    then_7:Exec
    then_8:Exec
    then_9:Exec
    then_10:Exec
    then_11:Exec
   )
   > Branch(Condition=Get bIsRRWindUp() -> (bIsRRWindUp:bool))
   > PrintString(InString=WindUp, TextColor=(R=1,G=0.0203,B=0,A=1), Duration=0.1)
   Branch(execute=@n1.then_2, Condition=Get bIsRRFreeFall() -> (bIsRRFreeFall:bool))
   > PrintString(InString=FreeFall, TextColor=(R=1,G=0.0203,B=0,A=1), Duration=0.1)
   Branch(execute=@n1.then_4, Condition=Get bIsRRBraking() -> (bIsRRBraking:bool))
   > PrintString(InString=Braking, TextColor=(R=1,G=0.0203,B=0,A=1), Duration=0.1)
   Branch(execute=@n1.then_6, Condition=Get bIsRRSlowAcceleration() -> (bIsRRSlowAcceleration:bool))
   > PrintString(InString=RRSlowAcceleration, TextColor=(R=1,G=0.0203,B=0,A=1), Duration=0.1)
   Branch(execute=@n1.then_7, Condition=Get bIsLandingDetected() -> (bIsLandingDetected:bool))
   > PrintString(InString=LandingDetected, TextColor=(R=1,G=0.0203,B=0,A=1), Duration=0.1)
   Branch(execute=@n1.then_8, Condition=Get bIsWallImpact() -> (bIsWallImpact:bool))
   > PrintString(InString=WallImpact, TextColor=(R=1,G=0.0203,B=0,A=1), Duration=0.1)
   Branch(execute=@n1.then_9, Condition=Get bIsRRRecovery() -> (bIsRRRecovery:bool))
   > PrintString(InString=RRRecovery, TextColor=(R=1,G=0.0444,B=0,A=1), Duration=0.1)
   > Branch(Condition=Get bIsImpactWallBackup() -> (bIsImpactWallBackup:bool))
   > PrintString(InString=Impact Wall Backup, TextColor=(R=1,G=0.0444,B=0,A=1), Duration=0.1)
   Branch(execute=@n1.then_1, Condition=Get bIsRRActiveThrust() -> (bIsRRActiveThrust:bool))
   > PrintString(InString=RRActiveThrust, TextColor=(R=1,G=0.0203,B=0,A=1), Duration=0.1)
   Branch(execute=@n1.then_3, Condition=Get bIsRRFastFreeFall() -> (bIsRRFastFreeFall:bool))
   > PrintString(InString=FastFreeFall, TextColor=(R=1,G=0.0203,B=0,A=1), Duration=0.1)
   Branch(execute=@n1.then_5, Condition=Get bIsRROff() -> (bIsRROff:bool))
   > PrintString(InString=Off, TextColor=(R=1,G=0.0203,B=0,A=1), Duration=0.1)
   n2 = SetTimeDilationRateScale!() -> (OutputDelegate, then, TimeDilationRateScale:double=~) {is=title K2Node_CustomEvent}
   > Set TimeDilationRateScale(TimeDilationRateScale:double=@n2.TimeDilationRateScale) -> (Output_Get:double)
   n3 = NotifyImpactDetected!() -> (OutputDelegate, then, bIsWallImpact:bool=~) {is=title K2Node_CustomEvent}
   > Set bIsWallImpact(bIsWallImpact:bool=@n3.bIsWallImpact) -> (Output_Get:bool)
   > n4 = Set bIsLandingDetected(bIsLandingDetected:bool=true) -> (Output_Get:bool)
   Do Once(execute=@n6.then_2, Reset=@n4.then, Start Closed=true)
   > Set ImpactPitch(ImpactPitch:float=Get Pitch() -> (Pitch:float)) -> (Output_Get:float)
   > Delay(Duration=0.2)
   > Set bIsLandingDetected(bIsLandingDetected=false)
   n5 = Event BlueprintUpdateAnimation()
   > Set Delta Time X(Delta Time X=@n5.DeltaTimeX)
   > n6 = Sequence!(execute) -> (then_0, then_1, then_2, then_3, then_4, then_5, then_6, then_7, then_8)
   > n7 = Cast To FortPlayerPawn(Object=TryGetPawnOwner())
   > n8 = Set As Fort Player Pawn(As Fort Player Pawn:Fort Player Pawn Object Reference=@n7) -> (Output_Get:Fort Player Pawn Object Reference)
   > n9 = Cast To B_Weapon_RocketRam!(execute, Object=GetCurrentWeapon!(self:Fort Pawn Object Reference=@n8) -> (ReturnValue:Fort Weapon Object Reference)) -> (then, CastFailed, AsB Weapon Rocket Ram:B Weapon Rocket Ram Object Reference) {is=title K2Node_DynamicCast}
   > n10 = Set As B Weapon Rocket Ram(As B Weapon Rocket Ram:B Weapon Rocket Ram Object Reference=@n9) -> (Output_Get:B Weapon Rocket Ram Object Reference)
   > Set WeaponLaunchDirection(WeaponLaunchDirection:Vector=Get LaunchDirection!(self:B Weapon Rocket Ram Object Reference=@n10) -> (LaunchDirection:Vector)) -> (Output_Get:Vector)
   RocketRamState(execute=@n6.then_1)
   Branch(execute=@n6.then_3, Condition=Get bIsRROff())
   > Delay(Duration=0.5)
   > Set bIsWallImpact(bIsWallImpact=false)
   n11 = Branch(execute=@n6.then_4, Condition=Get bIsRRFreeFall())
   > Do Once(Reset=@n12.then_0)
   > Delay(Duration=1)
   > Set bIsApexTimerEnd(bIsApexTimerEnd:bool=true) -> (Output_Get:bool)
   n12 = Sequence!(execute=@n11.else) -> (then_0, then_1)
   Set bIsApexTimerEnd(execute=@n12.then_1, bIsApexTimerEnd=false)
   n13 = Branch(execute=@n6.then_5, Condition=Get bIsRRWindUp())
   > Do Once(Reset=@n14.then)
   > Delay(Duration=1.4)
   > Set SquatAdditiveTimer(SquatAdditiveTimer:bool=true) -> (Output_Get:bool)
   Delay(execute=@n13.else, Duration=0.2)
   > n14 = Set SquatAdditiveTimer(SquatAdditiveTimer=false)
   n15 = Branch(execute=@n6.then_6, Condition=Get bIsRRBraking())
   > Set FalseRotationBlendTime(FalseRotationBlendTime:double=0.3) -> (Output_Get:double)
   n16 = Branch(execute=@n15.else, Condition=OR Boolean(A=Get bIsWallImpact(), B=Get bIsImpactWallBackup()))
   > Set FalseRotationBlendTime(FalseRotationBlendTime=0.3)
   Delay(execute=@n16.else, Duration=0.5)
   > Set FalseRotationBlendTime(FalseRotationBlendTime=0)
   n17 = Branch(execute=@n6.then_7, Condition=Get bIsRRRecovery())
   > DelayUntilNextTick!(execute) -> (then)
   > Set bIsRRRecoveryDelayed(bIsRRRecoveryDelayed:bool=true) -> (Output_Get:bool)
   Set bIsRRRecoveryDelayed(execute=@n17.else, bIsRRRecoveryDelayed=false)
   n18 = Get bIsWallImpact()
   n19 = Get bIsImpactWallBackup()
   n20 = Branch(execute=@n6.then_8, Condition=AND Boolean(A=Get bIsRRRecovery(), B=OR Boolean(A=@n18, B=@n19)))
   > Set bIsRRRecoveryHalfSecDelay(bIsRRRecoveryHalfSecDelay:bool=true) -> (Output_Get:bool)
   Delay(execute=@n20.else, Duration=0.5)
   > Set bIsRRRecoveryHalfSecDelay(bIsRRRecoveryHalfSecDelay=false)
   n21 = Branch(Condition=NOT Boolean(A=Get bIsWallImpact()))
   Set Yaw(execute=@n21.else, Yaw:float=float - float(A=Get WallImpactYaw())) -> (Output_Get:float)
   DelayUntilNextTick()
   DelayUntilNextTick()
   n22 = Get bIsRRRecovery()
   OR Boolean(
    A=AND Boolean(A=Get bIsRRRecoveryDelayed() -> (bIsRRRecoveryDelayed:bool), B=Get bIsRRRecoveryHalfSecDelay() -> (bIsRRRecoveryHalfSecDelay:bool))
    B=AND Boolean(A=NOT Boolean(A=@n22), B=Get bIsRRRecoveryHalfSecDelay())
   )
  ]}
 ItemAimOffset {nodes [
   Input Pose!() -> (Pose, AimOffsetAlpha:double, Yaw:double, Pitch:double) {name=InPose_AimOffset}
   Input Pose!() -> (Pose) {name=InPose_PreAimOffsets}
   Output Pose() {layerGroup=Item Override}
  ]}
 ItemFallAdditive {nodes [
   Input Pose() {name=InPoseItemFallAdditive}
   Output Pose() {layerGroup=Item Override}
  ]}
 ItemFallLandAdditive {nodes [
   Input Pose() {name=InPoseItemFallLandAdditive}
   Output Pose() {layerGroup=Item Override}
  ]}
 ItemFallLevel2Override {nodes [
   Input Pose() {name=InPose_FallLevel2}
   Output Pose() {layerGroup=Item Override}
  ]}
 ItemFinalPoseOverride {nodes [
   n1 = "Sequence Player (additive) 'rocketRam_Jump_Pose'"() {bLoopAnimation=false, sequence=/RocketRamGameplay/Items/RocketRam/Animations/Animation/Locomotion/rocketRam_Jump_Pose}
   n2 = Use cached pose 'InFinalPose_Cached'()
   n3 = "Transform (Modify) Bone - Bone: pelvis"!(
    ComponentPose=Local To Component(LocalPose=Use cached pose 'InFinalPose_Cached'())
    Alpha=1
    Translation="0,0,0"
    Rotation_Roll:float=float * float(A=Get Pitch(), B=-1) {is=title K2Node_PromotableOperator}
    Rotation_Pitch:float=0
    Rotation_Yaw:float=0
    Scale="1,1,1"
   ) -> (Pose) {
    rotationMode=BMM_Additive
    rotationSpace=BCS_ParentBoneSpace
    translationMode=BMM_Additive
    translationSpace=BCS_BoneSpace
    alphaBoolBlend {blendOption=Linear}
    boneToModify {boneName=pelvis}
   }
   n4 = "Transform (Modify) Bone - Bone: pelvis"(ComponentPose=@n3, Alpha=1, Translation="0,0,0", Rotation_Roll=0, Rotation_Pitch=0, Rotation_Yaw=Get Yaw() -> (Yaw:float), Scale="1,1,1") {rotationMode=BMM_Additive, rotationSpace=BCS_ParentBoneSpace, alphaBoolBlend {blendOption=Linear}, boneToModify {boneName=pelvis}}
   n5 = "Transform (Modify) Bone - Bone: pelvis"(
    ComponentPose=Local To Component(LocalPose=Use cached pose 'InFinalPose_Cached'())
    Alpha=1
    Translation="0,0,0"
    Rotation_Roll=0
    Rotation_Pitch=0
    Rotation_Yaw=Get Yaw()
    Scale="1,1,1"
   ) {rotationMode=BMM_Additive, rotationSpace=BCS_ParentBoneSpace, alphaBoolBlend {blendOption=Linear}, boneToModify {boneName=pelvis}}
   n6 = Blend Poses by bool(BlendPose_0=Component To Local(ComponentPose=@n4), BlendPose_1=Component To Local(ComponentPose=@n5), BlendTime_0=0.15, BlendTime_1=0) {blendType=Linear}
   n7 = Use cached pose 'InFinalPose_Cached'()
   n8 = Get FalseRotationBlendTime() -> (FalseRotationBlendTime:double)
   n9 = Modify Curve!(
    SourcePose=Layered blend per bone(
     BasePose=Blend Poses by bool(BlendPose_0=@n6, BlendPose_1=@n7, BlendTime_0=0.05, BlendTime_1=@n8) {blendType=Linear}
     BlendPoses_0=Apply Additive(Base=@n2, Additive=@n1) {alphaBoolBlend {blendOption=Linear}}
     BlendWeights_0=Get JumpAddOffsetAlpha() -> (JumpAddOffsetAlpha:float)
    ) {layerSetup [{branchFilters [{boneName=thigh_l}, {boneName=thigh_r}]}]}
    CurveValues_0=1
    CurveValues_1:float=1
    CurveValues_2:float=1
    CurveValues_3:float=1
    Alpha=Get JumpAddCurveAlpha() -> (JumpAddCurveAlpha:float)
   ) -> (Pose) {curveNames [RemoveJumpAdditive, RemoveJumpFallAdditive, RemoveJumpLoopAdditive, RemoveJumpUpAdditive], curveValues [0,0,0,0]}
   n10 = Get ArmsHeadAddAlpha() -> (ArmsHeadAddAlpha:float)
   Output Pose(Result=Inertialization!(Source:Pose Link Structure=Modify Curve!(SourcePose=@n9, CurveValues_0=1, Alpha=@n10) -> (Pose) {curveNames [DisableArmsHeadAdditive], curveValues [0]}) -> (Pose:Pose Link Structure) {is=title AnimGraphNode_Inertialization}) {layerGroup=Item Override}
   Save cached pose 'InFinalPose_Cached'(Pose=Input Pose() {name=InFinalPose}) {cacheName=InFinalPose_Cached}
  ]}
 ItemFullBodyAimOffsets {nodes [
   Input Pose!() -> (Pose, FullBody_AimOffsetAlpha:double, Aim_Yaw:double, Aim_Pitch:double) {name=InPoseFullBodyAimOffsets}
   Input Pose!() -> (Pose) {name=InPosePreFullBodyAimOffsets}
   Output Pose() {layerGroup=Item Override}
  ]}
 ItemFullBodyOverride {nodes [
   Dead Blending!(Source:Pose Link Structure) -> (Pose:Pose Link Structure) {is=title AnimGraphNode_DeadBlending, defaultBlendMode=Linear}
   Inertialization()
   AnimGraphNode_StateMachine_0 = FB_RocketRam_SM() {is=title AnimGraphNode_StateMachine}
   Output Pose(Result=@AnimGraphNode_StateMachine_0) {layerGroup=Item Override}
   Save cached pose 'FullBody_Cached'(Pose=Input Pose() {name=InPoseFullBody}) {cacheName=FullBody_Cached}
  ]}
 FB_RocketRam_SM {
  entry=FB Wait
  path=ItemFullBodyOverride.AnimGraphNode_StateMachine_0.FB_RocketRam_SM
  ids=id
  nodes [
   FB Detect Slide() {is=title AnimStateAliasNode, also {aliases [FB Apex, FB Thrusting, FB To Apex]}}
   FB Off() {
    is=title AnimStateAliasNode
    also {aliases [
      FB Apex
      FB Ground Slide
      FB Jump
      FB Mortar
      FB Pull Lever
      FB Squat
      FB Thrusting
      FB To Apex
      FB To Impact Ground
      FB To Impact Wall Apex
      FB To Mortar
      FB To Slide
      FB To Thrust
     ]}
   }
   FB Detect Wall Apex() {is=title AnimStateAliasNode, also {aliases [FB Apex, FB Thrusting, FB To Apex, FB To Thrust]}}
   FB Braking() {
    is=title AnimStateAliasNode
    also {aliases [FB Apex, FB Impact Ground, FB Impact Wall, FB Mortar, FB To Impact Ground, FB To Impact Wall Apex, FB To Mortar, FB To Slide]}
   }
   FB Off Ground() {
    is=title AnimStateAliasNode
    also {aliases [FB Apex, FB Mortar, FB Thrusting, FB To Apex, FB To Impact Ground, FB To Mortar, FB To Slide, FB To Thrust]}
   }
   FB Off Wall() {
    is=title AnimStateAliasNode
    also {aliases [FB Apex, FB Mortar, FB Thrusting, FB To Apex, FB To Impact Wall Apex, FB To Impact Wall Mortar, FB To Mortar, FB To Slide, FB To Thrust]}
   }
   FB Detect Ground() {is=title AnimStateAliasNode, also {aliases [FB Apex, FB Mortar, FB Thrusting, FB To Apex, FB To Mortar, FB To Thrust]}}
   FB Detect Wall Mortar() {is=title AnimStateAliasNode, also {aliases [FB Mortar]}}
   AnimStateNode_0 = FB To Apex() {is=title AnimStateNode}
   AnimStateNode_1 = FB Squat() {is=title AnimStateNode}
   AnimStateNode_14 = FB Wait() {is=title AnimStateNode}
   AnimStateNode_15 = FB Impact Ground() {
    is=title AnimStateNode
    stateEntered {guid=#1}
    stateFullyBlended {guid=#2}
    stateLeft {guid=#3}
   }
   AnimStateNode_16 = FB Thrusting() {
    is=title AnimStateNode
    stateEntered {guid=#4}
    stateFullyBlended {guid=#5}
    stateLeft {guid=#6}
   }
   AnimStateNode_17 = FB Ground Slide() {
    is=title AnimStateNode
    stateEntered {guid=#7}
    stateFullyBlended {guid=#8}
    stateLeft {guid=#9}
   }
   AnimStateNode_18 = FB Pull Lever() {is=title AnimStateNode}
   AnimStateNode_19 = FB Apex() {is=title AnimStateNode}
   AnimStateNode_2 = FB To Thrust() {is=title AnimStateNode}
   AnimStateNode_20 = FB To Mortar() {is=title AnimStateNode}
   AnimStateNode_21 = FB Mortar() {
    is=title AnimStateNode
    stateEntered {guid=#10}
    stateFullyBlended {guid=#11}
    stateLeft {guid=#12}
   }
   AnimStateNode_22 = FB To Impact Ground() {is=title AnimStateNode}
   AnimStateNode_23 = FB Impact Wall() {is=title AnimStateNode}
   AnimStateNode_24 = FB To Impact Wall Mortar() {is=title AnimStateNode}
   AnimStateNode_25 = FB Jump() {
    is=title AnimStateNode
    stateEntered {guid=#13}
    stateFullyBlended {guid=#14}
    stateLeft {guid=#15}
   }
   AnimStateNode_3 = FB To Slide() {is=title AnimStateNode}
   AnimStateNode_4 = FB To Impact Wall Apex() {is=title AnimStateNode}
   AnimStateTransitionNode_0 = AnimStateTransitionNode() {automaticRuleTriggerTime=0, bAutomaticRuleBasedOnSequencePlayerInState=true, blendMode=Linear}
   AnimStateTransitionNode_1 = AnimStateTransitionNode() {bAutomaticRuleBasedOnSequencePlayerInState=true, blendMode=Linear, crossfadeDuration=0.3}
   AnimStateTransitionNode_10 = AnimStateTransitionNode()
   AnimStateTransitionNode_11 = AnimStateTransitionNode()
   AnimStateTransitionNode_15 = AnimStateTransitionNode() {
    blendMode=Linear
    crossfadeDuration=0.35
    transitionEnd {guid=#16}
    transitionInterrupt {guid=#17}
    transitionStart {guid=#18}
   }
   AnimStateTransitionNode_17 = AnimStateTransitionNode()
   AnimStateTransitionNode_2 = AnimStateTransitionNode() {crossfadeDuration=0.05}
   AnimStateTransitionNode_21 = AnimStateTransitionNode() {bAutomaticRuleBasedOnSequencePlayerInState=true}
   AnimStateTransitionNode_22 = AnimStateTransitionNode() {crossfadeDuration=0.1}
   AnimStateTransitionNode_23 = AnimStateTransitionNode() {crossfadeDuration=0.1, priorityOrder=3}
   AnimStateTransitionNode_24 = AnimStateTransitionNode() {crossfadeDuration=0}
   AnimStateTransitionNode_25 = AnimStateTransitionNode() {bAutomaticRuleBasedOnSequencePlayerInState=true}
   AnimStateTransitionNode_26 = AnimStateTransitionNode() {crossfadeDuration=0, priorityOrder=2}
   AnimStateTransitionNode_27 = AnimStateTransitionNode() {crossfadeDuration=0.5}
   AnimStateTransitionNode_28 = AnimStateTransitionNode() {crossfadeDuration=0.5}
   AnimStateTransitionNode_29 = AnimStateTransitionNode() {automaticRuleTriggerTime=0, blendMode=Linear, crossfadeDuration=0.1}
   AnimStateTransitionNode_3 = AnimStateTransitionNode()
   AnimStateTransitionNode_5 = AnimStateTransitionNode() {crossfadeDuration=0.02}
   AnimStateTransitionNode_6 = AnimStateTransitionNode() {bAutomaticRuleBasedOnSequencePlayerInState=true}
   AnimStateTransitionNode_7 = AnimStateTransitionNode() {crossfadeDuration=0.5}
   AnimStateTransitionNode_8 = AnimStateTransitionNode() {crossfadeDuration=0.1}
   AnimStateTransitionNode_9 = AnimStateTransitionNode()
  ]
  links [
   {
    dst=FB To Mortar
    rule=ItemFullBodyOverride.AnimGraphNode_StateMachine_0.FB_RocketRam_SM.AnimStateTransitionNode_17.AnimationTransitionGraph_3
    src=FB Apex
    via=AnimStateTransitionNode_17
   }
   {
    dst=FB Ground Slide
    rule=ItemFullBodyOverride.AnimGraphNode_StateMachine_0.FB_RocketRam_SM.AnimStateTransitionNode_23.AnimationTransitionGraph_2
    src=FB Braking
    via=AnimStateTransitionNode_23
   }
   {
    dst=FB To Impact Ground
    rule=ItemFullBodyOverride.AnimGraphNode_StateMachine_0.FB_RocketRam_SM.AnimStateTransitionNode_27.AnimationTransitionGraph_0
    src=FB Detect Ground
    via=AnimStateTransitionNode_27
   }
   {
    dst=FB To Slide
    rule=ItemFullBodyOverride.AnimGraphNode_StateMachine_0.FB_RocketRam_SM.AnimStateTransitionNode_7.Transition
    src=FB Detect Slide
    via=AnimStateTransitionNode_7
   }
   {
    dst=FB To Impact Wall Apex
    rule=ItemFullBodyOverride.AnimGraphNode_StateMachine_0.FB_RocketRam_SM.AnimStateTransitionNode_28.AnimationTransitionGraph_0
    src=FB Detect Wall Apex
    via=AnimStateTransitionNode_28
   }
   {
    dst=FB To Impact Wall Mortar
    rule=ItemFullBodyOverride.AnimGraphNode_StateMachine_0.FB_RocketRam_SM.AnimStateTransitionNode_9.Transition
    src=FB Detect Wall Mortar
    via=AnimStateTransitionNode_9
   }
   {
    dst=FB Wait
    rule=ItemFullBodyOverride.AnimGraphNode_StateMachine_0.FB_RocketRam_SM.AnimStateTransitionNode_3.AnimationTransitionGraph_3
    src=FB Ground Slide
    via=AnimStateTransitionNode_3
   }
   {
    dst=FB Wait
    rule=ItemFullBodyOverride.AnimGraphNode_StateMachine_0.FB_RocketRam_SM.AnimStateTransitionNode_0.AnimationTransitionGraph_3
    src=FB Impact Ground
    via=AnimStateTransitionNode_0
   }
   {
    dst=FB Wait
    rule=ItemFullBodyOverride.AnimGraphNode_StateMachine_0.FB_RocketRam_SM.AnimStateTransitionNode_11.Transition
    src=FB Impact Ground
    via=AnimStateTransitionNode_11
   }
   {
    dst=FB Wait
    rule=ItemFullBodyOverride.AnimGraphNode_StateMachine_0.FB_RocketRam_SM.AnimStateTransitionNode_25.AnimationTransitionGraph_2
    src=FB Impact Wall
    via=AnimStateTransitionNode_25
   }
   {
    dst=FB To Thrust
    rule=ItemFullBodyOverride.AnimGraphNode_StateMachine_0.FB_RocketRam_SM.AnimStateTransitionNode_29.AnimationTransitionGraph_2
    src=FB Jump
    via=AnimStateTransitionNode_29
   }
   {
    dst=FB Wait
    rule=ItemFullBodyOverride.AnimGraphNode_StateMachine_0.FB_RocketRam_SM.AnimStateTransitionNode_10.Transition
    src=FB Off
    via=AnimStateTransitionNode_10
   }
   {
    dst=FB Impact Ground
    rule=ItemFullBodyOverride.AnimGraphNode_StateMachine_0.FB_RocketRam_SM.AnimStateTransitionNode_24.AnimationTransitionGraph_2
    src=FB Off Ground
    via=AnimStateTransitionNode_24
   }
   {
    dst=FB Impact Wall
    rule=ItemFullBodyOverride.AnimGraphNode_StateMachine_0.FB_RocketRam_SM.AnimStateTransitionNode_26.AnimationTransitionGraph_2
    src=FB Off Wall
    via=AnimStateTransitionNode_26
   }
   {
    dst=FB Squat
    rule=ItemFullBodyOverride.AnimGraphNode_StateMachine_0.FB_RocketRam_SM.AnimStateTransitionNode_8.Transition
    src=FB Pull Lever
    via=AnimStateTransitionNode_8
   }
   {
    dst=FB Wait
    rule=ItemFullBodyOverride.AnimGraphNode_StateMachine_0.FB_RocketRam_SM.AnimStateTransitionNode_22.AnimationTransitionGraph_2
    src=FB Pull Lever
    via=AnimStateTransitionNode_22
   }
   {
    dst=FB Jump
    rule=ItemFullBodyOverride.AnimGraphNode_StateMachine_0.FB_RocketRam_SM.AnimStateTransitionNode_5.Transition
    src=FB Squat
    via=AnimStateTransitionNode_5
   }
   {
    dst=FB To Apex
    rule=ItemFullBodyOverride.AnimGraphNode_StateMachine_0.FB_RocketRam_SM.AnimStateTransitionNode_15.AnimationTransitionGraph_3
    src=FB Thrusting
    via=AnimStateTransitionNode_15
   }
   {
    dst=FB Apex
    rule=ItemFullBodyOverride.AnimGraphNode_StateMachine_0.FB_RocketRam_SM.AnimStateTransitionNode_1.Transition
    src=FB To Apex
    via=AnimStateTransitionNode_1
   }
   {
    dst=FB Mortar
    rule=ItemFullBodyOverride.AnimGraphNode_StateMachine_0.FB_RocketRam_SM.AnimStateTransitionNode_21.AnimationTransitionGraph_3
    src=FB To Mortar
    via=AnimStateTransitionNode_21
   }
   {
    dst=FB Thrusting
    rule=ItemFullBodyOverride.AnimGraphNode_StateMachine_0.FB_RocketRam_SM.AnimStateTransitionNode_6.Transition
    src=FB To Thrust
    via=AnimStateTransitionNode_6
   }
   {
    dst=FB Pull Lever
    rule=ItemFullBodyOverride.AnimGraphNode_StateMachine_0.FB_RocketRam_SM.AnimStateTransitionNode_2.Transition
    src=FB Wait
    via=AnimStateTransitionNode_2
   }
  ]
 }
 FB To Apex {
  path=ItemFullBodyOverride.AnimGraphNode_StateMachine_0.FB_RocketRam_SM.AnimStateNode_0.FB To Apex
  nodes [Output Animation Pose(Result=Two Way Blend(
     A=Sequence Player 'Antigravity_Intro'() {bLoopAnimation=false, sequence=/RocketRamGameplay/Items/RocketRam/Animations/Animation/Ramming/Apex/Antigravity_Intro}
     B=Sequence Player 'RocketRam_Flight_To_Slide'() {bLoopAnimation=false, sequence=/RocketRamGameplay/Items/RocketRam/Animations/Animation/Ramming/Impact/RocketRam_Flight_To_Slide}
     Alpha=Get FBToApexStateAlpha() -> (FBToApexStateAlpha:float)
    ) {blendNode {alphaBoolBlend {blendOption=Linear}}})]
 }
 FB Squat {
  path=ItemFullBodyOverride.AnimGraphNode_StateMachine_0.FB_RocketRam_SM.AnimStateNode_1.FB Squat
  nodes [
   n1 = Apply Additive(
    Base=Use cached pose 'FullBody_Cached'()
    Additive="Sequence Player (additive) 'RocketRam_Squat_Additive'"() {bLoopAnimation=false, sequence=/RocketRamGameplay/Items/RocketRam/Animations/Animation/Ramming/PreBoost/RocketRam_Squat_Additive}
   ) {alphaBoolBlend {blendOption=Linear}}
   n2 = Use cached pose 'FullBody_Cached'()
   Output Animation Pose(Result=Two Way Blend(
     A=Use cached pose 'FullBody_Cached'()
     B=Layered blend per bone(BasePose=@n2, BlendPoses_0=@n1) {layerSetup [{branchFilters [{boneName=pelvis}, {blendDepth=-1, boneName=spine_01}]}]}
     Alpha=Get AdditiveSquatAlpha() -> (AdditiveSquatAlpha:float)
    ) {blendNode {alphaBoolBlend {blendOption=Linear}}})
  ]
 }
 FB Wait {
  path=ItemFullBodyOverride.AnimGraphNode_StateMachine_0.FB_RocketRam_SM.AnimStateNode_14.FB Wait
  nodes [Output Animation Pose(Result=Use cached pose 'FullBody_Cached'())]
 }
 FB Impact Ground {
  path=ItemFullBodyOverride.AnimGraphNode_StateMachine_0.FB_RocketRam_SM.AnimStateNode_15.FB Impact Ground
  nodes [
   n1 = Apply Additive(
    Base=Use cached pose 'FullBody_Cached'()
    Additive="Sequence Player (additive) 'Slam_LowerBody_Additive'"() {bLoopAnimation=false, sequence=/RocketRamGameplay/Items/RocketRam/Animations/Animation/Ramming/Impact/Slam_LowerBody_Additive}
   ) {alphaBoolBlend {blendOption=Linear}}
   n2 = Use cached pose 'FullBody_Cached'()
   Output Animation Pose(Result=Two Way Blend(
     A=Layered blend per bone(BasePose=@n2, BlendPoses_0=@n1) {layerSetup [{branchFilters [{boneName=pelvis}, {blendDepth=-1, boneName=spine_01}]}]}
     B=Use cached pose 'FullBody_Cached'()
     Alpha=Get FBImpactGroundStateAlpha() -> (FBImpactGroundStateAlpha:float)
    ) {blendNode {alphaBoolBlend {blendOption=Linear}}})
  ]
 }
 FB Thrusting {
  path=ItemFullBodyOverride.AnimGraphNode_StateMachine_0.FB_RocketRam_SM.AnimStateNode_16.FB Thrusting
  nodes [Output Animation Pose(Result=Sequence Player 'RocketRam_FlyingPose'() {bLoopAnimation=false, sequence=/RocketRamGameplay/Items/RocketRam/Animations/Animation/Ramming/Thrusting/RocketRam_FlyingPose})]
 }
 FB Ground Slide {
  path=ItemFullBodyOverride.AnimGraphNode_StateMachine_0.FB_RocketRam_SM.AnimStateNode_17.FB Ground Slide
  nodes [Output Animation Pose(Result=Blendspace Player 'RocketRam_StopSlide_BS'(X=Get YInput() -> (YInput:float), Y=0) {blendSpace=/RocketRamGameplay/Items/RocketRam/Animations/Animation/Ramming/Impact/RocketRam_StopSlide_BS})]
 }
 FB Pull Lever {
  path=ItemFullBodyOverride.AnimGraphNode_StateMachine_0.FB_RocketRam_SM.AnimStateNode_18.FB Pull Lever
  nodes [Output Animation Pose(Result=Use cached pose 'FullBody_Cached'())]
 }
 FB Apex {
  path=ItemFullBodyOverride.AnimGraphNode_StateMachine_0.FB_RocketRam_SM.AnimStateNode_19.FB Apex
  nodes [
   n1 = AimOffset Player 'RocketRam_Apex_AO'!(
    X:float=Get YInput()
    Y:float=0
    BasePose:Pose Link Structure=Sequence Player 'RocketRam_Antigravity'() {bLoopAnimation=false, sequence=/RocketRamGameplay/Items/RocketRam/Animations/Animation/Ramming/Apex/RocketRam_Antigravity}
    Alpha:float=1
   ) -> (Pose:Pose Link Structure) {blendSpace=/RocketRamGameplay/Items/RocketRam/Animations/Animation/Ramming/Apex/RocketRam_Apex_AO, alphaBoolBlend {blendOption=Linear}}
   Output Animation Pose(Result=@n1)
  ]
 }
 FB To Thrust {
  path=ItemFullBodyOverride.AnimGraphNode_StateMachine_0.FB_RocketRam_SM.AnimStateNode_2.FB To Thrust
  nodes [Output Animation Pose(Result=Sequence Player 'IntroToThrust'() {bLoopAnimation=false, sequence=/RocketRamGameplay/Items/RocketRam/Animations/Animation/Ramming/Thrusting/IntroToThrust})]
 }
 FB To Mortar {
  path=ItemFullBodyOverride.AnimGraphNode_StateMachine_0.FB_RocketRam_SM.AnimStateNode_20.FB To Mortar
  nodes [Output Animation Pose(Result=Sequence Player 'FlyingDownIntro'() {bLoopAnimation=false, sequence=/RocketRamGameplay/Items/RocketRam/Animations/Animation/Ramming/Mortar/FlyingDownIntro})]
 }
 FB Mortar {
  path=ItemFullBodyOverride.AnimGraphNode_StateMachine_0.FB_RocketRam_SM.AnimStateNode_21.FB Mortar
  nodes [
   n1 = AimOffset Player 'RocketRam_Mortar_AO'!(
    X:float=Get YInput()
    Y:float=0
    BasePose:Pose Link Structure=Sequence Player 'FlyingDown'() {bLoopAnimation=false, sequence=/RocketRamGameplay/Items/RocketRam/Animations/Animation/Ramming/Mortar/FlyingDown}
    Alpha:float=1
   ) -> (Pose:Pose Link Structure) {blendSpace=/RocketRamGameplay/Items/RocketRam/Animations/Animation/Ramming/Mortar/RocketRam_Mortar_AO, alphaBoolBlend {blendOption=Linear}}
   Output Animation Pose(Result=@n1)
  ]
 }
 FB To Impact Ground {
  path=ItemFullBodyOverride.AnimGraphNode_StateMachine_0.FB_RocketRam_SM.AnimStateNode_22.FB To Impact Ground
  nodes [
   AimOffset Player 'RocketRam_Mortar_AO'(X=Get YInput(), Y=0, Alpha=1) {blendSpace=&RocketRam_Mortar_AO, alphaBoolBlend {blendOption=Linear}}
   Output Animation Pose(Result=Sequence Player 'MortarLoopEnd'() {bLoopAnimation=false, sequence=/RocketRamGameplay/Items/RocketRam/Animations/Animation/Ramming/Mortar/MortarLoopEnd})
  ]
 }
 FB Impact Wall {
  path=ItemFullBodyOverride.AnimGraphNode_StateMachine_0.FB_RocketRam_SM.AnimStateNode_23.FB Impact Wall
  nodes [Output Animation Pose(Result=Sequence Player 'RocketRam_Stop'() {bLoopAnimation=false, sequence=/RocketRamGameplay/Items/RocketRam/Animations/Animation/Ramming/Impact/RocketRam_Stop})]
 }
 FB To Impact Wall Mortar {
  path=ItemFullBodyOverride.AnimGraphNode_StateMachine_0.FB_RocketRam_SM.AnimStateNode_24.FB To Impact Wall Mortar
  nodes [
   n1 = AimOffset Player 'RocketRam_Mortar_AO'(
    X=Get YInput()
    Y=0
    BasePose=Sequence Player 'FlyingDown'() {bLoopAnimation=false, sequence=&FlyingDown}
    Alpha=1
   ) {blendSpace=&RocketRam_Mortar_AO, alphaBoolBlend {blendOption=Linear}}
   Output Animation Pose(Result=@n1)
  ]
 }
 FB Jump {
  path=ItemFullBodyOverride.AnimGraphNode_StateMachine_0.FB_RocketRam_SM.AnimStateNode_25.FB Jump
  nodes [
   n1 = Get!() -> (AdjustedAim_ResultingYaw:float=~, AdjustedAim_ResultingPitch:float=~, AdjustedAim_ZeroOutPitchWeightCurveName:Name)
   n2 = Blendspace Player 'RocketRam_PreBoostJump_BS'(X=0, Y=@n1.AdjustedAim_ResultingPitch) {bLoop=false, blendSpace=/RocketRamGameplay/Items/RocketRam/Animations/Animation/Ramming/PreBoost/RocketRam_PreBoostJump_BS}
   Output Animation Pose(Result=Two Way Blend(
     A=Use cached pose 'FullBody_Cached'()
     B=AimOffset Player 'SawBladeWeapon_AO'!(X:float=@n1, Y:float=0, BasePose:Pose Link Structure=@n2, Alpha:float=1) -> (Pose:Pose Link Structure) {blendSpace=/SharedWeaponAnims/Poses/SawbladeGun/SawBladeWeapon_AO, alphaBoolBlend {blendOption=Linear}}
     Alpha=Get FBJumpStateAlpha() -> (FBJumpStateAlpha:float)
    ) {blendNode {alphaBoolBlend {blendOption=Linear}}})
  ]
 }
 FB To Slide {
  path=ItemFullBodyOverride.AnimGraphNode_StateMachine_0.FB_RocketRam_SM.AnimStateNode_3.FB To Slide
  nodes [
   n1 = AimOffset Player 'RocketRam_Apex_AO'(
    X=Get YInput()
    Y=0
    BasePose=Sequence Player 'RocketRam_Antigravity'() {bLoopAnimation=false, sequence=&RocketRam_Antigravity}
    Alpha=1
   ) {blendSpace=&RocketRam_Apex_AO, alphaBoolBlend {blendOption=Linear}}
   Output Animation Pose(Result=@n1)
  ]
 }
 FB To Impact Wall Apex {
  path=ItemFullBodyOverride.AnimGraphNode_StateMachine_0.FB_RocketRam_SM.AnimStateNode_4.FB To Impact Wall Apex
  nodes [
   n1 = AimOffset Player 'RocketRam_Apex_AO'(
    X=Get YInput()
    Y=0
    BasePose=Sequence Player 'RocketRam_Antigravity'() {bLoopAnimation=false, sequence=&RocketRam_Antigravity}
    Alpha=1
   ) {blendSpace=&RocketRam_Apex_AO, alphaBoolBlend {blendOption=Linear}}
   Output Animation Pose(Result=@n1)
  ]
 }
 AnimationTransitionGraph_3 {
  path=ItemFullBodyOverride.AnimGraphNode_StateMachine_0.FB_RocketRam_SM.AnimStateTransitionNode_0.AnimationTransitionGraph_3
  nodes [Result(bCanEnterTransition=true)]
 }
 Transition {
  path=ItemFullBodyOverride.AnimGraphNode_StateMachine_0.FB_RocketRam_SM.AnimStateTransitionNode_1.Transition
  nodes [Result(bCanEnterTransition=true)]
 }
 Transition {
  path=ItemFullBodyOverride.AnimGraphNode_StateMachine_0.FB_RocketRam_SM.AnimStateTransitionNode_10.Transition
  nodes [
   n1 = Get bIsRRRecoveryHalfSecDelay()
   Result(bCanEnterTransition=AND Boolean(A=Get bIsRROff(), B=NOT Boolean(A=@n1)))
  ]
 }
 Transition {
  path=ItemFullBodyOverride.AnimGraphNode_StateMachine_0.FB_RocketRam_SM.AnimStateTransitionNode_11.Transition
  nodes [Result(bCanEnterTransition=Get bIsRRBraking())]
 }
 AnimationTransitionGraph_3 {
  path=ItemFullBodyOverride.AnimGraphNode_StateMachine_0.FB_RocketRam_SM.AnimStateTransitionNode_15.AnimationTransitionGraph_3
  nodes [Result(bCanEnterTransition=Get bFBThrustingToFBApex() -> (bFBThrustingToFBApex:bool))]
 }
 AnimationTransitionGraph_3 {
  path=ItemFullBodyOverride.AnimGraphNode_StateMachine_0.FB_RocketRam_SM.AnimStateTransitionNode_17.AnimationTransitionGraph_3
  nodes [Result(bCanEnterTransition=Get bFBApexToFBToMortar() -> (bFBApexToFBToMortar:bool))]
 }
 Transition {
  path=ItemFullBodyOverride.AnimGraphNode_StateMachine_0.FB_RocketRam_SM.AnimStateTransitionNode_2.Transition
  nodes [Result(bCanEnterTransition=Get bIsRRWindUp())]
 }
 AnimationTransitionGraph_3 {
  path=ItemFullBodyOverride.AnimGraphNode_StateMachine_0.FB_RocketRam_SM.AnimStateTransitionNode_21.AnimationTransitionGraph_3
  nodes [Result()]
 }
 AnimationTransitionGraph_2 {
  path=ItemFullBodyOverride.AnimGraphNode_StateMachine_0.FB_RocketRam_SM.AnimStateTransitionNode_22.AnimationTransitionGraph_2
  nodes [Result(bCanEnterTransition=Get bIsRROff())]
 }
 AnimationTransitionGraph_2 {
  path=ItemFullBodyOverride.AnimGraphNode_StateMachine_0.FB_RocketRam_SM.AnimStateTransitionNode_23.AnimationTransitionGraph_2
  nodes [Result(bCanEnterTransition=Get bIsRRBraking())]
 }
 AnimationTransitionGraph_2 {
  path=ItemFullBodyOverride.AnimGraphNode_StateMachine_0.FB_RocketRam_SM.AnimStateTransitionNode_24.AnimationTransitionGraph_2
  nodes [Result(bCanEnterTransition=Get bFBOffGroundToFBImpactGround() -> (bFBOffGroundToFBImpactGround:bool))]
 }
 AnimationTransitionGraph_2 {
  path=ItemFullBodyOverride.AnimGraphNode_StateMachine_0.FB_RocketRam_SM.AnimStateTransitionNode_25.AnimationTransitionGraph_2
  nodes [Result()]
 }
 AnimationTransitionGraph_2 {
  path=ItemFullBodyOverride.AnimGraphNode_StateMachine_0.FB_RocketRam_SM.AnimStateTransitionNode_26.AnimationTransitionGraph_2
  nodes [Result(bCanEnterTransition=Get bFBOffWallToFBImpactWall() -> (bFBOffWallToFBImpactWall:bool))]
 }
 AnimationTransitionGraph_0 {
  path=ItemFullBodyOverride.AnimGraphNode_StateMachine_0.FB_RocketRam_SM.AnimStateTransitionNode_27.AnimationTransitionGraph_0
  nodes [Result(bCanEnterTransition=Get bFBDetectGroundToFBToImpactGround() -> (bFBDetectGroundToFBToImpactGround:bool))]
 }
 AnimationTransitionGraph_0 {
  path=ItemFullBodyOverride.AnimGraphNode_StateMachine_0.FB_RocketRam_SM.AnimStateTransitionNode_28.AnimationTransitionGraph_0
  nodes [Result(bCanEnterTransition=Get bIsWallImpact())]
 }
 AnimationTransitionGraph_2 {
  path=ItemFullBodyOverride.AnimGraphNode_StateMachine_0.FB_RocketRam_SM.AnimStateTransitionNode_29.AnimationTransitionGraph_2
  nodes [Result(bCanEnterTransition=Get bIsRRActiveThrust())]
 }
 AnimationTransitionGraph_3 {
  path=ItemFullBodyOverride.AnimGraphNode_StateMachine_0.FB_RocketRam_SM.AnimStateTransitionNode_3.AnimationTransitionGraph_3
  nodes [Result(bCanEnterTransition=Get bFBGroundSlideToFBWait() -> (bFBGroundSlideToFBWait:bool))]
 }
 Transition {
  path=ItemFullBodyOverride.AnimGraphNode_StateMachine_0.FB_RocketRam_SM.AnimStateTransitionNode_5.Transition
  nodes [Result(bCanEnterTransition=Get bIsRRSlowAcceleration())]
 }
 Transition {
  path=ItemFullBodyOverride.AnimGraphNode_StateMachine_0.FB_RocketRam_SM.AnimStateTransitionNode_6.Transition
  nodes [Result(bCanEnterTransition=true)]
 }
 Transition {
  path=ItemFullBodyOverride.AnimGraphNode_StateMachine_0.FB_RocketRam_SM.AnimStateTransitionNode_7.Transition
  nodes [Result(bCanEnterTransition=Get bFBDetectSlideToFBToSlide() -> (bFBDetectSlideToFBToSlide:bool))]
 }
 Transition {
  path=ItemFullBodyOverride.AnimGraphNode_StateMachine_0.FB_RocketRam_SM.AnimStateTransitionNode_8.Transition
  nodes [Result(bCanEnterTransition=Get SquatAdditiveTimer() -> (SquatAdditiveTimer:bool))]
 }
 Transition {
  path=ItemFullBodyOverride.AnimGraphNode_StateMachine_0.FB_RocketRam_SM.AnimStateTransitionNode_9.Transition
  nodes [Result(bCanEnterTransition=Get bIsWallImpact())]
 }
 ItemIdleOverride {nodes [
   Input Pose() {name=InPose_Idle}
   Output Pose() {layerGroup=Item Override}
  ]}
 ItemJetPackJumpAdditive {nodes [Output Pose() {layerGroup=Item Override}]}
 ItemJetPackStartAdditive {nodes [Output Pose() {layerGroup=Item Override}]}
 ItemJumpLoopAdditive {nodes [
   Input Pose() {name=InPoseItemJumpLoopAdditive}
   Output Pose() {layerGroup=Item Override}
  ]}
 ItemJumpOverride {nodes [
   Input Pose() {name=InPoseJump}
   Output Pose() {layerGroup=Item Override}
  ]}
 ItemJumpUpAdditive {nodes [
   Input Pose() {name=InPoseItemJumpUpAdditive}
   Output Pose() {layerGroup=Item Override}
  ]}
 ItemPreAimOffsets {nodes [
   Input Pose() {name=InPosePreAimOffsets}
   Output Pose() {layerGroup=Item Override}
  ]}
 ItemPreIK {nodes [
   Input Pose() {name=InPose_PreIK}
   Output Pose() {layerGroup=Item Override}
  ]}
 ItemRightHandIK {nodes [
   Input Pose() {name=InPosePreRightHandIK}
   Input Pose() {name=InPosePostRightHandIK}
   Output Pose() {layerGroup=Item Override}
  ]}
 ItemSkeletalControl {nodes [
   Input Pose()
   Output Pose() {layerGroup=Item Override}
  ]}
 ItemSwimJumpFallLoop {nodes [Output Pose() {layerGroup=Item Override}]}
 ItemSwimJumpSurfaceEnd {nodes [Output Pose() {layerGroup=Item Override}]}
 ItemSwimJumpSurfaceLoop {nodes [Output Pose() {layerGroup=Item Override}]}
 ItemUpperBody {nodes [
   AnimGraphNode_StateMachine_0 = UB_RocketRam_SM() {is=title AnimGraphNode_StateMachine}
   Output Pose(Result=Inertialization(Source=@AnimGraphNode_StateMachine_0)) {layerGroup=Item Override}
   Save cached pose 'UpperBody_Cached'(Pose=Input Pose!() -> (Pose, InputParam:Fort Anim Input Adjusted Aim Structure) {name=InPoseUpperBody}) {cacheName=UpperBody_Cached}
  ]}
 UB_RocketRam_SM {
  entry=UB Wait
  path=ItemUpperBody.AnimGraphNode_StateMachine_0.UB_RocketRam_SM
  ids=id
  nodes [
   UB Off() {is=title AnimStateAliasNode, also {aliases [UB Jump, UB Pull Lever, UB Squat, UB To Ground Impact]}}
   AnimStateNode_0 = UB Wait() {is=title AnimStateNode}
   AnimStateNode_1 = UB Squat() {is=title AnimStateNode}
   AnimStateNode_2 = UB Jump() {is=title AnimStateNode}
   AnimStateNode_3 = UB Impact Ground() {is=title AnimStateNode}
   AnimStateNode_4 = UB To Ground Impact() {is=title AnimStateNode}
   AnimStateNode_5 = UB Pull Lever() {is=title AnimStateNode}
   AnimStateTransitionNode_1 = AnimStateTransitionNode() {crossfadeDuration=0.02}
   AnimStateTransitionNode_10 = AnimStateTransitionNode() {crossfadeDuration=0}
   AnimStateTransitionNode_11 = AnimStateTransitionNode() {crossfadeDuration=0}
   AnimStateTransitionNode_12 = AnimStateTransitionNode() {crossfadeDuration=0}
   AnimStateTransitionNode_2 = AnimStateTransitionNode() {blendMode=Linear, crossfadeDuration=0.02}
   AnimStateTransitionNode_3 = AnimStateTransitionNode() {crossfadeDuration=0.05}
   AnimStateTransitionNode_4 = AnimStateTransitionNode()
   AnimStateTransitionNode_5 = AnimStateTransitionNode() {crossfadeDuration=0}
   AnimStateTransitionNode_6 = AnimStateTransitionNode() {blendMode=Linear, crossfadeDuration=0.1}
   AnimStateTransitionNode_8 = AnimStateTransitionNode() {bAutomaticRuleBasedOnSequencePlayerInState=true}
   AnimStateTransitionNode_9 = AnimStateTransitionNode() {crossfadeDuration=0}
  ]
  links [
   {
    dst=UB Wait
    rule=ItemUpperBody.AnimGraphNode_StateMachine_0.UB_RocketRam_SM.AnimStateTransitionNode_12.Transition
    src=UB Impact Ground
    via=AnimStateTransitionNode_12
   }
   {
    dst=UB Wait
    rule=ItemUpperBody.AnimGraphNode_StateMachine_0.UB_RocketRam_SM.AnimStateTransitionNode_8.Transition
    src=UB Impact Ground
    via=AnimStateTransitionNode_8
   }
   {
    dst=UB To Ground Impact
    rule=ItemUpperBody.AnimGraphNode_StateMachine_0.UB_RocketRam_SM.AnimStateTransitionNode_9.Transition
    src=UB Jump
    via=AnimStateTransitionNode_9
   }
   {
    dst=UB Wait
    rule=ItemUpperBody.AnimGraphNode_StateMachine_0.UB_RocketRam_SM.AnimStateTransitionNode_5.Transition
    src=UB Jump
    via=AnimStateTransitionNode_5
   }
   {
    dst=UB Wait
    rule=ItemUpperBody.AnimGraphNode_StateMachine_0.UB_RocketRam_SM.AnimStateTransitionNode_4.Transition
    src=UB Off
    via=AnimStateTransitionNode_4
   }
   {
    dst=UB Squat
    rule=ItemUpperBody.AnimGraphNode_StateMachine_0.UB_RocketRam_SM.AnimStateTransitionNode_6.Transition
    src=UB Pull Lever
    via=AnimStateTransitionNode_6
   }
   {
    dst=UB Jump
    rule=ItemUpperBody.AnimGraphNode_StateMachine_0.UB_RocketRam_SM.AnimStateTransitionNode_1.Transition
    src=UB Squat
    via=AnimStateTransitionNode_1
   }
   {
    dst=UB Jump
    rule=ItemUpperBody.AnimGraphNode_StateMachine_0.UB_RocketRam_SM.AnimStateTransitionNode_2.Transition
    src=UB Squat
    via=AnimStateTransitionNode_2
   }
   {
    dst=UB Impact Ground
    rule=ItemUpperBody.AnimGraphNode_StateMachine_0.UB_RocketRam_SM.AnimStateTransitionNode_10.Transition
    src=UB To Ground Impact
    via=AnimStateTransitionNode_10
   }
   {
    dst=UB Impact Ground
    rule=ItemUpperBody.AnimGraphNode_StateMachine_0.UB_RocketRam_SM.AnimStateTransitionNode_11.Transition
    src=UB Wait
    via=AnimStateTransitionNode_11
   }
   {
    dst=UB Pull Lever
    rule=ItemUpperBody.AnimGraphNode_StateMachine_0.UB_RocketRam_SM.AnimStateTransitionNode_3.Transition
    src=UB Wait
    via=AnimStateTransitionNode_3
   }
  ]
 }
 UB Wait {
  path=ItemUpperBody.AnimGraphNode_StateMachine_0.UB_RocketRam_SM.AnimStateNode_0.UB Wait
  nodes [Output Animation Pose(Result=Use cached pose 'UpperBody_Cached'())]
 }
 UB Squat {
  path=ItemUpperBody.AnimGraphNode_StateMachine_0.UB_RocketRam_SM.AnimStateNode_1.UB Squat
  nodes [
   n1 = Get() -> (AdjustedAim_ResultingYaw=~, AdjustedAim_ResultingPitch=~)
   n2 = Blendspace Player 'RocketRam_Squat_BS'(X=0, Y=@n1.AdjustedAim_ResultingPitch) {bLoop=false, blendSpace=/RocketRamGameplay/Items/RocketRam/Animations/Animation/Ramming/PreBoost/RocketRam_Squat_BS}
   Output Animation Pose(Result=Two Way Blend(
     A=AimOffset Player 'SawBladeWeapon_AO'(X=@n1, Y=0, BasePose=@n2, Alpha=1) {blendSpace=&SawBladeWeapon_AO, alphaBoolBlend {blendOption=Linear}}
     B=Use cached pose 'UpperBody_Cached'()
     Alpha=Get UBSquatStateAlpha() -> (UBSquatStateAlpha:float)
    ) {blendNode {alphaBoolBlend {blendOption=Linear}}})
  ]
 }
 UB Jump {
  path=ItemUpperBody.AnimGraphNode_StateMachine_0.UB_RocketRam_SM.AnimStateNode_2.UB Jump
  nodes [
   n1 = Get() -> (AdjustedAim_ResultingYaw=~, AdjustedAim_ResultingPitch=~)
   AimOffset Player 'SawBladeWeapon_AO'(
    X=@n1
    Y=@n1.AdjustedAim_ResultingPitch
    BasePose=Sequence Player 'RocketRam_BasePose'() {bLoopAnimation=false, sequence=/RocketRamGameplay/Items/RocketRam/Animations/Animation/Locomotion/RocketRam_BasePose}
    Alpha=1
   ) {blendSpace=&SawBladeWeapon_AO, alphaBoolBlend {blendOption=Linear}}
   n2 = Get() -> (AdjustedAim_ResultingYaw=~, AdjustedAim_ResultingPitch=~)
   n3 = Blendspace Player 'RocketRam_PreBoostJump_BS'(X=0, Y=@n2.AdjustedAim_ResultingPitch) {bLoop=false, blendSpace=&RocketRam_PreBoostJump_BS}
   Output Animation Pose(Result=Two Way Blend(
     A=AimOffset Player 'SawBladeWeapon_AO'(X=@n2, Y=0, BasePose=@n3, Alpha=1) {blendSpace=&SawBladeWeapon_AO, alphaBoolBlend {blendOption=Linear}}
     B=Use cached pose 'UpperBody_Cached'()
     Alpha=Get UBJumpStateAlpha() -> (UBJumpStateAlpha:float)
    ) {blendNode {alphaBoolBlend {blendOption=Linear}}})
  ]
 }
 UB Impact Ground {
  path=ItemUpperBody.AnimGraphNode_StateMachine_0.UB_RocketRam_SM.AnimStateNode_3.UB Impact Ground
  nodes [Output Animation Pose(Result=Blendspace Player 'RocketRam_Slam_BS'(X=0, Y=Get ImpactPitch() -> (ImpactPitch:float)) {blendSpace=/RocketRamGameplay/Items/RocketRam/Animations/Animation/Ramming/Impact/RocketRam_Slam_BS})]
 }
 UB To Ground Impact {
  path=ItemUpperBody.AnimGraphNode_StateMachine_0.UB_RocketRam_SM.AnimStateNode_4.UB To Ground Impact
  nodes [Output Animation Pose(Result=Use cached pose 'UpperBody_Cached'())]
 }
 UB Pull Lever {
  path=ItemUpperBody.AnimGraphNode_StateMachine_0.UB_RocketRam_SM.AnimStateNode_5.UB Pull Lever
  nodes [
   n1 = Sequence Player 'RocketRam_LeverPull_and_Charge_CC'!(PlayRate:float=1) -> (Pose) {bLoopAnimation=false, sequence=/RocketRamGameplay/Items/RocketRam/Animations/Animation/Ramming/LeverPull/RocketRam_LeverPull_and_Charge_CC}
   n2 = Sequence Player 'RocketRam_LeverPull_and_Charge_Crouch_CC'() {
    bLoopAnimation=false
    sequence=/RocketRamGameplay/Items/RocketRam/Animations/Animation/Ramming/LeverPull/RocketRam_LeverPull_and_Charge_Crouch_CC
   }
   n3 = Get UBPullLeverStateAlpha() -> (UBPullLeverStateAlpha:float)
   n4 = Get() -> (AdjustedAim_ResultingYaw=~, AdjustedAim_ResultingPitch=~)
   Output Animation Pose(Result=AimOffset Player 'SawBladeWeapon_AO'(
     X=@n4
     Y=@n4.AdjustedAim_ResultingPitch
     BasePose=Two Way Blend(A=@n1, B=@n2, Alpha=@n3) {blendNode {alphaBoolBlend {blendOption=Linear}}}
     Alpha=1
    ) {bLoop=false, blendSpace=&SawBladeWeapon_AO, alphaBoolBlend {blendOption=Linear}})
  ]
 }
 Transition {
  path=ItemUpperBody.AnimGraphNode_StateMachine_0.UB_RocketRam_SM.AnimStateTransitionNode_1.Transition
  nodes [Result(bCanEnterTransition=Get bUBSquatToUBJump() -> (bUBSquatToUBJump:bool))]
 }
 Transition {
  path=ItemUpperBody.AnimGraphNode_StateMachine_0.UB_RocketRam_SM.AnimStateTransitionNode_10.Transition
  nodes [Result(bCanEnterTransition=Get bIsRRRecoveryDelayed())]
 }
 Transition {
  path=ItemUpperBody.AnimGraphNode_StateMachine_0.UB_RocketRam_SM.AnimStateTransitionNode_11.Transition
  nodes [Result(bCanEnterTransition=Get bUBWaitToUBImpactGround() -> (bUBWaitToUBImpactGround:bool))]
 }
 Transition {
  path=ItemUpperBody.AnimGraphNode_StateMachine_0.UB_RocketRam_SM.AnimStateTransitionNode_12.Transition
  nodes [Result(bCanEnterTransition=Get bIsRRBraking())]
 }
 Transition {
  path=ItemUpperBody.AnimGraphNode_StateMachine_0.UB_RocketRam_SM.AnimStateTransitionNode_2.Transition
  nodes [Result(bCanEnterTransition=Get bUBSquatToUBJump())]
 }
 Transition {
  path=ItemUpperBody.AnimGraphNode_StateMachine_0.UB_RocketRam_SM.AnimStateTransitionNode_3.Transition
  nodes [Result(bCanEnterTransition=Get bIsRRWindUp())]
 }
 Transition {
  path=ItemUpperBody.AnimGraphNode_StateMachine_0.UB_RocketRam_SM.AnimStateTransitionNode_4.Transition
  nodes [Result(bCanEnterTransition=Get bUBOffToUBWait() -> (bUBOffToUBWait:bool))]
 }
 Transition {
  path=ItemUpperBody.AnimGraphNode_StateMachine_0.UB_RocketRam_SM.AnimStateTransitionNode_5.Transition
  nodes [Result(bCanEnterTransition=Get bUBJumpToUBWait() -> (bUBJumpToUBWait:bool))]
 }
 Transition {
  path=ItemUpperBody.AnimGraphNode_StateMachine_0.UB_RocketRam_SM.AnimStateTransitionNode_6.Transition
  nodes [Result(bCanEnterTransition=Get SquatAdditiveTimer())]
 }
 Transition {path=ItemUpperBody.AnimGraphNode_StateMachine_0.UB_RocketRam_SM.AnimStateTransitionNode_8.Transition, nodes [Result()]}
 Transition {
  path=ItemUpperBody.AnimGraphNode_StateMachine_0.UB_RocketRam_SM.AnimStateTransitionNode_9.Transition
  nodes [Result(bCanEnterTransition=Get bUBJumpToUBToGroundImpact() -> (bUBJumpToUBToGroundImpact:bool))]
 }
 ItemZipLineStartAdditive {nodes [Output Pose() {layerGroup=Item Override}]}
 RocketRamState {nodes [
   RocketRamState() {is=title K2Node_FunctionEntry}
   > Is Valid(InputObject=Get As B Weapon Rocket Ram() -> (As B Weapon Rocket Ram:B Weapon Rocket Ram Object Reference))
   > Sequence()
   > Set bIsRRWindUp(bIsRRWindUp:bool=false) -> (Output_Get:bool)
   > Set bIsRRSlowAcceleration(bIsRRSlowAcceleration:bool=false) -> (Output_Get:bool)
   > Set bIsRRActiveThrust(bIsRRActiveThrust:bool=false) -> (Output_Get:bool)
   > Set bIsRRFreeFall(bIsRRFreeFall:bool=false) -> (Output_Get:bool)
   > Set bIsRRFastFreeFall(bIsRRFastFreeFall:bool=false) -> (Output_Get:bool)
   > Set bIsRRBraking(bIsRRBraking:bool=false) -> (Output_Get:bool)
   > Set bIsRRRecovery(bIsRRRecovery:bool=false) -> (Output_Get:bool)
   > Set bIsRROff(bIsRROff:bool=false) -> (Output_Get:bool)
   > n1 = Switch on Enum_RocketRamState!(
    execute
    Selection:Enum_RocketRamState Enum=Get RamState!(self:B Weapon Rocket Ram Object Reference=Get As B Weapon Rocket Ram()) -> (RamState:Enum_RocketRamState Enum=NewEnumerator0)
   ) -> (
    NewEnumerator0:Exec
    NewEnumerator6:Exec
    NewEnumerator1:Exec
    NewEnumerator2:Exec
    NewEnumerator5:Exec
    NewEnumerator3:Exec
    NewEnumerator7:Exec
    NewEnumerator4:Exec
   ) {is=title K2Node_SwitchEnum}
   > Set bIsRRWindUp(bIsRRWindUp=true)
   Set bIsRRActiveThrust(execute=@n1.NewEnumerator1, bIsRRActiveThrust=true)
   Set bIsRRFreeFall(execute=@n1.NewEnumerator2, bIsRRFreeFall=true)
   Set bIsRRBraking(execute=@n1.NewEnumerator3, bIsRRBraking=true)
   Set bIsRROff(execute=@n1.NewEnumerator4, bIsRROff=true)
   Set bIsRRFastFreeFall(execute=@n1.NewEnumerator5, bIsRRFastFreeFall=true)
   Set bIsRRSlowAcceleration(execute=@n1.NewEnumerator6, bIsRRSlowAcceleration=true)
   Set bIsRRRecovery(execute=@n1.NewEnumerator7, bIsRRRecovery=true)
  ]}
]
variables [
 As B Weapon Rocket Ram {type=B_Weapon_RocketRam_C, default=null}
 As Fort Player Pawn {type=FortPlayerPawn}
 Delta Time X {type=number}
 FalseRotationBlendTime {type=number}
 Impact Direction {type=Vector}
 ImpactPitch {type=number}
 SquatAdditiveTimer {type=boolean}
 TimeDilationRateScale {type=number, default=1}
]
