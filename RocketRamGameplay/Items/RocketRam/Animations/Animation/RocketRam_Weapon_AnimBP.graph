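# Animation Blueprint for the Rocket Ram weapon mesh, animating it from wind-up and jump through thrust, apex, mortar, impacts and braking slide.
# Each update turns the weapon's RamState into transition booleans, delaying recovery two ticks.
# It copies aim, landing and impact values from the player's linked RocketRam_Player_AnimBP layer.
# An aim offset, eased in during free fall or braking and sign-flipped by UserCurve01, feeds DefaultSlot and inertialization.
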AnimBlueprint /RocketRamGameplay/Items/RocketRam/Animations/Animation/RocketRam_Weapon_AnimBP
target_skeleton=/RocketRamGameplay/Items/RocketRam/Meshes/RocketRam_Heavy_Skeleton
graphs [
 AnimGraph {nodes [
   AnimGraphNode_StateMachine_0 = RocketRam_SM() {is=title AnimGraphNode_StateMachine}
   n1 = Get X Input() -> (X Input:double)
   n2 = Get AOMultiplier() -> (AOMultiplier:double)
   n3 = Get YInput() -> (YInput:float)
   n4 = AimOffset Player 'RocketRam_AO_WEP'!(
    X:float=float * float(A="Clamp (Float)"(Value=@n3, Min=-0.8, Max=0.8), B=@n2) {is=title K2Node_PromotableOperator}
    Y:float=float * float(A=@n2, B="Clamp (Float)"(Value=@n1, Min=-0.8, Max=0.8))
    BasePose:Pose Link Structure=@AnimGraphNode_StateMachine_0
    Alpha:float=Get AO Alpha() -> (AO Alpha:double)
   ) -> (Pose:Pose Link Structure) {blendSpace=/RocketRamGameplay/Items/RocketRam/Animations/Animation/Ramming/Thrusting/RocketRam_AO_WEP, alphaBoolBlend {blendOption=Linear}}
   Output Pose(Result=Inertialization!(Source:Pose Link Structure=Slot 'DefaultSlot'(Source=@n4)) -> (Pose:Pose Link Structure) {is=title AnimGraphNode_Inertialization}) {layerGroup=None}
  ]}
 RocketRam_SM {
  entry=Wait
  path=AnimGraph.AnimGraphNode_StateMachine_0.RocketRam_SM
  ids=id
  nodes [
   FB Off() {is=title AnimStateAliasNode, also {aliases [Apex, Jump, LeverPull+Charge, Squat, Thrusting, To Apex, To Impact, To Mortar, Wait]}}
   To Braking() {is=title AnimStateAliasNode, also {aliases [Apex, Impact Wall, Thrusting, To Apex, To Impact, To Mortar]}}
   To Impact Ground() {is=title AnimStateAliasNode, also {aliases [Thrusting, To Impact]}}
   To Impact Wall() {is=title AnimStateAliasNode, also {aliases [Apex, Thrusting, To Apex, To Impact, To Mortar]}}
   AnimStateNode_0 = Wait() {is=title AnimStateNode}
   AnimStateNode_11 = Squat() {is=title AnimStateNode}
   AnimStateNode_12 = To Apex() {is=title AnimStateNode}
   AnimStateNode_15 = Impact Ground() {
    is=title AnimStateNode
    stateEntered {guid=#1}
    stateFullyBlended {guid=#2}
    stateLeft {guid=#3}
   }
   AnimStateNode_17 = Ground Slide() {
    is=title AnimStateNode
    stateEntered {guid=#4}
    stateFullyBlended {guid=#5}
    stateLeft {guid=#6}
   }
   AnimStateNode_2 = Thrusting() {is=title AnimStateNode}
   AnimStateNode_23 = Impact Wall() {is=title AnimStateNode}
   AnimStateNode_4 = Jump() {is=title AnimStateNode}
   AnimStateNode_5 = LeverPull+Charge() {is=title AnimStateNode}
   AnimStateNode_6 = Apex() {is=title AnimStateNode}
   AnimStateNode_7 = To Mortar() {is=title AnimStateNode}
   AnimStateNode_9 = To Impact() {is=title AnimStateNode}
   AnimStateTransitionNode_10 = AnimStateTransitionNode() {blendMode=Linear, crossfadeDuration=0.25}
   AnimStateTransitionNode_11 = AnimStateTransitionNode() {blendMode=Linear, crossfadeDuration=0.1}
   AnimStateTransitionNode_12 = AnimStateTransitionNode()
   AnimStateTransitionNode_13 = AnimStateTransitionNode() {bAutomaticRuleBasedOnSequencePlayerInState=true}
   AnimStateTransitionNode_18 = AnimStateTransitionNode() {bAutomaticRuleBasedOnSequencePlayerInState=true, crossfadeDuration=0.3}
   AnimStateTransitionNode_2 = AnimStateTransitionNode() {crossfadeDuration=0.02}
   AnimStateTransitionNode_20 = AnimStateTransitionNode() {crossfadeDuration=0.1}
   AnimStateTransitionNode_21 = AnimStateTransitionNode()
   AnimStateTransitionNode_23 = AnimStateTransitionNode() {crossfadeDuration=0.1, priorityOrder=3}
   AnimStateTransitionNode_24 = AnimStateTransitionNode() {crossfadeDuration=0}
   AnimStateTransitionNode_26 = AnimStateTransitionNode() {crossfadeDuration=0.1, logicType=TLT_Inertialization, priorityOrder=2}
   AnimStateTransitionNode_4 = AnimStateTransitionNode()
   AnimStateTransitionNode_6 = AnimStateTransitionNode() {crossfadeDuration=0.1}
   AnimStateTransitionNode_8 = AnimStateTransitionNode() {bAutomaticRuleBasedOnSequencePlayerInState=true}
   AnimStateTransitionNode_9 = AnimStateTransitionNode() {crossfadeDuration=0.05}
  ]
  links [
   {
    dst=To Mortar
    rule=AnimGraph.AnimGraphNode_StateMachine_0.RocketRam_SM.AnimStateTransitionNode_12.AnimationTransitionGraph_0
    src=Apex
    via=AnimStateTransitionNode_12
   }
   {
    dst=Wait
    rule=AnimGraph.AnimGraphNode_StateMachine_0.RocketRam_SM.AnimStateTransitionNode_21.Transition
    src=FB Off
    via=AnimStateTransitionNode_21
   }
   {
    dst=Wait
    rule=AnimGraph.AnimGraphNode_StateMachine_0.RocketRam_SM.AnimStateTransitionNode_4.Transition
    src=Ground Slide
    via=AnimStateTransitionNode_4
   }
   {
    dst=Wait
    rule=AnimGraph.AnimGraphNode_StateMachine_0.RocketRam_SM.AnimStateTransitionNode_8.Transition
    src=Impact Ground
    via=AnimStateTransitionNode_8
   }
   {
    dst=Wait
    rule=AnimGraph.AnimGraphNode_StateMachine_0.RocketRam_SM.AnimStateTransitionNode_13.Transition
    src=Impact Wall
    via=AnimStateTransitionNode_13
   }
   {
    dst=Thrusting
    rule=AnimGraph.AnimGraphNode_StateMachine_0.RocketRam_SM.AnimStateTransitionNode_6.AnimationTransitionGraph_0
    src=Jump
    via=AnimStateTransitionNode_6
   }
   {
    dst=Squat
    rule=AnimGraph.AnimGraphNode_StateMachine_0.RocketRam_SM.AnimStateTransitionNode_11.Transition
    src=LeverPull+Charge
    via=AnimStateTransitionNode_11
   }
   {
    dst=Jump
    rule=AnimGraph.AnimGraphNode_StateMachine_0.RocketRam_SM.AnimStateTransitionNode_2.Transition
    src=Squat
    via=AnimStateTransitionNode_2
   }
   {
    dst=To Apex
    rule=AnimGraph.AnimGraphNode_StateMachine_0.RocketRam_SM.AnimStateTransitionNode_10.Transition
    src=Thrusting
    via=AnimStateTransitionNode_10
   }
   {
    dst=Apex
    rule=AnimGraph.AnimGraphNode_StateMachine_0.RocketRam_SM.AnimStateTransitionNode_18.Transition
    src=To Apex
    via=AnimStateTransitionNode_18
   }
   {
    dst=Ground Slide
    rule=AnimGraph.AnimGraphNode_StateMachine_0.RocketRam_SM.AnimStateTransitionNode_23.AnimationTransitionGraph_0
    src=To Braking
    via=AnimStateTransitionNode_23
   }
   {
    dst=Impact Ground
    rule=AnimGraph.AnimGraphNode_StateMachine_0.RocketRam_SM.AnimStateTransitionNode_24.AnimationTransitionGraph_0
    src=To Impact Ground
    via=AnimStateTransitionNode_24
   }
   {
    dst=Impact Wall
    rule=AnimGraph.AnimGraphNode_StateMachine_0.RocketRam_SM.AnimStateTransitionNode_26.AnimationTransitionGraph_0
    src=To Impact Wall
    via=AnimStateTransitionNode_26
   }
   {
    dst=To Impact
    rule=AnimGraph.AnimGraphNode_StateMachine_0.RocketRam_SM.AnimStateTransitionNode_20.Transition
    src=To Mortar
    via=AnimStateTransitionNode_20
   }
   {
    dst=LeverPull+Charge
    rule=AnimGraph.AnimGraphNode_StateMachine_0.RocketRam_SM.AnimStateTransitionNode_9.AnimationTransitionGraph_0
    src=Wait
    via=AnimStateTransitionNode_9
   }
  ]
 }
 Wait {
  path=AnimGraph.AnimGraphNode_StateMachine_0.RocketRam_SM.AnimStateNode_0.Wait
  nodes [Output Animation Pose(Result=Sequence Player 'RocketRam_BasePose_WEP'() {bLoopAnimation=false, sequence=/RocketRamGameplay/Items/RocketRam/Animations/RocketRam_BasePose_WEP})]
 }
 Squat {
  path=AnimGraph.AnimGraphNode_StateMachine_0.RocketRam_SM.AnimStateNode_11.Squat
  nodes [
   Sequence Player 'RocketRam_Squat_CC_WEP'() {bLoopAnimation=false, sequence=/RocketRamGameplay/Items/RocketRam/Animations/Animation/Ramming/PreBoost/RocketRam_Squat_CC_WEP}
   Sequence Player 'RocketRam_BasePose_WEP'() {bLoopAnimation=false, sequence=&RocketRam_BasePose_WEP}
   n1 = Get AdjustedPitch() -> (AdjustedPitch:float)
   n2 = Get Is Jump from Ground() -> (Is Jump from Ground:bool)
   Output Animation Pose(Result=Two Way Blend(
     A=Local Space Ref Pose()
     B=Blendspace Player 'RocketRam_Squat_BS_WEP'(X=0, Y=@n1) {bLoop=false, blendSpace=/RocketRamGameplay/Items/RocketRam/Animations/Animation/Ramming/PreBoost/RocketRam_Squat_BS_WEP}
     Alpha="To Float (Boolean)"(InBool=@n2)
    ) {blendNode {alphaBoolBlend {blendOption=Linear}}})
  ]
 }
 To Apex {
  path=AnimGraph.AnimGraphNode_StateMachine_0.RocketRam_SM.AnimStateNode_12.To Apex
  nodes [
   n1 = Get IsRRBraking() -> (IsRRBraking:bool)
   Output Animation Pose(Result=Two Way Blend(
     A=Sequence Player 'RocketRam_Antigravity_Intro_WEP'() {bLoopAnimation=false, sequence=/RocketRamGameplay/Items/RocketRam/Animations/Animation/Ramming/Apex/RocketRam_Antigravity_Intro_WEP}
     B=Sequence Player 'RocketRam_Flight_To_Slide_WEP'() {bLoopAnimation=false, sequence=/RocketRamGameplay/Items/RocketRam/Animations/Animation/Ramming/Impact/RocketRam_Flight_To_Slide_WEP}
     Alpha="To Float (Boolean)"(InBool=@n1)
    ) {blendNode {alphaBoolBlend {blendOption=Linear}}})
  ]
 }
 Impact Ground {
  path=AnimGraph.AnimGraphNode_StateMachine_0.RocketRam_SM.AnimStateNode_15.Impact Ground
  nodes [Output Animation Pose(Result=Sequence Player 'RocketRam_BasePose_WEP'() {bLoopAnimation=false, sequence=&RocketRam_BasePose_WEP})]
 }
 Ground Slide {
  path=AnimGraph.AnimGraphNode_StateMachine_0.RocketRam_SM.AnimStateNode_17.Ground Slide
  nodes [Output Animation Pose(Result=Sequence Player 'RocketRam_BasePose_WEP'() {bLoopAnimation=false, sequence=&RocketRam_BasePose_WEP})]
 }
 Thrusting {
  path=AnimGraph.AnimGraphNode_StateMachine_0.RocketRam_SM.AnimStateNode_2.Thrusting
  nodes [
   Sequence Player 'RocketRam_BasePose_WEP'() {bLoopAnimation=false, sequence=&RocketRam_BasePose_WEP}
   Output Animation Pose(Result=Sequence Player 'RocketRam_FlyingLoop_WEP'() {bLoopAnimation=false, sequence=/RocketRamGameplay/Items/RocketRam/Animations/Animation/Ramming/Thrusting/RocketRam_FlyingLoop_WEP})
  ]
 }
 Impact Wall {
  path=AnimGraph.AnimGraphNode_StateMachine_0.RocketRam_SM.AnimStateNode_23.Impact Wall
  nodes [Output Animation Pose(Result=Sequence Player 'RocketRam_BasePose_WEP'() {bLoopAnimation=false, sequence=&RocketRam_BasePose_WEP})]
 }
 Jump {
  path=AnimGraph.AnimGraphNode_StateMachine_0.RocketRam_SM.AnimStateNode_4.Jump
  nodes [
   Output Animation Pose(Result=Blendspace Player 'RocketRam_PreBoostJump_BS_WEP'(X=0, Y=Get AdjustedPitch()) {bLoop=false, blendSpace=/RocketRamGameplay/Items/RocketRam/Animations/Animation/Ramming/PreBoost/RocketRam_PreBoostJump_BS_WEP})
   Two Way Blend(
    A=Sequence Player 'RocketRam_BasePose_WEP'() {bLoopAnimation=false, sequence=&RocketRam_BasePose_WEP}
    Alpha="To Float (Boolean)"(InBool=Get Is Jump from Ground())
   ) {blendNode {alphaBoolBlend {blendOption=Linear}}}
  ]
 }
 LeverPull+Charge {
  path=AnimGraph.AnimGraphNode_StateMachine_0.RocketRam_SM.AnimStateNode_5.LeverPull+Charge
  nodes [
   n1 = Sequence Player 'RocketRam_LeverPull_and_Charge_CC_WEP'!(PlayRate:float=1) -> (Pose) {
    bLoopAnimation=false
    sequence=/RocketRamGameplay/Items/RocketRam/Animations/Animation/Ramming/LeverPull/RocketRam_LeverPull_and_Charge_CC_WEP
   }
   Output Animation Pose(Result=@n1)
  ]
 }
 Apex {
  path=AnimGraph.AnimGraphNode_StateMachine_0.RocketRam_SM.AnimStateNode_6.Apex
  nodes [
   Sequence Player 'RocketRam_BasePose_WEP'() {bLoopAnimation=false, sequence=&RocketRam_BasePose_WEP}
   Output Animation Pose(Result=Sequence Player 'RocketRam_FlyingLoop_WEP'() {bLoopAnimation=false, sequence=&RocketRam_FlyingLoop_WEP})
  ]
 }
 To Mortar {
  path=AnimGraph.AnimGraphNode_StateMachine_0.RocketRam_SM.AnimStateNode_7.To Mortar
  nodes [Output Animation Pose(Result=Sequence Player 'RocketRam_MortarPose_WEP'() {bLoopAnimation=false, sequence=/RocketRamGameplay/Items/RocketRam/Animations/Animation/Ramming/Mortar/RocketRam_MortarPose_WEP})]
 }
 To Impact {
  path=AnimGraph.AnimGraphNode_StateMachine_0.RocketRam_SM.AnimStateNode_9.To Impact
  nodes [Output Animation Pose(Result=Sequence Player 'RocketRam_BasePose_WEP'() {bLoopAnimation=false, sequence=&RocketRam_BasePose_WEP})]
 }
 Transition {
  path=AnimGraph.AnimGraphNode_StateMachine_0.RocketRam_SM.AnimStateTransitionNode_10.Transition
  nodes [
   n1 = Get IsRRFreeFall() -> (IsRRFreeFall:bool)
   n2 = Get IsFalling() -> (IsFalling:bool)
   Result(bCanEnterTransition=OR Boolean(A=AND Boolean(A=@n1, B=@n2), B=Get IsRRBraking()))
  ]
 }
 Transition {
  path=AnimGraph.AnimGraphNode_StateMachine_0.RocketRam_SM.AnimStateTransitionNode_11.Transition
  nodes [Result(bCanEnterTransition=Get Squat Additive Timer() -> (Squat Additive Timer:bool))]
 }
 AnimationTransitionGraph_0 {
  path=AnimGraph.AnimGraphNode_StateMachine_0.RocketRam_SM.AnimStateTransitionNode_12.AnimationTransitionGraph_0
  nodes [
   n1 = Get Is Wall Impact() -> (Is Wall Impact:bool)
   n2 = Get IsGoingUp() -> (IsGoingUp:bool)
   n3 = Get Is No Mortar() -> (Is No Mortar:bool)
   Result(bCanEnterTransition=AND Boolean!(A=NOT Boolean(A=@n2), B=NOT Boolean(A=@n3), C:bool=Get Is Apex Timer End() -> (Is Apex Timer End:bool), D:bool=NOT Boolean(A=@n1)) -> (ReturnValue))
  ]
 }
 Transition {path=AnimGraph.AnimGraphNode_StateMachine_0.RocketRam_SM.AnimStateTransitionNode_13.Transition, nodes [Result()]}
 Transition {path=AnimGraph.AnimGraphNode_StateMachine_0.RocketRam_SM.AnimStateTransitionNode_18.Transition, nodes [Result()]}
 Transition {
  path=AnimGraph.AnimGraphNode_StateMachine_0.RocketRam_SM.AnimStateTransitionNode_2.Transition
  nodes [Result(bCanEnterTransition=Get IsRRSlowAcceleration() -> (IsRRSlowAcceleration:bool))]
 }
 Transition {
  path=AnimGraph.AnimGraphNode_StateMachine_0.RocketRam_SM.AnimStateTransitionNode_20.Transition
  nodes [Result(bCanEnterTransition=Get IsLandingDetected() -> (IsLandingDetected:bool))]
 }
 Transition {
  path=AnimGraph.AnimGraphNode_StateMachine_0.RocketRam_SM.AnimStateTransitionNode_21.Transition
  nodes [
   n1 = Get IsRRRecovery() -> (IsRRRecovery:bool)
   Result(bCanEnterTransition=AND Boolean!(A=Get IsRROff() -> (IsRROff:bool), B=NOT Boolean(A=@n1)) -> (ReturnValue))
  ]
 }
 AnimationTransitionGraph_0 {
  path=AnimGraph.AnimGraphNode_StateMachine_0.RocketRam_SM.AnimStateTransitionNode_23.AnimationTransitionGraph_0
  nodes [Result(bCanEnterTransition=Get IsRRBraking())]
 }
 AnimationTransitionGraph_0 {
  path=AnimGraph.AnimGraphNode_StateMachine_0.RocketRam_SM.AnimStateTransitionNode_24.AnimationTransitionGraph_0
  nodes [
   n1 = Get Is Wall Impact()
   n2 = Get IsRROff()
   n3 = Get IsRRRecoveryDelayed() -> (IsRRRecoveryDelayed:bool)
   Result(bCanEnterTransition=AND Boolean(A=OR Boolean(A=@n3, B=@n2), B=NOT Boolean(A=@n1)))
  ]
 }
 AnimationTransitionGraph_0 {
  path=AnimGraph.AnimGraphNode_StateMachine_0.RocketRam_SM.AnimStateTransitionNode_26.AnimationTransitionGraph_0
  nodes [
   n1 = Get IsRROff()
   n2 = Get IsRRRecoveryDelayed()
   Result(bCanEnterTransition=AND Boolean(A=OR Boolean(A=@n2, B=@n1), B=Get Is Wall Impact()))
  ]
 }
 Transition {
  path=AnimGraph.AnimGraphNode_StateMachine_0.RocketRam_SM.AnimStateTransitionNode_4.Transition
  nodes [
   n1 = Get IsRRBraking()
   Result(bCanEnterTransition=OR Boolean(A=NOT Boolean(A=@n1), B=Get IsRROff()))
  ]
 }
 AnimationTransitionGraph_0 {
  path=AnimGraph.AnimGraphNode_StateMachine_0.RocketRam_SM.AnimStateTransitionNode_6.AnimationTransitionGraph_0
  nodes [Result(bCanEnterTransition=Get IsRRActiveThrust() -> (IsRRActiveThrust:bool))]
 }
 Transition {path=AnimGraph.AnimGraphNode_StateMachine_0.RocketRam_SM.AnimStateTransitionNode_8.Transition, nodes [Result()]}
 AnimationTransitionGraph_0 {
  path=AnimGraph.AnimGraphNode_StateMachine_0.RocketRam_SM.AnimStateTransitionNode_9.AnimationTransitionGraph_0
  nodes [Result(bCanEnterTransition=Get IsRRWindUp() -> (IsRRWindUp:bool))]
 }
 EventGraph {nodes [
   n1 = Event BlueprintUpdateAnimation()
   > Set Delta Time X(Delta Time X=@n1.DeltaTimeX)
   > n2 = Sequence!(execute) -> (then_0, then_1, then_2:Exec, then_3:Exec, then_4:Exec, then_5:Exec)
   n3 = GetOwningActor()
   > Is Valid(InputObject=@n3)
   > n4 = Cast To B_Weapon_RocketRam!(execute, Object=@n3) -> (then, CastFailed, AsB Weapon Rocket Ram:B Weapon Rocket Ram Object Reference) {is=title K2Node_DynamicCast}
   > Set As B Weapon Rocket Ram(As B Weapon Rocket Ram:B Weapon Rocket Ram Object Reference=@n4) -> (Output_Get:B Weapon Rocket Ram Object Reference)
   n5 = TryGetFortPawn()
   > Is Valid(InputObject=@n5)
   > n6 = Cast To FortPlayerPawn(Object=@n5)
   > Set As Fort Player Pawn(As Fort Player Pawn:Fort Player Pawn Object Reference=@n6) -> (Output_Get:Fort Player Pawn Object Reference)
   RocketRamState(execute=@n2.then_1)
   Is Valid(exec=@n2.then_2, InputObject=Get As Fort Player Pawn() -> (As Fort Player Pawn:Fort Player Pawn Object Reference))
   n7 = Get As Fort Player Pawn()
   > Set IsFalling(IsFalling:bool=IsFalling!(self:Nav Movement Component Object Reference=Get CharacterMovement!(self:Character Object Reference=@n7) -> (CharacterMovement:Character Movement Component Object Reference)) -> (ReturnValue:bool)) -> (Output_Get:bool)
   n8 = Get As Fort Player Pawn()
   > n9 = Cast To RocketRam_Player_AnimBP!(
    execute
    Object=GetLinkedAnimLayerInstanceByClass!(
     self:Skeletal Mesh Component Object Reference=Get Mesh(self=@n8)
     InClass:Anim Instance Class Reference=&/RocketRam_Player_AnimBP.RocketRam_Player_AnimBP_C
    ) -> (ReturnValue:Anim Instance Object Reference)
   ) -> (then, CastFailed, AsRocket Ram Player Anim BP:Rocket Ram Player Anim BP Object Reference) {is=title K2Node_DynamicCast}
   > Set IsLandingDetected(IsLandingDetected:bool=Get bIsLandingDetected!(self:Rocket Ram Item Layer Anim Instance Object Reference=@n9) -> (bIsLandingDetected:bool)) -> (Output_Get:bool)
   > Set AdjustedPitch(AdjustedPitch:float=Get!(self:Fort Item Layer Anim Instance Object Reference=@n9) -> (AdjustedAim_ResultingYaw:float=~, AdjustedAim_ResultingPitch:float=~, AdjustedAim_ZeroOutPitchWeightCurveName:Name).AdjustedAim_ResultingPitch) -> (Output_Get:float)
   > Set YInput(YInput:float=Get YInput!(self:Rocket Ram Item Layer Anim Instance Object Reference=@n9) -> (YInput)) -> (Output_Get:float)
   > Set Squat Additive Timer(Squat Additive Timer:bool=Get SquatAdditiveTimer!(self:Rocket Ram Player Anim BP Object Reference=@n9) -> (SquatAdditiveTimer:bool)) -> (Output_Get:bool)
   > Set IsGoingUp(IsGoingUp:bool=Get bIsGoingUp!(self:Rocket Ram Item Layer Anim Instance Object Reference=@n9) -> (bIsGoingUp:bool)) -> (Output_Get:bool)
   > Set Is No Mortar(Is No Mortar:bool=Get bIsNoMortar!(self:Rocket Ram Item Layer Anim Instance Object Reference=@n9) -> (bIsNoMortar:bool)) -> (Output_Get:bool)
   > Set X Input(X Input:double=Get XInput!(self:Rocket Ram Item Layer Anim Instance Object Reference=@n9) -> (XInput:float)) -> (Output_Get:double)
   > Set Is Apex Timer End(Is Apex Timer End:bool=Get bIsApexTimerEnd!(self:Rocket Ram Item Layer Anim Instance Object Reference=@n9) -> (bIsApexTimerEnd:bool)) -> (Output_Get:bool)
   > Set Is Wall Impact(Is Wall Impact:bool=Get bIsWallImpact!(self:Rocket Ram Item Layer Anim Instance Object Reference=@n9) -> (bIsWallImpact:bool)) -> (Output_Get:bool)
   > Set Time Dilation Rate Scale(Time Dilation Rate Scale:double=Get TimeDilationRateScale!(self:Rocket Ram Player Anim BP Object Reference=@n9) -> (TimeDilationRateScale:double)) -> (Output_Get:double)
   > Set Is Jump from Ground(Is Jump from Ground:bool=Get bIsJumpFromGround!(self:Rocket Ram Item Layer Anim Instance Object Reference=@n9) -> (bIsJumpFromGround:bool)) -> (Output_Get:bool)
   n10 = Branch(execute=@n2.then_4, Condition="float > float"(A=GetCurveValue(CurveName=UserCurve01), B=0.5))
   > Set AOMultiplier(AOMultiplier:double=-1) -> (Output_Get:double)
   Set AOMultiplier(execute=@n10.else, AOMultiplier=1)
   n11 = Branch(execute=@n2.then_5, Condition=Get IsRRRecovery())
   > DelayUntilNextTick!(execute) -> (then)
   > DelayUntilNextTick()
   > Set IsRRRecoveryDelayed(IsRRRecoveryDelayed:bool=true) -> (Output_Get:bool)
   Set IsRRRecoveryDelayed(execute=@n11.else, IsRRRecoveryDelayed=false)
   n12 = OR Boolean!(A=Get IsRRFreeFall(), B=Get IsRRFastFreeFall() -> (IsRRFastFreeFall:bool), C:bool=Get IsRRBraking()) -> (ReturnValue)
   Set AO Alpha(
    execute=@n2.then_3
    AO Alpha:double=FInterpTo(Current=Get AO Alpha(), Target="To Float (Boolean)"(InBool=@n12), DeltaTime=Get Delta Time X(), InterpSpeed=6)
   ) -> (Output_Get:double)
  ]}
 RocketRamState {nodes [
   RocketRamState() {is=title K2Node_FunctionEntry}
   > Is Valid(InputObject=Get As B Weapon Rocket Ram() -> (As B Weapon Rocket Ram:B Weapon Rocket Ram Object Reference))
   > n1 = Switch on Enum_RocketRamState!(
    execute
    Selection:Enum_RocketRamState Enum=Get RamState!(self:B Weapon Rocket Ram Object Reference=Get As B Weapon Rocket Ram()) -> (RamState:Enum_RocketRamState Enum=NewEnumerator0)
   ) -> (
    NewEnumerator0:Exec
    NewEnumerator6:Exec
    NewEnumerator1:Exec
    NewEnumerator2:Exec
    NewEnumerator5:Exec
    NewEnumerator3:Exec
    NewEnumerator7:Exec
    NewEnumerator4:Exec
   ) {is=title K2Node_SwitchEnum}
   > Set IsRRWindUp(IsRRWindUp:bool=true) -> (Output_Get:bool)
   > n2 = Sequence!(execute) -> (then_0, then_1, then_2, then_3, then_4, then_5, then_6:Exec)
   Set IsRRActiveThrust(execute=[@n2.then_0, @n6.then_1, @n8.then_1, @n4.then_1, @n7.then_1, @n5.then_1, @n9.then_2], IsRRActiveThrust:bool=false) -> (Output_Get:bool)
   Set IsRRFreeFall(execute=[@n2.then_1, @n6.then_2, @n8.then_2, @n3.then_1, @n7.then_2, @n5.then_2, @n9.then_3], IsRRFreeFall:bool=false) -> (Output_Get:bool)
   Set IsRRFastFreeFall(execute=[@n2.then_2, @n6.then_3, @n8.then_3, @n3.then_2, @n4.then_2, @n5.then_3, @n9.then_4], IsRRFastFreeFall:bool=false) -> (Output_Get:bool)
   Set IsRRBraking(execute=[@n2.then_3, @n6.then_4, @n8.then_4, @n3.then_3, @n4.then_3, @n7.then_3, @n9.then_5], IsRRBraking:bool=false) -> (Output_Get:bool)
   Set IsRROff(execute=[@n2.then_4, @n8.then_5, @n3.then_4, @n4.then_4, @n7.then_4, @n5.then_4, @n9.then_6], IsRROff:bool=false) -> (Output_Get:bool)
   Set IsRRSlowAcceleration(execute=[@n2.then_5, @n6.then_5, @n3.then_5, @n4.then_5, @n7.then_5, @n5.then_5, @n9.then_1], IsRRSlowAcceleration:bool=false) -> (Output_Get:bool)
   Set IsRRRecovery(execute=[@n2.then_6, @n6.then_6, @n8.then_6, @n3.then_6, @n4.then_6, @n7.then_6, @n5.then_6], IsRRRecovery:bool=false) -> (Output_Get:bool)
   Set IsRRActiveThrust(execute=@n1.NewEnumerator1, IsRRActiveThrust=true)
   > n3 = Sequence()
   Set IsRRWindUp(execute=[@n6.then_0, @n8.then_0, @n3.then_0, @n4.then_0, @n7.then_0, @n5.then_0, @n9.then_0], IsRRWindUp=false)
   Set IsRRFreeFall(execute=@n1.NewEnumerator2, IsRRFreeFall=true)
   > n4 = Sequence()
   Set IsRRBraking(execute=@n1.NewEnumerator3, IsRRBraking=true)
   > n5 = Sequence()
   Set IsRROff(execute=@n1.NewEnumerator4, IsRROff=true)
   > n6 = Sequence()
   Set IsRRFastFreeFall(execute=@n1.NewEnumerator5, IsRRFastFreeFall=true)
   > n7 = Sequence()
   Set IsRRSlowAcceleration(execute=@n1.NewEnumerator6, IsRRSlowAcceleration=true)
   > n8 = Sequence()
   Set IsRRRecovery(execute=@n1.NewEnumerator7, IsRRRecovery=true)
   > n9 = Sequence()
  ]}
]
variables [
 AdjustedPitch {type=number}
 AO Alpha {type=number}
 AOMultiplier {type=number}
 As B Weapon Rocket Ram {type=B_Weapon_RocketRam_C, default=null}
 As Fort Player Pawn {type=FortPlayerPawn}
 Delta Time X {type=number}
 Is Apex Timer End {type=boolean}
 Is Jump from Ground {type=boolean}
 Is No Mortar {type=boolean}
 Is Wall Impact {type=boolean}
 IsFalling {type=boolean}
 IsGoingUp {type=boolean}
 IsLandingDetected {type=boolean}
 IsRRActiveThrust {type=boolean}
 IsRRBraking {type=boolean}
 IsRRFastFreeFall {type=boolean}
 IsRRFreeFall {type=boolean}
 IsRROff {type=boolean}
 IsRRRecovery {type=boolean}
 IsRRRecoveryDelayed {type=boolean}
 IsRRSlowAcceleration {type=boolean}
 IsRRWindUp {type=boolean}
 Squat Additive Timer {type=boolean}
 Time Dilation Rate Scale {type=number}
 X Input {type=number}
 YInput {type=number}
]
