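; Weapon Blueprint for the Rocket Ram, tracking ram state (mirrored for replays) and removing the limited-ammo item when empty.
; Gameplay tags drive idle thruster FX and audio, plus client-side trajectory-aim blocking until cooldown ends.
; Holstering waits for unequip, then lasts until landing or a duration, with a brief local equip-animation block covering replication lag.
; Impacts multicast to the player AnimBP and camera modifier; cheat setters support debugging.
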(event EventBeginPlay
  (bind _player (Variables|Default|GetPlayer))
  (|CallParentFunction)
  (Utilities|FlowControl|Sequence
    (:then_0
      (CallFunction|OnRamStateChanged "NewEnumerator4")
      (bind _instigator (Variables|Actor|GetInstigator)
        (:then
          (bind _asplayer_pawn_athena (Utilities|Casting|CastToPlayerPawn_Athena _instigator))
          (Variables|Default|SetPlayer _asplayer_pawn_athena)
          (Ability|Tasks|WaitGameplayTagAddToActor (Variables|Default|GetPlayer) "(TagName=\"Granted.Athena.RocketRam.ReadyToUse\")"
            (:then
              (Ability|Tasks|WaitGameplayTagRemoveFromActor (Variables|Default|GetPlayer) "(TagName=\"Granted.Athena.RocketRam.ReadyToUse\")"
                (:Removed
                  (Components|Activation|SetActive (Variables|Default|GetNS_RocketRam_Thruster_Idle) false false)
                  ;[1 1 1] Stopping idle audio when ready to use
                  (bind _rocketram_equipsoundcomponent (Variables|Default|GetRocketRam_EquipSoundComponent)
                    (:then
                      (Class|WeaponEquipSoundComponent|StopAudio _rocketram_equipsoundcomponent))
                    (:else)))))
            (:Added
              (Utilities|Time|SetTimerbyEvent (delegate Custom|ActivateIdleEffects) (select (CallFunction|IsBeingEquipped) 0.1 0.3)))))
        (:else)))
    (:then_1
      ;[1 0.4614 0 0.5] HIDE IDLE FX WHEN IN A VEHICLE #JIRA FORT-851363
      (Ability|Tasks|WaitGameplayTagQueryOnActor _player (Variables|Default|GetVehicleTag)
        (:Triggered
          (Development|SetHiddeninGame (Variables|Default|GetNS_RocketRam_Thruster_Idle) true false)
          (Ability|Tasks|WaitGameplayTagQueryOnActor _player (Variables|Default|GetVehicleTag) "WhenFalse"
            (:Triggered
              (Development|SetHiddeninGame (Variables|Default|GetNS_RocketRam_Thruster_Idle) false false))))))))

(event Custom|ActivateIdleEffects
  (bind _player (Variables|Default|GetPlayer)
    (:then
      (if (GameplayTags|HasMatchingGameplayTag _player "(TagName=\"Granted.Athena.RocketRam.ReadyToUse\")")
        (Components|Activation|SetActive (Variables|Default|GetNS_RocketRam_Thruster_Idle) true false)
        ;[1 1 1] Starting idle audio when ready to use
        (bind _rocketram_equipsoundcomponent (Variables|Default|GetRocketRam_EquipSoundComponent)
          (:then
            (Class|WeaponEquipSoundComponent|OnUnHolestered _rocketram_equipsoundcomponent))
          (:else))))
    (:else)))

(event EventEndPlay
  (params (EndPlayReason :type EEndPlayReason%20Enum :default "Destroyed"))
  (|CallParentFunction EndPlayReason)
  (CallFunction|SetEquipAnimBlocked false)
  (bind _asfort_player_pawn (Utilities|Casting|CastToFortPlayerPawn (Game|GetInstigator)))
  ;[1 0.4182 0] Remove the tag that blocks trajectory aim.
  ; We only need it for a few frames, and by the time the player reacquires this weapon we will no longer care about it.
  (Ability|GameplayEffect|RemoveLooseGameplayTags _asfort_player_pawn (Variables|BlockTrajectoryAim|GetBlockTrajectoryAimTags)))

(event Custom|OnRamStateChanged
  (params (NewState :type Enum_RocketRamState%20Enum :default "NewEnumerator0"))
  (bind _ramstate (Variables|Default|GetRamState))
  ;[0.1437 0.2378 1] On Ram State Changed
  (CallFunction|LogMessage (Utilities|String|Append "OnRamStateChanged(), previous state: [" (Utilities|String|EnumtoString _ramstate) "] new state: [" (Utilities|String|EnumtoString NewState) "]"))
  (Variables|Default|SetPreviousRamState _ramstate)
  (Networking|SwitchHasAuthority
    (:Authority
      ;;[1 0.8798 0] Since RamState only replicates to simulators we need to record the Ram State to the owning client for replays.
      (Variables|Default|SetRamStateForReplays NewState))
    (:Remote
      _))
  (bind _output_get (Variables|Default|SetRamState NewState))
  (CallFunction|UpdateRamState)
  (Default|CallOnRamStateChangedEvent _output_get)
  (CallFunction|UpdateRamVFX)
  (if (and (Utilities|Enum|Equal(Enum) _output_get "NewEnumerator4") (GameplayTags|HasMatchingGameplayTag (Variables|Default|GetPlayer) "(TagName=\"Granted.Athena.RocketRam.ShouldConsumeAmmo\")"))
    (if (<= (FortWeapon|GetRemainingAmmo) 0)
      (Utilities|Time|SetTimerbyEvent (delegate Custom|RemoveSelf) 0.5) ; Wait for GA to finish and for anim of picking up the rocket to finish
    )))

(event Custom|UpdateRamVFX
  (if (Utilities|Enum|Equal(Enum) (Variables|Default|GetRamState) "NewEnumerator0")
    (CallFunction|ResetHolster)
    (CallFunction|SetEquipAnimBlocked false)))

(event Custom|RemoveSelf
  (Inventory|RemoveItemFromPlayer :PlayerController (Fortnite|GetFortPlayerControllerfromActor (Variables|Default|GetPlayer)) :ItemDefinition "/RocketRamGameplay/Items/RocketRam/WID_RocketLimitedRam_Comp.WID_RocketLimitedRam_Comp" :AmountToRemove -1 :bForceRemoval true :bSkipVerbRemoval false))

(event Custom|HolsterEquippedWeapon
  (params (bHolsterTillLand :type Boolean) (ImmediateHolsterDuration :type Float%20%28double-precision%29))
  ;[1 0.1146 0.6587] Holstering
  (Variables|Holster|SetbHolsterTillLand bHolsterTillLand)
  (Variables|Holster|SetImmediateHolsterDuration ImmediateHolsterDuration)
  (Variables|Holster|SetbReadyToHolster true)
  (CallFunction|TryHolsterWeapon))

(event Custom|ResetHolster
  ;[1 0.1146 0.6587] Holstering
  (Variables|Holster|SetbReadyToHolster false)
  (Variables|Holster|SetImmediateHolsterDuration -1.0)
  (Variables|Holster|SetbHolsterTillLand false)
  (Variables|Holster|SetbHasUnequipped false))

(event FortWeapon|EventOnLocalUnEquip
  ;[1 0.1146 0.6587] Holstering
  (Variables|Holster|SetbHasUnequipped true)
  (Utilities|FlowControl|Sequence
    (:then_0
      (Networking|SwitchHasAuthority
        (:Remote
          (switch Utilities|FlowControl|Switch|SwitchonEnum_RocketRamState (Variables|Default|GetRamState)
            (:NewEnumerator2
              _)
            (:NewEnumerator5
              _)
            (:NewEnumerator3
              _)
            (:NewEnumerator7
              _))
          (CallFunction|SetEquipAnimBlocked true) ; It takes a moment for the equip anim being blocked \nby the holster GE to be replicated, so temporarily \nblock the equip anim locally to prevent it from starting at all.
          (Utilities|FlowControl|Delay 0.1)
          (CallFunction|SetEquipAnimBlocked false))))
    (:then_1
      (CallFunction|TryHolsterWeapon))))

(event Custom|TryHolsterWeapon
  ;[1 0.1146 0.6587] Holstering
  (if (and (Variables|Holster|GetbReadyToHolster) (Variables|Holster|GetbHasUnequipped))
    (Utilities|FlowControl|Sequence
      (:then_0
        (if (Variables|Holster|GetbHolsterTillLand)
          (bind _asfort_player_pawn (Utilities|Casting|CastToFortPlayerPawn (Game|GetInstigator)))
          (GameplayEffects|ApplyGameplayEffectToSelf :self (Ability|GetAbilitySystemComponent _asfort_player_pawn) :GameplayEffectClass "/RocketRamGameplay/Items/RocketRam/Effects/Holster/GE_TillLandHolster_Grant_RocketRam.GE_TillLandHolster_Grant_RocketRam_C" :Level 0.0)
          (else
            (Default|HolsterforDuration (Variables|Holster|GetImmediateHolsterDuration)))))
      (:then_1
        (CallFunction|ResetHolster)))))

(event Custom|SetEquipAnimBlocked
  (params (bIsBlocked :type Boolean))
  (bind _blockequipanimtag (Variables|Holster|GetBlockEquipAnimTag))
  ;[1 0.1146 0.6587] Holstering
  (bind _instigator (Variables|Actor|GetInstigator)
    (:then
      (if bIsBlocked
        (if (not (Variables|Holster|GetbHasAppliedBlockEquipAnimTag))
          (Ability|GameplayEffect|AddLooseGameplayTags _instigator _blockequipanimtag)
          (Variables|Holster|SetbHasAppliedBlockEquipAnimTag true))
        (elif (Variables|Holster|GetbHasAppliedBlockEquipAnimTag)
          (Ability|GameplayEffect|RemoveLooseGameplayTags _instigator _blockequipanimtag)
          (Variables|Holster|SetbHasAppliedBlockEquipAnimTag false))))
    (:else)))

(event Custom|NotifyImpactDetected
  (params (bIsWallImpact :type Boolean))
  ;[0.2113 1 0.2598] Notify Impact Detected
  (CallFunction|InternalNotifyImpactDetected bIsWallImpact)
  (Networking|SwitchHasAuthority
    (:Authority
      (CallFunction|MulticastNotifyImpactDetected bIsWallImpact)
      (Networking|ForceNetUpdate (Variables|Actor|GetInstigator)))))

(event Custom|Multicast_NotifyImpactDetected
  (params (bIsWallImpact :type Boolean))
  ;[0.2113 1 0.2598] Notify Impact Detected
  (bind _asfort_player_pawn (Utilities|Casting|CastToFortPlayerPawn (Variables|Actor|GetInstigator)))
  (if (not (Pawn|IsLocallyControlled _asfort_player_pawn))
    (CallFunction|InternalNotifyImpactDetected bIsWallImpact)))

(event Custom|Internal_NotifyImpactDetected
  (params (bIsWallImpact :type Boolean))
  ;[0.2113 1 0.2598] Notify Impact Detected
  (bind _asfort_player_pawn (Utilities|Casting|CastToFortPlayerPawn (Variables|Actor|GetInstigator)))
  (bind _asrocket_ram_player_anim_bp (Utilities|Casting|CastToRocketRam_Player_AnimBP (Components|SkeletalMesh|AnimationBlueprintLinking|GetLinkedAnimLayerInstancebyClass (Variables|/Script/Engine.Character|GetMesh _asfort_player_pawn) "/RocketRamGameplay/Items/RocketRam/Animations/Animation/RocketRam_Player_AnimBP.RocketRam_Player_AnimBP_C")))
  (Class|RocketRamPlayerAnimBP|NotifyImpactDetected _asrocket_ram_player_anim_bp bIsWallImpact)
  (Class|CameraModifierRocketRam|IncomingImpactNotify (Variables|Default|GetCameraModifier) bIsWallImpact))

(event Custom|SetLaunchDirection
  (params (LaunchDirection :type Vector))
  ;[1 0.2854 0.7225] Set Launch Direction
  (Variables|Animation|SetLaunchDirection LaunchDirection))

(event Custom|SetImpactDirection
  (params (LaunchDirection :type Vector))
  ;[1 0.2854 0.7225] Set Impact Direction
  (Variables|Animation|SetImpactDirection LaunchDirection))

(event Custom|UpdateRamState
  (bind _cooldown_removal_async_action (Variables|BlockTrajectoryAim|GetCooldown%20Removal%20Async%20Action))
  (bind _cooldowntag (Variables|BlockTrajectoryAim|GetCooldownTag))
  (bind _blocktrajectoryaimtags (Variables|BlockTrajectoryAim|GetBlockTrajectoryAimTags))
  ;[0 1 0.9785 0.5] UpdateRamState: When the local client exits the Off state adds the BlockTrajectoryAim to them, then waits for the ram cooldown to end before removing the tag
  (if (Math|Boolean|ANDBoolean (Utilities|Enum|Equal(Enum) (Variables|Default|GetPreviousRamState) "NewEnumerator4") (Utilities|Enum|NotEqual(Enum) (Variables|Default|GetPreviousRamState) (Variables|Default|GetRamState)) (not (Variables|Cheats|GetCheat_CooldownDisabled)))
    ;;[1 0 0.9793] If this is the local player.
    (bind _instigator (Variables|Actor|GetInstigator)
      (:then
        (if (and (Pawn|IsLocallyControlled _instigator) (not (Networking|IsServer)))
          (bind _asfort_player_pawn (Utilities|Casting|CastToFortPlayerPawn _instigator))
          (Variables|BlockTrajectoryAim|SetShouldIgnoreNextCooldownRemoval (not (GameplayTags|HasMatchingGameplayTag _asfort_player_pawn _cooldowntag))) ; If this player does NOT have the cooldown tag, then ignore the first removal.\nWhen the tag is not present, the first removal will trigger immediately, and we only care when the cooldown edge actually changes.
          (Ability|GameplayEffect|AddLooseGameplayTags _asfort_player_pawn _blocktrajectoryaimtags)
          (Utilities|FlowControl|Sequence
            (:then_0
              (Utilities|IsValid _cooldown_removal_async_action
                (:Is%20Valid
                  (EventDispatchers|UnbindallEventsfromRemoved _cooldown_removal_async_action))))
            (:then_1
              (bind _asynctaskproxy (Ability|Tasks|WaitGameplayTagRemoveFromActor _asfort_player_pawn _cooldowntag)
                (:then
                  (Variables|BlockTrajectoryAim|SetCooldown%20Removal%20Async%20Action _asynctaskproxy))
                (:Removed
                  ;;[0.04585 1 0.02567] Cooldown Tag Removed
                  (if (Variables|BlockTrajectoryAim|GetShouldIgnoreNextCooldownRemoval)
                    (Variables|BlockTrajectoryAim|SetShouldIgnoreNextCooldownRemoval false) ; We are ignoring this "removal" now make sure we do NOT ignore the next removal
                    (else
                      ;;;[1 0.3093 0] When cooldown is removed, remove the block trajectory aim tag.
                      (bind _asfort_player_pawn_2 (Utilities|Casting|CastToFortPlayerPawn (Game|GetInstigator)))
                      (Ability|GameplayEffect|RemoveLooseGameplayTags _asfort_player_pawn_2 _blocktrajectoryaimtags)
                      (EventDispatchers|UnbindallEventsfromRemoved (Variables|BlockTrajectoryAim|GetCooldown%20Removal%20Async%20Action))))))))))
      (:else))))

(fn BP_AllowRaiseWeaponForDeferredFireInput
  (bind _instigator (Variables|Actor|GetInstigator))
  (bind _returnvalue (|CallParentFunction))
  (if _returnvalue
    (Utilities|IsValid _instigator
      (:Is%20Valid
        (bind _returnvalue_2 (GameplayTags|GetOwnedGameplayTags _instigator :!class /Script/GameplayTags.GameplayTagAssetInterface))
        ;[1 0.4479 0.4479] Do NOT raise the weapon if it is still on cooldown. This prevents a jerky animation from playing while we cannot fire.
        (return (not (GameplayTags|HasTag _returnvalue_2 "(TagName=\"Granted.Cooldown.RocketRam\")" true))))
      (:Is%20Not%20Valid
        _))
    (else
      _))
  ;[1 0.4479 0.4479] Failed, do NOT allow the weapon to raise.
  (return false))

(fn LogMessage
  (params (InString :type String) (bLogAsWarning :type Boolean))
  (if (Variables|Default|GetLoggingEnabled)
    (Fortnite|FortShippingLog self (Utilities|String|Append "B_Weapon_RocketRam " InString ".") bLogAsWarning)))

(fn NoRamCooldown
  ;[1 0 0] CHEAT ONLY
  (Variables|Cheats|SetCheat_CooldownDisabled true)
  (Cheats|RemoveRocketRamCooldown))

(fn OnRep_RamStateForReplays
  (if (Replay|GetIsPlayingReplay)
    (Variables|Default|SetRamState (Variables|Default|GetRamStateForReplays))))

(fn RemoveRocketRamCooldown
  ;[1 0 0] CHEAT ONLY
  (bind _asfort_player_pawn (Utilities|Casting|CastToFortPlayerPawn (Game|GetInstigator)))
  (GameplayEffects|RemoveActiveEffectswithGrantedTags (Ability|GetFortAbilitySystemComponent _asfort_player_pawn) "(GameplayTags=((TagName=\"Granted.Cooldown.RocketRam\")))"))

(fn SetADSCameraEnabled
  (params (bADSCameraEnabled :type Boolean))
  ;[1 0 0] CHEAT ONLY
  (bind _output_get (Variables|Cheats|SetCheat_ADSCameraEnabled bADSCameraEnabled))
  (Networking|ForceNetUpdate self)
  (Class|RocketRamSharedLibrary|RamLog (Variables|Default|GetPlayer) (Utilities|Text|ToText(String) (Utilities|String|Append "Cheat Set -- ADSCameraEnabled: " (Utilities|String|ToString(Boolean) _output_get))) false true))

(fn SetDebugDrawRamArc
  (params (bDebugDrawRamArc :type Boolean))
  ;[1 0 0] CHEAT ONLY
  (Variables|Cheats|SetCheat_DebugDrawRamArc bDebugDrawRamArc)
  (Networking|ForceNetUpdate self))

(fn SetGravityWaitSyncEnabled
  (params (bIsGravityWaitSyncEnabled :type Boolean))
  ;[1 0 0] CHEAT ONLY
  (Variables|Cheats|SetCheat_GravityWaitSyncEnabled bIsGravityWaitSyncEnabled)
  (Networking|ForceNetUpdate self))

(fn SetHardIgnoreCameraModifierWhenNoTag
  (params (bShouldIgnore :type Boolean))
  ;[1 0 0] CHEAT ONLY
  (bind _output_get (Variables|Cheats|SetCheat_HardIgnoreCameraModifierWhenNoTag bShouldIgnore))
  (Networking|ForceNetUpdate self)
  (Class|RocketRamSharedLibrary|RamLog (Variables|Default|GetPlayer) (Utilities|Text|ToText(String) (Utilities|String|Append "Cheat Set -- HardIgnoreCameraModifierWhenNoTag: " (Utilities|String|ToString(Boolean) _output_get))) false true))

(fn SetRamBlockingEnabled
  (params (bIsBlockingEnabled :type Boolean))
  ;[1 0 0] CHEAT ONLY
  (Variables|Cheats|SetCheat_BlockingDisabled (not bIsBlockingEnabled))
  (Networking|ForceNetUpdate self))

(fn SetRamCameraEnabled
  (params (bIsCameraEnabled :type Boolean))
  ;[1 0 0] CHEAT ONLY
  (bind _output_get (Variables|Cheats|SetCheat_CameraModifierDisabled (not bIsCameraEnabled)))
  (Networking|ForceNetUpdate self)
  (Class|RocketRamSharedLibrary|RamLog (Variables|Default|GetPlayer) (Utilities|Text|ToText(String) (Utilities|String|Append "Cheat Set -- CameraModifierDisabled: " (Utilities|String|ToString(Boolean) _output_get))) false true))

(fn SetRamStaticGravity
  (params (bUseStaticGravity :type Boolean))
  ;[1 0 0] CHEAT ONLY
  (Variables|Cheats|SetCheat_StaticGravityEnabled bUseStaticGravity)
  (Networking|ForceNetUpdate self))

(fn ConstructionScript
  (|CallParentFunction))
