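; Camera modifier for the Rocket Ram item that offsets camera position and FOV during a ram.
; Each update switches on the weapon's ram state, using windup and impact curves, launch and peak positions, and player velocity to drive forward, right, up and FOV offsets.
; Offsets are scaled by the modifier's alpha.
; When the player lacks the RocketRam movement tag, it disables itself, waits out the blend, then removes itself from the camera manager.
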(event Custom|DisableAndRemove
  ;[1 0.4811 0.1007] Disable this modifier, wait for it to fully blend out, and remove it from the owning player's camera manager.
  (CameraModifier|DisableModifier false)
  (Utilities|FlowControl|Delay (Variables|CameraModifier|GetAlphaOutTime))
  (Variables|Default|SetRemoved true)
  (bind _cameraowner (Variables|CameraModifier|GetCameraOwner)
    (:then
      (Utilities|IsValid (Game|Player|GetOwningPlayerController _cameraowner)
        (:Is%20Valid
          (Game|Player|RemoveCameraModifier (Variables|/Script/Engine.PlayerController|GetPlayerCameraManager (Game|Player|GetOwningPlayerController _cameraowner)) self))))
    (:else)))

(fn BlueprintModifyCamera
  (params (DeltaTime :type Float%20%28single-precision%29) (ViewLocation :type Vector :default "0, 0, 0") (ViewRotation :type Rotator :default "0, 0, 0") (FOV :type Float%20%28single-precision%29))
  (bind _player (Variables|Default|GetPlayer))
  (bind (_returnvalue_x _returnvalue_y _returnvalue_z) (Math|Vector|BreakVector (Transformation|GetVelocity _player)))
  (bind _startedtime (Variables|Default|GetStartedTime))
  (bind _returnvalue (Utilities|Time|GetGameTimeinSeconds))
  (bind _returnvalue_2 (- _returnvalue _startedtime))
  (bind _returnvalue_3 (/ _returnvalue_2 1.75))
  (bind _initialfovcurve (Variables|Default|GetInitialFOVCurve))
  (bind _returnvalue_4 (Math|Curves|GetFloatValue _initialfovcurve _returnvalue_3))
  (bind _player_2 (Variables|Default|GetPlayer))
  (bind (_returnvalue_x_2 _returnvalue_y_2 _returnvalue_z_2) (Math|Vector|BreakVector (Transformation|GetVelocity _player_2)))
  (bind (_returnvalue_x_3 _returnvalue_y_3 _returnvalue_z_3) (Math|Vector|BreakVector (Transformation|GetVelocity (Variables|Default|GetPlayer))))
  (bind (_returnvalue_x_4 _returnvalue_y_4 _returnvalue_z_4) (Math|Vector|BreakVector (Transformation|GetVelocity (Variables|Default|GetPlayer))))
  (bind (_returnvalue_x_5 _returnvalue_y_5 _returnvalue_z_5) (Math|Vector|BreakVector (Transformation|GetActorLocation (Variables|Default|GetPlayer))))
  (bind _player_3 (Variables|Default|GetPlayer))
  (bind _returnvalue_5 (Transformation|GetActorLocation _player_3))
  (bind _peakedposition (Variables|Default|GetPeakedPosition))
  (bind _returnvalue_6 (- _peakedposition _returnvalue_5))
  (bind _returnvalue_7 (Math|Vector|VectorLength _returnvalue_6))
  (bind _returnvalue_8 (Utilities|Time|GetGameTimeinSeconds))
  (bind _impacttime (Variables|Default|GetImpactTime))
  (bind _returnvalue_9 (- _returnvalue_8 _impacttime))
  (bind _impactzoomincurve (Variables|Default|GetImpactZoomInCurve))
  (bind _returnvalue_10 (Math|Curves|GetFloatValue _impactzoomincurve _returnvalue_9))
  (bind (_launchedposition_x _launchedposition_y _launchedposition_z) (Math|Vector|BreakVector (Variables|GetLaunchedPosition_X)))
  (Utilities|FlowControl|Sequence
    (:then_0
      (Utilities|IsValid (Variables|Default|GetWeapon)
        (:Is%20Not%20Valid
          (bind _output_get (Variables|Default|SetPlayer (Game|Player|GetPlayerPawn (Utilities|Casting|CastToFortPlayerController (Game|Player|GetOwningPlayerController (Variables|CameraModifier|GetCameraOwner))))))
          (bind _asb_weapon_rocket_ram (Utilities|Casting|CastToB_Weapon_RocketRam (Weapon|GetCurrentWeapon _output_get)))
          (Variables|Default|SetWeapon _asb_weapon_rocket_ram)
          (Variables|Default|SetStartedTime (Utilities|Time|GetGameTimeinSeconds)))))
    (:then_1
      (if (and (Utilities|IsValid (Variables|Default|GetWeapon) :!class /Script/Engine.KismetSystemLibrary) (Utilities|IsValid (Variables|Default|GetPlayer) :!class /Script/Engine.KismetSystemLibrary))
        (if (Variables|Default|GetRemoved)
          _
          (elif (GameplayTags|HasMatchingGameplayTag (Variables|Default|GetPlayer) "(TagName=\"Granted.Athena.RocketRam.Movement\")")
            _
            (else
              (CallFunction|DisableandRemove)
              (if (Variables|/RocketRamGameplay/Items/RocketRam/B_Weapon_RocketRam.B_Weapon_RocketRam_C|GetCheat_HardIgnoreCameraModifierWhenNoTag (Variables|Default|GetWeapon))
                (return (Variables|GetViewLocation) (Variables|GetViewRotation) (Variables|GetFOV)))))))
      (return (Variables|GetViewLocation) (Variables|GetViewRotation) (Variables|GetFOV)))
    (:then_2
      (Utilities|FlowControl|Sequence
        (:then_0
          (switch Utilities|FlowControl|Switch|SwitchonEnum_RocketRamState (Variables|/RocketRamGameplay/Items/RocketRam/B_Weapon_RocketRam.B_Weapon_RocketRam_C|GetRamState (Variables|Default|GetWeapon))
            (:NewEnumerator6
              (if (not (Variables|Default|GetLaunched))
                (Variables|Default|SetLaunchedPosition (Transformation|GetActorLocation (Variables|Default|GetPlayer)))
                (Variables|Default|SetLaunched true)))
            (:NewEnumerator2
              _)
            (:NewEnumerator5
              _))
          (if (and (< _returnvalue_z_4 0) (not (Variables|Default|GetPeaked)))
            (Variables|Default|SetPeakedPosition (Transformation|GetActorLocation (Variables|Default|GetPlayer)))
            (Variables|Default|SetPeaked true)
            (Variables|Default|SetPeakedForward (Variables|Default|GetCurrent%20Forward))
            (Variables|Default|SetPeakedFOV (Variables|Default|GetFOV%20Offset))))
        (:then_1
          (switch Utilities|FlowControl|Switch|SwitchonEnum_RocketRamState (Variables|/RocketRamGameplay/Items/RocketRam/B_Weapon_RocketRam.B_Weapon_RocketRam_C|GetRamState (Variables|Default|GetWeapon))
            (:NewEnumerator0
              (bind _output_get_2 (Variables|Default|SetCurrent%20Forward (* _returnvalue_4 -8e+01)))
              (Variables|Default|SetForwardFromWindup _output_get_2)
              (Variables|Default|SetFOV%20Offset (* _returnvalue_4 1e+01)))
            (:NewEnumerator6
              (Variables|Default|SetCurrent%20Up (* (- _launchedposition_z _returnvalue_z_5) 0.5)))
            (:NewEnumerator1
              (Variables|Default|SetFOV%20Offset (+ (Variables|Default|GetFOV%20Offset) (* (Variables|GetDeltaTime) 3e+01))))
            (:NewEnumerator2
              (Variables|Default|SetFOV%20Offset (Math|Interpolation|FInterpTo (Variables|Default|GetFOV%20Offset) (Utilities|RemapFloat _returnvalue_z_2 "(X=0.000000,Y=-4000.000000)" "(X=-10.000000,Y=20.000000)" true) (Variables|GetDeltaTime) (Math|Float|SelectFloat 4.0 1.5 (< _returnvalue_z_3 0))) :!label _label)
              (Variables|Default|SetCurrent%20Forward (Math|Interpolation|FInterpTo (Variables|Default|GetCurrent%20Forward) (Utilities|RemapFloat _returnvalue_z_2 "(X=0.000000,Y=-4000.000000)" "(X=0.000000,Y=-750.000000)" true) (Variables|GetDeltaTime) 0.5)))
            (:NewEnumerator5
              (goto _label))
            (:NewEnumerator3
              (goto _label))
            (:NewEnumerator7
              (goto _label))
            (:NewEnumerator4
              (goto _label)))
          (Variables|Default|SetCurrent%20Forward (* (Math|Vector|VectorLength (- (Variables|Default|GetLaunchedPosition) (Transformation|GetActorLocation (Variables|Default|GetPlayer)))) (Utilities|RemapFloat (Math|Vector|VectorLength (- (Variables|Default|GetLaunchedPosition) (Transformation|GetActorLocation (Variables|Default|GetPlayer)))) "(X=1000.000000,Y=3000.000000)" "(X=-0.500000,Y=-0.300000)" true))))
        (:then_2
          (switch Utilities|FlowControl|Switch|SwitchonEnum_RocketRamState (Variables|/RocketRamGameplay/Items/RocketRam/B_Weapon_RocketRam.B_Weapon_RocketRam_C|GetRamState (Variables|Default|GetWeapon))
            (:NewEnumerator6
              _)
            (:NewEnumerator1
              _)
            (:NewEnumerator2
              (Utilities|FlowControl|Sequence :!label _label_2
                (:then_0
                  (if (Variables|Default|GetPeaked)
                    (Variables|Default|SetCurrent%20Forward (Math|Float|Clamp(Float) (+ (* _returnvalue_7 -0.07) (Variables|Default|GetPeakedForward)) -6e+02 2e+02))
                    (Variables|Default|SetFOV%20Offset (Utilities|RemapFloat _returnvalue_7 "(X=0.000000,Y=7000.000000)" (Math|Vector2D|MakeVector2D (Variables|Default|GetPeakedFOV) 2e+01) true))
                    (if (Variables|Default|GetImpacting)
                      (Variables|Default|SetFOV%20Offset (+ (Variables|Default|GetFOV%20Offset) (Utilities|RemapFloat _returnvalue_10 "(X=0.000000,Y=1.000000)" "(X=0.000000,Y=-20.000000)" true)))
                      (Variables|Default|SetCurrent%20Forward (+ (Variables|Default|GetCurrent%20Forward) (Utilities|RemapFloat _returnvalue_10 "(X=0.000000,Y=1.000000)" "(X=0.000000,Y=300.000000)" true))))))
                (:then_1
                  _)))
            (:NewEnumerator5
              (goto _label_2))
            (:NewEnumerator3
              (goto _label_2))
            (:NewEnumerator7
              (goto _label_2))
            (:NewEnumerator4
              (goto _label_2)))
          (LocalVariables|SetZoomOutCoefficient (Utilities|RemapFloat (+ (* (Variables|Default|GetFOV%20Offset) 3e+01) (* (Variables|Default|GetCurrent%20Forward) -1.0)) "(X=0.000000,Y=1500.000000)" "(X=0.300000,Y=1.500000)" true))
          (Variables|Default|SetCurrent%20Up (Math|Interpolation|FInterpTo (Variables|Default|GetCurrent%20Up) (* (Utilities|RemapFloat _returnvalue_z "(X=-4000.000000,Y=2500.000000)" "(X=500.000000,Y=-300.000000)" true) (LocalVariables|GetZoomOutCoefficient)) (Variables|GetDeltaTime) 2.0))
          (Variables|Default|SetCurrent%20Right (Math|Interpolation|FInterpTo (Variables|Default|GetCurrent%20Right) (* (Utilities|RemapFloat (Math|Vector|DotProduct (Transformation|GetVelocity _player) (Math|Vector|GetRightVector (Variables|GetViewRotation))) "(X=-700.000000,Y=700.000000)" "(X=240.000000,Y=-240.000000)" true) (LocalVariables|GetZoomOutCoefficient)) (Variables|GetDeltaTime) 2.0)))
        (:then_3
          (return (+ (Variables|GetViewLocation) (* (Utilities|Operators|Add (* (Math|Vector|GetForwardVector (Variables|GetViewRotation)) (Variables|Default|GetCurrent%20Forward)) (* (Math|Vector|GetRightVector (Variables|GetViewRotation)) (Variables|Default|GetCurrent%20Right)) (Math|Vector|MakeVector 0.0 0.0 (Variables|Default|GetCurrent%20Up))) (Variables|CameraModifier|GetAlpha))) (Variables|GetViewRotation) (+ (Variables|GetFOV) (* (Variables|Default|GetFOV%20Offset) (Variables|CameraModifier|GetAlpha)))))))))

(fn IncomingImpactNotify
  (params (IsWallImpact :type Boolean))
  (Variables|Default|SetImpacting true)
  (Variables|Default|SetImpactTime (Utilities|Time|GetGameTimeinSeconds)))
