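; Blueprint class DataClass_RocketRam in the RocketRam gameplay item folder, by name a data class for the Rocket Ram item.
; It declares a single function, NewFunction, with an empty body and default name.
; No variables, components or other logic appear, so it holds no behavior of its own.
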(fn NewFunction)
