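# Translucent deferred decal material for the Rocket Ram's ground path, with a fade tied to decal lifetime.
# The Tile UV parameter and Time scale and pan the path texture, whose red channel blends two dark brown tones.
# The panned green channel multiplies an unpanned blue-channel mask to build opacity.
# That green channel also sets the SmoothStep upper edge against DecalLifetimeOpacity, giving a noise-eroded dissolve alongside a plain lifetime fade.
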Material /RocketRamGameplay/Items/RocketRam/FX/Materials/M_RocketRam_Path_Decal_Ground
graph {
 properties {blendMode=BLEND_Translucent, materialDomain=MD_DeferredDecal, pixelDepthOffsetMode=PDOM_Legacy, usedShadingModels=Default Lit}
 nodes [
  n1 = TextureCoordinate()
  n2 = VectorParameter() {group=Noise2, parameterName=Tile UV, channelNames {a=Speed V, b=Speed U, g=Tile V, r=Tile U}, defaultValue=<rgba 1 2 0 -0.25>}
  n3 = MaterialFunctionCall(X=@n2.Tile U, Y=@n2.Tile U) {materialFunction=&MakeFloat2}
  n4 = MaterialFunctionCall(X=@n2.Speed U, Y=@n2.Speed U) {materialFunction=&MakeFloat2}
  n5 = Time()
  n6 = TextureSample(UVs=Add(A=Multiply(A=@n1, B=@n3.Result), B=Multiply(A=@n4.Result, B=@n5))) {texture=/RocketRamGameplay/Items/RocketRam/FX/Textures/T_RocketRam_Path_Decal_Ground_Single}
  LinearInterpolate(
   A=Constant3Vector() {constant=<rgba 0.015 0.0113 0.0075 1>}
   B=Constant3Vector() {constant=<rgba 0.15 0.1232 0.1125 1>}
   Alpha=Multiply(A=@n6.R) {constB=15}
  )
  MaterialFunctionCall() {materialFunction=&DebugTimeSine}
  n7 = Multiply(
   A=Multiply(A=@n6.G) {constB=5}
   B=Multiply(A=TextureSample() {texture=&T_RocketRam_Path_Decal_Ground_Single}.B) {constB=2}
  )
  n8 = DecalLifetimeOpacity()
  n9 = SmoothStep(Max=@n6.G, Value=@n8.Opacity)
  Multiply(A=@n7, B=@n9)
  Multiply(A=@n7, B=@n8.Opacity)
  OneMinus(None=@n9)
 ]
}
