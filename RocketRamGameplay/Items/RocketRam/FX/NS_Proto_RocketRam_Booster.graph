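# Prototype Niagara booster exhaust effect for the Rocket Ram item, combining GPU sparks and ashes, flare, blast meshes, wind planes and an air layer.
# Combined emitters give one particle per-renderer custom attributes to drive several meshes and a corona sprite.
# User.uPhase scales sparks and blast size, uDuration times the burst and flare, and FlareFlicker adds flare jitter.
# Air opacity follows system velocity; air and wind also adjust for time of day.
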NiagaraSystem /RocketRamGameplay/Items/RocketRam/FX/NS_Proto_RocketRam_Booster
dependencies {
 usedDataInterfaces [&NiagaraDataInterfaceColorCurve, &NiagaraDataInterfaceCurve, &NiagaraDataInterfaceVectorCurve]
 usedDynamicInputs [&ClampFloat, &RemapRange, &VectorLength]
 usedModules [
  &SystemState
  &InterpolateOverTime
  &self:SystemSpawnScript.SystemScriptSource.SystemScriptGraph.NiagaraNodeAssignment_0.SetVariables_#1
  &self:SystemSpawnScript.SystemScriptSource.SystemScriptGraph.NiagaraNodeAssignment_1.SetVariables_#2
 ]
}
emitters {
 Air {
  emitterUpdateScript [
   &EmitterState()
   &SpawnBurst_Instantaneous(Use Spawn Probability=false, UseLoopCountLimit=false)
   &SpawnRate(SpawnRate=2, Use Spawn Probability=false)
  ]
  particleSpawnScript [
   &InitializeParticle(
    Color=<rgba 1 1 1 0.025>
    Lifetime=0.5
    Mesh Scale=<1.5 3 3>
    Mesh Scale Mode=::NewEnumerator4
    UsePositionOffset=false
    Write Material Random=true
   )
   &InitialMeshOrientation(Mesh Orientation Mode=::NewEnumerator7, Rotation=<0.5 0 0>, Use Rotation Vector=true)
  ]
  particleUpdateScript [
   &ParticleState()
   &ScaleColor(
    Scale Alpha=&Multiply_Float(
     A=&FloatFromCurve(
      FloatCurve=DataInterface() {propertyValues {Curve {keys [
          {arriveTangent=-0.2342, interpMode=RCIM_Cubic, leaveTangent=-0.2342, tangentMode=RCTM_User, time=0.3, value=0}
          {arriveTangent=3.1159, interpMode=RCIM_Cubic, leaveTangent=3.1159, tangentMode=RCTM_User, time=1, value=1}
         ]}}}
      CurveIndex=$System.NormalizedVelocityRange
     )
     B=&FloatFromCurve(
      FloatCurve=DataInterface() {propertyValues {Curve {keys [
          {arriveTangent=3.053, interpMode=RCIM_Cubic, leaveTangent=3.053, tangentMode=RCTM_User, time=0, value=0}
          {arriveTangent=1.3198, interpMode=RCIM_Cubic, leaveTangent=1.3198, tangentMode=RCTM_User, time=0.2574, value=0.6346}
          {arriveTangent=0.0224, interpMode=RCIM_Cubic, leaveTangent=0.0224, tangentMode=RCTM_User, time=1, value=1}
         ]}}}
      CurveIndex=$System.NormalizedLoopAge
     )
    )
    ScaleA=true
    ScaleRGB=false
   )
   &ScaleColor(
    Scale RGB=&VectorFromFloat(Value=&FNTimeOfDayScalar(Night Scale=0.35, Morning Scale=0.5, Day Scale=0.8, Evening Scale=0.4))
    ScaleA=true
    ScaleRGB=true
   ) {module=ScaleColor002}
  ]
  rendererClasses [&NiagaraMeshRendererProperties]
  renderers [{rendererClass=&NiagaraMeshRendererProperties, rendererIndex=0}]
 }
 Ashes {
  simTarget=GPUComputeSim
  emitterUpdateScript [
   &EmitterState(Inactive Response=::NewEnumerator1, Life Cycle Mode=::NewEnumerator1, Loop Behavior=::NewEnumerator1, Loop Duration=2, UseLoopDelay=false)
   &SpawnBurst_Instantaneous(Spawn Count=10, Use Spawn Probability=false, UseLoopCountLimit=false)
  ]
  particleSpawnScript [
   /Niagara/Modules/Spawn/Initialization/InitializeParticle(
    Color=<rgba 1 0.4454 0.0447 1>
    Lifetime=&UniformRangedFloat(Minimum=0.1, Maximum=0.4)
    Module.Write Color=true
    Module.Write Lifetime=true
    Module.Write Mass=true
    Module.Write Material Random=true
    Module.Write Position=false
    Module.Write Scale=false
    Module.Write SpriteRotation=true
    Module.Write SpriteSize=true
    Position=&SimulationPosition()
    Sprite Rotation=&UniformRangedFloat(Minimum=0, Maximum=360)
    Sprite Size=&Vector2DFromFloat(Value=&Multiply_Float(A=&UniformRangedFloat(Minimum=2.5, Maximum=5), B=1))
   )
   /Niagara/Modules/Spawn/Location/SphereLocation(
    Hemisphere Z=true
    MaskSpawnGroup=false
    Module.Override Randomness=false
    Offset=<-25 0 0>
    Sphere Radius=&Multiply_Float(A=10, B=1)
    Surface Only=true
    Surface Only Band Thickness=2
   )
   &AddVelocity(
    Cone Angle=25
    Cone Axis=<-1 0 0>
    Module.UseConeFalloff=false
    Module.UseVelDistribution=false
    Override Seed=false
    Velocity Mode=::NewEnumerator2
    Velocity Speed=&RandomRangeFloat(Minimum=200, Maximum=2000)
   )
  ]
  particleUpdateScript [
   &ParticleState()
   &AccelerationForce(Acceleration=&RandomRangeVector(Minimum=<0 0 -50>, Maximum=<0 0 -500>))
   &Drag(Drag=5, Use Linear Drag=true, Use Rotational Drag=false)
   &SolveForcesAndVelocity(Clamp Velocity=false, Limit Acceleration=false)
   &ScaleColor(
    Scale Mode=::NewEnumerator1
    Scale RGBA=&Vector4FromCurve(Vector4Curve=DataInterface() {propertyValues {
       AlphaCurve {keys [<at 0.0024 0>,<at 0.1271 15>,<at 0.7369 0.9976>,<at 0.9368 0>]}
       BlueCurve {keys [<at 0.1075 300>,<at 0.4233 0>]}
       GreenCurve {keys [<at 0.1075 15.9452>,<at 0.4233 0>]}
       RedCurve {keys [<at 0.1075 6.1503>,<at 0.4233 0>]}
      }})
    ScaleRGBA=true
   )
   &SpriteRotationRate(Rotation Rate=2)
  ]
  rendererClasses [&NiagaraSpriteRendererProperties]
  renderers [{rendererClass=&NiagaraSpriteRendererProperties}]
 }
 Big_Sparks {
  simTarget=GPUComputeSim
  emitterUpdateScript [
   &EmitterState(
    Enable Distance Culling=true
    Max Distance Response=::KillAfterParticlesFinish
    MaxDistance=3000
    Module.bUseMaxDistance=true
    Module.bUseMinDistance=false
    Scalability Mode=::NewEnumerator1
   )
   &SpawnRate(
    SpawnRate=&FloatFromCurve(
     FloatCurve=DataInterface() {propertyValues {Curve {keys [
         {arriveTangent=-0.107, interpMode=RCIM_Cubic, leaveTangent=-0.107, tangentMode=RCTM_User, time=0, value=0}
         {interpMode=RCIM_Cubic, tangentMode=RCTM_User, time=0.5, value=1}
         {arriveTangent=-1.2381, interpMode=RCIM_Cubic, leaveTangent=-1.2381, tangentMode=RCTM_User, time=1, value=0.25}
        ]}}}
     CurveIndex=$User.uPhase
     Scale Curve=150
    )
    Use Spawn Probability=false
   )
   &SpawnBurst_Instantaneous(Spawn Count=12, Use Spawn Probability=false, UseLoopCountLimit=false)
  ]
  particleSpawnScript [
   /Niagara/Modules/Spawn/Initialization/V2/InitializeParticle(
    Color Minimum=<rgba 0.6198 0.6198 0.6198 1>
    Color Mode=::NewEnumerator2
    Lifetime Max=0.3
    Lifetime Min=0.1
    Lifetime Mode=::NewEnumerator1
    Sprite Size Max=<0.8 0.8>
    Sprite Size Min=<0.4 0.4>
    Sprite Size Mode=::NewEnumerator2
    UsePositionOffset=false
    Write Material Random=true
   ) {module=InitializeParticle001}
   &ShapeLocation(MaskSpawnGroup=false, Offset=<-10 0 0>, Offset Mode=::NewEnumerator0, Override Seed=false, Sphere Radius=5, UseOwnerScale=true)
   &AddVelocityInCone(
    Cone Angle=25
    Cone Axis=<-1 0 0>
    Module.Use Velocity Falloff On Cone Axis=false
    Velocity Strength=&Multiply_Float(
     A=&UniformRangedFloat(Minimum=100, Maximum=2500)
     B=&Lerp_Float(A=0.75, B=1.5, Alpha=$User.uPhase)
    )
   )
  ]
  particleUpdateScript [
   &ParticleState()
   &AccelerationForce(Acceleration=&RandomRangeVector(Minimum=<0 0 -400>, Maximum=<0 0 -980>))
   &Drag(Drag=5, Use Linear Drag=true, Use Rotational Drag=false)
   &SolveForcesAndVelocity(Clamp Velocity=false, Limit Acceleration=false)
   &ScaleSpriteSize(Scale Sprite Size Mode=::NewEnumerator0, Uniform Scale Factor=&Lerp_Float(A=4, B=5, Alpha=$User.uPhase))
   &ScaleSpriteSizeBySpeed(Max Scale Factor=<1 5>, Velocity Threshold:NiagaraFloat=&RandomRangeFloat(Minimum=600, Maximum=1000))
   &ScaleColor(
    Linear Color Curve=DataInterface() {propertyValues {
      AlphaCurve {keys [<at 0.4164 1>,<at 1 0>]}
      BlueCurve {keys [<at 0 0>,<at 0.8465 0>,<at 0.9965 0.2543>]}
      GreenCurve {keys [<at 0 2.5>,<at 0.8465 0.4167>,<at 0.9965 0>]}
      RedCurve {keys [<at 0 15>,<at 0.8465 5>,<at 0.9965 1>]}
     }}
    Scale Mode=::NewEnumerator2
   )
  ]
  rendererClasses [&NiagaraSpriteRendererProperties]
  renderers [{rendererClass=&NiagaraSpriteRendererProperties}]
 }
 Combined_Burst_Once {
  emitterUpdateScript [
   &EmitterState(
    Inactive Response=::NewEnumerator1
    Life Cycle Mode=::NewEnumerator1
    Loop Behavior=::NewEnumerator1
    Loop Duration=$User.uDuration
    UseLoopDelay=false
   )
   &SpawnBurst_Instantaneous(Use Spawn Probability=false, UseLoopCountLimit=false)
  ]
  particleSpawnScript [
   &InitializeParticle(Color Mode=::NewEnumerator0, Lifetime=0.25, Position Mode=::NewEnumerator0, Write Material Random=true)
   /SetVariables(
    Particles.Color_WindCone:LinearColor=<rgba 0.4583 0.4032 0.3558 1.5>
    Particles.DynamicParameter_WindCone:Vector4f=<0.7 2.5 0 0>
    Particles.Orientation_WindCone:Quat4f=&OrientMeshToVector(Look At Direction=<1 0 0>, Reference Axis Vector=<1 0 0>)
    Particles.Size_WindCone:Vector3f=<0.5 1.2 1.2>
   ) {isSetParameters=true}
   /SetVariables(
    Particles.Color_AirGlow:LinearColor=<rgba 0.125 0.125 0.125 1.5>
    Particles.DynamicParameter_AirGlow:Vector4f=<1 1 0 1>
    Particles.Orientation_AirGlow:Quat4f=&OrientMeshToVector(Look At Direction=<1 0 0>, Reference Axis Vector=<1 0 0>)
    Particles.Size_AirGlow:Vector3f=&Multiply_VectorByFloat(Vector=<2 2.25 2.25>, Float=0.8)
   ) {isSetParameters=true}
   /SetVariables(
    Particles.Color_GlowCorona:LinearColor=<rgba 1 0.1667 0 0.5>
    Particles.Position_GlowCorona:NiagaraPosition=<-5 0 0>
    Particles.Size_GlowCorona:Vector2f=&Multiply_Vector2D_ByFloat(Float=125, Vector2D=<1.5 0.8>)
   ) {isSetParameters=true}
   /SetVariables(Particles.DynamicParameter_Sprite:Vector4f=<1 1 1 1>) {isSetParameters=true}
   /SetVariables(Particles.CameraOffsetGlowCorona=15) {isSetParameters=true}
  ]
  particleUpdateScript [
   &ParticleState()
   /SetVariables(Particles.Color_WindCone:LinearColor=&MultiplyLinearColor(
     Linear Color=$Particles.Initial.Color_WindCone
     Scale Factor=&MakeFloat4FromVectorAndFloat(
      "Vector (RGB)"=<1 1 1>
      "Float (Alpha)"=&FloatFromCurve(FloatCurve=DataInterface() {propertyValues {Curve {keys [
           {arriveTangent=11.4472, interpMode=RCIM_Cubic, leaveTangent=11.4472, tangentMode=RCTM_User, time=0, value=0}
           {arriveTangent=1.1179, interpMode=RCIM_Cubic, leaveTangent=1.1179, tangentMode=RCTM_User, time=0.2, value=0.75}
           {arriveTangent=0.0224, interpMode=RCIM_Cubic, leaveTangent=0.0224, tangentMode=RCTM_User, time=1, value=0}
          ]}}})
     )
    )) {isSetParameters=true}
   /SetVariables(Particles.Color_AirGlow:LinearColor=&MultiplyLinearColor(
     Linear Color=$Particles.Initial.Color_AirGlow
     Scale Factor=&MakeFloat4FromVectorAndFloat(
      "Vector (RGB)"=&VectorFromFloat(Value=5)
      "Float (Alpha)"=&FloatFromCurve(FloatCurve=DataInterface() {propertyValues {Curve {keys [
           {arriveTangent=11.4472, interpMode=RCIM_Cubic, leaveTangent=11.4472, tangentMode=RCTM_User, time=0, value=0}
           {arriveTangent=1.1179, interpMode=RCIM_Cubic, leaveTangent=1.1179, tangentMode=RCTM_User, time=0.2, value=0.75}
           {arriveTangent=0.0224, interpMode=RCIM_Cubic, leaveTangent=0.0224, tangentMode=RCTM_User, time=1, value=0}
          ]}}})
     )
    )) {isSetParameters=true}
   /SetVariables(
    Particles.Color_GlowCorona:LinearColor=&MultiplyLinearColor(
     Linear Color=$Particles.Initial.Color_GlowCorona
     Scale Factor=&MakeFloat4FromVectorAndFloat(
      "Vector (RGB)"=&VectorFromFloat(Value=&FloatFromCurve(
        FloatCurve=DataInterface() {propertyValues {Curve {keys [
            {arriveTangent=47.5885, interpMode=RCIM_Cubic, time=0, value=0}
            {arriveTangent=2.8125, interpMode=RCIM_Cubic, leaveTangent=2.8125, time=0.1, value=5}
            <at 0.8 2.25>
            <at 1 0 RCIM_Cubic>
           ]}}}
        Scale Curve=15
       ))
      "Float (Alpha)"=&FloatFromCurve(FloatCurve=DataInterface() {propertyValues {Curve {keys [
           <at 0 1 RCIM_Cubic>
           {arriveTangent=-1.3333, interpMode=RCIM_Cubic, leaveTangent=-1.3333, time=0.1, value=1.25}
           {arriveTangent=-1.3889, interpMode=RCIM_Cubic, leaveTangent=-1.3889, time=0.3, value=0.6}
           <at 1 0 RCIM_Cubic>
          ]}}})
     )
    )
    Particles.Size_GlowCorona:Vector2f=&Multiply_Vector2D_ByFloat(
     Float=&FloatFromCurve(
      FloatCurve=DataInterface() {propertyValues {Curve {keys [
          {arriveTangent=0.225, interpMode=RCIM_Cubic, time=0, value=0.025}
          {arriveTangent=1.2812, interpMode=RCIM_Cubic, leaveTangent=1.2812, time=0.1, value=0.575}
          {arriveTangent=-0.6389, interpMode=RCIM_Cubic, leaveTangent=-0.6389, time=0.8, value=1.05}
          {arriveTangent=-11.4662, interpMode=RCIM_Cubic, leaveTangent=-11.4662, tangentMode=RCTM_User, time=1, value=0}
         ]}}}
      Scale Curve=5
     )
     Vector2D=$Particles.Initial.Size_GlowCorona
    )
   ) {isSetParameters=true}
   &FN_Light_Attributes(
    Color=<rgba 1 0.0833 0 1>
    Create Color Attribute=true
    Create Exponent Attribute=true
    Create Radius Attribute=true
    Radius=100
    Scale Color=&Multiply_Float(
     A=$System.Interpolation_Thrust_EndLife
     B=&Waveform("Waveform [1]"=::NewEnumerator0, "[1] Frequency"=5, "[1] Amplitude Min/Max"=<0.5 1>)
    )
   )
  ]
  rendererClasses [&NiagaraMeshRendererProperties, &NiagaraSpriteRendererProperties]
  renderers [
   {rendererClass=&NiagaraMeshRendererProperties, rendererIndex=0}
   {rendererClass=&NiagaraMeshRendererProperties, rendererIndex=1}
   {rendererClass=&NiagaraSpriteRendererProperties, rendererIndex=2}
  ]
 }
 Combined_Rate {
  emitterUpdateScript [
   &EmitterState()
   &SpawnRate(SpawnRate=12, Use Spawn Probability=false)
  ]
  particleSpawnScript [
   &InitializeParticle(
    Color Mode=::NewEnumerator0
    Lifetime Max=1
    Lifetime Min=0.25
    Lifetime Mode=::NewEnumerator1
    Position Mode=::NewEnumerator0
    Write Material Random=true
   )
   /SetVariables(
    Particles.Color_Blaster:LinearColor=<rgba 1 0.0833 0 0.1>
    Particles.Orientation_Blaster:Quat4f=&OrientMeshToVector(Look At Direction=<-1 0 0>, Reference Axis Vector=<1 0 0>)
    Particles.Size_Blaster:Vector3f=<4 4 4>
   ) {isSetParameters=true}
   /SetVariables(
    Particles.DynamicParameter_Thruster:Vector4f=&MakeVector4(
     X=&RandomRangeFloat(Minimum=0, Maximum=1)
     Y=0
     Z=0
     W=&Lerp_Float(A=10, B=20, Alpha=0.5)
    )
    Particles.Orientation_Thruster:Quat4f=&OrientMeshToVector(Look At Direction=<0 0 1>, Reference Axis Vector=<1 0 0>)
    Particles.Size_Thruster:Vector3f=&RandomRangeVector(Minimum=<0.05 -0.5 0.05>, Maximum=<0.125 -0.75 0.125>)
   ) {enabled=false, isSetParameters=true}
  ]
  particleUpdateScript [
   /SetVariables(
    Particles.Color_Blaster:LinearColor=&MultiplyLinearColor(
     Linear Color=$Particles.Initial.Color_Blaster
     Scale Factor=&MakeFloat4FromVectorAndFloat(
      "Vector (RGB)"=&VectorFromCurve(
       VectorCurve=DataInterface() {propertyValues {
         XCurve {keys [<at 0 2 RCIM_Cubic>,<at 1 1 RCIM_Cubic>]}
         YCurve {keys [<at 0 2 RCIM_Cubic>,<at 1 1 RCIM_Cubic>]}
         ZCurve {keys [<at 0 2 RCIM_Cubic>,<at 1 1 RCIM_Cubic>]}
        }}
       Scale Curve=<5 5 5>
      )
      "Float (Alpha)"=&FloatFromCurve(FloatCurve=DataInterface() {propertyValues {Curve {keys [<at 0 0 RCIM_Cubic>,<at 0.5 1>,<at 1.0011 0 RCIM_Cubic>]}}})
     )
    )
    Particles.DynamicParameter_Blaster:Vector4f=&MakeVector4(
     X=&FloatFromCurve(FloatCurve=DataInterface() {propertyValues {Curve {keys [<at 0 1>,<at 1 0>]}}})
     Y=50
     Z=150
     W=1
    )
    Particles.Size_Blaster:Vector3f=&Multiply_Vector(
     A=<0.5 0.5 0.5>
     B=&Multiply_Vector(
      A=&VectorFromCurve(
       VectorCurve=DataInterface() {propertyValues {
         XCurve {keys [
           <at 0 0.75 RCIM_Cubic>
           <at 0.4908 3.9579>
           {arriveTangent=4.355, interpMode=RCIM_Cubic, leaveTangent=4.355, tangentMode=RCTM_User, time=0.5, value=4}
           <at 1 0.5 RCIM_Cubic>
          ]}
         YCurve {keys [
           <at 0 1 RCIM_Cubic>
           <at 0.4908 0.9933>
           {arriveTangent=0.75, interpMode=RCIM_Cubic, leaveTangent=0.75, time=0.5, value=1}
           <at 1 1.75 RCIM_Cubic>
          ]}
         ZCurve {keys [<at -0.0014 1 RCIM_Cubic>, {arriveTangent=0.7489, interpMode=RCIM_Cubic, leaveTangent=0.7489, time=0.5, value=1}, <at 1 1.75 RCIM_Cubic>]}
        }}
       Scale Curve=&MakeVector(
        X=&Lerp_Float(A=0, B=1, Alpha=$User.uPhase)
        Y=&Lerp_Float(A=0.5, B=1, Alpha=$User.uPhase)
        Z=&Lerp_Float(A=0.5, B=1, Alpha=$User.uPhase)
       )
      )
      B=&VectorFromFloat(Value=&FloatFromCurve(
        FloatCurve=DataInterface() {propertyValues {Curve {keys [
            {interpMode=RCIM_Cubic, tangentMode=RCTM_User, time=0.3, value=1}
            {arriveTangent=-3.2987, interpMode=RCIM_Cubic, leaveTangent=-3.2987, tangentMode=RCTM_User, time=1, value=0.2}
           ]}}}
        CurveIndex=$Emitter.NormalizedLoopAge
       ))
     )
    )
   ) {isSetParameters=true}
   /SetVariables(
    Particles.Color_Thruster:LinearColor=&ScaleLinearColorByCurve(
     InitialValue=$Particles.Initial.Color_Thruster
     ScaleCurve=DataInterface() {propertyValues {
       AlphaCurve {keys [<at 0 0.05 RCIM_Cubic>]}
       BlueCurve {keys [<at 0.0203 0>, <at 0.0654 0.0499>, {arriveTangent=-15.6046, time=0.133, value=0}, <at 0.3627 0>, <at 0.6962 0>]}
       GreenCurve {keys [<at 0.0203 0.8333>, <at 0.0654 0.2874>, {arriveTangent=-5.6465, time=0.133, value=0.1667}, <at 0.3627 0.0833>, <at 0.6962 0.0167>]}
       RedCurve {keys [<at 0.0203 2>, <at 0.0654 1>, {arriveTangent=-0.716, time=0.133, value=1}, <at 0.3627 1>, <at 0.6962 1>]}
      }}
     CurveIndex=$Particles.NormalizedAge
    )
    Particles.Size_Thruster:Vector3f=&Multiply_Vector(
     A=$Particles.Initial.Size_Thruster
     B=&Multiply_Vector(
      A=&VectorFromCurve(
       VectorCurve=DataInterface() {propertyValues {
         XCurve {keys [<at 0 1>,<at 1 1>]}
         YCurve {keys [<at 0 1>,<at 1 1>]}
         ZCurve {keys [<at 0.9 1 RCIM_Cubic>, {arriveTangent=-4.873, interpMode=RCIM_Cubic, leaveTangent=-4.873, tangentMode=RCTM_User, time=1, value=0}]}
        }}
       Scale Curve=&Lerp_Vector(A=<0.5 0.5 1.5>, B=<0.75 0.75 3>, Alpha=$User.uPhase)
      )
      B=&MakeVector(
       X=1
       Y=1
       Z=&FloatFromCurve(FloatCurve=DataInterface() {propertyValues {Curve {keys [{interpMode=RCIM_Cubic, tangentMode=RCTM_User, time=0, value=1}, {interpMode=RCIM_Cubic, tangentMode=RCTM_User, time=1, value=0}]}}})
      )
     )
    )
   ) {isSetParameters=true}
   &ParticleState()
  ]
  rendererClasses [&NiagaraMeshRendererProperties]
  renderers [
   {rendererClass=&NiagaraMeshRendererProperties, rendererIndex=0}
   {rendererClass=&NiagaraMeshRendererProperties, rendererIndex=1}
  ]
 }
 MainEngine_LightingFX {
  emitterSpawnScript [
   /SetVariables(Emitter.Distortion_Input=1, Emitter.LightConeOverallSize=2.2) {isSetParameters=true}
   /SetVariables(Emitter.FlareSizeScalar=0.2) {isSetParameters=true}
  ]
  emitterUpdateScript [
   &EmitterState(
    Inactive Response=::NewEnumerator1
    Life Cycle Mode=::NewEnumerator1
    Loop Behavior=::NewEnumerator1
    Loop Duration Mode=::NewEnumerator1
    UseLoopDelay=true
   )
   &SpawnBurst_Instantaneous(Spawn Count=1, Use Spawn Probability=false)
  ]
  particleSpawnScript [&InitializeParticle(
    Color=<rgba 1 0.1667 0 1>
    Lifetime=$User.uDuration
    Position Offset=<-5 0 0>
    Sprite Size=&Multiply_Vector2D_ByFloat(Float=$Emitter.FlareSizeScalar, Vector2D=&Multiply_Vector2D_ByFloat(Float=1000, Vector2D=<1 0.5>))
    Sprite Size Mode=::NewEnumerator4
    UsePositionOffset=true
    Write Material Random=true
   )]
  particleUpdateScript [
   &ParticleState(
    Kill Particles When Lifetime Has Elapsed=false
    Let Infinitely Lived Particles Die When Emitter Deactivates=true
    Loop Particles Lifetime=true
   )
   &ScaleSpriteSize(Scale Factor=&Vector2DFromFloat(Value=&Lerp_Float(
      A=1
      B=&RandomRangeFloat(Minimum=0.8, Maximum=1, Evaluation Type=::NewEnumerator1)
      Alpha=$User.FlareFlicker
     )))
   &ScaleColor(
    Scale Alpha=&FloatFromCurve(
     FloatCurve=DataInterface() {propertyValues {Curve {keys [
         {arriveTangent=4.1785, interpMode=RCIM_Cubic, leaveTangent=4.1785, tangentMode=RCTM_User, time=0, value=0}
         {arriveTangent=1.1179, interpMode=RCIM_Cubic, leaveTangent=1.1179, tangentMode=RCTM_User, time=0.2042, value=0.6228}
         {arriveTangent=0.0224, interpMode=RCIM_Cubic, leaveTangent=0.0224, tangentMode=RCTM_User, time=1, value=1}
        ]}}}
     CurveIndex=$System.NormalizedLoopAge
     Scale Curve=2.5
    )
    ScaleA=true
    ScaleRGB=true
   )
   &CameraOffset(Camera Offset Amount=15) {module=CameraOffset001}
   &DynamicMaterialParameters(
    Index 0 Param 1=0.2
    Index 0 Param 2=1
    Index 1 Param 1=5
    Module.Index0Param0Write=true
    Module.Index0Param1Write=true
    Module.Index0Param2Write=false
    Module.Index0Param3Write=false
    Module.Index1Param0Write=true
    Module.Index1Param1Write=true
    Module.Index1Param2Write=true
    Module.Index1Param3Write=true
    Write Parameter Index 1=true
   )
  ]
  rendererClasses [&NiagaraSpriteRendererProperties]
  renderers [{rendererClass=&NiagaraSpriteRendererProperties}]
 }
 Small_Sparks {
  simTarget=GPUComputeSim
  emitterUpdateScript [
   &EmitterState()
   &SpawnBurst_Instantaneous(Spawn Count=12, Use Spawn Probability=false, UseLoopCountLimit=false)
   &SpawnRate(
    SpawnRate=&FloatFromCurve(
     FloatCurve=DataInterface() {propertyValues {Curve {keys [
         {arriveTangent=-0.0168, interpMode=RCIM_Cubic, leaveTangent=-0.0168, tangentMode=RCTM_User, time=0, value=0}
         {interpMode=RCIM_Cubic, tangentMode=RCTM_User, time=0.3, value=0.725}
         {arriveTangent=-1.2381, interpMode=RCIM_Cubic, leaveTangent=-1.2381, tangentMode=RCTM_User, time=1, value=0.25}
        ]}}}
     CurveIndex=$User.uPhase
     Scale Curve=100
    )
    Use Spawn Probability=false
   )
  ]
  particleSpawnScript [
   /Niagara/Modules/Spawn/Initialization/InitializeParticle(
    Color=<rgba 20 20 20 1>
    Lifetime=&RandomRangeFloat(Minimum=0.2, Maximum=0.4)
    Module.Write Color=true
    Module.Write Lifetime=true
    Module.Write Mass=true
    Module.Write Material Random=true
    Module.Write Position=false
    Module.Write Scale=false
    Module.Write SpriteRotation=false
    Module.Write SpriteSize=true
    Position=&SimulationPosition()
    Sprite Size=&UniformRangedVector2D(Minimum=<1 2>, Maximum=<2 10>)
   )
   &ShapeLocation(MaskSpawnGroup=false, Offset=<-10 0 0>, Offset Mode=::NewEnumerator0, Override Seed=false, Sphere Radius=5, UseOwnerScale=true)
   &AddVelocityInCone(
    Cone Angle=25
    Cone Axis=<-1 0 0>
    Module.Use Velocity Falloff On Cone Axis=false
    Velocity Strength=&Multiply_Float(
     A=&UniformRangedFloat(Minimum=500, Maximum=750)
     B=&Lerp_Float(A=0.75, B=1.5, Alpha=$User.uPhase)
    )
   )
  ]
  particleUpdateScript [
   &SolveForcesAndVelocity(Clamp Velocity=false, Limit Acceleration=false)
   &ParticleState()
   &ScaleColor(
    Linear Color Curve=DataInterface() {propertyValues {
      AlphaCurve {keys [<at 0.4414 1>,<at 1 0>]}
      BlueCurve {keys [<at 0 0>,<at 0.9576 0>,<at 0.9965 0.2543>]}
      GreenCurve {keys [<at 0 0.5>,<at 0.9576 0.0833>,<at 0.9965 0>]}
      RedCurve {keys [<at 0 2>,<at 0.9576 1>,<at 0.9965 1>]}
     }}
    Scale Mode=::NewEnumerator2
   )
  ]
  rendererClasses [&NiagaraSpriteRendererProperties]
  renderers [{rendererClass=&NiagaraSpriteRendererProperties}]
 }
 WindPlanes {
  emitterUpdateScript [
   &EmitterState(Inactive Response=::NewEnumerator1, Life Cycle Mode=::NewEnumerator1, Loop Behavior=::NewEnumerator1, Loop Duration=1, UseLoopDelay=false)
   &SpawnBurst_Instantaneous(Use Spawn Probability=false, UseLoopCountLimit=false)
   &SpawnBurst_Instantaneous(Spawn Time=0.1, Use Spawn Probability=false, UseLoopCountLimit=false) {module=SpawnBurst_Instantaneous001}
   &SpawnBurst_Instantaneous(Spawn Time=0.15, Use Spawn Probability=false, UseLoopCountLimit=false) {module=SpawnBurst_Instantaneous002}
  ]
  particleSpawnScript [
   /Niagara/Modules/Spawn/Initialization/V2/InitializeParticle(
    Lifetime=0.15
    Mesh Scale Mode=::NewEnumerator3
    Mesh Uniform Scale=1.5
    Position Offset=<-25 0 0>
    UsePositionOffset=true
    Write Material Random=true
   )
   &CameraOffset(Camera Offset Amount=50)
   &InitialMeshOrientation(Mesh Orientation Mode=::NewEnumerator2, Orientation Vector=<0 0 1>, Use Rotation Vector=false)
  ]
  particleUpdateScript [
   &ParticleState()
   &ScaleMeshSize(Scale Factor=&VectorFromFloat(Value=&FloatFromCurve(
      FloatCurve=DataInterface() {propertyValues {Curve {keys [
          {arriveTangent=2.1062, interpMode=RCIM_Cubic, leaveTangent=2.1062, tangentMode=RCTM_User, time=0, value=0}
          {arriveTangent=1.0293, interpMode=RCIM_Cubic, leaveTangent=1.0293, tangentMode=RCTM_User, time=1, value=1}
         ]}}}
      Scale Curve=1.5
     )))
   &Color(Color=&ColorFromCurve(ColorCurve=DataInterface() {propertyValues {
       AlphaCurve {keys [{arriveTangent=0.25, interpMode=RCIM_Cubic, time=0, value=0}, <at 0.3 0.15>, {interpMode=RCIM_Cubic, leaveTangent=-0.1071, time=1, value=0}]}
       BlueCurve {keys [<at 0 0.9263 RCIM_Cubic>]}
       GreenCurve {keys [<at 0 0.9816 RCIM_Cubic>]}
       RedCurve {keys [<at 0 1 RCIM_Cubic>]}
      }})) {enabled=false}
   &ScaleColor(
    Scale Alpha=&FloatFromCurve(
     FloatCurve=DataInterface() {propertyValues {Curve {keys [
         {arriveTangent=10.9641, interpMode=RCIM_Cubic, leaveTangent=10.9641, tangentMode=RCTM_User, time=0, value=0}
         {interpMode=RCIM_Cubic, tangentMode=RCTM_User, time=0.2, value=1}
         {interpMode=RCIM_Cubic, tangentMode=RCTM_User, time=1, value=0}
        ]}}}
     Scale Curve=&FNTimeOfDayScalar(Night Scale=0.2, Morning Scale=0.35, Day Scale=0.45, Evening Scale=0.35)
    )
    ScaleA=true
    ScaleRGB=false
   )
  ]
  rendererClasses [&NiagaraMeshRendererProperties]
  renderers [{rendererClass=&NiagaraMeshRendererProperties, rendererIndex=0}]
 }
}
summary {
 systemName=NS_Proto_RocketRam_Booster
 emitters [
  {emitterName=Ashes, simTarget=GPUComputeSim, rendererClasses [&NiagaraSpriteRendererProperties]}
  {emitterName=Big_Sparks, simTarget=GPUComputeSim, rendererClasses [&NiagaraSpriteRendererProperties]}
  {emitterName=Small_Sparks, simTarget=GPUComputeSim, rendererClasses [&NiagaraSpriteRendererProperties]}
  {emitterName=MainEngine_LightingFX, rendererClasses [&NiagaraSpriteRendererProperties]}
  {emitterName=Air, rendererClasses [&NiagaraMeshRendererProperties]}
  {emitterName=WindPlanes, rendererClasses [&NiagaraMeshRendererProperties]}
  {
   emitterName=Combined_Burst_Once
   rendererClasses [&NiagaraMeshRendererProperties, &NiagaraSpriteRendererProperties]
  }
  {emitterName=Combined_Rate, rendererClasses [&NiagaraMeshRendererProperties]}
 ]
 userVariables [
  User.FlareFlicker {type=NiagaraFloat, defaultValue=1, description=""}
  User.uDuration {type=NiagaraFloat, defaultValue=1.5, description=""}
  User.uPhase {type=NiagaraFloat, defaultValue=1, description=""}
 ]
}
