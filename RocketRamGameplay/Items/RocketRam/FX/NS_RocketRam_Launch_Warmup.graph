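# Niagara system for the Rocket Ram launch warmup: backward-shooting sparks, ashes and smoke, a pulsing flame mesh, and wind and air-pressure meshes that build up.
# Combined_Burst_Once packs wind cone, glows, streak, growing flare and light into custom per-element attributes, feeding mesh, sprite and light renderers.
# User.uDuration times most emitters, uPhase scales spark speed and size, and FlareFlicker blends in flare size jitter.
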NiagaraSystem /RocketRamGameplay/Items/RocketRam/FX/NS_RocketRam_Launch_Warmup
dependencies {
 usedDataInterfaces [&NiagaraDataInterfaceCurve, &NiagaraDataInterfaceColorCurve]
 usedDynamicInputs [
  /Niagara/DynamicInputs/Helpers/SimulationPosition
  &ColorFromCurve
  &RemapRange
 ]
 usedModules [
  &SystemState
  &self:SystemSpawnScript.SystemScriptSource.SystemScriptGraph.NiagaraNodeAssignment_0.SetVariables_#1
  &self:SystemSpawnScript.SystemScriptSource.SystemScriptGraph.NiagaraNodeAssignment_1.SetVariables_#2
 ]
}
emitters {
 AIR_PRESSURE {
  emitterUpdateScript [
   &EmitterState(Life Cycle Mode=::NewEnumerator1, Loop Behavior=::NewEnumerator1, Loop Duration=1.5, UseLoopDelay=false)
   &SpawnRate(SpawnRate=20, Use Spawn Probability=false)
  ]
  particleSpawnScript [
   &InitializeParticle(
    Color Maximum=<rgba 0.2292 0.212 0.1946 0.25>
    Color Minimum=<rgba 0.1354 0.1253 0.115 0.15>
    Color Mode=::NewEnumerator2
    Lifetime Max=0.4
    Lifetime Min=0.25
    Lifetime Mode=::NewEnumerator1
    Mesh Scale=&RandomRangeVector(Minimum=<1.8 1.8 2>, Maximum=<2 2 2>)
    Mesh Scale Mode=::NewEnumerator4
    Position Mode=::NewEnumerator1
    Position Offset=<0 0 50>
    Position Offset Coordinate Space=::World
    UsePositionOffset=true
    Write Material Random=true
   )
   &InitialMeshOrientation(
    Mesh Orientation Mode=::NewEnumerator2
    Orientation Coordinate Space=::Local
    Rotation=&RandomRangeVector(Minimum=<-0.05 0 -0.5>, Maximum=<-0.05 0 0.5>)
    Use Rotation Vector=true
   )
  ]
  particleUpdateScript [
   &ParticleState()
   &ScaleMeshSize(Scale Factor=&MakeVector(
     X=&FloatFromCurve(FloatCurve=DataInterface() {propertyValues {Curve {keys [
          {interpMode=RCIM_Cubic, tangentMode=RCTM_User, time=0, value=0.05}
          {arriveTangent=0.9455, interpMode=RCIM_Cubic, leaveTangent=0.9455, tangentMode=RCTM_User, time=0.3, value=0.8}
          {interpMode=RCIM_Cubic, tangentMode=RCTM_User, time=1, value=1}
         ]}}})
     Y=&FloatFromCurve(FloatCurve=DataInterface() {propertyValues {Curve {keys [
          {interpMode=RCIM_Cubic, tangentMode=RCTM_User, time=0, value=0.05}
          {arriveTangent=0.9455, interpMode=RCIM_Cubic, leaveTangent=0.9455, tangentMode=RCTM_User, time=0.4, value=0.8}
          {interpMode=RCIM_Cubic, tangentMode=RCTM_User, time=1, value=1}
         ]}}})
     Z=1
    ))
   &ScaleColor(
    Linear Color Curve=DataInterface() {propertyValues {
      AlphaCurve {keys [<at 0 1>,<at 0.4544 0.5>,<at 0.9945 0>]}
      BlueCurve {keys [<at 0 1>,<at 1 1>]}
      GreenCurve {keys [<at 0 1>,<at 1 1>]}
      RedCurve {keys [<at 0 1>,<at 1 1>]}
     }}
    Scale Mode=::NewEnumerator2
   )
   &SolveForcesAndVelocity(Clamp Velocity=false, Limit Acceleration=false)
   &ScaleColor(
    Scale Alpha=&FNTimeOfDayScalar(Night Scale=1, Morning Scale=1.2, Day Scale=1.5, Evening Scale=1.2)
    ScaleA=true
    ScaleRGB=false
   ) {module=ScaleColor001}
   &ScaleColor(
    Scale Alpha=&FloatFromCurve(
     FloatCurve=DataInterface() {propertyValues {Curve {keys [
         {interpMode=RCIM_Cubic, tangentMode=RCTM_User, time=0, value=0}
         {arriveTangent=2.0129, interpMode=RCIM_Cubic, leaveTangent=2.0129, tangentMode=RCTM_User, time=0.3, value=0.3}
         {interpMode=RCIM_Cubic, tangentMode=RCTM_User, time=0.6, value=1}
        ]}}}
     CurveIndex=$Emitter.NormalizedLoopAge
    )
    ScaleA=true
    ScaleRGB=false
   ) {module=ScaleColor002}
  ]
  rendererClasses [&NiagaraMeshRendererProperties]
  renderers [{rendererClass=&NiagaraMeshRendererProperties, rendererIndex=0}]
 }
 Ashes {
  emitterUpdateScript [
   &EmitterState(Life Cycle Mode=::NewEnumerator1, Loop Behavior=::NewEnumerator1, Loop Duration=$User.uDuration, UseLoopDelay=false)
   &SpawnBurst_Instantaneous(Spawn Count=4, Use Spawn Probability=false, UseLoopCountLimit=false)
  ]
  particleSpawnScript [
   /Niagara/Modules/Spawn/Initialization/InitializeParticle(
    Color=<rgba 1 0.4454 0.0447 0.1>
    Lifetime=&UniformRangedFloat(Minimum=0.25, Maximum=0.5)
    Module.Write Color=true
    Module.Write Lifetime=true
    Module.Write Mass=true
    Module.Write Material Random=true
    Module.Write Position=false
    Module.Write Scale=false
    Module.Write SpriteRotation=true
    Module.Write SpriteSize=true
    Position=/Niagara/Functions/Localspace/SimulationPosition()
    Sprite Rotation=&UniformRangedFloat(Minimum=0, Maximum=360)
    Sprite Size=&Vector2DFromFloat(Value=&Multiply_Float(A=&UniformRangedFloat(Minimum=2.5, Maximum=5), B=1))
   )
   /Niagara/Modules/Spawn/Location/SphereLocation(
    Hemisphere Z=true
    MaskSpawnGroup=false
    Module.Override Randomness=false
    Offset=<-25 0 0>
    Sphere Radius=&Multiply_Float(A=10, B=1)
    Surface Only=true
    Surface Only Band Thickness=2
   )
   &AddVelocity(
    Cone Angle=25
    Cone Axis=<-1 0 0>
    Module.UseConeFalloff=false
    Module.UseVelDistribution=false
    Override Seed=false
    Velocity Mode=::NewEnumerator2
    Velocity Speed=&RandomRangeFloat(Minimum=250, Maximum=1000)
   )
  ]
  particleUpdateScript [
   &ParticleState()
   &AerodynamicDrag(
    Aerodynamic Drag:NiagaraFloat=&RandomRangeFloat(Minimum=0.1, Maximum=0.05)
    Aerodynamic Rotational Drag=0.1
    Lift Contribution=0.2
    Pivot Offset Maximum:Vector3f=<1 0.1 0.1>
    Pivot Offset Minimum:Vector3f=<-1 -0.1 -0.1>
    Rotation Strength=0.25
    Use Linear Drag=true
    Use Rotational Drag=true
   )
   &GravityForce(Gravity=<0 0 -240>)
   &SolveForcesAndVelocity(Clamp Velocity=false, Limit Acceleration=false)
   &ScaleColor(
    Scale Mode=::NewEnumerator1
    Scale RGBA=&Vector4FromCurve(Vector4Curve=DataInterface() {propertyValues {
       AlphaCurve {keys [<at 0.0024 0>,<at 0.1271 15>,<at 0.7369 0.9976>,<at 0.9368 0>]}
       BlueCurve {keys [<at 0.1075 300>,<at 0.4233 0.4062>]}
       GreenCurve {keys [<at 0.1075 15.9452>,<at 0.4233 0.4062>]}
       RedCurve {keys [<at 0.1075 6.1503>,<at 0.4233 0.4062>]}
      }})
    ScaleRGBA=true
   )
   &AlignSpriteToMeshOrientation(SetAlignment=true, SetFacing=true)
   &SpriteRotationRate(Rotation Rate=2)
  ]
  rendererClasses [&NiagaraSpriteRendererProperties]
  renderers [{rendererClass=&NiagaraSpriteRendererProperties}]
 }
 Big_Sparks {
  emitterUpdateScript [
   &EmitterState(
    Enable Distance Culling=true
    Max Distance Response=::KillAfterParticlesFinish
    MaxDistance=3000
    Module.bUseMaxDistance=true
    Module.bUseMinDistance=false
    Scalability Mode=::NewEnumerator1
   )
   &SpawnBurst_Instantaneous(Spawn Count=20, Use Spawn Probability=false, UseLoopCountLimit=false)
  ]
  particleSpawnScript [
   /Niagara/Modules/Spawn/Initialization/V2/InitializeParticle(
    Color Minimum=<rgba 0.6198 0.6198 0.6198 1>
    Color Mode=::NewEnumerator2
    Lifetime Max=0.3
    Lifetime Min=0.1
    Lifetime Mode=::NewEnumerator1
    Sprite Size Max=<0.8 0.8>
    Sprite Size Min=<0.4 0.4>
    Sprite Size Mode=::NewEnumerator2
    UsePositionOffset=false
    Write Material Random=true
   ) {module=InitializeParticle001}
   &ShapeLocation(MaskSpawnGroup=false, Offset=<-10 0 0>, Offset Mode=::NewEnumerator0, Override Seed=false, Sphere Radius=5, UseOwnerScale=true)
   &AddVelocityInCone(
    Cone Angle=25
    Cone Axis=<-1 0 0>
    Module.Use Velocity Falloff On Cone Axis=false
    Velocity Strength=&Multiply_Float(
     A=&UniformRangedFloat(Minimum=100, Maximum=2500)
     B=&Lerp_Float(A=0.75, B=1.5, Alpha=$User.uPhase)
    )
   )
  ]
  particleUpdateScript [
   &ParticleState()
   &AccelerationForce(Acceleration=&RandomRangeVector(Minimum=<0 0 -400>, Maximum=<0 0 -980>))
   &Drag(Drag=5, Use Linear Drag=true, Use Rotational Drag=false)
   &AerodynamicDrag(
    Aerodynamic Drag:NiagaraFloat=&RandomRangeFloat(Minimum=0.1, Maximum=0.05)
    Aerodynamic Rotational Drag=0.1
    Drag Shape Curve=DataInterface() {propertyValues {Curve {keys [
        {interpMode=RCIM_Cubic, tangentMode=RCTM_User, time=0, value=0.1}
        {arriveTangent=2.0799, interpMode=RCIM_Cubic, leaveTangent=2.0799, tangentMode=RCTM_User, time=1, value=1}
       ]}}}
    Lift Contribution=0.2
    Lift Shape Curve=DataInterface() {propertyValues {Curve {keys [
        {interpMode=RCIM_Cubic, tangentMode=RCTM_User, time=0, value=0}
        {interpMode=RCIM_Cubic, tangentMode=RCTM_User, time=0.5, value=1}
        {tangentMode=RCTM_User, time=1, value=0}
       ]}}}
    Pivot Offset Maximum:Vector3f=<1 0.1 0.1>
    Pivot Offset Minimum:Vector3f=<-1 -0.1 -0.1>
    Rotation Strength=0.25
    Use Linear Drag=true
    Use Rotational Drag=true
   )
   &SolveForcesAndVelocity(Clamp Velocity=false, Limit Acceleration=false)
   &ScaleSpriteSize(Scale Sprite Size Mode=::NewEnumerator0, Uniform Scale Factor=&Lerp_Float(A=2, B=4, Alpha=$User.uPhase))
   &ScaleSpriteSizeBySpeed(Max Scale Factor=<1 5>, Velocity Threshold:NiagaraFloat=&RandomRangeFloat(Minimum=600, Maximum=1000))
   &AlignSpriteToMeshOrientation(SetAlignment=true, SetFacing=true)
   &ScaleColor(
    Linear Color Curve=DataInterface() {propertyValues {
      AlphaCurve {keys [<at 0.4164 1>,<at 1 0>]}
      BlueCurve {keys [<at 0 0>,<at 0.4153 0>,<at 0.9965 0.2543>]}
      GreenCurve {keys [<at 0 2.5>,<at 0.4153 0.4167>,<at 0.9965 0>]}
      RedCurve {keys [<at 0 15>,<at 0.4153 5>,<at 0.9965 1>]}
     }}
    Scale Mode=::NewEnumerator2
   )
  ]
  rendererClasses [&NiagaraSpriteRendererProperties]
  renderers [{rendererClass=&NiagaraSpriteRendererProperties}]
 }
 Combined_Burst_Once {
  emitterUpdateScript [
   &EmitterState()
   &SpawnBurst_Instantaneous(Use Spawn Probability=false, UseLoopCountLimit=false)
  ]
  particleSpawnScript [
   &InitializeParticle(Color Mode=::NewEnumerator0, Lifetime=$User.uDuration, Position Mode=::NewEnumerator0, Write Material Random=true)
   /SetVariables(
    Particles.Color_WindCone:LinearColor=<rgba 0.4583 0.4032 0.3558 0.75>
    Particles.DynamicParameter_WindCone:Vector4f=<0.7 2.5 0 0>
    Particles.Orientation_WindCone:Quat4f=&OrientMeshToVector(Look At Direction=<1 0 0>, Reference Axis Vector=<1 0 0>)
    Particles.Position_WindCone:NiagaraPosition=<-5 0 0>
    Particles.Size_WindCone:Vector3f=<0.5 0.5 0.5>
   ) {isSetParameters=true}
   /SetVariables(
    Particles.Color_AirGlow:LinearColor=<rgba 0.125 0.125 0.125 1>
    Particles.DynamicParameter_AirGlow:Vector4f=<1 1 0 1>
    Particles.Orientation_AirGlow:Quat4f=&OrientMeshToVector(Look At Direction=<0 1 0>, Reference Axis Vector=<1 0 0>)
    Particles.Position_AirGlow:NiagaraPosition=<0 75 0>
    Particles.Size_AirGlow:Vector3f=&Multiply_VectorByFloat(Vector=<7.5 2.25 2.25>, Float=0.6)
   ) {isSetParameters=true}
   /SetVariables(
    Particles.Color_GlowCorona:LinearColor=<rgba 1 0.1667 0 1>
    Particles.Position_GlowCorona:NiagaraPosition=<-5 0 0>
    Particles.Size_GlowCorona:Vector2f=<315 210>
   ) {isSetParameters=true}
   /SetVariables(
    Particles.Streak_01_Color:LinearColor=<rgba 2.5 0.1667 0 1>
    Particles.Streak_01_Offset:NiagaraPosition=&AddVectorToPosition(Position=$Particles.Position, Vector=<0 0 0>)
    Particles.Streak_01_Size:Vector2f=<350 25>
   ) {isSetParameters=true}
   /SetVariables(
    Particles.Color_FlareIncrease:LinearColor=<rgba 1 0.125 0 1>
    Particles.DynamicParameter_FlareIncrease:Vector4f=<0.5 1 0 0>
    Particles.DynamicParameter_FlareIncrease_01:Vector4f=<5 0 0 0>
    Particles.Size_FlareIncrease:Vector2f=<35 35>
   ) {isSetParameters=true}
   /SetVariables(Particles.Sprite_CameraOffset=15) {isSetParameters=true}
  ]
  particleUpdateScript [
   &ParticleState()
   /SetVariables(Particles.Color_WindCone:LinearColor=&MultiplyLinearColor(
     Linear Color=&LerpLinearColors(
      StartColor=<rgba 0.4583 0.4032 0.3558 1>
      EndColor=<rgba 0.4583 0.4032 0.3558 1>
      LerpFactor=&FloatFromCurve(
       FloatCurve=DataInterface() {propertyValues {Curve {keys [
           {interpMode=RCIM_Cubic, tangentMode=RCTM_User, time=0, value=0}
           {interpMode=RCIM_Cubic, tangentMode=RCTM_User, time=0.5, value=1}
           {tangentMode=RCTM_User, time=1, value=0}
          ]}}}
       CurveIndex=&Multiply_Float(A=$Particles.NormalizedAge, B=2.5)
      )
     )
     Scale Factor=&MakeFloat4FromVectorAndFloat(
      "Vector (RGB)"=<1 1 1>
      "Float (Alpha)"=&FloatFromCurve(
       FloatCurve=DataInterface() {propertyValues {Curve {keys [
           {arriveTangent=9.0725, interpMode=RCIM_Cubic, leaveTangent=9.0725, tangentMode=RCTM_User, time=0, value=0}
           {arriveTangent=-2.5249, interpMode=RCIM_Cubic, leaveTangent=-2.5249, tangentMode=RCTM_User, time=0.1408, value=0.9847}
           {arriveTangent=-0.0405, interpMode=RCIM_Cubic, leaveTangent=-0.0405, tangentMode=RCTM_User, time=1, value=0}
          ]}}}
       CurveIndex=&Multiply_Float(A=$Particles.NormalizedAge, B=2.5)
      )
     )
    )) {isSetParameters=true}
   /SetVariables(
    Particles.Color_GlowCorona:LinearColor=&MultiplyLinearColor(
     Linear Color=$Particles.Initial.Color_GlowCorona
     Scale Factor=&MakeFloat4FromVectorAndFloat(
      "Vector (RGB)"=&VectorFromFloat(Value=&FloatFromCurve(
        FloatCurve=DataInterface() {propertyValues {Curve {keys [
            {arriveTangent=47.5885, interpMode=RCIM_Cubic, time=0, value=0}
            {arriveTangent=3.5714, interpMode=RCIM_Cubic, leaveTangent=3.5714, time=0.1, value=5}
            <at 0.7 2.5>
            <at 1 0 RCIM_Cubic>
           ]}}}
        CurveIndex=&Multiply_Float(A=$Particles.NormalizedAge, B=9.6)
        Scale Curve=2
       ))
      "Float (Alpha)"=&FloatFromCurve(
       FloatCurve=DataInterface() {propertyValues {Curve {keys [
           <at 0 1 RCIM_Cubic>
           {arriveTangent=-1.8333, interpMode=RCIM_Cubic, leaveTangent=-1.8333, time=0.1, value=1.25}
           {arriveTangent=-1.3889, interpMode=RCIM_Cubic, leaveTangent=-1.3889, time=0.3, value=0.45}
           <at 1 0 RCIM_Cubic>
          ]}}}
       CurveIndex=&Multiply_Float(A=$Particles.NormalizedAge, B=9.6)
      )
     )
    )
    Particles.Size_GlowCorona:Vector2f=&Multiply_Vector2D_ByFloat(
     Float=&FloatFromCurve(
      FloatCurve=DataInterface() {propertyValues {Curve {keys [
          {arriveTangent=0.225, interpMode=RCIM_Cubic, time=0, value=0.025}
          {arriveTangent=1.2, interpMode=RCIM_Cubic, leaveTangent=1.2, time=0.1, value=0.575}
          <at 1 1.225 RCIM_Cubic>
         ]}}}
      CurveIndex=&Multiply_Float(A=$Particles.NormalizedAge, B=9.6)
     )
     Vector2D=$Particles.Initial.Size_GlowCorona
    )
   ) {isSetParameters=true}
   /SetVariables(Particles.Streak_01_Color:LinearColor=&ScaleLinearColorByCurve(
     InitialValue=$Particles.Initial.Streak_01_Color
     ScaleCurve=DataInterface() {propertyValues {
       AlphaCurve {keys [
         {interpMode=RCIM_Cubic, tangentMode=RCTM_User, time=0, value=1}
         {arriveTangent=-1.8333, arriveTangentWeight=0.0333, interpMode=RCIM_Cubic, leaveTangent=-1.8333, tangentMode=RCTM_User, time=0.1, value=1.25}
         {arriveTangent=-1.3889, arriveTangentWeight=0.0667, interpMode=RCIM_Cubic, leaveTangent=-1.3889, tangentMode=RCTM_User, time=0.3, value=0.45}
         {arriveTangentWeight=0.2333, interpMode=RCIM_Cubic, tangentMode=RCTM_User, time=1, value=0}
        ]}
       BlueCurve {keys [
         {interpMode=RCIM_Cubic, tangentMode=RCTM_User, time=0, value=0}
         {arriveTangent=3.5714, arriveTangentWeight=0.0333, interpMode=RCIM_Cubic, leaveTangent=3.5714, tangentMode=RCTM_User, time=0.1, value=5}
         <at 0.7 2.5>
         {arriveTangentWeight=0.1, interpMode=RCIM_Cubic, tangentMode=RCTM_User, time=1, value=0}
        ]}
       GreenCurve {keys [
         {interpMode=RCIM_Cubic, tangentMode=RCTM_User, time=0, value=0}
         {arriveTangent=3.5714, arriveTangentWeight=0.0333, interpMode=RCIM_Cubic, leaveTangent=3.5714, tangentMode=RCTM_User, time=0.1, value=5}
         <at 0.7 2.5>
         {arriveTangentWeight=0.1, interpMode=RCIM_Cubic, tangentMode=RCTM_User, time=1, value=0}
        ]}
       RedCurve {keys [
         {interpMode=RCIM_Cubic, tangentMode=RCTM_User, time=0, value=0}
         {arriveTangent=3.5714, arriveTangentWeight=0.0333, interpMode=RCIM_Cubic, leaveTangent=3.5714, tangentMode=RCTM_User, time=0.1, value=5}
         <at 0.7 2.5>
         {arriveTangentWeight=0.1, interpMode=RCIM_Cubic, tangentMode=RCTM_User, time=1, value=0}
        ]}
      }}
     CurveIndex=&Multiply_Float(A=$Particles.NormalizedAge, B=9.6)
    )) {isSetParameters=true}
   /SetVariables(
    Particles.Color_FlareIncrease:LinearColor=&MultiplyLinearColor(
     Linear Color=$Particles.Initial.Color_FlareIncrease
     Scale Factor=&MakeFloat4FromVectorAndFloat(
      "Vector (RGB)"=&VectorFromFloat(Value=&FloatFromCurve(
        FloatCurve=DataInterface() {propertyValues {Curve {keys [
            {arriveTangent=-1.8256, interpMode=RCIM_Cubic, leaveTangent=-1.8256, tangentMode=RCTM_User, time=0, value=1}
            {tangentMode=RCTM_User, time=1, value=10}
           ]}}}
        Scale Curve=10
       ))
      "Float (Alpha)"=&Multiply_Float(
       A=&FloatFromCurve(
        FloatCurve=DataInterface() {propertyValues {Curve {keys [{tangentMode=RCTM_User, time=0, value=1}, {tangentMode=RCTM_User, time=1, value=0}]}}}
        CurveIndex=$System.NormalizedLoopAge
       )
       B=$System.SputterFloat
      )
     )
    )
    Particles.Size_FlareIncrease:Vector2f=&Multiply_Vector2D_ByFloat(
     Float=&Multiply_Float(
      A=&Lerp_Float(
       A=1
       B=&RandomRangeFloat(Minimum=0.8, Maximum=1, Evaluation Type=::NewEnumerator1)
       Alpha=$User.FlareFlicker
      )
      B=&Lerp_Float(
       A=1
       B=6
       Alpha=&FloatFromCurve(FloatCurve=DataInterface() {propertyValues {Curve {keys [
            {interpMode=RCIM_Cubic, tangentMode=RCTM_User, time=0, value=0}
            {arriveTangent=1.5936, interpMode=RCIM_Cubic, leaveTangent=1.5936, tangentMode=RCTM_User, time=0.4, value=0.3}
            <at 0.6 1>
           ]}}})
      )
     )
     Vector2D=$Particles.Initial.Size_FlareIncrease
    )
   ) {isSetParameters=true}
   &FN_Light_Attributes(
    Color=<rgba 1 0.0833 0 1>
    Create Color Attribute=true
    Create Exponent Attribute=true
    Create Radius Attribute=true
    Exponent=1
    Radius=3
    Scale Color=&FloatFromCurve(
     FloatCurve=DataInterface() {propertyValues {Curve {keys [{tangentMode=RCTM_User, time=0, value=0}, <at 0.8 0.875>, {tangentMode=RCTM_User, time=1, value=0}]}}}
     Scale Curve=250
    )
   )
  ]
  rendererClasses [
   &NiagaraMeshRendererProperties
   &NiagaraLightRendererProperties
   &NiagaraSpriteRendererProperties
  ]
  renderers [
   {rendererClass=&NiagaraMeshRendererProperties, rendererIndex=0}
   {rendererClass=&NiagaraLightRendererProperties, rendererIndex=1}
   {rendererClass=&NiagaraSpriteRendererProperties, rendererIndex=2}
   {rendererClass=&NiagaraSpriteRendererProperties, rendererIndex=3}
   {rendererClass=&NiagaraSpriteRendererProperties, rendererIndex=4}
  ]
 }
 Smoke_Local {
  emitterUpdateScript [
   &EmitterState(Life Cycle Mode=::NewEnumerator1, Loop Behavior=::NewEnumerator1, Loop Duration=$User.uDuration, UseLoopDelay=false)
   &SpawnBurst_Instantaneous(Spawn Count=6, Use Spawn Probability=false, UseLoopCountLimit=false)
  ]
  particleSpawnScript [
   &InitializeParticle(
    Color=&UniformRangedLinearColor(Minimum=<rgba 0.1351 0.1423 0.15 2>, Maximum=<rgba 0.12 0.1344 0.15 2>)
    Lifetime=$User.uDuration
    Mass Mode=::NewEnumerator1
    Position Mode=::NewEnumerator0
    Sprite Rotation Angle=&RandomRangeFloat(Minimum=-360, Maximum=360)
    Sprite Rotation Mode=::NewEnumerator0
    Sprite Size=&Vector2DFromFloat(Value=&UniformRangedFloat(Minimum=80, Maximum=120))
    Sprite Size Mode=::NewEnumerator4
    Write Material Random=true
   ) {module=InitializeParticle001}
   /SetVariables(
    Particles.CameraOffset_FlameMesh=-5
    Particles.DynamicParameter_FlameMesh:Vector4f=&MakeVector4(
     X=&RandomRangeFloat(Minimum=0, Maximum=1)
     Y=1
     Z=1
     W=&Lerp_Float(A=10, B=20, Alpha=0.5)
    )
    Particles.Orientation_FlameMesh:Quat4f=&OrientMeshToVector(Look At Direction=<0 0 1>, Reference Axis Vector=<1 0 0>)
    Particles.Position_FlameMesh:NiagaraPosition=<10 0 0>
    Particles.Size_FlameMesh:Vector3f=&RandomRangeVector(Minimum=<0.025 0.025 0.2>, Maximum=<0.05 0.05 0.5>)
   ) {isSetParameters=true}
   &AddVelocityInCone(
    Cone Angle=5
    Cone Axis=<-1 0 0>
    Module.Use Velocity Falloff On Cone Axis=false
    Velocity Strength=&FloatFromCurve(
     FloatCurve=DataInterface() {propertyValues {Curve {keys [<at 0 1 RCIM_Cubic>, {arriveTangent=-1.4526, interpMode=RCIM_Cubic, leaveTangent=-1.4526, tangentMode=RCTM_User, time=1, value=0.5}]}}}
     Scale Curve=&RandomRangeFloat(Minimum=500, Maximum=1000)
    )
   )
   /Niagara/Modules/Spawn/Location/V2/SphereLocation(
    MaskSpawnGroup=false
    Override Seed=false
    Sphere Radius=5
    Surface Only=false
    UseNonUniformScale=false
    UseOffset=true
    UseSphereOrientation=false
   )
  ]
  particleUpdateScript [
   &ParticleState()
   /SetVariables(Particles.ScaleXYCurve:NiagaraFloat=&Lerp_Float(
     A=1.2
     B=2
     Alpha=&FloatFromCurve(FloatCurve=DataInterface() {propertyValues {Curve {keys [
          {interpMode=RCIM_Cubic, tangentMode=RCTM_User, time=0, value=0}
          {arriveTangent=0.3693, interpMode=RCIM_Cubic, leaveTangent=0.3693, tangentMode=RCTM_User, time=0.5, value=0.1}
          <at 0.6 1>
         ]}}})
    )) {isSetParameters=true}
   /SetVariables(
    Particles.Color_FlameMesh:LinearColor=&MultiplyLinearColor(
     Linear Color=<rgba 2.5 0.4167 0 0.8588>
     Scale Factor=&MakeFloat4FromVectorAndFloat(
      "Vector (RGB)"=&VectorFromFloat(Value=&Lerp_Float(
        A=1
        B=10
        Alpha=&FloatFromCurve(FloatCurve=DataInterface() {propertyValues {Curve {keys [
             {interpMode=RCIM_Cubic, tangentMode=RCTM_User, time=0, value=0}
             {arriveTangent=0.3693, interpMode=RCIM_Cubic, leaveTangent=0.3693, tangentMode=RCTM_User, time=0.5, value=0.1}
             <at 0.6 1>
            ]}}})
       ))
      "Float (Alpha)"=&FloatFromCurve(FloatCurve=DataInterface() {propertyValues {Curve {keys [
           {tangentMode=RCTM_User, time=0, value=1}
           {arriveTangent=-0.0405, interpMode=RCIM_Cubic, leaveTangent=-0.0405, tangentMode=RCTM_User, time=0.7009, value=1}
           {arriveTangent=-8.0089, interpMode=RCIM_Cubic, leaveTangent=-8.0089, tangentMode=RCTM_User, time=1, value=0}
          ]}}})
     )
    )
    Particles.Size_FlameMesh:Vector3f=&Multiply_Vector(
     A=$Particles.Initial.Size_FlameMesh
     B=&MakeVector(
      X=$Particles.ScaleXYCurve
      Y=$Particles.ScaleXYCurve
      Z=&Lerp_Float(
       A=0.4
       B=&Waveform("Waveform [1]"=::NewEnumerator0, "[1] Phase Scale"=1.5, "[1] Frequency"=4, "[1] Amplitude Min/Max"=<0.7 1>, "[1] Scale"=1.6)
       Alpha=&FloatFromCurve(FloatCurve=DataInterface() {propertyValues {Curve {keys [
            {interpMode=RCIM_Cubic, tangentMode=RCTM_User, time=0, value=0}
            {arriveTangent=0.3693, interpMode=RCIM_Cubic, leaveTangent=0.3693, tangentMode=RCTM_User, time=0.3, value=0.3}
            <at 0.6 1>
           ]}}})
      )
     )
    )
   ) {isSetParameters=true}
   &SpriteRotationRate(Rotation Rate=&Multiply_Float(
     A=&UniformRangedFloat(Minimum=-3, Maximum=3)
     B=&FloatFromCurve(
      FloatCurve=DataInterface() {propertyValues {Curve {keys [
          {arriveTangent=-1.9214, interpMode=RCIM_Cubic, leaveTangent=-1.9214, tangentMode=RCTM_User, time=-0.0046, value=1.0113}
          {arriveTangent=-0.2496, interpMode=RCIM_Cubic, leaveTangent=-0.2496, tangentMode=RCTM_User, time=0.9896, value=0.16}
         ]}}}
      CurveIndex=$Emitter.Age
      Scale Curve=150
     )
    ))
   &ScaleColor(
    Scale Alpha=&FloatFromCurve(
     FloatCurve=DataInterface() {propertyValues {Curve {keys [<at 0 0>,<at 0.4 0.9765>,<at 0.7791 0.9107>,<at 1 0>]}}}
     CurveIndex=&Multiply_Float(A=$Particles.NormalizedAge, B=&UniformRangedFloat(Minimum=6, Maximum=2.66))
    )
    Scale RGB=&VectorFromFloat(Value=&FloatFromCurve(
      FloatCurve=DataInterface() {propertyValues {Curve {keys [<at 0.1505 1>,<at 1 0>,<at 5 2>]}}}
      CurveIndex=&Multiply_Float(A=$Particles.NormalizedAge, B=&UniformRangedFloat(Minimum=6, Maximum=2.66))
     ))
    ScaleA=true
    ScaleRGB=true
   )
   &Drag(Use Linear Drag=true, Use Rotational Drag=false)
   &GravityForce(Gravity=&Lerp_Vector(A=<0 0 400>, B=<0 0 -100>, Alpha=$Particles.MaterialRandom))
   &SolveForcesAndVelocity(Clamp Velocity=false, Limit Acceleration=false)
   &ScaleSpriteSize(Scale Factor=&Vector2DFromFloat(Value=&FloatFromCurve(
      FloatCurve=DataInterface() {propertyValues {Curve {keys [
          {arriveTangent=1.6544, interpMode=RCIM_Cubic, leaveTangent=1.6544, tangentMode=RCTM_User, time=0, value=0.15}
          {interpMode=RCIM_Cubic, leaveTangent=1.6352, time=1, value=1}
         ]}}}
      CurveIndex=&Multiply_Float(A=$Particles.NormalizedAge, B=&UniformRangedFloat(Minimum=6, Maximum=2.66))
     )))
   &DynamicMaterialParameters(
    Index 0 Param 1=&Add_Float(
     A=$Particles.RandomFloat
     B=&FloatFromCurve(
      FloatCurve=DataInterface() {propertyValues {Curve {keys [
          {arriveTangent=0.9088, interpMode=RCIM_Cubic, leaveTangent=0.9088, tangentMode=RCTM_User, time=0, value=1}
          {arriveTangent=0.5473, interpMode=RCIM_Cubic, leaveTangent=0.5473, tangentMode=RCTM_User, time=1, value=2}
         ]}}}
      CurveIndex=&Multiply_Float(A=$Particles.NormalizedAge, B=&UniformRangedFloat(Minimum=6, Maximum=2.66))
     )
    )
    Index 0 Param 3=&FloatFromCurve(
     FloatCurve=DataInterface() {propertyValues {Curve {keys [{arriveTangent=-12.6531, interpMode=RCIM_Cubic, leaveTangent=-12.6531, tangentMode=RCTM_User, time=0, value=1}, <at 0.1 0 RCIM_Cubic>]}}}
     CurveIndex=&Multiply_Float(A=$Particles.NormalizedAge, B=&UniformRangedFloat(Minimum=6, Maximum=2.66))
    )
    Index 1 Param 1=&FloatFromCurve(
     FloatCurve=DataInterface() {propertyValues {Curve {keys [<at 0 0>,<at 0.25 1>]}}}
     CurveIndex=&Multiply_Float(A=$Particles.NormalizedAge, B=&UniformRangedFloat(Minimum=6, Maximum=2.66))
     Scale Curve=&RandomRangeFloat(Minimum=2, Maximum=1)
    )
    Index 1 Param 2=&FloatFromCurve(
     FloatCurve=DataInterface() {propertyValues {Curve {keys [
         {interpMode=RCIM_Cubic, leaveTangent=3.7422, tangentMode=RCTM_Break, time=0, value=0.2}
         {interpMode=RCIM_Cubic, leaveTangent=0.9066, time=1, value=2}
        ]}}}
     CurveIndex=&Multiply_Float(A=$Particles.NormalizedAge, B=&UniformRangedFloat(Minimum=6, Maximum=2.66))
     Scale Curve=&RandomRangeFloat(Minimum=0.5, Maximum=1)
    )
    Index 1 Param 3=&FloatFromCurve(
     FloatCurve=DataInterface() {propertyValues {Curve {keys [
         {interpMode=RCIM_Cubic, leaveTangent=2.8445, tangentMode=RCTM_Break, time=0, value=0.2}
         {interpMode=RCIM_Cubic, leaveTangent=63.2978, time=0.25, value=1}
        ]}}}
     CurveIndex=&Multiply_Float(A=$Particles.NormalizedAge, B=&UniformRangedFloat(Minimum=6, Maximum=2.66))
    )
    Index 1 Param 4=&FloatFromCurve(
     FloatCurve=DataInterface() {propertyValues {Curve {keys [<at 0 1>,<at 1 0.8>]}}}
     CurveIndex=&Multiply_Float(A=$Particles.NormalizedAge, B=&UniformRangedFloat(Minimum=6, Maximum=2.66))
     Scale Curve=&RandomRangeFloat(Minimum=0.5, Maximum=1)
    )
    Index 2 Param 1=&FloatFromCurve(
     FloatCurve=DataInterface() {propertyValues {Curve {keys [{interpMode=RCIM_Cubic, leaveTangent=3.2309, tangentMode=RCTM_Break, time=0.1, value=0}, <at 1 1>]}}}
     CurveIndex=&Multiply_Float(A=$Particles.NormalizedAge, B=&UniformRangedFloat(Minimum=6, Maximum=2.66))
    )
    Index 2 Param 2=&FloatFromCurve(
     FloatCurve=DataInterface() {propertyValues {Curve {keys [<at 0 0.2 RCIM_Cubic>,<at 1 0.8 RCIM_Cubic>]}}}
     CurveIndex=&Multiply_Float(A=$Particles.NormalizedAge, B=&UniformRangedFloat(Minimum=6, Maximum=2.66))
     Scale Curve=&RandomRangeFloat(Minimum=0.2, Maximum=1.2)
    )
    Index 2 Param 3=&FloatFromCurve(
     FloatCurve=DataInterface() {propertyValues {Curve {keys [
         {interpMode=RCIM_Cubic, tangentMode=RCTM_Break, time=0, value=1}
         {interpMode=RCIM_Cubic, leaveTangent=11.1233, tangentMode=RCTM_Break, time=1, value=2}
        ]}}}
     CurveIndex=&Multiply_Float(A=$Particles.NormalizedAge, B=&UniformRangedFloat(Minimum=6, Maximum=2.66))
    )
    Index 2 Param 4=&FloatFromCurve(
     FloatCurve=DataInterface() {propertyValues {Curve {keys [<at 0 1 RCIM_Cubic>, {interpMode=RCIM_Cubic, leaveTangent=-1.2956, tangentMode=RCTM_Break, time=1, value=0.75}]}}}
     CurveIndex=&Multiply_Float(A=$Particles.NormalizedAge, B=&UniformRangedFloat(Minimum=6, Maximum=2.66))
     Scale Curve=0.8
    )
    Module.Param0WriteEnabled=true
    Module.Param0WriteEnabled1=true
    Module.Param0WriteEnabled2=true
    Module.Param1WriteEnabled=false
    Module.Param1WriteEnabled1=true
    Module.Param1WriteEnabled2=true
    Module.Param2WriteEnabled=true
    Module.Param2WriteEnabled1=true
    Module.Param2WriteEnabled2=true
    Module.Param3WriteEnabled=true
    Module.Param3WriteEnabled1=true
    Module.Param3WriteEnabled2=true
    Write Parameter Index 1=true
    Write Parameter Index 2=true
   )
  ]
  rendererClasses [&NiagaraMeshRendererProperties, &NiagaraSpriteRendererProperties]
  renderers [
   {rendererClass=&NiagaraMeshRendererProperties, rendererIndex=0}
   {rendererClass=&NiagaraSpriteRendererProperties, rendererIndex=1}
  ]
 }
 Wind {
  emitterUpdateScript [
   &EmitterState(Life Cycle Mode=::NewEnumerator1, Loop Behavior=::NewEnumerator1, Loop Duration=$User.uDuration, UseLoopDelay=false)
   &SpawnRate(SpawnRate=8, Use Spawn Probability=false)
  ]
  particleSpawnScript [
   &InitializeParticle(
    Color Maximum=<rgba 0.6094 0.5078 0.457 1>
    Color Minimum=<rgba 0.4167 0.3299 0.3125 1>
    Color Mode=::NewEnumerator2
    Lifetime Max=0.6
    Lifetime Min=0.3
    Lifetime Mode=::NewEnumerator1
    Mesh Scale Mode=::NewEnumerator3
    Mesh Uniform Scale=&Multiply_Float(A=&RandomRangeFloat(Minimum=0.25, Maximum=0.4), B=0.5)
    Position Offset=<5 0 0>
    UsePositionOffset=true
    Write Material Random=true
   )
   &InitialMeshOrientation(Mesh Orientation Mode=::NewEnumerator2, Use Rotation Vector=true)
  ]
  particleUpdateScript [
   &ParticleState()
   &ScaleMeshSize(Scale Factor=&MakeVector(
     X=&FloatFromCurve(
      FloatCurve=DataInterface() {propertyValues {Curve {keys [
          {arriveTangent=4.1785, interpMode=RCIM_Cubic, leaveTangent=4.1785, tangentMode=RCTM_User, time=0, value=0}
          {arriveTangent=2.2158, interpMode=RCIM_Cubic, leaveTangent=2.2158, tangentMode=RCTM_User, time=0.2042, value=0.6228}
          {arriveTangent=0.0224, interpMode=RCIM_Cubic, leaveTangent=0.0224, tangentMode=RCTM_User, time=0.5, value=1}
          {arriveTangentWeight=0.1692, interpMode=RCIM_Cubic, tangentMode=RCTM_User, time=1, value=0.2}
         ]}}}
      Scale Curve=&RandomRangeFloat(Minimum=2, Maximum=2.75)
     )
     Y=1.5
     Z=1.5
    ))
   &ScaleMeshSize(Scale Factor=&VectorFromFloat(Value=&FloatFromCurve(
      FloatCurve=DataInterface() {propertyValues {Curve {keys [
          {arriveTangent=1.519, interpMode=RCIM_Cubic, leaveTangent=1.519, tangentMode=RCTM_User, time=0, value=0}
          {interpMode=RCIM_Cubic, tangentMode=RCTM_User, time=0.6, value=1}
         ]}}}
      CurveIndex=$Emitter.NormalizedLoopAge
      Scale Curve=&RandomRangeFloat(Minimum=0.75, Maximum=1.5)
     ))) {module=ScaleMeshSize001}
   &ScaleColor(
    Scale Alpha=&FloatFromCurve(
     FloatCurve=DataInterface() {propertyValues {Curve {keys [
         {arriveTangent=4.1785, interpMode=RCIM_Cubic, leaveTangent=4.1785, tangentMode=RCTM_User, time=0, value=0}
         {
          arriveTangent=1.9289
          arriveTangentWeight=0.1253
          interpMode=RCIM_Cubic
          leaveTangent=1.9289
          leaveTangentWeight=0.2429
          tangentMode=RCTM_User
          time=0.2042
          value=0.6228
         }
         {arriveTangent=0.0224, interpMode=RCIM_Cubic, leaveTangent=0.0224, tangentMode=RCTM_User, time=0.5, value=1}
         {arriveTangent=-3.0357, interpMode=RCIM_Cubic, leaveTangent=-3.0357, tangentMode=RCTM_User, time=1, value=0.2}
        ]}}}
     CurveIndex=$Emitter.NormalizedLoopAge
     Scale Curve=4
    )
    ScaleA=true
    ScaleRGB=false
   )
  ]
  rendererClasses [&NiagaraMeshRendererProperties]
  renderers [{rendererClass=&NiagaraMeshRendererProperties, rendererIndex=0}]
 }
}
summary {
 systemName=NS_RocketRam_Launch_Warmup
 emitters [
  {emitterName=Wind, rendererClasses [&NiagaraMeshRendererProperties]}
  {emitterName=Ashes, rendererClasses [&NiagaraSpriteRendererProperties]}
  {emitterName=Big_Sparks, rendererClasses [&NiagaraSpriteRendererProperties]}
  {emitterName=Smoke_Local, rendererClasses [&NiagaraMeshRendererProperties, &NiagaraSpriteRendererProperties]}
  {emitterName=AIR_PRESSURE, rendererClasses [&NiagaraMeshRendererProperties]}
  {
   emitterName=Combined_Burst_Once
   rendererClasses [
    &NiagaraMeshRendererProperties
    &NiagaraLightRendererProperties
    &NiagaraSpriteRendererProperties
   ]
  }
 ]
 userVariables [
  User.DebrisAmount {type=NiagaraFloat, defaultValue=0, description=""}
  User.ExplosionScale {type=NiagaraFloat, defaultValue=0, description=""}
  User.FlareFlicker {type=NiagaraFloat, defaultValue=1, description=""}
  User.FuelAmount {type=NiagaraFloat, defaultValue=1, description=""}
  User.OverallOpacity {type=NiagaraFloat, defaultValue=1, description=""}
  User.uDuration {type=NiagaraFloat, defaultValue=2.4, description=""}
  User.uPhase {type=NiagaraFloat, defaultValue=1, description=""}
 ]
}
