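; Gameplay ability driving Rocket Ram movement as a state machine: wind-up, hop, active thrust, free fall, braking, recovery and off.
; Each state applies effects and tunes time dilation, air control or gravity, with thrust using slide root motion along a locked launch direction.
; Per-tick destruction traces damage builds, vehicles and pawns, with blocking hits triggering impact AOE and recovery.
; Net syncs, replicated state correction and client-reported blocks keep client and server aligned.
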(event Ability|EventActivateAbility
  ;[0.4838 1 0.2276] Activate Ability
  (|CallParentFunction)
  (Utilities|FlowControl|Sequence
    (:then_1
      (bind _asfort_player_pawn (Utilities|Casting|CastToFortPlayerPawn (Ability|GetActivatingPawn)))
      (bind _output_get (Variables|Default|SetPlayerPawn _asfort_player_pawn))
      (bind _asb_weapon_rocket_ram (Utilities|Casting|CastToB_Weapon_RocketRam (Variables|/Script/FortniteGame.FortPawn|GetCurrentWeapon _output_get)))
      (Variables|Default|SetWeapon _asb_weapon_rocket_ram)
      (CallFunction|AddInitialThrustAOEGameplayEffectContainer))
    (:then_2
      (Variables|Default|GetPlayerPawn
        (:then
          (Variables|Default|GetWeapon
            (:then
              (Class|RocketRamSharedLibrary|RamLog (Variables|Default|GetPlayerPawn) (Utilities|Text|FormatText "NSLOCTEXT(\"[63668553DC3B7B41603CFE6FD08FF0F2]\", \"E7EAFC5A4D968070E3EEFEBFD4920A44\", \"Movement - Triggered.\")") false true)
              (Default|SetRamState "NewEnumerator0")
              (Variables|Default|SetDidInitialThrustAOE false)
              (Variables|Default|SetHasBlockBeenHandled false))
            (:else
              _)))
        (:else
          _))
      (Class|RocketRamSharedLibrary|RamLog (Variables|Default|GetPlayerPawn) (Utilities|Text|FormatText "NSLOCTEXT(\"[63668553DC3B7B41603CFE6FD08FF0F2]\", \"E34E13DA43A0DD4A98C50BA68F3B5173\", \"Movement - Failed to activate.\")") true)
      (Ability|EndAbility))))

(event Custom|Do%20Initial%20Thrust%20AOE%20%28Server%29
  (params (TargetData :type Gameplay%20Ability%20Target%20Data%20Handle%20Structure%20%28by%20ref%29))
  (bind _playerpawn (Variables|Default|GetPlayerPawn))
  ;[0.1025 1 0.6739] On Enter State
  ;;[0.07432 0.3973 1] On Early Eject Complete
  (if (and (not (Variables|Default|GetDidInitialThrustAOE)) (Utilities|Enum|Equal(Enum) (Variables|Default|GetCurrentRamState) "NewEnumerator0"))
    (Utilities|IsValid _playerpawn
      (:Is%20Valid
        (Variables|Default|SetDidInitialThrustAOE true)
        (AOE|AOEInitialThrust _playerpawn (Transformation|GetActorLocation _playerpawn) (Variables|Transient|DestructionTrace|GetDestructionTraceIgnoredActors) TargetData)
        (if (GameplayTags|HasMatchingGameplayTag (Variables|Default|GetPlayerPawn) "(TagName=\"Granted.Athena.RocketRam.ShouldConsumeAmmo\")")
          (FortWeapon|Ammo|ConsumeAmmo (Variables|Default|GetWeapon) 1 false))))))

(event Custom|OnEarlyEjectComplete
  ;[0.1025 1 0.6739] On Enter State
  ;;[0.07432 0.3973 1] On Early Eject Complete
  (bind _asynctaskproxy (Ability|Tasks|WaitNetSync "BothWait")
    (:then
      (Variables|Transient|SetWaitEjectSyncAsyncTask _asynctaskproxy))
    (:OnSync
      (Utilities|FlowControl|Sequence
        (:then_0
          (Class|RocketRamSharedLibrary|RamLog (Variables|Default|GetPlayerPawn) (Utilities|Text|FormatText "NSLOCTEXT(\"[63668553DC3B7B41603CFE6FD08FF0F2]\", \"B65A0EA24C5681C7FFCA029FED69C714\", \"Movement - Early ejection complete.\")") false true)
          (Default|SetUseHighAccelerationTolerance true) ; Use High Acceleration Tolerance to prevent minor net corrections
          (CallFunction|BlockWeaponSwitching)
          (CallFunction|ApplyGameplayEffectSafe (Variables|Config|GetGE_FallDamageImmunity_Class) (Variables|Transient|GetGE_FallDamageImmunity))
          (CallFunction|ApplyGameplayEffectSafe (Variables|Config|GetGE_BlockGrindingButDontCancel_Class) (Variables|Transient|GetGE_BlockGrindingButDontCancel))
          (CallFunction|StartTimeDilation (Math|Float|SelectFloat (Utilities|GetValueOfScalableFloat (Variables|/RocketRamGameplay/Items/RocketRam/DataClass_RocketRam.DataClass_RocketRam_C|GetWindUpTargetTimeDilation_WhenFalling (Variables|Config|GetConfigData))) (Utilities|GetValueOfScalableFloat (Variables|/RocketRamGameplay/Items/RocketRam/DataClass_RocketRam.DataClass_RocketRam_C|GetWindUpTargetTimeDilation (Variables|Config|GetConfigData))) (AI|Components|NavMovement|IsFalling (Variables|/Script/Engine.Character|GetCharacterMovement (Variables|Default|GetPlayerPawn)))) (Math|Float|SelectFloat (Utilities|GetValueOfScalableFloat (Variables|/RocketRamGameplay/Items/RocketRam/DataClass_RocketRam.DataClass_RocketRam_C|GetWindUpTimeDilationEaseInDuration_WhenFalling (Variables|Config|GetConfigData))) (Utilities|GetValueOfScalableFloat (Variables|/RocketRamGameplay/Items/RocketRam/DataClass_RocketRam.DataClass_RocketRam_C|GetWindUpTimeDilationEaseInDuration (Variables|Config|GetConfigData))) (AI|Components|NavMovement|IsFalling (Variables|/Script/Engine.Character|GetCharacterMovement (Variables|Default|GetPlayerPawn)))))
          (CallFunction|AddCameraModifier)
          (bind _returnvalue (Utilities|Time|SetTimerbyEvent (EventDispatchers|CreateEvent :!event Custom|OnWindUpPhaseComplete) (- (Utilities|GetValueOfScalableFloat (Variables|/RocketRamGameplay/Items/RocketRam/DataClass_RocketRam.DataClass_RocketRam_C|GetWindUpPhaseDuration (Variables|Config|GetConfigData))) (Utilities|GetValueOfScalableFloat (Variables|/RocketRamGameplay/Items/RocketRam/DataClass_RocketRam.DataClass_RocketRam_C|GetAverageNetworkWaitingTime (Variables|Config|GetConfigData))))))
          (Variables|Transient|SetWindUpTimerHandle _returnvalue)
          (Utilities|FlowControl|Sequence
            (:then_0
              (CallFunction|ApplyGameplayEffectSafe (Variables|Config|GetGE_State_WindingUp_Class) (Variables|Transient|GetGE_State_WindingUp))
              (CallFunction|StartListeningtoPawnLandedCancelTimeDilation) ; If the player was in the air when the Wind Up phase starts, remove the in \nair time dilation that was applied when they land.
            )
            (:then_1
              (Variables|Default|SetCurrentCustomTargetingTransform (Math|Transform|MakeTransform (Transformation|GetActorLocation (Variables|Default|GetPlayerPawn)) "0.000000,45.000000,0.000000"))
              (bind (_asynctaskproxy_2 _targetdata) (Ability|Tasks|DoTargetSelection "Rocket Ram AOE Initial Thrust Targeting" "(TagName=\"Abilities.Generic.RocketRam.AOEInitialThrust\")")
                (:Targeted
                  (Ability|Tasks|WaitGameplayTagAdd "(TagName=\"Granted.Athena.RocketRam.EquipmentChangeIsBlocked\")" 0 true
                    (:Added
                      (Ability|HasAuthority
                        (:False
                          (Ability|IsLocallyControlled
                            (:True
                              (CallFunction|DoInitialThrustAOE(Server) _targetdata) ; Wait for equipment changing to be blocked on the CLIENT,\nthen tell the server to do the AOE
                            )))))) ; Make sure equipment change is blocked before we do the destruction,\nto avoid the possibility of being able to cancel ability when they've already\ngotten the destruction.
                )))))
        (:then_2
          (CallFunction|SendWindUpEvent true) ; doesn't work if placed after Wait Net Sync
        )))))

(event Custom|OnWindUpPhaseComplete
  ;[0.1025 1 0.6739] On Enter State
  ;;[0.07432 0.3973 1] On Wind Up Phase Complete
  (Default|SetRamState "NewEnumerator6")
  (CallFunction|SendWindUpEvent false) ; doesn't work if placed after Wait Net Sync
)

(event Custom|Add%20Camera%20Modifier
  ;[0.1025 1 0.6739] On Enter State
  ;;[0.07432 0.3973 1] Add Camera Modifier
  (bind _weapon (Variables|Default|GetWeapon)
    (:then
      (if (not (Variables|/RocketRamGameplay/Items/RocketRam/B_Weapon_RocketRam.B_Weapon_RocketRam_C|GetCheat_CameraModifierDisabled _weapon))
        (bind _playerpawn (Variables|Default|GetPlayerPawn)
          (:then
            (if (and (Pawn|IsLocallyControlled _playerpawn) (not (Networking|IsServer)))
              (Utilities|IsValid (Fortnite|GetFortPlayerControllerfromActor _playerpawn)
                (:Is%20Valid
                  (Class|RocketRamSharedLibrary|RamLog (Variables|Default|GetPlayerPawn) (Utilities|LogCameraModifierAdded) false true)
                  (bind _returnvalue (Game|Player|AddNewCameraModifier (Variables|/Script/Engine.PlayerController|GetPlayerCameraManager (Fortnite|GetFortPlayerControllerfromActor _playerpawn)) "/RocketRamGameplay/Items/RocketRam/CameraModifier_RocketRam.CameraModifier_RocketRam_C"))
                  (bind _ascamera_modifier_rocket_ram (Utilities|Casting|CastToCameraModifier_RocketRam _returnvalue))
                  (bind _output_get (Variables|Transient|SetCameraModifier _ascamera_modifier_rocket_ram))
                  (Variables|/RocketRamGameplay/Items/RocketRam/B_Weapon_RocketRam.B_Weapon_RocketRam_C|SetCameraModifier _output_get (Variables|Default|GetWeapon))))))
          (:else))))
    (:else)))

(event Custom|HandleServerReplicatedState
  (bind _previousramstate (Variables|Default|GetPreviousRamState))
  (bind _currentramstate (Variables|Default|GetCurrentRamState))
  (bind _serverreplicatedstate (Variables|Transient|GetServerReplicatedState))
  ;[0.2735 0.1586 1] Replicate State From Server
  (if (and (Utilities|Enum|NotEqual(Enum) _serverreplicatedstate _previousramstate) (Utilities|Enum|NotEqual(Enum) _serverreplicatedstate _currentramstate))
    (Class|RocketRamSharedLibrary|RamLog (Variables|Default|GetPlayerPawn) (Utilities|Text|FormatText "NSLOCTEXT(\"[63668553DC3B7B41603CFE6FD08FF0F2]\", \"8B76CB52428053EE3AFC5994DDCCB710\", \"Movement - Client state out of sync with server! Overwriting client state. Client Previous State: [{PreviousState}], Client Current State: [{CurrentState}], New State From Server: [{ServerState}].\")" (Utilities|String|EnumtoString _previousramstate) (Utilities|String|EnumtoString _currentramstate) (Utilities|String|EnumtoString _serverreplicatedstate)) true)
    (Default|SetRamState (Variables|Transient|GetServerReplicatedState)) ; Only overwrite the client state with the server state if it is transitioning to \nif the Previous and Current state on the client are different than the server state. \nWe check both because it is possible for the client to predictively transition into and \nout of the state that is being replicated from the server (i.e. client predicts Recovery, \nserver replicates Recovery, but client is in Off before the replicated event is received). \nIt is highly unlikely that the client transitions through two states before the replicated \nevent is received.
  ))

(event Custom|Client_ReplicateStateFromServer
  (params (ServerState :type Enum_RocketRamState%20Enum :default "NewEnumerator0"))
  ;[0.2735 0.1586 1] Replicate State From Server
  (Variables|Transient|SetServerReplicatedState ServerState)
  (bind _returnvalue (Utilities|Time|SetTimerbyEvent (EventDispatchers|CreateEvent :!event Custom|HandleServerReplicatedState) 0.1))
  (Variables|Transient|SetHandleServerReplicatedStateTimerHandle _returnvalue))

(event Ability|EventOnEndAbility
  (params (bWasCancelled :type Boolean))
  ;[1 0.5582 0.1752] End Ability
  (|CallParentFunction bWasCancelled)
  (Default|SetRamState "NewEnumerator4") ; Always transition to Off when ending this ability \nand always end this ability when transitioning to Off.
  (Class|RocketRamSharedLibrary|RamLog (Variables|Default|GetPlayerPawn) (Utilities|Text|FormatText "NSLOCTEXT(\"[63668553DC3B7B41603CFE6FD08FF0F2]\", \"25A328EB4A8E143EEF2C50B7BDF98030\", \"Movement - Ended.\")") false true)
  (CallFunction|SendWindUpEvent false)
  (CallFunction|ClearAllTimers))

(event Custom|Client_LockInLaunchDirection
  (params (LaunchDirection :type Vector))
  ;[0.1025 1 0.6739] On Enter State
  ;;[0.02924 0.08066 1] Lock in the launch direction and replicate it to the client.
  (Variables|Transient|SetLaunchDirection LaunchDirection)
  (Class|RocketRamSharedLibrary|RamLog (Variables|Default|GetPlayerPawn) (Utilities|Text|FormatText "NSLOCTEXT(\"[63668553DC3B7B41603CFE6FD08FF0F2]\", \"84EBBE7A4E0FD73A27306789E16D2307\", \"Movement - Launch Direction Locked In {0}.\")" (Utilities|String|ToString(Vector) (Variables|Transient|GetLaunchDirection))) false true)
  (Variables|Transient|SetbHasLaunchDirectionReplicated true)
  (CallFunction|TryMovetoActiveThrustPhase))

(event Custom|OnSlowAccelerationPhaseComplete
  ;[0.1025 1 0.6739] On Enter State
  ;;[0.02924 0.08066 1] Switch to Active Thrust when the launch direction has been replicated and the slow acceleration phase has completed.
  (Variables|Transient|SetbHasSlowAccelerationPhaseCompleted true)
  (CallFunction|TryMovetoActiveThrustPhase))

(event Custom|LockInLaunchDirection
  ;[0.1025 1 0.6739] On Enter State
  ;;[0.02924 0.08066 1] Lock in the launch direction and replicate it to the client.
  (Variables|Default|GetPlayerPawn
    (:then
      (bind _launchdirection (Default|CalculateLaunchDirection))
      (Variables|Transient|SetLaunchDirection _launchdirection)
      (bind _weapon (Variables|Default|GetWeapon)
        (:then
          (Class|BWeaponRocketRam|SetLaunchDirection _weapon (Variables|Transient|GetLaunchDirection) :!class /RocketRamGameplay/Items/RocketRam/B_Weapon_RocketRam.B_Weapon_RocketRam_C)
          (Ability|HasAuthority
            (:True
              (Utilities|IsValid self
                (:Is%20Valid
                  (CallFunction|ClientLockinLaunchDirection (Variables|Transient|GetLaunchDirection))
                  (Default|ForceNetUpdate)
                  (Class|RocketRamSharedLibrary|RamLog (Variables|Default|GetPlayerPawn) (Utilities|Text|FormatText "NSLOCTEXT(\"[63668553DC3B7B41603CFE6FD08FF0F2]\", \"84EBBE7A4E0FD73A27306789E16D2307\", \"Movement - Launch Direction Locked In {0}.\")" (Utilities|String|ToString(Vector) (Variables|Transient|GetLaunchDirection))) false true)
                  (Variables|Transient|SetbHasLaunchDirectionReplicated true)
                  (CallFunction|TryMovetoActiveThrustPhase))))))
        (:else)))
    (:else)))

(event Custom|TryMoveToActiveThrustPhase
  ;[0.1025 1 0.6739] On Enter State
  ;;[0.02924 0.08066 1] Switch to Active Thrust when the launch direction has been replicated and the slow acceleration phase has completed.
  (if (and (Variables|Transient|GetbHasLaunchDirectionReplicated) (Variables|Transient|GetbHasSlowAccelerationPhaseCompleted))
    (bind _asynctaskproxy (Ability|Tasks|WaitNetSync "BothWait")
      (:then
        (Variables|Transient|SetWaitLaunchDirectionReplicatedAsyncTask _asynctaskproxy))
      (:OnSync
        (switch Utilities|FlowControl|Switch|SwitchonEnum_RocketRamState (Variables|Default|GetCurrentRamState)
          (:NewEnumerator6
            (Default|SetRamState "NewEnumerator1")))))))

(event Custom|ResetIgnoredCollision
  (bind _destructiontracecameraignoreactors (Variables|Transient|DestructionTrace|GetDestructionTraceCameraIgnoreActors))
  (bind _destructiontraceignoredactors (Variables|Transient|DestructionTrace|GetDestructionTraceIgnoredActors))
  ;[1 0.2889 0.9783] Destruction Trace
  ;;[1 0.06918 0.5615] Reset ignored collision. Called on state changed to Off. This is a failsafe to ensure that if something wasn't destroyed, but had its collision ignored, we unignore it here to prevent exploits and bugs.
  (Utilities|FlowControl|Sequence
    (:then_0
      (bind _playerpawn (Variables|Default|GetPlayerPawn)
        (:then
          (for _array_element _destructiontraceignoredactors
            (Utilities|IsValid _array_element
              (:Is%20Valid
                (Collision|IgnoreActorwhenMoving (Variables|/Script/Engine.Character|GetCapsuleComponent _playerpawn) _array_element false))))
          (Utilities|Array|Clear _destructiontraceignoredactors))
        (:else)))
    (:then_1
      (Class|RocketRamDestructionTraceLibrary|ClearCameraCollisionActorTags _destructiontracecameraignoreactors)
      (Utilities|Array|Clear _destructiontracecameraignoreactors))))

(event Custom|OnEnterRamState
  (params (CurrentState :type Enum_RocketRamState%20Enum :default "NewEnumerator0"))
  ;[0.1025 1 0.6739] On Enter State
  (Development|PrintString (Utilities|String|EnumtoString CurrentState))
  (switch Utilities|FlowControl|Switch|SwitchonEnum_RocketRamState CurrentState
    (:NewEnumerator0
      ;;[0.07432 0.3973 1] On State Changed To Wind Up
      (CallFunction|StartWaitingforMovementModeChange)
      (CallFunction|TryEarlyEject))
    (:NewEnumerator6
      ;;[0.02924 0.08066 1] On State Changed To Slow Acceleration
      (Variables|Transient|SetbHasSlowAccelerationPhaseCompleted false)
      (bind _returnvalue (Utilities|Time|SetTimerbyEvent (EventDispatchers|CreateEvent :!event Custom|OnSlowAccelerationPhaseComplete) (Utilities|GetValueOfScalableFloat (Variables|/RocketRamGameplay/Items/RocketRam/DataClass_RocketRam.DataClass_RocketRam_C|GetHopPhaseDuration (Variables|Config|GetConfigData)))))
      (Variables|Transient|SetSlowAccelerationTimerHandle _returnvalue)
      (Ability|SendGameplayEvent "(TagName=\"Event.Athena.Balloons.RemoveAll\")") ; remove any balloons
      (bind _returnvalue_2 (Utilities|Time|SetTimerbyEvent (EventDispatchers|CreateEvent :!event Custom|LockInLaunchDirection) (Utilities|GetValueOfScalableFloat (Variables|/RocketRamGameplay/Items/RocketRam/DataClass_RocketRam.DataClass_RocketRam_C|GetHopDirectionLockInDelay (Variables|Config|GetConfigData))))) ; We lock in the Launch Direction just before the Slow Acceleration phase finishes to allow time for it to replicate.
      (Variables|Transient|SetSlowAccelerationLockInLaunchDirectionTimerHandle _returnvalue_2)
      (CallFunction|DoPreThrustHop)
      (CallFunction|StartTimeDilation (Utilities|GetValueOfScalableFloat (Variables|/RocketRamGameplay/Items/RocketRam/DataClass_RocketRam.DataClass_RocketRam_C|GetHopTargetTimeDilation (Variables|Config|GetConfigData))) (Utilities|GetValueOfScalableFloat (Variables|/RocketRamGameplay/Items/RocketRam/DataClass_RocketRam.DataClass_RocketRam_C|GetHopTimeDilationEaseInDuration (Variables|Config|GetConfigData))))
      (Default|SetAirControl (ScalableFloat|GetValueatLevel (Variables|/RocketRamGameplay/Items/RocketRam/DataClass_RocketRam.DataClass_RocketRam_C|GetAirControl_StartAmount (Variables|Config|GetConfigData))))
      (CallFunction|StartPredictingLaunchDirection) ; Predict the launch direction so it can be sent \nto the Anim BP which uses it to control Aim Offsets.
      (CallFunction|ApplyGameplayEffectSafe (Variables|Config|GetGE_State_Hopping_Class) (Variables|Transient|GetGE_State_Hopping))
      (if (ScalableFloat|GetValueasBool (Variables|Default|GetbApplyCollisionImmunity))
        (CallFunction|ApplyGameplayEffectSafe (Variables|Default|GetGE_CollisionImmunity_Class) (Variables|Default|GetGE_CollisionImmunity))))
    (:NewEnumerator1
      ;;[0.7199 0.1195 1] On State Changed To Active Thrust
      (CallFunction|TryLateEject)
      (Default|SetGravityMultiplier 0.0)
      (Variables|Transient|Gravity|SetBaseGravity (Class|RocketRamSharedLibrary|CalculateBaseGravity (Variables|Transient|GetLaunchDirection)))
      (bind _asynctaskproxy (Ability|Tasks|ApplyRootMotion_Slide "None" (Utilities|GetValueOfScalableFloat (Variables|/RocketRamGameplay/Items/RocketRam/DataClass_RocketRam.DataClass_RocketRam_C|GetActiveThrust_Duration (Variables|Config|GetConfigData))) (Variables|Transient|GetLaunchDirection) (Utilities|GetValueOfScalableFloat (Variables|/RocketRamGameplay/Items/RocketRam/DataClass_RocketRam.DataClass_RocketRam_C|GetActiveThrust_StartSpeed (Variables|Config|GetConfigData))) (Utilities|GetValueOfScalableFloat (Variables|/RocketRamGameplay/Items/RocketRam/DataClass_RocketRam.DataClass_RocketRam_C|GetActiveThrust_MaxSpeed (Variables|Config|GetConfigData))) "/RocketRamGameplay/Items/RocketRam/Curves/ActiveThrustRootMotionCurve.ActiveThrustRootMotionCurve" 4e+01 0 "SetVelocity" (* (Variables|Transient|GetLaunchDirection) (Utilities|GetValueOfScalableFloat (Variables|/RocketRamGameplay/Items/RocketRam/DataClass_RocketRam.DataClass_RocketRam_C|GetActiveThrust_SpeedClampAtEnd (Variables|Config|GetConfigData)))) 0.0)
        (:then
          (Variables|Transient|SetRootMotionAsyncTask _asynctaskproxy)
          (Ability|ApplyGameplayEffectToOwner (Variables|Config|GetGE_BrakingFriction_Class)) ; Grants a GA that removes itself and the applying \nGE when the Ram's state changes to Off.
          ;;;[1 0.6702 0] There is a race condition bug where triggering GE State Active Thrust on the client and server causes the GE to NOT be removed correctly, if we deactivate the GE too quickly. Just activating on the server bypasses this issue.
          (if (or (not (ScalableFloat|GetValueasBool (Variables|Default|GetHF%20Only%20Add%20Active%20Thrust%20GEOn%20Server))) (Networking|IsDedicatedServer))
            (CallFunction|ApplyGameplayEffectSafe (Variables|Config|GetGE_State_ActiveThrust_Class) (Variables|Transient|GetGE_State_ActiveThrust)))
          (CallFunction|ApplyGameplayEffectSafe (Variables|Config|GetGE_InAir_Class) (Variables|Transient|GetGE_InAir))
          (CallFunction|StartDestructionTrace) ; Needs to happen last; this node can sometimes result\nin changing the state
          (CallFunction|StartWaitingforUnderwater) ; Needs to happen last; this node sometimes results in ending the ability
        )
        (:OnFinish
          (bind _playerpawn (Variables|Default|GetPlayerPawn)
            (:then
              (if (AI|Components|NavMovement|IsFalling (Variables|/Script/Engine.Character|GetCharacterMovement _playerpawn))
                (Default|SetRamState "NewEnumerator2")
                (else
                  (Default|SetRamState "NewEnumerator3"))))
            (:else)))))
    (:NewEnumerator2
      _)
    (:NewEnumerator5
      _)
    (:NewEnumerator3
      ;;[1 0.9123 0.1124] On State Changed To Braking
      (CallFunction|StartBraking)
      (CallFunction|StopIncreasingGravity)
      (CallFunction|StopWaitingforFreeFallPeak)
      (CallFunction|StopDestructionTrace)
      (CallFunction|StopTimeDilation)
      (CallFunction|StopListeningforPawnLanded)
      (CallFunction|ApplyGameplayEffectSafe (Variables|Config|GetGE_State_Braking_Class) (Variables|Transient|GetGE_State_Braking))
      (Ability|RemoveGameplayEffectFromOwnerWithHandle (Variables|Transient|GetGE_BlockWeaponSwitching)) ; Make sure any existing one is removed
      (Ability|RemoveGameplayEffectFromOwnerWithHandle (Variables|Transient|GetGE_InAir)) ; Make sure any existing one is removed
    )
    (:NewEnumerator7
      ;;[0.08993 1 0.1404] On State Changed To Recovery
      (CallFunction|StartRecovery)
      (CallFunction|ApplyGameplayEffectSafe (Variables|Config|GetGE_State_Recovering_Class) (Variables|Transient|GetGE_State_Recovering))
      (Default|SetUseHighAccelerationTolerance false)
      (CallFunction|StopIncreasingGravity)
      (CallFunction|StopWaitingforFreeFallPeak)
      (CallFunction|StopDestructionTrace)
      (CallFunction|StopTimeDilation))
    (:NewEnumerator4
      ;;[0.4235 1 0.04638] On State Changed To Off
      (Utilities|FlowControl|Sequence
        (:then_0
          (Default|SetUseHighAccelerationTolerance false)
          (CallFunction|StopIncreasingGravity)
          (CallFunction|StopWaitingforFreeFallPeak)
          (CallFunction|StopDestructionTrace)
          (CallFunction|StopTimeDilation)
          (CallFunction|StopListeningforPawnLanded)
          (CallFunction|ResetIgnoredCollision)
          (CallFunction|StopWaitingforMovementModeChange)
          (CallFunction|StopPostAOEPause)
          (Utilities|Time|ClearandInvalidateTimerbyHandle (Variables|Default|GetTimer_WaitingForFreeFallTooLong)))
        (:then_1
          (Utilities|Time|ClearandInvalidateTimerbyHandle (Variables|Transient|GetHandleClientReplicatedBlockingHitTimerHandle))
          (Utilities|Time|ClearandInvalidateTimerbyHandle (Variables|Transient|GetHandleServerReplicatedStateTimerHandle))
          (Ability|RemoveGameplayEffectFromOwnerWithHandle (Variables|Transient|GetGE_AirControl))
          (Ability|RemoveGameplayEffectFromOwnerWithHandle (Variables|Transient|GetGE_BlockWeaponSwitching)) ; Make sure any existing one is removed
          (Ability|RemoveGameplayEffectFromOwnerWithHandle (Variables|Transient|GetGE_InAir)) ; Make sure any existing one is removed
          (Ability|RemoveGameplayEffectFromOwnerWithHandle (Variables|Transient|GetGE_FallDamageImmunity)) ; Make sure any existing one is removed
          (Ability|RemoveGameplayEffectFromOwnerWithHandle (Variables|Transient|GetGE_BlockGrindingButDontCancel)) ; Make sure any existing one is removed
          (Ability|RemoveGameplayEffectFromOwnerWithHandle (Variables|Default|GetGE_CollisionImmunity)) ; Make sure any existing one is removed
          (bind _cameramodifier (Variables|Transient|GetCameraModifier)
            (:then
              (Class|CameraModifierRocketRam|DisableandRemove _cameramodifier)
              (Class|RocketRamSharedLibrary|RamLog (Variables|Default|GetPlayerPawn) (Utilities|Text|ToText(String) (Utilities|String|Append "Camera modifier removed. Current value: " (Utilities|GetDisplayName _cameramodifier))) false true))
            (:else
              (Class|RocketRamSharedLibrary|RamLog 0 "NSLOCTEXT(\"[63668553DC3B7B41603CFE6FD08FF0F2]\", \"C4BD8613409AAFA623EAC4B8088BE72B\", \"Camera modifier removal skipped because it was null\")" true true))))
        (:then_2
          ;;;[0.1299 0.1558 1] If the player is falling, grant fall immunity
          (bind _playerpawn_2 (Variables|Default|GetPlayerPawn)
            (:then
              (if (AI|Components|NavMovement|IsFalling (Variables|/Script/Engine.Character|GetCharacterMovement _playerpawn_2))
                (Ability|ApplyGameplayEffectToOwner (Variables|Config|GetGE_TillLandFallDamageImmunity_Class))))
            (:else)))
        (:then_3
          ;;;[0.474 0.4679 0.09972] Apply cooldown to weapon.
          (switch Utilities|FlowControl|Switch|SwitchonEnum_RocketRamState (Variables|Default|GetPreviousRamState)
            (:NewEnumerator0
              _)
            (:NewEnumerator6
              _)
            (:NewEnumerator1
              _)
            (:NewEnumerator2
              _)
            (:NewEnumerator5
              _)
            (:NewEnumerator3
              _)
            (:NewEnumerator7
              _))
          (bind _weapon (Variables|Default|GetWeapon)
            (:then
              (if (not (Variables|/RocketRamGameplay/Items/RocketRam/B_Weapon_RocketRam.B_Weapon_RocketRam_C|GetCheat_CooldownDisabled _weapon))
                (Ability|ApplyGameplayEffectToOwner (select (Variables|Default|GetDidInitialThrustAOE) (Variables|Config|GetGE_Cooldown_Class) (Variables|Config|GetGE_CooldownOnCancel_Class)))))
            (:else)))
        (:then_5
          ;;;[1 0.581 0] Log and End Ability
          (Class|RocketRamSharedLibrary|RamLog (Variables|Default|GetPlayerPawn) (Utilities|Text|FormatText "NSLOCTEXT(\"[63668553DC3B7B41603CFE6FD08FF0F2]\", \"25F316B24E3796C9AB6E1EB397239069\", \"On Enter State(Off) Calling End Ability\")") false true)
          (Ability|EndAbility) ; Always transition to Off when ending this ability \nand always end this ability when transitioning to Off.
        ))))
  ;;[1 0.1707 0.6226] On State Changed To Free Fall
  (CallFunction|StartIncreasingGravity)
  (CallFunction|IncreaseAirControlatFreeFallPeak)
  (CallFunction|StartListeningforPawnLanded)
  (CallFunction|StartLookaheadTrace)
  (CallFunction|ApplyGameplayEffectSafe (Variables|Config|GetGE_State_FreeFalling_Class) (Variables|Transient|GetGE_State_FreeFalling))
  (CallFunction|StartWaitingforFreeFallTooLong))

(event Custom|DoPreThrustHop
  (bind _playerpawn (Variables|Default|GetPlayerPawn))
  (bind _charactermovement (Variables|/Script/Engine.Character|GetCharacterMovement _playerpawn))
  (bind _returnvalue (AI|Components|NavMovement|IsFalling _charactermovement))
  (bind _configdata (Variables|Config|GetConfigData))
  (bind _hop_upamount_onground (Variables|/RocketRamGameplay/Items/RocketRam/DataClass_RocketRam.DataClass_RocketRam_C|GetHop_UpAmount_OnGround _configdata))
  (bind _returnvalue_2 (Utilities|GetValueOfScalableFloat _hop_upamount_onground))
  (bind _hop_upamount_inair (Variables|/RocketRamGameplay/Items/RocketRam/DataClass_RocketRam.DataClass_RocketRam_C|GetHop_UpAmount_InAir _configdata))
  (bind _returnvalue_3 (Utilities|GetValueOfScalableFloat _hop_upamount_inair))
  (bind _returnvalue_4 (Math|Vector|SelectVector (Math|Vector|MakeVector 0.0 0.0 _returnvalue_3) (Math|Vector|MakeVector 0.0 0.0 _returnvalue_2) _returnvalue))
  (bind _hop_forwardamount (Variables|/RocketRamGameplay/Items/RocketRam/DataClass_RocketRam.DataClass_RocketRam_C|GetHop_ForwardAmount _configdata))
  (bind _returnvalue_5 (Utilities|GetValueOfScalableFloat _hop_forwardamount))
  ;[0.1025 1 0.6739] On Enter State
  ;;[0.02924 0.08066 1] Pre Thrust Hop (player jumps up before they take off)
  (bind _asynctaskproxy (Ability|Tasks|WaitNetSync "BothWait")
    (:then
      (Variables|Transient|SetPreThrustHopNetSyncAsyncTask _asynctaskproxy))
    (:OnSync
      (bind _playerpawn_2 (Variables|Default|GetPlayerPawn)
        (:then
          (if (GameplayTags|HasMatchingGameplayTag _playerpawn_2 "(TagName=\"MovementMode.Grinding\")")
            (CallFunction|SetGrindingBlocked true) ; Adds a loose tag that will force you off \ngrind rails by the next frame.
            (Utilities|FlowControl|DelayUntilNextTick))
          (bind _launchdirection (Default|CalculateLaunchDirection))
          (bind _playerpawn_3 (Variables|Default|GetPlayerPawn)
            (:then
              (Character|LaunchCharacter _playerpawn_3 (+ (* _launchdirection _returnvalue_5) _returnvalue_4) true true)
              (Class|RocketRamSharedLibrary|RamLog (Variables|Default|GetPlayerPawn) (Utilities|Text|FormatText "NSLOCTEXT(\"[63668553DC3B7B41603CFE6FD08FF0F2]\", \"8B88329D445039D5EC4651AF1F872ED8\", \"Movement - Pre Thrust Launch: {0}\")" (Utilities|String|ToString(Vector) (+ (* _launchdirection _returnvalue_5) _returnvalue_4))) false true))
            (:else)))
        (:else)))))

(event Custom|StartDestructionTrace
  ;[1 0.2889 0.9783] Destruction Trace
  (Variables|Transient|DestructionTrace|SetLastDestructionTraceTime (- (Utilities|Time|GetTimeSeconds) (Utilities|Time|GetWorldDeltaSeconds)))
  (Variables|Transient|DestructionTrace|SetbIsDestructionTraceActive true)
  (CallFunction|DoDestructionTrace))

(event Custom|StopDestructionTrace
  ;[1 0.2889 0.9783] Destruction Trace
  (Variables|Transient|DestructionTrace|SetbIsDestructionTraceActive false)
  (Utilities|Time|ClearandInvalidateTimerbyHandle (Variables|Transient|GetDestructionTraceTimerHandle)))

(event Custom|DoDestructionTrace
  ;[1 0.2889 0.9783] Destruction Trace
  (if (Variables|Transient|DestructionTrace|GetbIsDestructionTraceActive)
    (Utilities|FlowControl|Sequence
      (:then_0
        (bind _playerpawn (Variables|Default|GetPlayerPawn)
          (:then
            (bind (_pawnhits _vehiclehits _destructiblebuildinghits _blockinghit) (DestructionTrace|DestructionTrace _playerpawn (Utilities|Operators|Multiply (Pawn|Components|CharacterMovement|GetLastUpdateVelocity (Variables|/Script/Engine.Character|GetCharacterMovement _playerpawn)) (- (Utilities|Time|GetTimeSeconds) (Variables|Transient|DestructionTrace|GetLastDestructionTraceTime)) (ScalableFloat|GetValueatLevel (Variables|/RocketRamGameplay/Items/RocketRam/DataClass_RocketRam.DataClass_RocketRam_C|GetDestructionTrace_DistanceMultiplier (Variables|Config|GetConfigData)))) (Math|Vector|VectorLength (Utilities|Operators|Multiply (Pawn|Components|CharacterMovement|GetLastUpdateVelocity (Variables|/Script/Engine.Character|GetCharacterMovement _playerpawn)) (- (Utilities|Time|GetTimeSeconds) (Variables|Transient|DestructionTrace|GetLastDestructionTraceTime)) (ScalableFloat|GetValueatLevel (Variables|/RocketRamGameplay/Items/RocketRam/DataClass_RocketRam.DataClass_RocketRam_C|GetDestructionTrace_DistanceMultiplier (Variables|Config|GetConfigData))))) (select (Utilities|Enum|Equal(Enum) (Variables|Default|GetCurrentRamState) "NewEnumerator6") "NewEnumerator7" "NewEnumerator8") (Variables|Default|GetCurrentRamState) (Variables|Transient|DestructionTrace|GetDestructionTraceIgnoredActors) (Variables|Debug|GetbEnableDebugDesctructionTrace) (CallFunction|GetTimeBeenStoppedInActiveThrust)))
            (Utilities|FlowControl|Sequence
              (:then_0
                ;;[0.07198 0.9602 1] Before we process this hit, add what we hit to our list of actors to ignore (so we won't deal double damage.)
                ;; Then play cosmetics.
                (bind (_actorstoignore _actorsforcameratoignore _damaged_a_build _damaged_a_vehicle _damaged_a_pawn) (DestructionTrace|HitHandling|HandleDestructionTraceHits (Variables|Default|GetPlayerPawn) (Variables|Default|GetCurrentRamState) _pawnhits _vehiclehits _destructiblebuildinghits)) ; Hanlde the hits after performing the AOE and switching \nstates so the hit actor is still valid for those checks.
                (Utilities|Array|AppendArray (Variables|Transient|DestructionTrace|GetDestructionTraceIgnoredActors) _actorstoignore)
                (Utilities|Array|AppendArray (Variables|Transient|DestructionTrace|GetDestructionTraceCameraIgnoreActors) _actorsforcameratoignore)
                (CallFunction|PlayDestructionEffects _damaged_a_build _damaged_a_vehicle _damaged_a_pawn))
              (:then_1
                ;;[1 0.586 0.1692] Process that the rocket was blocked
                (Utilities|IsValid (.HitActor (Collision|BreakHitResult _blockinghit))
                  (:Is%20Valid
                    (bind _weapon (Variables|Default|GetWeapon)
                      (:then
                        (if (Variables|/RocketRamGameplay/Items/RocketRam/B_Weapon_RocketRam.B_Weapon_RocketRam_C|GetCheat_BlockingDisabled _weapon)
                          (Utilities|Array|Add (Variables|Transient|DestructionTrace|GetDestructionTraceIgnoredActors) (.HitActor (Collision|BreakHitResult _blockinghit)))
                          (else
                            (CallFunction|OnRamBlocked _blockinghit))))
                      (:else)))))
              (:then_2
                ;;[0.897 0 1] Maybe do some debug drawing.
                (bind _weapon_2 (Variables|Default|GetWeapon)
                  (:then
                    (if (Variables|/RocketRamGameplay/Items/RocketRam/B_Weapon_RocketRam.B_Weapon_RocketRam_C|GetCheat_DebugDrawRamArc _weapon_2)
                      (Rendering|Debug|DrawDebugPoint (Transformation|GetActorLocation (Variables|Default|GetPlayerPawn)) 5.0 (Math|Color|SelectColor "(R=1.000000,G=0.000000,B=0.040508,A=1.000000)" "(R=0.096205,G=0.000000,B=1.000000,A=1.000000)" (Networking|HasAuthority (Variables|Default|GetPlayerPawn))) 3e+01)))
                  (:else)))))
          (:else)))
      (:then_1
        ;;[0.2943 0 1] Set up the next destruction trace.
        (Variables|Default|SetPlayerLocationAtLastDestructionTrace (Transformation|GetActorLocation (Variables|Default|GetPlayerPawn)))
        (Variables|Transient|DestructionTrace|SetLastDestructionTraceTime (Utilities|Time|GetTimeSeconds))
        (bind _returnvalue (Utilities|Time|SetTimerforNextTickbyEvent (EventDispatchers|CreateEvent :!event Custom|DoDestructionTrace)))
        (Variables|Transient|SetDestructionTraceTimerHandle _returnvalue)
        (Variables|Transient|SetDidAtLeastOneDestructionTrace true)))))

(event Custom|OnRamBlocked
  (params (BlockingHitResult :type Hit%20Result%20Structure) (CalledFromClient :type Boolean))
  (bind (_bblockinghit _binitialoverlap _time _distance _location _impactpoint) (Collision|BreakHitResult (Variables|Transient|GetLastBlockingHitResult)))
  (bind _currentramstate (Variables|Default|GetCurrentRamState))
  (bind _playerpawn (Variables|Default|GetPlayerPawn))
  ;[1 0.2889 0.9783] Destruction Trace
  (Class|RocketRamSharedLibrary|RamLog (Variables|Default|GetPlayerPawn) (Utilities|Text|FormatText "NSLOCTEXT(\"[63668553DC3B7B41603CFE6FD08FF0F2]\", \"0814823C4C2B22688CCC19AB4DC8FD25\", \"OnRamBlocked(), CurrentRamState: [{RamState}], BlockAlready: [{BlockHandledThisMovement}], CalledFromClient: [{CalledFromClient}].\")" (Utilities|String|EnumtoString _currentramstate) (Variables|Default|GetHasBlockBeenHandled) CalledFromClient) false true)
  (switch Utilities|FlowControl|Switch|SwitchonEnum_RocketRamState _currentramstate
    (:NewEnumerator1
      (if (not CalledFromClient)
        (Ability|HasAuthority
          (:False
            (Utilities|IsValid self
              (:Is%20Valid
                (CallFunction|OnRamBlockedtoServer BlockingHitResult) ; Active Thrust blocking needs to be client \nauthoritative because the method it uses to\ndetermine a blockage can sometimes trigger \nfor client but not for server and vice versa
              ))))))
    (:NewEnumerator2
      _)
    (:NewEnumerator5
      _))
  (Variables|Transient|SetLastBlockingHitResult BlockingHitResult)
  (if (not (Variables|Default|GetHasBlockBeenHandled))
    (Utilities|FlowControl|Sequence
      (:then_0
        (Variables|Default|SetHasBlockBeenHandled true)
        (Utilities|Time|ClearandInvalidateTimerbyHandle (Variables|Transient|GetHandleClientReplicatedBlockingHitTimerHandle))
        (Ability|HasAuthority
          (:False
            (Utilities|IsValid self
              (:Is%20Valid
                (CallFunction|ServerNotifyClientBlocked (Variables|Transient|GetLastBlockingHitResult))
                (Default|ForceNetUpdate))))))
      (:then_1
        (bind _playerpawn_2 (Variables|Default|GetPlayerPawn)
          (:then
            (if (Utilities|Enum|Equal(Enum) (Variables|Default|GetCurrentRamState) "NewEnumerator1")
              (Variables|Transient|SetLastImpactDirection (Variables|Transient|GetLaunchDirection))
              (else
                (bind _result (AOE|ShouldDoAOEImpact _playerpawn_2 (Pawn|Components|CharacterMovement|GetLastUpdateVelocity (Variables|/Script/Engine.Character|GetCharacterMovement _playerpawn_2)) (Variables|Transient|GetLastBlockingHitResult)))
                (if _result
                  (Variables|Transient|SetLastImpactDirection (Math|Vector|Normalize (Pawn|Components|CharacterMovement|GetLastUpdateVelocity (Variables|/Script/Engine.Character|GetCharacterMovement _playerpawn_2)))))))
            (CallFunction|DoFinalDestructionTrace (Variables|Transient|GetLastBlockingHitResult))
            (Ability|HasAuthority
              (:True
                (CallFunction|DoAOEImpact _impactpoint (CallFunction|ShouldBlockingHitTriggerPostAOEPause (Variables|Transient|GetLastBlockingHitResult)) (Variables|Transient|GetLastImpactDirection)))))
          (:else)))
      (:then_2
        ;;[0.1656 0.9602 1] IF the ram hasn't been turned off, then enter the Recovery state.
        (if (Utilities|Enum|NotEqual(Enum) (Variables|Default|GetCurrentRamState) "NewEnumerator4")
          (Class|RocketRamSharedLibrary|RamLog _playerpawn (Utilities|Text|FormatText "NSLOCTEXT(\"[63668553DC3B7B41603CFE6FD08FF0F2]\", \"9C3D7BD64189E9C25BEE37A14E309CB6\", \"OnRamBlocked, setting Ram State to Recovery.\")") false true)
          (Default|SetRamState "NewEnumerator7")
          (else
            (Class|RocketRamSharedLibrary|RamLog _playerpawn (Utilities|Text|FormatText "NSLOCTEXT(\"[63668553DC3B7B41603CFE6FD08FF0F2]\", \"D542EDE14D09891CBD2139BC32E8BA3D\", \"OnRamBlocked(), Ram State is already Off, purposefully NOT setting Ram State to Recovery.\")") false true)))))))

(event Custom|OnRamBlocked_ToServer
  (params (BlockingHitResult :type Hit%20Result%20Structure))
  ;[1 0.2889 0.9783] Destruction Trace
  (CallFunction|OnRamBlocked BlockingHitResult true))

(event Custom|Server_NotifyClientBlocked
  (params (ClientHit :type Hit%20Result%20Structure))
  ;[1 0.2889 0.9783] Destruction Trace
  ;;[1 0 0.5261] Notify Client Blocked. Failsafe for if the client detects an AOE that the server doesn't. We don't do this immediately because it's possible for the server to just be a few frames behind the client.
  (Variables|Transient|DestructionTrace|SetClientReplicatedBlockingHit ClientHit)
  (bind _returnvalue (Utilities|Time|SetTimerbyEvent (EventDispatchers|CreateEvent :!event Custom|HandleClientReplicatedBlockingHit) 0.1))
  (Variables|Transient|SetHandleClientReplicatedBlockingHitTimerHandle _returnvalue))

(event Custom|HandleClientReplicatedBlockingHit
  (bind _currentramstate (Variables|Default|GetCurrentRamState))
  (bind _playerpawn (Variables|Default|GetPlayerPawn))
  ;[1 0.2889 0.9783] Destruction Trace
  ;;[1 0 0.5261] Notify Client Blocked. Failsafe for if the client detects an AOE that the server doesn't. We don't do this immediately because it's possible for the server to just be a few frames behind the client.
  (Class|RocketRamSharedLibrary|RamLog _playerpawn (Utilities|Text|FormatText "NSLOCTEXT(\"[63668553DC3B7B41603CFE6FD08FF0F2]\", \"B57A86524B1D6F074919DCB147B8682C\", \"HandleClientReplicatedBlockingHit, CurrentRamState: [{RamState}].\")" (Utilities|String|EnumtoString _currentramstate)))
  (switch Utilities|FlowControl|Switch|SwitchonEnum_RocketRamState _currentramstate
    (:NewEnumerator1
      _)
    (:NewEnumerator2
      _)
    (:NewEnumerator5
      _))
  (Class|RocketRamSharedLibrary|RamLog _playerpawn (Utilities|Text|FormatText "NSLOCTEXT(\"[63668553DC3B7B41603CFE6FD08FF0F2]\", \"DE3D66604FA64C6FE931C39C9F9FB039\", \"Movement - Client predicted hit not found by server! Blocking movement on server using client replicated hit result..\")") true)
  (CallFunction|OnRamBlocked :CalledFromClient false) ; If the client predicts a blocking hit that the server doesn't notice, \nwe still want to treat the hit as a block. The client notifies the server \nwhenever it's blocked and if the server hasn't been blocked by the \ntime the message is replicated, we simulate the block on the server \nwith the same hit.
)

(event Custom|BeginCheckForGrinding
  ;[0 1 0.9705] Wait Movement Mode Changed
  ;;[1 1 1] Cancel on started grinding, but restart the wait if the tag is removed (i.e. if you leave then re-hit a grind rail). Keep 'Only Trigger Once' on for safety.
  (bind _asynctaskproxy (Ability|Tasks|WaitGameplayTagAdd "(TagName=\"MovementMode.Grinding\")" 0 true)
    (:then
      (Variables|Transient|SetWaitGrindingAsyncTask _asynctaskproxy))
    (:Added
      (switch Utilities|FlowControl|Switch|SwitchonEnum_RocketRamState (Variables|Default|GetCurrentRamState)
        (:NewEnumerator0
          _)
        (:NewEnumerator6
          _)
        (:NewEnumerator1
          _)
        (:NewEnumerator2
          (Class|RocketRamSharedLibrary|RamLog (Variables|Default|GetPlayerPawn) (Utilities|Text|FormatText "NSLOCTEXT(\"[63668553DC3B7B41603CFE6FD08FF0F2]\", \"41D879A946E3060FE047C3A35250291E\", \"Movement - Started grinding, canceling movement.\")") false true :!label _label)
          (Ability|EndAbility))
        (:NewEnumerator5
          (goto _label))
        (:NewEnumerator3
          (goto _label))
        (:NewEnumerator7
          (goto _label)))
      (bind _asynctaskproxy_2 (Ability|Tasks|WaitGameplayTagRemove "(TagName=\"MovementMode.Grinding\")" 0 true)
        (:then
          (Variables|Transient|SetWaitGrindingEndAsyncTask _asynctaskproxy_2))
        (:Removed
          (CallFunction|BeginCheckforGrinding))) ; If the player is grinding when they initiate the Rocket Ram, this will instantly fire. So we need to reset the wait.
    )))

(event Custom|StopWaitingForMovementModeChange
  ;[0 1 0.9705] Wait Movement Mode Changed
  (GameplayTasks|EndTask (Variables|Transient|GetWaitGrindingAsyncTask)))

(event Custom|DoAOEImpact
  (params (ImpactLocation :type Vector) (bDoPostAOEPause :type Boolean) (ImpactDirection :type Vector))
  ;[1 0.2889 0.9783] Destruction Trace
  ;;[1 0.3083 0.9188] Do AOE Impact (Server Auth) - See AOE Impact Library
  (Ability|HasAuthority
    (:True
      (Class|RocketRamSharedLibrary|RamLog (Variables|Default|GetPlayerPawn) "NSLOCTEXT(\"[63668553DC3B7B41603CFE6FD08FF0F2]\", \"F87BBB0144D75813E998B6A525040C69\", \"DoAOEImpact\")")
      (Variables|Default|SetCurrentCustomTargetingTransform (Math|Transform|MakeTransform (Transformation|GetActorLocation (Variables|Default|GetPlayerPawn))))
      (bind (_asynctaskproxy _targetdata) (Ability|Tasks|DoTargetSelection "Rocket Ram AOE Impact Targeting" "(TagName=\"Abilities.Generic.GenericProjectile.Explode\")" true)
        (:Targeted
          (AOE|AOEImpact (Variables|Default|GetPlayerPawn) ImpactLocation (Variables|Transient|DestructionTrace|GetDestructionTraceIgnoredActors) (Variables|Config|GetGC_AOEImpact) _targetdata)
          (Class|RocketRamSharedLibrary|RamLog (Variables|Default|GetPlayerPawn) (Utilities|Text|FormatText "NSLOCTEXT(\"[63668553DC3B7B41603CFE6FD08FF0F2]\", \"086CFD424C6B06B8C696A49FB4F6B250\", \"Movement - AOE Requested\")") false true)
          (if bDoPostAOEPause
            (CallFunction|StartPostAOEPause)
            (bind _weapon (Variables|Default|GetWeapon)
              (:then
                (Class|BWeaponRocketRam|SetImpactDirection _weapon ImpactDirection :!class /RocketRamGameplay/Items/RocketRam/B_Weapon_RocketRam.B_Weapon_RocketRam_C))
              (:else))))))))

(event Custom|StartWaitingForMovementModeChange
  ;[0 1 0.9705] Wait Movement Mode Changed
  (Utilities|FlowControl|Sequence
    (:then_0
      ;;[1 1 1] Cancel on started grinding, but restart the wait if the tag is removed (i.e. if you leave then re-hit a grind rail). Keep 'Only Trigger Once' on for safety.
      (bind _asynctaskproxy (Ability|Tasks|WaitGameplayTagAdd "(TagName=\"MovementMode.Grinding\")" 0 true)
        (:then
          (Variables|Transient|SetWaitGrindingAsyncTask _asynctaskproxy))
        (:Added
          (switch Utilities|FlowControl|Switch|SwitchonEnum_RocketRamState (Variables|Default|GetCurrentRamState)
            (:NewEnumerator0
              _)
            (:NewEnumerator6
              _)
            (:NewEnumerator1
              _)
            (:NewEnumerator2
              (Class|RocketRamSharedLibrary|RamLog (Variables|Default|GetPlayerPawn) (Utilities|Text|FormatText "NSLOCTEXT(\"[63668553DC3B7B41603CFE6FD08FF0F2]\", \"41D879A946E3060FE047C3A35250291E\", \"Movement - Started grinding, canceling movement.\")") false true :!label _label)
              (Ability|EndAbility))
            (:NewEnumerator5
              (goto _label))
            (:NewEnumerator3
              (goto _label))
            (:NewEnumerator7
              (goto _label)))
          (bind _asynctaskproxy_2 (Ability|Tasks|WaitGameplayTagRemove "(TagName=\"MovementMode.Grinding\")" 0 true)
            (:then
              (Variables|Transient|SetWaitGrindingEndAsyncTask _asynctaskproxy_2))
            (:Removed
              (CallFunction|BeginCheckforGrinding))) ; If the player is grinding when they initiate the Rocket Ram, this will instantly fire. So we need to reset the wait.
        )))
    (:then_1
      ;;[1 1 1] Cancel on entered water
      (bind _asynctaskproxy_3 (Ability|Tasks|WaitGameplayTagAdd "(TagName=\"MovementMode.SurfaceSwimming\")")
        (:then
          (Variables|Transient|SetWaitSwimmingAsyncTask _asynctaskproxy_3))
        (:Added
          (switch Utilities|FlowControl|Switch|SwitchonEnum_RocketRamState (Variables|Default|GetCurrentRamState)
            (:NewEnumerator2
              _)
            (:NewEnumerator5
              _)
            (:NewEnumerator3
              _)
            (:NewEnumerator7
              _))
          (Class|RocketRamSharedLibrary|RamLog (Variables|Default|GetPlayerPawn) (Utilities|Text|FormatText "NSLOCTEXT(\"[63668553DC3B7B41603CFE6FD08FF0F2]\", \"8A3665E94556C93010CFF69C37CAD69D\", \"Movement - Started swimming, canceling movement.\")") false true)
          (Ability|EndAbility))))
    (:then_3
      ;;[1 1 1] Cancel on started ziplining
      (Utilities|FlowControl|Sequence
        (:then_0
          (bind _asynctaskproxy_4 (Ability|Tasks|WaitGameplayTagAdd "(TagName=\"MovementMode.ZipliningOnSpline\")")
            (:then
              (Variables|Transient|SetWaitZipliningOnSplineAsyncTask _asynctaskproxy_4))
            (:Added
              _)))
        (:then_1
          (bind _asynctaskproxy_5 (Ability|Tasks|WaitGameplayTagAdd "(TagName=\"MovementMode.Ziplining\")" 0 true)
            (:then
              (Variables|Transient|SetWaitZipliningAsyncTask _asynctaskproxy_5))
            (:Added
              _))))
      (switch Utilities|FlowControl|Switch|SwitchonEnum_RocketRamState (Variables|Default|GetCurrentRamState)
        (:NewEnumerator2
          _)
        (:NewEnumerator5
          _)
        (:NewEnumerator3
          _)
        (:NewEnumerator7
          _))
      (Class|RocketRamSharedLibrary|RamLog (Variables|Default|GetPlayerPawn) (Utilities|Text|FormatText "NSLOCTEXT(\"[63668553DC3B7B41603CFE6FD08FF0F2]\", \"33D0460F49AEC4B3CF7CED9A86BC0346\", \"Movement - Started ziplining, canceling movement.\")") false true)
      (Ability|EndAbility))
    (:then_4
      ;;[1 1 1] Cancel on started Parachuting
      (bind _asynctaskproxy_6 (Ability|Tasks|WaitGameplayTagAdd "(TagName=\"MovementMode.Parachuting\")" 0 true)
        (:then
          (Variables|Transient|SetWaitParachutingAsyncTask _asynctaskproxy_6))
        (:Added
          (Class|RocketRamSharedLibrary|RamLog (Variables|Default|GetPlayerPawn) (Utilities|Text|FormatText "NSLOCTEXT(\"[63668553DC3B7B41603CFE6FD08FF0F2]\", \"2656DDB947B15A18E4E640AC34C20E59\", \"Movement - Started parachuting, canceling movement.\")") false true)
          (Ability|EndAbility))))))

(event Custom|OnExitRamState
  (params (PreviousState :type Enum_RocketRamState%20Enum :default "NewEnumerator0"))
  ;[1 0.4181 0.09459] On Exit State
  (switch Utilities|FlowControl|Switch|SwitchonEnum_RocketRamState PreviousState
    (:NewEnumerator0
      ;;[0.07432 0.3973 1] On State Changed From Winding Up
      (Utilities|Time|ClearandInvalidateTimerbyHandle (Variables|Transient|GetWindUpTimerHandle))
      (Ability|RemoveGameplayEffectFromOwnerWithHandle (Variables|Transient|GetGE_State_WindingUp))
      (CallFunction|StopListeningtoPawnLandedCancelTimeDilation))
    (:NewEnumerator6
      ;;[0.02924 0.08066 1] On State Changed From Slow Acceleration
      (Utilities|Time|ClearandInvalidateTimerbyHandle (Variables|Transient|GetSlowAccelerationTimerHandle))
      (GameplayTasks|EndTask (Variables|Transient|GetWaitLaunchDirectionReplicatedAsyncTask))
      (CallFunction|StopTimeDilation)
      (CallFunction|StopPredictingLaunchDirection)
      (Ability|RemoveGameplayEffectFromOwnerWithHandle (Variables|Transient|GetGE_State_Hopping)))
    (:NewEnumerator1
      ;;[0.7199 0.1195 1] On State Changed From Active Thrust
      (GameplayTasks|EndTask (Variables|Transient|GetRootMotionAsyncTask))
      (Ability|RemoveGameplayEffectFromOwnerWithHandle (Variables|Transient|GetGE_State_ActiveThrust))
      (CallFunction|SetGrindingBlocked false)
      (Ability|RemoveGameplayEffectFromOwnerWithHandle (Variables|Transient|GetGE_BlockGrindingButDontCancel)) ; Make sure any existing one is removed
      (if (ScalableFloat|GetValueasBool (Variables|/RocketRamGameplay/Items/RocketRam/DataClass_RocketRam.DataClass_RocketRam_C|GetAllowMidairWeaponSwitching (Variables|Config|GetConfigData)))
        (Ability|RemoveGameplayEffectFromOwnerWithHandle (Variables|Transient|GetGE_BlockWeaponSwitching)) ; Make sure any existing one is removed
      ))
    (:NewEnumerator2
      _)
    (:NewEnumerator5
      _)
    (:NewEnumerator3
      ;;[1 0.9123 0.1124] On State Changed From Braking
      (bind _weapon (Variables|Default|GetWeapon)
        (:then
          (Class|BWeaponRocketRam|HolsterEquippedWeapon _weapon false (Utilities|Time|GetTimerRemainingTimebyHandle (Variables|Transient|GetBrakingTimerHandle)))
          (CallFunction|StopBraking)
          (Ability|RemoveGameplayEffectFromOwnerWithHandle (Variables|Transient|GetGE_State_Braking)))
        (:else)))
    (:NewEnumerator7
      ;;[0.08993 1 0.1404] On State Changed From Recovery
      (bind _weapon_2 (Variables|Default|GetWeapon)
        (:then
          (Class|BWeaponRocketRam|HolsterEquippedWeapon _weapon_2 false (Utilities|Time|GetTimerRemainingTimebyHandle (Variables|Transient|GetRecoveryTimerHandle)))
          (CallFunction|StopRecovery)
          (Ability|RemoveGameplayEffectFromOwnerWithHandle (Variables|Transient|GetGE_State_Recovering))
          (bind _weapon_3 (Variables|Default|GetWeapon)
            (:then
              (if (<= (FortWeapon|GetRemainingAmmo _weapon_3) 0)
                (bind _playerpawn (Variables|Default|GetPlayerPawn)
                  (:then
                    (Utilities|IsValid (Fortnite|GetFortPlayerControllerfromActor _playerpawn)
                      (:Is%20Valid
                        (Inventory|RemoveItemFromPlayer :PlayerController (Fortnite|GetFortPlayerControllerfromActor _playerpawn) :ItemDefinition "/RocketRamGameplay/Items/RocketRam/WID_RocketRam.WID_RocketRam" :AmountToRemove -1 :bForceRemoval true :bSkipVerbRemoval false))))
                  (:else))))
            (:else)))
        (:else))))
  ;;[1 0.1707 0.6226] On State Changed From Free Fall
  (CallFunction|StopLookaheadTrace)
  (Ability|RemoveGameplayEffectFromOwnerWithHandle (Variables|Transient|GetGE_State_FreeFalling))
  (CallFunction|StopWaitingforFreeFallTooLong)
  (Ability|RemoveGameplayEffectFromOwnerWithHandle (Variables|Transient|GetGE_BlockWeaponSwitching)) ; Make sure any existing one is removed
  (Ability|RemoveGameplayEffectFromOwnerWithHandle (Variables|Transient|GetGE_InAir)) ; Make sure any existing one is removed
  (bind _weapon_4 (Variables|Default|GetWeapon)
    (:then
      (Class|BWeaponRocketRam|HolsterEquippedWeapon _weapon_4 true 0.0))
    (:else)))

(event Custom|DoFinalDestructionTrace
  (params (BlockingHitResult :type Hit%20Result%20Structure))
  (bind (_bblockinghit _binitialoverlap _time _distance) (Collision|BreakHitResult BlockingHitResult))
  ;[1 0.2889 0.9783] Destruction Trace
  ;;[0.5068 0.1131 1] Final Destruction Trace. This is a destruction trace that is done after the movement is blocked that goes the distance from the player to the blocking hit plus some standard distance. This ensures that the destruction consistently destroyes buildings that are just behind the impacted building (such as floors that are behind an impacted wall).
  (bind _playerpawn (Variables|Default|GetPlayerPawn)
    (:then
      (bind (_pawnhits _vehiclehits _destructiblebuildinghits) (DestructionTrace|DestructionTrace _playerpawn (Pawn|Components|CharacterMovement|GetLastUpdateVelocity (Variables|/Script/Engine.Character|GetCharacterMovement _playerpawn)) (+ (ScalableFloat|GetValueatLevel (Variables|/RocketRamGameplay/Items/RocketRam/DataClass_RocketRam.DataClass_RocketRam_C|GetDestructionTrace_FinalTraceDistance (Variables|Config|GetConfigData))) _distance) "NewEnumerator8" (Variables|Default|GetCurrentRamState) (Variables|Transient|DestructionTrace|GetDestructionTraceIgnoredActors) (Variables|Debug|GetbEnableDebugDesctructionTrace)))
      (bind (_actorstoignore _actorsforcameratoignore) (DestructionTrace|HitHandling|HandleDestructionTraceHits (Variables|Default|GetPlayerPawn) (Variables|Default|GetCurrentRamState) _pawnhits _vehiclehits _destructiblebuildinghits))
      (Utilities|Array|AppendArray (Variables|Transient|DestructionTrace|GetDestructionTraceIgnoredActors) _actorstoignore)
      (Utilities|Array|AppendArray (Variables|Transient|DestructionTrace|GetDestructionTraceCameraIgnoreActors) _actorsforcameratoignore))
    (:else)))

(event Custom|StartListeningForPawnLanded
  ;[0.05203 1 0.9901] Pawn Landed
  (bind _playerpawn (Variables|Default|GetPlayerPawn)
    (:then
      (if (AI|Components|NavMovement|IsFalling (Variables|/Script/Engine.Character|GetCharacterMovement _playerpawn))
        (EventDispatchers|BindEventtoOnPawnLanded _playerpawn (EventDispatchers|CreateEvent :!event Custom|OnPawnLanded))
        (else
          (CallFunction|OnPawnLanded))))
    (:else)))

(event Custom|Start%20Waiting%20For%20Underwater
  ;[0 1 0.9705] Wait Movement Mode Changed
  ;;[1 1 1] Cancel on Underwater
  (Ability|HasAuthority
    (:True
      (bind _asynctaskproxy (Ability|Tasks|WaitGameplayTagAdd "(TagName=\"Gameplay.Status.Underwater\")")
        (:then
          (Variables|Transient|SetWaitUnderwaterAsyncTask _asynctaskproxy))
        (:Added
          (switch Utilities|FlowControl|Switch|SwitchonEnum_RocketRamState (Variables|Default|GetCurrentRamState)
            (:NewEnumerator1
              _)
            (:NewEnumerator2
              _)
            (:NewEnumerator5
              _)
            (:NewEnumerator3
              _)
            (:NewEnumerator7
              _))
          (Class|RocketRamSharedLibrary|RamLog (Variables|Default|GetPlayerPawn) (Utilities|Text|FormatText "NSLOCTEXT(\"[63668553DC3B7B41603CFE6FD08FF0F2]\", \"A9F1FA44458C089F3700DF9192B6F9FC\", \"Movement - Went underwater, canceling movement.\")") false true)
          (Ability|EndAbility))) ; NOTE: This tag is only ever added on the server
    )))

(event Custom|StartIncreasingGravity
  ;[1 0.2333 0.7025] Gravity
  (bind _weapon (Variables|Default|GetWeapon)
    (:then
      (if (Variables|/RocketRamGameplay/Items/RocketRam/B_Weapon_RocketRam.B_Weapon_RocketRam_C|GetCheat_StaticGravityEnabled _weapon)
        (Default|SetGravityMultiplier 0.7)
        (else
          (bind _returnvalue (Utilities|Time|SetTimerbyEvent (EventDispatchers|CreateEvent :!event Custom|IncreaseGravity) (Utilities|GetValueOfScalableFloat (Variables|/RocketRamGameplay/Items/RocketRam/DataClass_RocketRam.DataClass_RocketRam_C|GetGravity_UpdateFrequency (Variables|Config|GetConfigData))) true))
          (Variables|Transient|SetGravityIncreaseTimerHandle _returnvalue)
          (CallFunction|IncreaseGravity))))
    (:else)))

(event Custom|IncreaseAirControlAtFreeFallPeak
  ;[0.5846 0.3086 1] Air Control
  (bind _returnvalue (Utilities|Time|SetTimerbyEvent (EventDispatchers|CreateEvent :!event Custom|CheckHasHitFreeFallPeak) 0.2 true))
  (Variables|Transient|SetAirControlIncreaseTimerHandle _returnvalue)
  (CallFunction|CheckHasHitFreeFallPeak))

(event Custom|StopListeningForPawnLanded
  ;[0.05203 1 0.9901] Pawn Landed
  (bind _playerpawn (Variables|Default|GetPlayerPawn)
    (:then
      (EventDispatchers|UnbindEventfromOnPawnLanded _playerpawn (EventDispatchers|CreateEvent :!event Custom|OnPawnLanded)))
    (:else)))

(event Custom|StartWaitingForFreeFallTooLong
  (bind _configdata (Variables|Config|GetConfigData))
  ;[1 1 1] Wait For if Free Falling state lasts too long; if so, clamp velocity and air control to avoid going too far in one rocket ride, as it would cause streaming perf issues.
  ; This should only trigger if you activate it from VERY high
  (if (Class|AthenaFunctionLibrary|RowtoBool (Variables|/RocketRamGameplay/Items/RocketRam/DataClass_RocketRam.DataClass_RocketRam_C|GetClampVelocityIfFreeFallingForTooLong_Enabled _configdata))
    (bind _returnvalue (Utilities|Time|SetTimerbyEvent (delegate Custom|FreeFallLastedTooLong) (Utilities|GetValueOfScalableFloat (Variables|/RocketRamGameplay/Items/RocketRam/DataClass_RocketRam.DataClass_RocketRam_C|GetClampVelocityIfFreeFallingForTooLong_TimeBeforeClamp _configdata))))
    (Variables|Default|SetTimer_WaitingForFreeFallTooLong _returnvalue)))

(event Custom|StopWaitingForFreeFallPeak
  ;[0.5846 0.3086 1] Air Control
  (Utilities|Time|ClearandInvalidateTimerbyHandle (Variables|Transient|GetAirControlIncreaseTimerHandle)))

(event Custom|FreeFallLastedTooLong
  (bind _configdata (Variables|Config|GetConfigData))
  ;[1 1 1] Wait For if Free Falling state lasts too long; if so, clamp velocity and air control to avoid going too far in one rocket ride, as it would cause streaming perf issues.
  ; This should only trigger if you activate it from VERY high
  (switch Utilities|FlowControl|Switch|SwitchonEnum_RocketRamState (Variables|Default|GetCurrentRamState)
    (:NewEnumerator2
      (bind _playerpawn (Variables|Default|GetPlayerPawn)
        (:then
          (Variables|/Script/Engine.MovementComponent|SetVelocity (Math|Vector|VectorClampSizeMax2D (Variables|/Script/Engine.MovementComponent|GetVelocity (Variables|/Script/Engine.Character|GetCharacterMovement _playerpawn)) (Utilities|GetValueOfScalableFloat (Variables|/RocketRamGameplay/Items/RocketRam/DataClass_RocketRam.DataClass_RocketRam_C|GetClampVelocityIfFreeFallingForTooLong_ClampedVelocity _configdata))) (Variables|/Script/Engine.Character|GetCharacterMovement _playerpawn))
          (Default|SetAirControl (Utilities|GetValueOfScalableFloat (Variables|/RocketRamGameplay/Items/RocketRam/DataClass_RocketRam.DataClass_RocketRam_C|GetClampVelocityIfFreeFallingForTooLong_ClampedAirControl _configdata))) ; Clamping air control makes it so you can't use\ninput to increase your speed again.
        )
        (:else)))))

(event Custom|StopIncreasingGravity
  ;[1 0.2333 0.7025] Gravity
  (Utilities|Time|ClearandInvalidateTimerbyHandle (Variables|Transient|GetGravityIncreaseTimerHandle))
  (Default|SetGravityMultiplier 1.0)
  (for _array_element (Variables|Transient|GetWaitNetSyncIncreaseGravityAsyncTasks)
    (Utilities|IsValid _array_element
      (:Is%20Valid
        (GameplayTasks|EndTask _array_element)))))

(event Custom|CheckHasHitFreeFallPeak
  ;[0.5846 0.3086 1] Air Control
  (bind _playerpawn (Variables|Default|GetPlayerPawn)
    (:then
      (bind (_returnvalue_x _returnvalue_y _returnvalue_z) (Math|Vector|BreakVector (Transformation|GetVelocity _playerpawn)))
      (if (< _returnvalue_z (ScalableFloat|GetValueatLevel (Variables|/RocketRamGameplay/Items/RocketRam/DataClass_RocketRam.DataClass_RocketRam_C|GetAirControl_ZThreasholdForPeak (Variables|Config|GetConfigData))))
        (Default|SetAirControl (ScalableFloat|GetValueatLevel (Variables|/RocketRamGameplay/Items/RocketRam/DataClass_RocketRam.DataClass_RocketRam_C|GetAirControl_AfterPeakAmount (Variables|Config|GetConfigData))))
        (CallFunction|StopWaitingforFreeFallPeak)))
    (:else)))

(event Custom|OnPawnLanded
  (params (Hit :type Hit%20Result%20Structure))
  ;[0.05203 1 0.9901] Pawn Landed
  (switch Utilities|FlowControl|Switch|SwitchonEnum_RocketRamState (Variables|Default|GetCurrentRamState)
    (:NewEnumerator2
      _)
    (:NewEnumerator5
      _)
    (:NewEnumerator7
      (bind _playerpawn (Variables|Default|GetPlayerPawn)
        (:then
          (Variables|/Script/Engine.MovementComponent|SetVelocity "0, 0, 0" (Variables|/Script/Engine.Character|GetCharacterMovement _playerpawn)))
        (:else)) ; Landing while in the recovery phase means we've \nalready done an AOE, so stop the player's velocity.
    ))
  ;;[1 0.8488 0] If we have already handled a destruction trace then goto the bracking state, else attempt 1 more trace and then only enter bracking if that trace does not put us in a meaningfully different ram state.
  (if (not (Variables|Default|GetHasBlockBeenHandled))
    (Variables|Transient|DestructionTrace|SetLastDestructionTraceTime (- (Utilities|Time|GetTimeSeconds) 0.02)) ; Artifically wind back the clock a little over 1 frame to force a new trace.\nThis is necessary becaues of a flaw with the destruction trace where it uses the time since the last trace to decide the length of the next trace.\nI was not able to find a more reliable way of getting the tick rate so I am not changing that system, just working within it.
    (CallFunction|DoDestructionTrace)
    (switch Utilities|FlowControl|Switch|SwitchonEnum_RocketRamState (Variables|Default|GetCurrentRamState)
      (:NewEnumerator2
        _)
      (:NewEnumerator5
        _)))
  (Default|SetRamState "NewEnumerator3") ; If we are here, then we landed without being blocked (example: we are skiding across the ground.)\nIn this case, start braking.
)

(event Custom|IncreaseGravity
  ;[1 0.2333 0.7025] Gravity
  (bind _weapon (Variables|Default|GetWeapon)
    (:then
      (if (Variables|/RocketRamGameplay/Items/RocketRam/B_Weapon_RocketRam.B_Weapon_RocketRam_C|GetCheat_GravityWaitSyncEnabled _weapon)
        (bind _asynctaskproxy (Ability|Tasks|WaitNetSync "BothWait")
          (:then
            (Utilities|Array|Add (Variables|Transient|GetWaitNetSyncIncreaseGravityAsyncTasks) _asynctaskproxy))
          (:OnSync
            _)))
      (Default|SetGravityMultiplier (Class|RocketRamSharedLibrary|CalculateNextGravity (Variables|Transient|Gravity|GetCurrentGravity) (Transformation|GetVelocity (Variables|Default|GetPlayerPawn)) (Variables|Transient|GetLaunchDirection) (Variables|Transient|Gravity|GetBaseGravity) (Utilities|GetValueOfScalableFloat (Variables|/RocketRamGameplay/Items/RocketRam/DataClass_RocketRam.DataClass_RocketRam_C|GetGravity_UpdateFrequency (Variables|Config|GetConfigData))))))
    (:else)))

(event Custom|StopWaitingForFreeFallTooLong
  ;[1 1 1] Wait For if Free Falling state lasts too long; if so, clamp velocity and air control to avoid going too far in one rocket ride, as it would cause streaming perf issues.
  ; This should only trigger if you activate it from VERY high
  (Utilities|Time|ClearandInvalidateTimerbyHandle (Variables|Default|GetTimer_WaitingForFreeFallTooLong)))

(event Custom|StartBraking
  ;[1 0.9123 0.1124] Braking Phase Timer
  (bind _returnvalue (Utilities|Time|SetTimerbyEvent (EventDispatchers|CreateEvent :!event Custom|OnBrakingComplete) (Utilities|GetValueOfScalableFloat (Variables|/RocketRamGameplay/Items/RocketRam/DataClass_RocketRam.DataClass_RocketRam_C|GetBrakingPhaseDuration (Variables|Config|GetConfigData)))))
  (Variables|Transient|SetBrakingTimerHandle _returnvalue))

(event Custom|StartRecovery
  ;[0.2598 1 0.3744] Recovery Phase Timer
  (bind _returnvalue (Utilities|Time|SetTimerbyEvent (EventDispatchers|CreateEvent :!event Custom|OnRecoveryComplete) (Utilities|GetValueOfScalableFloat (Variables|/RocketRamGameplay/Items/RocketRam/DataClass_RocketRam.DataClass_RocketRam_C|GetRecoveryPhaseDuration (Variables|Config|GetConfigData)))))
  (Variables|Transient|SetRecoveryTimerHandle _returnvalue))

(event Custom|StartTimeDilation
  (params (TargetTimeDilation :type Float%20%28double-precision%29) (EaseInDuration :type Float%20%28double-precision%29))
  ;[0.1637 0.4503 1] Time Dilation
  (Variables|Transient|TimeDilation|SetCurrentTargetTimeDilation TargetTimeDilation)
  (Variables|Transient|TimeDilation|SetCurrentTimeDilationEaseInDuration EaseInDuration)
  (Variables|Transient|TimeDilation|SetTimeDilationEaseInStartTime (Utilities|Time|GetTimeSeconds))
  (Utilities|Time|ClearandInvalidateTimerbyHandle (Variables|Transient|GetTimeDilationEaseInTimerHandle))
  (bind _returnvalue (Utilities|Time|SetTimerbyEvent (EventDispatchers|CreateEvent :!event Custom|UpdateTimeDilationEaseIn) 0.05 true))
  (Variables|Transient|SetTimeDilationEaseInTimerHandle _returnvalue))

(event Custom|StopBraking
  ;[1 0.9123 0.1124] Braking Phase Timer
  (Utilities|Time|ClearandInvalidateTimerbyHandle (Variables|Transient|GetBrakingTimerHandle)))

(event Custom|StopRecovery
  ;[0.2598 1 0.3744] Recovery Phase Timer
  (Utilities|Time|ClearandInvalidateTimerbyHandle (Variables|Transient|GetRecoveryTimerHandle)))

(event Custom|StopTimeDilation
  ;[0.1637 0.4503 1] Time Dilation
  (Utilities|Time|ClearandInvalidateTimerbyHandle (Variables|Transient|GetTimeDilationEaseInTimerHandle))
  (CallFunction|SetTimeDilation 1.0))

(event Custom|OnBrakingComplete
  ;[1 0.9123 0.1124] Braking Phase Timer
  (Default|SetRamState "NewEnumerator4"))

(event Custom|OnRecoveryComplete
  ;[0.2598 1 0.3744] Recovery Phase Timer
  (Class|RocketRamSharedLibrary|RamLog (Variables|Default|GetPlayerPawn) "NSLOCTEXT(\"[63668553DC3B7B41603CFE6FD08FF0F2]\", \"5367B894445E66CBF5DB69A2799BC630\", \"OnRecoveryComplete - setting ram state to Off.\")" false true)
  (Default|SetRamState "NewEnumerator4"))

(event Custom|UpdateTimeDilationEaseIn
  (bind _returnvalue (Utilities|Time|GetTimeSeconds))
  (bind _currenttimedilationeaseinduration (Variables|Transient|TimeDilation|GetCurrentTimeDilationEaseInDuration))
  (bind _timedilationeaseinstarttime (Variables|Transient|TimeDilation|GetTimeDilationEaseInStartTime))
  (bind _returnvalue_2 (- _returnvalue _timedilationeaseinstarttime))
  (bind _output (Utilities|RemapFloat _returnvalue_2 (Math|Vector2D|MakeVector2D 0.0 _currenttimedilationeaseinduration) "(X=0.000000,Y=1.000000)" true))
  ;[0.1637 0.4503 1] Time Dilation
  (if (>= _output 1.0)
    (Utilities|Time|ClearandInvalidateTimerbyHandle (Variables|Transient|GetTimeDilationEaseInTimerHandle))
    (else
      (Variables|Default|GetPlayerPawn
        (:then
          (bind _timedilationeaseincurve (Variables|/RocketRamGameplay/Items/RocketRam/DataClass_RocketRam.DataClass_RocketRam_C|GetTimeDilationEaseInCurve (Variables|Config|GetConfigData))
            (:then
              (CallFunction|SetTimeDilation (Utilities|RemapFloat (Math|Curves|GetFloatValue _timedilationeaseincurve _output) "(X=0.000000,Y=1.000000)" (Math|Vector2D|MakeVector2D (Variables|Transient|TimeDilation|GetCurrentTargetTimeDilation) 1.0) true)))
            (:else)))
        (:else)))))

(event Custom|SetGrindingBlocked
  (params (bIsBlocked :type Boolean))
  (bind _blockgrindrailstag (Variables|Config|GetBlockGrindRailsTag))
  ;[1 0.2517 0.7354] Block Grinding
  (bind _playerpawn (Variables|Default|GetPlayerPawn)
    (:then
      (if bIsBlocked
        (if (not (Variables|Transient|Grinding|GetbHasAppliedBlockGrindRailsTag))
          (Ability|GameplayEffect|AddLooseGameplayTags _playerpawn _blockgrindrailstag)
          (Variables|Transient|Grinding|SetbHasAppliedBlockGrindRailsTag true))
        (elif (Variables|Transient|Grinding|GetbHasAppliedBlockGrindRailsTag)
          (Ability|GameplayEffect|RemoveLooseGameplayTags _playerpawn _blockgrindrailstag)
          (Variables|Transient|Grinding|SetbHasAppliedBlockGrindRailsTag false))))
    (:else)))

(event Custom|SetTimeDilation
  (params (TimeDilation :type Float%20%28single-precision%29))
  ;[0.1637 0.4503 1] Time Dilation
  (bind _playerpawn (Variables|Default|GetPlayerPawn)
    (:then
      (Variables|/Script/Engine.Actor|SetCustomTimeDilation TimeDilation _playerpawn) ; Set time dilation directly instead of via a GE because we need it to \npredict on the client, and GEs take time to replicate (which \nnecessarily causes net corrections).\n\nNOTE: This is NOT ideal, it allows for the possibility of other things\nto override our work here. In practice it's not a problem, this time dilation\nmultiplies with the GE time dilation and nothing else in FN that\ncan be active at the same time currently modifies the time dilation like this.
      (bind _asrocket_ram_player_anim_bp (Utilities|Casting|CastToRocketRam_Player_AnimBP (Components|SkeletalMesh|AnimationBlueprintLinking|GetLinkedAnimLayerInstancebyClass (Variables|/Script/Engine.Character|GetMesh _playerpawn) "/RocketRamGameplay/Items/RocketRam/Animations/Animation/RocketRam_Player_AnimBP.RocketRam_Player_AnimBP_C")))
      (Class|RocketRamPlayerAnimBP|SetTimeDilationRateScale _asrocket_ram_player_anim_bp (Math|Float|SafeDivide 1.0 TimeDilation) :!class /RocketRamGameplay/Items/RocketRam/Animations/Animation/RocketRam_Player_AnimBP.RocketRam_Player_AnimBP_C) ; Counter act the effect of the time dilation on the \nanimation rate scale so they still play at full speed.
    )
    (:else)))

(event Custom|StartPredictingLaunchDirection
  ;[0.02924 0.08066 1] Predict Launch Direction
  (bind _returnvalue (Utilities|Time|SetTimerbyEvent (EventDispatchers|CreateEvent :!event Custom|PredictLaunchDirection) 0.05 true))
  (Variables|Transient|SetPredictLaunchDirectionTimerHandle _returnvalue)
  (CallFunction|PredictLaunchDirection))

(event Custom|StartLookaheadTrace
  ;[0.05708 1 0.3069] Lookahead Trace
  (bind _returnvalue (Utilities|Time|SetTimerbyEvent (EventDispatchers|CreateEvent :!event Custom|DoLookaheadTrace) (Utilities|GetValueOfScalableFloat (Variables|/RocketRamGameplay/Items/RocketRam/DataClass_RocketRam.DataClass_RocketRam_C|GetLookaheadTraceFrequency (Variables|Config|GetConfigData))) true))
  (Variables|Transient|SetLookaheadTraceTimerHandle _returnvalue)
  (CallFunction|DoLookaheadTrace))

(event Custom|StopPredictingLaunchDirection
  ;[0.02924 0.08066 1] Predict Launch Direction
  (Utilities|Time|ClearandInvalidateTimerbyHandle (Variables|Transient|GetPredictLaunchDirectionTimerHandle)))

(event Custom|PredictLaunchDirection
  ;[0.02924 0.08066 1] Predict Launch Direction
  (bind _launchdirection (Default|CalculateLaunchDirection))
  (Variables|Transient|SetLaunchDirection _launchdirection)
  (bind _weapon (Variables|Default|GetWeapon)
    (:then
      (Class|BWeaponRocketRam|SetLaunchDirection _weapon (Variables|Transient|GetLaunchDirection) :!class /RocketRamGameplay/Items/RocketRam/B_Weapon_RocketRam.B_Weapon_RocketRam_C))
    (:else)))

(event Custom|StopLookaheadTrace
  ;[0.05708 1 0.3069] Lookahead Trace
  (Utilities|Time|ClearandInvalidateTimerbyHandle (Variables|Transient|GetLookaheadTraceTimerHandle)))

(event Custom|DoLookaheadTrace
  ;[0.05708 1 0.3069] Lookahead Trace
  (bind _playerpawn (Variables|Default|GetPlayerPawn)
    (:then
      (bind (_pawnhits _vehiclehits _destructiblebuildinghits _blockinghit) (DestructionTrace|DestructionTrace _playerpawn (Math|Vector|Normalize (Transformation|GetVelocity _playerpawn)) (Utilities|GetValueOfScalableFloat (Variables|/RocketRamGameplay/Items/RocketRam/DataClass_RocketRam.DataClass_RocketRam_C|GetLookaheadTraceDistance (Variables|Config|GetConfigData))) "NewEnumerator4" (Variables|Default|GetCurrentRamState)))
      (Utilities|IsValid (.HitActor (Collision|BreakHitResult _blockinghit))
        (:Is%20Valid
          (bind _weapon (Variables|Default|GetWeapon)
            (:then
              (Class|BWeaponRocketRam|NotifyImpactDetected _weapon (CallFunction|IsHitAgainstWall _blockinghit))
              (Class|RocketRamSharedLibrary|RamLog (Variables|Default|GetPlayerPawn) (Utilities|Text|FormatText "NSLOCTEXT(\"[63668553DC3B7B41603CFE6FD08FF0F2]\", \"3351A8A048CE88DB2AEC998E1814CA78\", \"Movement - Lookahead trace detected imminent impact against {1}.\")" (Utilities|String|SelectString "Wall" "Floor" (CallFunction|IsHitAgainstWall _blockinghit))) false true)
              (CallFunction|StopLookaheadTrace))
            (:else)))))
    (:else)))

(event Custom|StartListeningToPawnLanded_CancelTimeDilation
  ;[0.386 1 0.1536] On Pawn Landed Cancel Time Dilation
  (bind _playerpawn (Variables|Default|GetPlayerPawn)
    (:then
      (if (AI|Components|NavMovement|IsFalling (Variables|/Script/Engine.Character|GetCharacterMovement _playerpawn))
        (EventDispatchers|BindEventtoOnPawnLanded _playerpawn (EventDispatchers|CreateEvent :!event Custom|OnPawnLanded_CancelTimeDilation))
        (else
          (CallFunction|OnPawnLandedCancelTimeDilation))))
    (:else)))

(event Custom|TryEarlyEject
  (bind _playerpawn (Variables|Default|GetPlayerPawn))
  (bind _returnvalue (Actor|GetComponentbyClass _playerpawn "/Script/RidingCodeRuntime.RiderComponent"))
  (bind _playerpawn_2 (Variables|Default|GetPlayerPawn))
  (bind _returnvalue_2 (Vehicle|GetVehicle _playerpawn_2))
  (bind _donotuse_activeridable (Variables|/Script/RidingCodeRuntime.RiderComponent|GetDoNotUse_ActiveRidable _returnvalue))
  (bind _destructiontraceignoredactors (Variables|Transient|DestructionTrace|GetDestructionTraceIgnoredActors))
  ;[0.07432 0.3973 1] Try Early Eject
  (bind _playerpawn_3 (Variables|Default|GetPlayerPawn)
    (:then
      (if (GameplayTags|HasMatchingGameplayTag _playerpawn_3 "(TagName=\"Gameplay.Status.InVehicle\")")
        (Class|RocketRamSharedLibrary|RamLog (Variables|Default|GetPlayerPawn) (Utilities|Text|FormatText "NSLOCTEXT(\"[63668553DC3B7B41603CFE6FD08FF0F2]\", \"05CFBB4D4255DF7F4FAB55B82E1B1CF7\", \"Movement - Ejecting from vehicle.\")") false true)
        (Utilities|IsValid _returnvalue_2
          (:Is%20Valid
            (Utilities|Array|Add _destructiontraceignoredactors _returnvalue_2)
            (Utilities|Array|AppendArray _destructiontraceignoredactors (Vehicle|GetAllPassengers _returnvalue_2))
            (bind _returnvalue_3 (Vehicle|EjectPlayerPawn _returnvalue_2 _playerpawn_2))
            (if _returnvalue_3
              (bind _asynctaskproxy (Ability|Tasks|WaitGameplayTagRemove "(TagName=\"Gameplay.Status.InVehicle\")" 0 true)
                (:then
                  (Variables|Transient|SetWaitEjectFromVehicleAsyncTask _asynctaskproxy))
                (:Removed
                  (CallFunction|OnEarlyEjectComplete)))
              (else
                (Class|RocketRamSharedLibrary|RamLog (Variables|Default|GetPlayerPawn) (Utilities|Text|FormatText "NSLOCTEXT(\"[63668553DC3B7B41603CFE6FD08FF0F2]\", \"706FABF94878930095D5F5A22C60751E\", \"Movement - Failed to eject from vehicle.\")") true)
                (Ability|EndAbility)))))
        (elif (GameplayTags|HasMatchingGameplayTag (Variables|Default|GetPlayerPawn) "(TagName=\"Granted.Riding.Player.IsRiding\")")
          (Class|RocketRamSharedLibrary|RamLog (Variables|Default|GetPlayerPawn) (Utilities|Text|FormatText "NSLOCTEXT(\"[63668553DC3B7B41603CFE6FD08FF0F2]\", \"66ECC2A44821F4627215879BA54040B7\", \"Movement - Ejecting from riding animal.\")") false true)
          (Utilities|IsValid _donotuse_activeridable
            (:Is%20Valid
              (Utilities|Array|Add (Variables|Transient|DestructionTrace|GetDestructionTraceIgnoredActors) (Components|GetOwner _donotuse_activeridable))
              (Utilities|IsValid _returnvalue
                (:Is%20Valid
                  (bind _returnvalue_4 (Riding|StopRidingActor _returnvalue true true false true))
                  (if (or _returnvalue_4 (not (Networking|IsServer)))
                    (bind _asynctaskproxy_2 (Ability|Tasks|WaitGameplayTagRemove "(TagName=\"Granted.Riding.Player.IsRiding\")" 0 true)
                      (:then
                        (Variables|Transient|SetWaitEjectFromRidingAsyncTask _asynctaskproxy_2))
                      (:Removed
                        (CallFunction|OnEarlyEjectComplete)))
                    (else
                      _)))
                (:Is%20Not%20Valid
                  _)))
            (:Is%20Not%20Valid
              _))
          (Class|RocketRamSharedLibrary|RamLog (Variables|Default|GetPlayerPawn) (Utilities|Text|FormatText "NSLOCTEXT(\"[63668553DC3B7B41603CFE6FD08FF0F2]\", \"8FF42041412F9D87919C65A7258C6CF5\", \"Movement - Failed to eject from riding animal.\")") true)
          (Ability|EndAbility)
          (else
            (CallFunction|OnEarlyEjectComplete)))))
    (:else)))

(event Custom|StopListeningToPawnLanded_CancelTimeDilation
  ;[0.386 1 0.1536] On Pawn Landed Cancel Time Dilation
  (bind _playerpawn (Variables|Default|GetPlayerPawn)
    (:then
      (EventDispatchers|UnbindEventfromOnPawnLanded _playerpawn (EventDispatchers|CreateEvent :!event Custom|OnPawnLanded_CancelTimeDilation)))
    (:else)))

(event Custom|OnPawnLanded_CancelTimeDilation
  (params (Hit :type Hit%20Result%20Structure))
  ;[0.386 1 0.1536] On Pawn Landed Cancel Time Dilation
  (CallFunction|StopTimeDilation))

(event Custom|StartPostAOEPause
  ;[1 0.6752 0.09549] Post AOE Pause
  (bind _returnvalue (Utilities|Time|SetTimerbyEvent (EventDispatchers|CreateEvent :!event Custom|StopPostAOEPause) 0.5))
  (Variables|Transient|SetPostAOEPauseTimerHandle _returnvalue)
  (bind _playerpawn (Variables|Default|GetPlayerPawn)
    (:then
      (Variables|/Script/Engine.MovementComponent|SetVelocity "0, 0, 0" (Variables|/Script/Engine.Character|GetCharacterMovement _playerpawn))
      (Default|SetGravityMultiplier 0.0)
      (Variables|Transient|PostAOEPause|SetbIsPostAOEPauseActive true))
    (:else)))

(event Custom|TryLateEject
  ;[1 0.2746 0.9747] Try Late Eject
  ;;[1 1 1] Comment
  (bind _playerpawn (Variables|Default|GetPlayerPawn)
    (:then
      (if (GameplayTags|HasMatchingGameplayTag _playerpawn "(TagName=\"MovementMode.ZipliningOnSpline\")")
        (Class|RocketRamSharedLibrary|RamLog (Variables|Default|GetPlayerPawn) (Utilities|Text|FormatText "NSLOCTEXT(\"[63668553DC3B7B41603CFE6FD08FF0F2]\", \"EF405A0948F0AD8639A8D7B634F1A3F8\", \"Movement - Ejecting from zipline.\")") false true)
        (Zipline|EndZiplining (Variables|Default|GetPlayerPawn) false false)
        (elif (GameplayTags|HasMatchingGameplayTag (Variables|Default|GetPlayerPawn) "(TagName=\"MovementMode.Grinding\")")
          (Class|RocketRamSharedLibrary|RamLog (Variables|Default|GetPlayerPawn) (Utilities|Text|FormatText "NSLOCTEXT(\"[63668553DC3B7B41603CFE6FD08FF0F2]\", \"5148F3F149159D50E6505CAAB2E7D488\", \"Movement - Ejecting from grinding.\")") false true)
          (CallFunction|SetGrindingBlocked true)
          (Pawn|Components|CharacterMovement|SetMovementMode (Variables|/Script/Engine.Character|GetCharacterMovement (Variables|Default|GetPlayerPawn)) "MOVE_Falling" 0))))
    (:else)))

(event Custom|StopPostAOEPause
  ;[1 0.6752 0.09549] Post AOE Pause
  (Variables|Transient|PostAOEPause|SetbIsPostAOEPauseActive false)
  (Utilities|Time|ClearandInvalidateTimerbyHandle (Variables|Transient|GetPostAOEPauseTimerHandle))
  (Default|SetGravityMultiplier 1.0))

(fn AddInitialThrustAOEGameplayEffectContainer
  (Ability|AddDynamicGameplayEffectContainer (GameplayTags|MakeLiteralGameplayTag "(TagName=\"Abilities.Generic.RocketRam.AOEInitialThrust\")") (Utilities|Struct|MakeFortGameplayEffectContainer "(TagName=\"Abilities.Generic.RocketRam.AOEInitialThrust\")" (Utilities|Struct|MakeFortAbilityTargetSelectionList (Utilities|Array|MakeArray (Utilities|MakeAOEInitialThrustTargetSelection (Variables|Default|GetInitialThrustAOETargetFilter_Pawns) true) (Utilities|MakeAOEInitialThrustTargetSelection (Variables|Default|GetInitialThrustAOETargetFilter_Environment))) true false false) (Utilities|Array|MakeArray "/RocketRamGameplay/Items/RocketRam/Effects/GE_Damage_AOEInitialThrust_RocketRam.GE_Damage_AOEInitialThrust_RocketRam_C"))))

(fn ApplyGameplayEffectSafe
  (params (Class :type Gameplay%20Effect%20Class%20Reference) (Handle :type Active%20Gameplay%20Effect%20Handle%20Structure%20%28by%20ref%29))
  (Utilities|IsValid (Ability|GameplayEffect|GetGameplayEffectfromActiveEffectHandle Handle)
    (:Is%20Valid
      (Class|RocketRamSharedLibrary|RamLog 0 (Utilities|Text|ToText(String) (Utilities|String|Append "Attempted to apply Gameplay Effect: [" (Utilities|GetClassDisplayName Class) "], but this gameplay effect has already been applied!")) true))
    (:Is%20Not%20Valid
      (bind _returnvalue (Ability|ApplyGameplayEffectToOwner Class))
      (Variables|SetBy-RefVar Handle _returnvalue))) ; If this is valid then it's trying to double-apply the\nGE, which is an error.
)

(fn BlockWeaponSwitching
  (bind _configdata (Variables|Config|GetConfigData))
  (if (not (Class|AthenaFunctionLibrary|RowtoBool (Variables|/RocketRamGameplay/Items/RocketRam/DataClass_RocketRam.DataClass_RocketRam_C|GetDisableBlockingWeaponSwitching _configdata)))
    (CallFunction|ApplyGameplayEffectSafe (select (Class|AthenaFunctionLibrary|RowtoBool (Variables|/RocketRamGameplay/Items/RocketRam/DataClass_RocketRam.DataClass_RocketRam_C|GetDisableDarkeningInventoryHUD _configdata)) (Variables|Config|GetGE_BlockWeaponSwitching_Class) (Variables|Config|GetGE_BlockWeaponSwitchingAndShowOnHUD_Class)) (Variables|Transient|GetGE_BlockWeaponSwitching))))

(fn CalculateLaunchDirection
  (bind _playerpawn (Variables|Default|GetPlayerPawn)
    (:then
      (bind (_returnvalue_location _returnvalue_rotation) (Math|Transform|BreakTransform (Targeting|GetTargetingTransform _playerpawn "Camera" 0 :!class /Script/FortniteGame.FortPawn)))
      (bind (_pawnhits _vehiclehits _destructiblebuildinghits _blockinghit) (DestructionTrace|DestructionTrace _playerpawn (Math|Vector|GetForwardVector _returnvalue_rotation) (ScalableFloat|GetValueatLevel (Variables|/RocketRamGameplay/Items/RocketRam/DataClass_RocketRam.DataClass_RocketRam_C|GetDestructionTrace_GroundDetectionTraceDistance (Variables|Config|GetConfigData))) "NewEnumerator4" "NewEnumerator2")) ; Do a destruction trace using the FreeFall state so terrain blocks us.
      (if (Math|Boolean|ANDBoolean (not (CallFunction|IsHitAgainstWall _blockinghit)) (Collision|BreakHitResult _blockinghit) (Utilities|IsValid (.HitActor (Collision|BreakHitResult _blockinghit)) :!class /Script/Engine.KismetSystemLibrary))
        (bind _asbuilding_actor (Utilities|Casting|CastToBuildingActor (.HitActor (Collision|BreakHitResult _blockinghit)))
          (:then
            (if (DestructionTrace|CanDestroyBuildingActor (Variables|Default|GetPlayerPawn) _asbuilding_actor)
              (return (Math|Vector|GetForwardVector _returnvalue_rotation))
              (else
                _)))
          (:CastFailed
            _))
        (return (Math|Vector|CrossProduct (Transformation|GetActorRightVector (Variables|Default|GetPlayerPawn)) (.ImpactNormal (Collision|BreakHitResult _blockinghit)))))
      (return (Math|Vector|GetForwardVector _returnvalue_rotation)))
    (:else)))

(fn ClearAllTimers
  (for _array_element (Utilities|Array|MakeArray (Variables|Transient|GetWindUpTimerHandle) (Variables|Transient|GetSlowAccelerationTimerHandle) (Variables|Transient|GetDestructionTraceTimerHandle) (Variables|Transient|GetGravityIncreaseTimerHandle) (Variables|Transient|GetAirControlIncreaseTimerHandle) (Variables|Transient|GetRecoveryTimerHandle) (Variables|Transient|GetBrakingTimerHandle) (Variables|Transient|GetTimeDilationEaseInTimerHandle) (Variables|Transient|GetLookaheadTraceTimerHandle) (Variables|Transient|GetSlowAccelerationLockInLaunchDirectionTimerHandle) (Variables|Transient|GetPredictLaunchDirectionTimerHandle) (Variables|Transient|GetPostAOEPauseTimerHandle) (Variables|Transient|GetHandleClientReplicatedBlockingHitTimerHandle) (Variables|Transient|GetHandleServerReplicatedStateTimerHandle))
    (Utilities|Time|ClearandInvalidateTimerbyHandle _array_element)))

(fn ForceNetUpdate
  (bind _playerpawn (Variables|Default|GetPlayerPawn)
    (:then
      (Networking|ForceNetUpdate _playerpawn))
    (:else)))

(fn GetTimeBeenStoppedinActiveThrust
  (Utilities|FlowControl|Sequence
    (:then_0
      (if (and (Variables|Transient|GetDidAtLeastOneDestructionTrace) (Utilities|Enum|Equal(Enum) (Variables|Default|GetCurrentRamState) "NewEnumerator1"))
        (if (not (> (Math|Vector|VectorLength (Math|Vector|ProjectVectorontoVector (select (Class|AthenaFunctionLibrary|RowtoBool (Variables|/RocketRamGameplay/Items/RocketRam/DataClass_RocketRam.DataClass_RocketRam_C|GetDestructionTrace_UsePositionBasedVelocityForActiveThrustBlockingCheck (Variables|Config|GetConfigData))) (/ (- (Transformation|GetActorLocation (Variables|Default|GetPlayerPawn)) (Variables|Default|GetPlayerLocationAtLastDestructionTrace)) (Utilities|Time|GetWorldDeltaSeconds)) (Pawn|Components|CharacterMovement|GetLastUpdateVelocity (Variables|/Script/Engine.Character|GetCharacterMovement (Variables|Default|GetPlayerPawn)))) (Variables|Transient|GetLaunchDirection))) (Utilities|GetValueOfScalableFloat (Variables|/RocketRamGameplay/Items/RocketRam/DataClass_RocketRam.DataClass_RocketRam_C|GetDestructionTrace_SpeedThresholdToIndicateBlockedDuringActiveThrust (Variables|Config|GetConfigData)))))
          (bind _output_get (Variables|Default|SetTimeBeenStoppedInActiveThrust (+ (Variables|Default|GetTimeBeenStoppedInActiveThrust) (Utilities|Time|GetWorldDeltaSeconds))))
          (return _output_get))))
    (:then_1
      (Variables|Default|SetTimeBeenStoppedInActiveThrust 0.0)
      (return 0.0))))

(fn GetCustomAbilitySourceTransform
  (return (Variables|Default|GetCurrentCustomTargetingTransform)))

(fn IsHitAgainstWall
  (params (HitResult :type Hit%20Result%20Structure))
  (bind (_bblockinghit _binitialoverlap _time _distance _location _impactpoint _normal _impactnormal) (Collision|BreakHitResult HitResult))
  (return (>= (Math|Trig|Acos(Degrees) (Math|Vector|DotProduct _impactnormal "0.000000,0.000000,1.000000")) 5e+01)))

(fn PlayDestructionEffects
  (params (Damaged%20A%20Build :type Boolean) (Damaged%20A%20Vehicle :type Boolean) (Damaged%20A%20Pawn :type Boolean))
  (Variables|Default|GetPlayerPawn
    (:then
      (if (and (Math|Boolean|ORBoolean Damaged%20A%20Build Damaged%20A%20Vehicle Damaged%20A%20Pawn) (> (- (Utilities|Time|GetGameTimeinSeconds) (Variables|Transient|GetLastHitSomethingGCTime)) (Utilities|GetValueOfScalableFloat (Variables|/RocketRamGameplay/Items/RocketRam/DataClass_RocketRam.DataClass_RocketRam_C|GetMinTimeBetweenHitSomethingGCPlays (Variables|Config|GetConfigData)))))
        (GameplayCue|ExecuteGameplayCueOnActor(Burst) (Variables|Default|GetPlayerPawn) (Variables|Config|GetGC_HitSomething) (Ability|GameplayCue|MakeGameplayCueParameters (Math|Conversions|ToFloat(Integer) (Math|Integer|SelectInt 0 1 (Variables|GetDamaged%20A%20Pawn)))))
        (Variables|Transient|SetLastHitSomethingGCTime (Utilities|Time|GetGameTimeinSeconds))))
    (:else)))

(fn SendWindUpEvent
  (params (IsStarting :type Boolean))
  (if (!= IsStarting (Variables|UIWindUpEvent|GetIsWindUpEventActive))
    (bind _output_get (Variables|UIWindUpEvent|SetIsWindUpEventActive IsStarting))
    (Ability|SendGameplayEventtoActor (Variables|Default|GetPlayerPawn) (select _output_get (Variables|UIWindUpEvent|GetGE_WindingUpStart) (Variables|UIWindUpEvent|GetGE_WindUpEnd)) (Utilities|Struct|MakeGameplayEventData "(TagName=\"\")" 0 0 0 0 "()" "(GameplayTags=)" "(GameplayTags=)" (Utilities|GetValueOfScalableFloat (Variables|/RocketRamGameplay/Items/RocketRam/DataClass_RocketRam.DataClass_RocketRam_C|GetWindUpPhaseDuration (Variables|Config|GetConfigData)))))))

(fn SetAirControl
  (params (AirControl :type Float%20%28single-precision%29))
  (if (!= (Variables|Transient|AirControl|GetCurrentAirControl) AirControl)
    (Variables|Transient|AirControl|SetCurrentAirControl AirControl)
    (bind _playerpawn (Variables|Default|GetPlayerPawn)
      (:then
        (Ability|RemoveGameplayEffectFromOwnerWithHandle (Variables|Transient|GetGE_AirControl))
        (bind _returnvalue (Ability|GameplayEffect|AssignTagSetbyCallerMagnitude (GameplayEffects|MakeOutgoingSpec :self (Ability|GetFortAbilitySystemComponent _playerpawn) :GameplayEffectClass (Variables|Config|GetGE_AirControl_Class) :Level 0.0) "(TagName=\"SetByCaller.Magnitude\")" AirControl))
        (bind _returnvalue_2 (Ability|ApplyGameplayEffectSpecToOwner _returnvalue))
        (Variables|Transient|SetGE_AirControl _returnvalue_2))
      (:else))))

(fn SetGravityMultiplier
  (params (Multiplier :type Float%20%28single-precision%29))
  (if (Variables|Transient|PostAOEPause|GetbIsPostAOEPauseActive)
    _
    (elif (!= (Variables|Transient|Gravity|GetCurrentGravity) Multiplier)
      (Variables|Transient|Gravity|SetCurrentGravity Multiplier)
      (bind _playerpawn (Variables|Default|GetPlayerPawn)
        (:then
          (Variables|/Script/Engine.CharacterMovementComponent|SetGravityScale (Variables|Transient|Gravity|GetCurrentGravity) (Variables|/Script/Engine.Character|GetCharacterMovement _playerpawn)) ; Set gravity directly instead of via a GE because we need it to \npredict on the client, and GEs take time to replicate (which \nnecessarily causes net corrections).\n\nNOTE: This is NOT ideal, it allows for the possibility of other things\nto override our work here. In practice it's not a problem, this gravity\nscale multiplies with the GE gravity scale and nothing else in FN that\ncan be active at the same time currently modifies the gravity scale like this.
        )
        (:else)))))

(fn SetRamState
  (params (NewState :type Enum_RocketRamState%20Enum :default "NewEnumerator0"))
  (bind _currentramstate (Variables|Default|GetCurrentRamState))
  (bind _newstate (Variables|GetNewState))
  (if (Utilities|Enum|NotEqual(Enum) _currentramstate _newstate)
    (Class|RocketRamSharedLibrary|RamLog (Variables|Default|GetPlayerPawn) (Utilities|Text|FormatText "NSLOCTEXT(\"[63668553DC3B7B41603CFE6FD08FF0F2]\", \"D9438C63438D3DB427F046A4BF5E2C2F\", \"State Changed. PreviousState: {Prev}  NewState: {New}\")" (Utilities|String|EnumtoString _currentramstate) (Utilities|String|EnumtoString _newstate)) false true)
    (Variables|Default|SetPreviousRamState (Variables|Default|GetCurrentRamState))
    (Variables|Default|SetCurrentRamState (Variables|GetNewState))
    (CallFunction|OnExitRamState (Variables|Default|GetPreviousRamState))
    (CallFunction|OnEnterRamState (Variables|Default|GetCurrentRamState))
    (Utilities|FlowControl|Sequence
      (:then_0
        (bind _weapon (Variables|Default|GetWeapon)
          (:then
            (Class|BWeaponRocketRam|OnRamStateChanged _weapon (Variables|Default|GetCurrentRamState)))
          (:else)))
      (:then_1
        (Ability|HasAuthority
          (:True
            (Utilities|IsValid self
              (:Is%20Valid
                (CallFunction|ClientReplicateStatefromServer (Variables|Default|GetCurrentRamState))
                (Default|ForceNetUpdate)))))))))

(fn SetUseHighAccelerationTolerance
  (params (bUseHighAccelerationTolerance :type Boolean))
  (bind _playerpawn (Variables|Default|GetPlayerPawn)
    (:then
      (bind _asfort_movement_comp_character (Utilities|Casting|CastToFortMovementComp_Character (Variables|/Script/Engine.Character|GetCharacterMovement _playerpawn)))
      (NetTolerance|SetUseHighAccelerationTolerance _asfort_movement_comp_character bUseHighAccelerationTolerance "RocketRamMovement"))
    (:else)))

(fn ShouldBlockingHitTriggerPostAOEPause
  (params (Hit :type Hit%20Result%20Structure))
  (bind (_bblockinghit _binitialoverlap _time _distance _location _impactpoint _normal _impactnormal _physmat _hitactor) (Collision|BreakHitResult Hit))
  (if (CallFunction|IsHitAgainstWall Hit)
    (bind _asbuilding_actor (Utilities|Casting|CastToBuildingActor _hitactor)
      (:then
        (return (not (DestructionTrace|CanDestroyBuildingActor (Variables|Default|GetPlayerPawn) _asbuilding_actor))))
      (:CastFailed
        (Utilities|IsValid _hitactor
          (:Is%20Valid
            (return true))
          (:Is%20Not%20Valid
            (return false)))))
    (else
      (return false))))
