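; Function library for the Rocket Ram item's area-of-effect knockback and damage at launch and impact.
; AOEImpact launches and damages target-data actors, then executes a gameplay cue carrying scaled radius in its hit result.
; AOEInitialThrust does likewise on authority, adding phased and overhead buildings with separate player-built damage; non-authority only spawns Niagara effects.
; ShouldDoAOEImpact gates impacts by velocity angle against up and surface normal; GetUniqueActors deduplicates hit actors.
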(fn AOEImpact
  (params (PlayerPawn :type Fort%20Player%20Pawn%20Object%20Reference) (ImpactLocation :type Vector) (AdditionalActorsToIgnore :type Array%20of%20Actor%20Object%20References) (GameplayCueTag :type Gameplay%20Tag%20Structure :default "(TagName=\"GameplayCue.RocketRam.AOEImpact\")") (TargetData :type Gameplay%20Ability%20Target%20Data%20Handle%20Structure%20%28by%20ref%29))
  (bind _playerpawn (Variables|GetPlayerPawn))
  (bind _playerpawn_2 (Variables|GetPlayerPawn)
    (:then
      (Class|RocketRamSharedLibrary|RamLog _playerpawn_2 "NSLOCTEXT(\"[528C0908F72682AF1A0FFCE08992315F]\", \"E1049ADF48F63791461594ADB0C8F4D7\", \"AOEImpact\")" false false _playerpawn_2)
      (Utilities|FlowControl|Sequence
        (:then_0
          ;[1 1 1] Ignore the pawn, its weapon and any actors that the caller requested to ignore.
          (LocalVariables|SetTraceActorsToIgnore (Utilities|Array|MakeArray (Variables|GetPlayerPawn) (Weapon|GetCurrentWeapon (Variables|GetPlayerPawn))))
          (Utilities|Array|AppendArray (LocalVariables|GetTraceActorsToIgnore) (Variables|GetAdditionalActorsToIgnore)))
        (:then_1
          (Class|RocketRamSharedLibrary|RamLog (Variables|GetPlayerPawn) (Utilities|Text|FormatText "NSLOCTEXT(\"[528C0908F72682AF1A0FFCE08992315F]\", \"0805DEEE4B60D66942A4E7B58333BE56\", \"AOE - Performing AOE with radius {0}.\")" (ScalableFloat|GetValueatLevel (LocalVariables|GetAOEImpact_ImpactRadius))) false true)
          (LocalVariables|SetIgnoreFallDamage (Class|AthenaFunctionLibrary|RowtoBool (LocalVariables|GetGranteFallImmunityToKnockbackVictims) 0.0 (Variables|GetPlayerPawn)))
          (for _array_element (Ability|TargetData|GetAllActorsfromTargetData (Variables|GetTargetData))
            (Utilities|IsValid _array_element
              (:Is%20Valid
                (LocalVariables|SetCurrentHitActor _array_element)
                (LocalVariables|SetbShouldApplyDamageToHitActor false)
                (Class|RocketRamSharedLibrary|RamLog (Variables|GetPlayerPawn) (Utilities|Text|FormatText "NSLOCTEXT(\"[528C0908F72682AF1A0FFCE08992315F]\", \"75128D9440BF441FD09576B42EEDF019\", \"AOE - Actor hit by AOE: {0}\")" (LocalVariables|GetCurrentHitActor)) false true)
                (Utilities|FlowControl|Sequence
                  (:then_0
                    (bind _asfort_pawn (Utilities|Casting|CastToFortPawn (LocalVariables|GetCurrentHitActor))
                      (:then
                        ;[1 1 1] Launch pawns away from the point of impact.
                        (Utilities|LaunchPawns (Variables|GetPlayerPawn) _asfort_pawn (* (+ (Math|Vector|Normalize2D(Vector) (- (Transformation|GetActorLocation _asfort_pawn) (Variables|GetImpactLocation))) "0.000000,0.000000,0.900000") (ScalableFloat|GetValueatLevel (LocalVariables|GetAOEImpactPawnKnockbackMagnitude))) true true (LocalVariables|GetIgnoreFallDamage) true false true true true)
                        (LocalVariables|SetbShouldApplyDamageToHitActor true)
                        (GameplayEffects|ApplyGameplayEffectToSelf :self (Ability|GetFortAbilitySystemComponent (LocalVariables|GetCurrentHitActor)) :GameplayEffectClass "/RocketRamGameplay/Items/RocketRam/Effects/GE_VehicleLaunchImmunity_RocketRam.GE_VehicleLaunchImmunity_RocketRam_C" :Level 0.0))
                      (:CastFailed
                        ;[1 1 1] Just apply damage to buildings
                        (Utilities|Casting|CastToBuildingActor (LocalVariables|GetCurrentHitActor))
                        (LocalVariables|SetbShouldApplyDamageToHitActor true))))
                  (:then_1
                    ;[1 1 1] Apply damage to the hit actor if necessary.
                    (if (and (LocalVariables|GetbShouldApplyDamageToHitActor) (not (Utilities|Array|ContainsItem (LocalVariables|GetTraceActorsToIgnore) (LocalVariables|GetCurrentHitActor))))
                      (GameplayEffects|ApplyGameplayEffectToTarget (Ability|GetFortAbilitySystemComponent (Variables|GetPlayerPawn)) (LocalVariables|GetGE_AOEDamage_Class) (Ability|GetFortAbilitySystemComponent (LocalVariables|GetCurrentHitActor)) 0.0 (GameplayEffects|MakeFortEffectContext (Ability|GetFortAbilitySystemComponent (Variables|GetPlayerPawn)) (Variables|GetPlayerPawn) (Variables|GetPlayerPawn) (Variables|GetPlayerPawn) 0)))))))))
        (:then_2
          ;[0.6242 0 1] Create Impact Gameplay Cue
          (bind _output_get (LocalVariables|SetGameplayCueContext (Class|AthenaFunctionLibrary|MakeSimpleFortEffectContext _playerpawn _playerpawn _playerpawn)))
          (Ability|EffectContext|AddHitResult _output_get (Collision|MakeHitResult false false (* (ScalableFloat|GetValueatLevel (LocalVariables|GetAOEImpact_ImpactRadius)) 0.0015)))
          (Ability|EffectContext|SetOrigin _output_get (+ (Variables|GetImpactLocation) "0.000000,0.000000,25.000000")) ; Set Origin AFTER AddHitResult because AddHitResultCan reset the origin.
          (GameplayCue|ExecuteGameplayCue _playerpawn (Variables|GetGameplayCueTag) _output_get))))
    (:else)))

(fn AOEInitialThrust
  (params (PlayerPawn :type Fort%20Player%20Pawn%20Object%20Reference) (ImpactLocation :type Vector) (AdditionalActorsToIgnore :type Array%20of%20Actor%20Object%20References) (TargetData :type Gameplay%20Ability%20Target%20Data%20Handle%20Structure%20%28by%20ref%29))
  (bind _charactermovement (Variables|/Script/Engine.Character|GetCharacterMovement (Variables|GetPlayerPawn)))
  (bind _playerpawn (Variables|GetPlayerPawn))
  (bind _traceactorstoignore (LocalVariables|GetTraceActorsToIgnore))
  (bind _playerpawn_2 (Variables|GetPlayerPawn))
  (bind _playerpawn_3 (Variables|GetPlayerPawn)
    (:then
      (if (Networking|HasAuthority _playerpawn_3)
        (Utilities|FlowControl|Sequence
          (:then_0
            ;[0.3906 0.2512 0] Ignore the pawn, its weapon, the actor the player is standing on, and any actors that the caller requested to ignore.
            (LocalVariables|SetTraceActorsToIgnore (Utilities|Array|MakeArray _playerpawn (Weapon|GetCurrentWeapon _playerpawn)))
            (Utilities|Array|AppendArray _traceactorstoignore (Variables|GetAdditionalActorsToIgnore))
            (Utilities|IsValid _charactermovement
              (:Is%20Valid
                (bind _asprimitive_component (Utilities|Casting|CastToPrimitiveComponent (Pawn|Components|CharacterMovement|GetMovementBaseObject _charactermovement)))
                (Utilities|IsValid (Components|GetOwner _asprimitive_component)
                  (:Is%20Valid
                    (Utilities|Array|Add _traceactorstoignore (Components|GetOwner _asprimitive_component)))))))
          (:then_1
            (LocalVariables|SetActorsToDestroy (Ability|TargetData|GetAllActorsfromTargetData (Variables|GetTargetData)))
            ;[1 0.5889 0.3475] Add any "phased" buildings the player may be overlaping with
            (if (ScalableFloat|GetValueasBool (LocalVariables|GetDamagePhasedBuildingsOnInitialThrust))
              (Utilities|Array|AppendArray (LocalVariables|GetActorsToDestroy) (Class|FortPawn|GetOverlappingBuildings (Variables|GetPlayerPawn)))
              (else
                _))
            (bind _outactors (Collision|BoxOverlapActors 0 (Utilities|Operators|Add (Variables|GetImpactLocation) (Math|Vector|MakeVector 0.0 0.0 (ScalableFloat|GetValueatLevel (LocalVariables|GetAOEThrustCielingPunchThroughZOffset)))) (Math|Vector|MakeVector 5e+01 5e+01 (ScalableFloat|GetValueatLevel (LocalVariables|GetAOEThrustCielingPunchThroughHeight))) (Utilities|Array|MakeArray "ObjectTypeQuery1") "/Script/FortniteGame.BuildingSMActor" (LocalVariables|GetTraceActorsToIgnore) :!class /Script/Engine.KismetSystemLibrary)) ; Need to add in any buildings directly above the player\n(this part doesn't need to effect anything other than builds)
            (Utilities|Array|AppendArray (LocalVariables|GetActorsToDestroy) _outactors)
            (for _array_element (LocalVariables|GetActorsToDestroy)
              (bind _output_get (LocalVariables|SetCurrentHitActor _array_element))
              (Utilities|IsValid _output_get
                (:Is%20Valid
                  (if (not (Utilities|Array|ContainsItem (LocalVariables|GetTraceActorsToIgnore) _array_element))
                    (LocalVariables|SetbShouldApplyDamageToHitActor false)
                    (Class|RocketRamSharedLibrary|RamLog (Variables|GetPlayerPawn) (Utilities|Text|FormatText "NSLOCTEXT(\"[528C0908F72682AF1A0FFCE08992315F]\", \"676D7AA2405ACFAFF00E02828DA87083\", \"AOE - Actor hit by initial thrust AOE: {0}\")" (LocalVariables|GetCurrentHitActor)) false true)
                    (Utilities|FlowControl|Sequence
                      (:then_0
                        (bind _asfort_pawn (Utilities|Casting|CastToFortPawn (LocalVariables|GetCurrentHitActor))
                          (:then
                            ;[1 1 1] Launch pawns away from the point of impact.
                            (Utilities|LaunchPawns (Variables|GetPlayerPawn) _asfort_pawn (* (+ (Math|Vector|Normalize (- (Transformation|GetActorLocation _asfort_pawn) (Variables|GetImpactLocation))) "0.000000,0.000000,1.000000") (ScalableFloat|GetValueatLevel (LocalVariables|GetAOEThrustPawnKnockback))) true true (Class|AthenaFunctionLibrary|RowtoBool (LocalVariables|GetGrantFallImmunityToKnockbackVictims)) true false true true true)
                            (LocalVariables|SetbShouldApplyDamageToHitActor true))
                          (:CastFailed
                            ;[1 1 1] Just apply damage to buildings
                            (Utilities|Casting|CastToBuildingActor (LocalVariables|GetCurrentHitActor))
                            (LocalVariables|SetbShouldApplyDamageToHitActor true))))
                      (:then_1
                        ;[1 1 1] Apply damage to the hit actor if necessary.
                        (if (LocalVariables|GetbShouldApplyDamageToHitActor)
                          (bind _returnvalue (GameplayTags|HasMatchingGameplayTag(Message) (LocalVariables|GetCurrentHitActor) "(TagName=\"Building.Type.Player\")"))
                          (GameplayEffects|ApplyGameplayEffectToTarget (Ability|GetFortAbilitySystemComponent (Variables|GetPlayerPawn)) (select _returnvalue "/RocketRamGameplay/Items/RocketRam/Effects/GE_Damage_AOEInitialThrust_PlayerBuiltWalls_RocketRam.GE_Damage_AOEInitialThrust_PlayerBuiltWalls_RocketRam_C" "/RocketRamGameplay/Items/RocketRam/Effects/GE_Damage_AOEInitialThrust_RocketRam.GE_Damage_AOEInitialThrust_RocketRam_C") (Ability|GetFortAbilitySystemComponent (LocalVariables|GetCurrentHitActor)) 0.0 (GameplayEffects|MakeFortEffectContext (Ability|GetFortAbilitySystemComponent (Variables|GetPlayerPawn)) (Variables|GetPlayerPawn) (Variables|GetPlayerPawn) (Variables|GetPlayerPawn) 0)))))))))))
        (else
          ;[0 0 0] TODO: all FX and sounds need to be in GC's because this is a GA (won't show up on remote clients)
          (Niagara|SpawnSystemAttached "/RocketRamGameplay/Items/RocketRam/FX/NS_RocketRam_Launch_Thrust.NS_RocketRam_Launch_Thrust" (Visual|GetWeaponMesh (Variables|/Script/FortniteGame.FortPawn|GetCurrentWeapon _playerpawn_2)) "front_socket")
          (Niagara|SpawnSystematLocation 0 "/RocketRamGameplay/Items/RocketRam/FX/NS_RocketRam_AOE_Launch.NS_RocketRam_AOE_Launch" (Transformation|GetActorLocation _playerpawn_2)))))
    (:else)))

(fn GetUniqueActors
  (params (Array :type Array%20of%20Hit%20Result%20Structures))
  (for _array_element Array
    (Utilities|Array|AddUnique (LocalVariables|GetActors) (.HitActor (Collision|BreakHitResult _array_element))))
  (return (LocalVariables|GetActors)))

(fn ShouldDoAOEImpact
  (params (PlayerPawn :type Fort%20Player%20Pawn%20Object%20Reference) (Impact%20Velocity :type Vector) (HitResult :type Hit%20Result%20Structure))
  (bind (_bblockinghit _binitialoverlap _time _distance _location _impactpoint _normal _impactnormal) (Collision|BreakHitResult (Variables|GetHitResult)))
  (return (or (>= (Math|Trig|Acos(Degrees) (Math|Vector|DotProduct (Math|Vector|Normalize (Variables|GetImpact%20Velocity)) "0.000000,0.000000,1.000000")) (ScalableFloat|GetValueatLevel (LocalVariables|GetAOEImpactMinAngleFromUpToForceAOE))) (>= (Math|Trig|Acos(Degrees) (Math|Vector|DotProduct (Math|Vector|Normalize (Variables|GetImpact%20Velocity)) _impactnormal)) (ScalableFloat|GetValueatLevel (LocalVariables|GetAOEImpactMinAngleForAOE))))))
