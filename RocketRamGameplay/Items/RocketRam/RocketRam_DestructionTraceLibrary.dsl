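; Blueprint function library for the Rocket Ram item's destruction trace, which finds and smashes what the riding player hits.
; It iterates line or capsule traces, dedupes hits per actor, then sorts them into pawns, vehicles, destructible buildings and a blocking hit.
; Blocking depends on ram state, AOE impact checks and a raised floor-clip retrace.
; Handlers apply knockback and damage, return actors to ignore, and tag doomed buildings for the camera to ignore.
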(fn CanDestroyBuildingActor
  (params (PlayerPawn :type Fort%20Player%20Pawn%20Object%20Reference) (BuildingActor :type Building%20Actor%20Object%20Reference))
  (Utilities|FlowControl|Sequence
    (:then_0
      (Utilities|IsValid BuildingActor
        (:Is%20Valid
          (if (GameplayTags|HasAnyMatchingGameplayTags BuildingActor (LocalVariables|GetIndestructibleGameplayTags))
            (return false)
            (else
              (bind _asbuilding_static_mesh (Utilities|Casting|CastToBuildingSMActor BuildingActor)
                (:then
                  (bind _playerpawn (Variables|GetPlayerPawn)
                    (:then
                      (return (Class|BuildingStaticMesh|CanbeDamaged _asbuilding_static_mesh (Pawn|GetController _playerpawn) _playerpawn)))
                    (:else)))
                (:CastFailed
                  (return (Variables|/Script/Engine.Actor|GetbCanBeDamaged BuildingActor)))))))))
    (:then_1
      (return false))))

(fn ClearCameraCollisionActorTags
  (params (Actors :type Array%20of%20Actor%20Object%20References))
  (for _array_element Actors
    (Utilities|IsValid _array_element
      (:Is%20Valid
        (Utilities|Array|RemoveItem (Variables|/Script/Engine.Actor|GetTags _array_element) (LocalVariables|GetIgnoreCameraCollisionTagName))))))

(fn DestructionTrace
  (params (PlayerPawn :type Fort%20Player%20Pawn%20Object%20Reference) (TraceDirection :type Vector) (TraceDistance :type Float%20%28double-precision%29) (TraceType :type Enum_DestructionTraceType%20Enum :default "NewEnumerator4") (RamState :type Enum_RocketRamState%20Enum :default "NewEnumerator0") (AdditionalActorsToIgnore :type Array%20of%20Actor%20Object%20References) (bDebugDrawTrace :type Boolean) (TimeBeenStopped :type Float%20%28double-precision%29))
  (bind (_bbuildingshouldblock _bshouldbedestroyed) (DestructionTrace|ShouldBuildingBlockDuringFreefall (Variables|GetPlayerPawn) (LocalVariables|GetCurrentHit) (Variables|GetTraceDirection) (Variables|GetPlayerPawn)))
  (bind (_bblockinghit _binitialoverlap _time _distance _location _impactpoint _normal _impactnormal _physmat _hitactor) (Collision|BreakHitResult (LocalVariables|GetBlockingHitResult)))
  (Utilities|FlowControl|Sequence
    (:then_0
      (LocalVariables|SetTraceActorsToIgnore (Utilities|Array|MakeArray (Variables|GetPlayerPawn) (Weapon|GetCurrentWeapon (Variables|GetPlayerPawn))))
      (Utilities|Array|AppendArray (LocalVariables|GetTraceActorsToIgnore) (Variables|GetAdditionalActorsToIgnore))
      (Utilities|Array|Clear (LocalVariables|GetPawnHitResults))
      (Utilities|Array|Clear (LocalVariables|GetDestructibleBuildingHitResults)))
    (:then_1
      (bind _playerpawn (Variables|GetPlayerPawn)
        (:then
          (LocalVariables|SetTraceStartLocation (Transformation|GetActorLocation _playerpawn))
          (LocalVariables|SetTraceEndLocation (+ (Transformation|GetActorLocation _playerpawn) (* (Math|Vector|Normalize (Variables|GetTraceDirection)) (Math|Float|Max(Float) (Variables|GetTraceDistance) (ScalableFloat|GetValueatLevel (LocalVariables|GetDestructionTraceMinTraceDistance))))))
          (bind _tracehitresults (DestructionTrace|DoActualDestructionTrace (Variables|GetTraceType) (Variables|GetPlayerPawn) (LocalVariables|GetTraceStartLocation) (LocalVariables|GetTraceEndLocation) (LocalVariables|GetTraceActorsToIgnore) (Variables|GetbDebugDrawTrace) (Variables|GetPlayerPawn))) ; The actual trace done is different depending on the given Trace Type.
          (LocalVariables|SetTraceHits _tracehitresults))
        (:else)))
    (:then_2
      (bind _array_element (Utilities|Array|ForEachLoopwithBreak (LocalVariables|GetTraceHits))
        (:LoopBody
          (bind _output_get (LocalVariables|SetCurrentHit _array_element))
          (if (Collision|BreakHitResult _output_get)
            (bind _output_get_2 (LocalVariables|SetCurrentHitActor (.HitActor (Collision|BreakHitResult _output_get))))
            (Utilities|IsValid _output_get_2
              (:Is%20Valid
                (Utilities|FlowControl|Sequence
                  (:then_0
                    (Utilities|Casting|CastToFortPawn (LocalVariables|GetCurrentHitActor)
                      (:then
                        (Utilities|Array|Add (LocalVariables|GetPawnHitResults) (LocalVariables|GetCurrentHit)))
                      (:CastFailed
                        (Utilities|Casting|CastToFortAthenaVehicle (LocalVariables|GetCurrentHitActor)
                          (:then
                            (Utilities|Array|Add (LocalVariables|GetVehicleHitResults) (LocalVariables|GetCurrentHit)))
                          (:CastFailed
                            (bind _asbuilding_actor (Utilities|Casting|CastToBuildingActor (LocalVariables|GetCurrentHitActor)))
                            (if (DestructionTrace|CanDestroyBuildingActor (Variables|GetPlayerPawn) _asbuilding_actor (Variables|GetPlayerPawn))
                              (switch Utilities|FlowControl|Switch|SwitchonEnum_RocketRamState (Variables|GetRamState)
                                (:NewEnumerator0
                                  _)
                                (:NewEnumerator6
                                  _)
                                (:NewEnumerator1
                                  _)
                                (:NewEnumerator2
                                  (if (DestructionTrace|HasPlayerHitFreeFallPeak (Variables|GetPlayerPawn) (Variables|GetPlayerPawn)) :!label _label
                                    ;[0 0 0] During Freefall, only destroy builds if we have enough velocity to do an AOE impact.
                                    (if _bshouldbedestroyed
                                      _)))
                                (:NewEnumerator5
                                  (goto _label))
                                (:NewEnumerator3
                                  _)
                                (:NewEnumerator7
                                  _)
                                (:NewEnumerator4
                                  _))
                              (Utilities|Array|Add (LocalVariables|GetDestructibleBuildingHitResults) (LocalVariables|GetCurrentHit))))))))
                  (:then_1
                    (Utilities|IsValid _hitactor
                      (:Is%20Not%20Valid
                        (bind _bshouldblock (DestructionTrace|ShouldHitBlockDestructionTrace (Variables|GetPlayerPawn) (Variables|GetRamState) (Variables|GetTraceType) (Variables|GetTraceDirection) (Variables|GetTraceDistance) (LocalVariables|GetTraceActorsToIgnore) (LocalVariables|GetCurrentHit) (Variables|GetTimeBeenStopped) (Variables|GetPlayerPawn)))
                        (if _bshouldblock
                          (LocalVariables|SetBlockingHitResult (LocalVariables|GetCurrentHit))
                          (Class|RocketRamSharedLibrary|RamLog (Variables|GetPlayerPawn) (Utilities|Text|FormatText "NSLOCTEXT(\"[5F0C5B7A54E665784329CE12CADAC24E]\", \"E73F80F348AEB41F0EB0119E6F15D3C7\", \"Destruction - Hit Blocker {0}.\")" (LocalVariables|GetCurrentHitActor)) false true))))))))))))
    (:then_3
      (return (LocalVariables|GetPawnHitResults) (LocalVariables|GetVehicleHitResults) (LocalVariables|GetDestructibleBuildingHitResults) (LocalVariables|GetBlockingHitResult)))))

(fn DoActualDestructionTrace
  (params (TraceType :type Enum_DestructionTraceType%20Enum :default "NewEnumerator4") (PlayerPawn :type Fort%20Player%20Pawn%20Object%20Reference) (TraceStartLocation :type Vector) (TraceEndLocation :type Vector) (TraceActorsToIgnore :type Array%20of%20Actor%20Object%20References) (bDebugDrawTrace :type Boolean))
  (bind (_bblockinghit _binitialoverlap _time _distance _location _impactpoint _normal _impactnormal _physmat _hitactor) (Collision|BreakHitResult (Utilities|Array|Get(aref) (LocalVariables|GetTraceHits) (LocalVariables|GetDuplicateLastIndex))))
  (bind (_bblockinghit_2 _binitialoverlap_2 _time_2 _distance_2 _location_2 _impactpoint_2 _normal_2 _impactnormal_2 _physmat_2 _hitactor_2) (Collision|BreakHitResult (Utilities|Array|Get(aref) (LocalVariables|GetTraceHits) (LocalVariables|GetDuplicateCheckIndex))))
  (Utilities|Array|AppendArray (LocalVariables|GetIgnoreActors) TraceActorsToIgnore)
  (Utilities|FlowControl|Sequence
    (:then_0
      (Utilities|FlowControl|ForLoopwithBreak 0 (LocalVariables|GetMaxTraceIterations) :!label _label
        (:LoopBody
          (if (LocalVariables|GetbShouldStopTrace)
            (goto _label Break)
            (else
              (Utilities|FlowControl|Sequence
                (:then_0
                  (switch Utilities|FlowControl|Switch|SwitchonEnum_DestructionTraceType (Variables|GetTraceType)
                    (:NewEnumerator4
                      (bind (_outhits _returnvalue) (Collision|MultiLineTraceByProfile (Variables|GetPlayerPawn) (Variables|GetTraceStartLocation) (Variables|GetTraceEndLocation) (LocalVariables|GetTraceProfileName) false (LocalVariables|GetIgnoreActors) (select (Variables|GetbDebugDrawTrace) "ForDuration" "None") true "(R=1.000000,G=0.000000,B=0.000000,A=1.000000)" "(R=0.000000,G=1.000000,B=0.000000,A=1.000000)" 3e+01))
                      (Utilities|Array|AppendArray (LocalVariables|GetTraceHits) _outhits)
                      (LocalVariables|SetbShouldStopTrace (not _returnvalue)))
                    (:NewEnumerator7
                      (bind (_outhits_2 _returnvalue_2) (Collision|MultiCapsuleTraceByProfile (Variables|GetPlayerPawn) (Variables|GetTraceStartLocation) (Variables|GetTraceEndLocation) (ScalableFloat|GetValueatLevel (LocalVariables|GetDestructionTraceSmallRadius)) (ScalableFloat|GetValueatLevel (LocalVariables|GetDestructionTraceSmallHalfHeight)) (LocalVariables|GetTraceProfileName) false (LocalVariables|GetIgnoreActors) (select (Variables|GetbDebugDrawTrace) "ForDuration" "None") true "(R=1.000000,G=0.000000,B=0.000000,A=1.000000)" "(R=0.000000,G=1.000000,B=0.000000,A=1.000000)" 3e+01))
                      (Utilities|Array|AppendArray (LocalVariables|GetTraceHits) _outhits_2)
                      (LocalVariables|SetbShouldStopTrace (not _returnvalue_2)))
                    (:NewEnumerator8
                      (bind (_outhits_3 _returnvalue_3) (Collision|MultiCapsuleTraceByProfile (Variables|GetPlayerPawn) (Variables|GetTraceStartLocation) (Variables|GetTraceEndLocation) (ScalableFloat|GetValueatLevel (LocalVariables|GetDestructionTraceFullRadius)) (ScalableFloat|GetValueatLevel (LocalVariables|GetDestructionTraceFullHalfHeight)) (LocalVariables|GetTraceProfileName) false (LocalVariables|GetIgnoreActors) (select (Variables|GetbDebugDrawTrace) "ForDuration" "None") true "(R=1.000000,G=0.000000,B=0.000000,A=1.000000)" "(R=0.000000,G=1.000000,B=0.000000,A=1.000000)" 3e+01))
                      (Utilities|Array|AppendArray (LocalVariables|GetTraceHits) _outhits_3)
                      (LocalVariables|SetbShouldStopTrace (not _returnvalue_3)))))
                (:then_1
                  (for _array_element (LocalVariables|GetTraceHits)
                    (Utilities|Array|Add (LocalVariables|GetIgnoreActors) (.HitActor (Collision|BreakHitResult _array_element)))))))))))
    (:then_1
      ;[1 0.4892 0.01151] Remove any duplicate hits that affect the same actor.
      ; But prioritizing keeping the last hit of same actor, so that final blocking hit is guaranteed to be kept
      ; This can happen because a single trace can hit multiple components on the same actor.
      ; Since we do not care about the specific component hit we need to remove duplicate actor hits so we don't apply velocity or damage multiple times.
      (bind _output_get (LocalVariables|SetDuplicateLastIndex (Utilities|Array|LastIndex (LocalVariables|GetTraceHits))))
      (while (>= _output_get 0)
        (Utilities|FlowControl|Sequence
          (:then_0
            (LocalVariables|SetDuplicateCheckActor _hitactor)
            (LocalVariables|SetDuplicateCheckIndex (- (LocalVariables|GetDuplicateLastIndex) 1))
            (while (>= (LocalVariables|GetDuplicateCheckIndex) 0)
              (Utilities|FlowControl|Sequence
                (:then_0
                  (if (== _hitactor_2 (LocalVariables|GetDuplicateCheckActor))
                    (Utilities|Array|RemoveIndex (LocalVariables|GetTraceHits) (LocalVariables|GetDuplicateCheckIndex))
                    (Math|Integer|DecrementInt (LocalVariables|GetDuplicateLastIndex))))
                (:then_1
                  (Math|Integer|DecrementInt (LocalVariables|GetDuplicateCheckIndex))))))
          (:then_1
            (Math|Integer|DecrementInt (LocalVariables|GetDuplicateLastIndex))))))
    (:then_2
      ;[1 0.4583 0.4583] Return
      (return (LocalVariables|GetTraceHits)))))

(fn DoFloorClipCheck
  (params (PlayerPawn :type Fort%20Player%20Pawn%20Object%20Reference) (HitResult :type Hit%20Result%20Structure) (TraceType :type Enum_DestructionTraceType%20Enum :default "NewEnumerator4") (TraceDirection :type Vector) (TraceDistance :type Float%20%28double-precision%29) (TraceActorsToIgnore :type Array%20of%20Actor%20Object%20References))
  (bind (_bblockinghit _binitialoverlap _time _distance _location _impactpoint) (Collision|BreakHitResult (Variables|GetHitResult)))
  (bind (_bblockinghit_2 _binitialoverlap_2 _time_2 _distance_2 _location_2 _impactpoint_2 _normal _impactnormal _physmat _hitactor) (Collision|BreakHitResult (Variables|GetHitResult)))
  (Utilities|FlowControl|Sequence
    (:then_0
      (Variables|GetPlayerPawn
        (:then
          (switch Utilities|FlowControl|Switch|SwitchonEnum_DestructionTraceType (Variables|GetTraceType)
            (:NewEnumerator7
              _)
            (:NewEnumerator8
              _))
          (if (> (- (.z _location) (.z _impactpoint)) (Utilities|Select :Index (Variables|GetTraceType) :NewEnumerator4 0.0 :NewEnumerator7 (- (ScalableFloat|GetValueatLevel (LocalVariables|GetDestructioonTraceSmallHalfHeight)) (ScalableFloat|GetValueatLevel (LocalVariables|GetDestructionTraceSmallRadius))) :NewEnumerator8 (- (ScalableFloat|GetValueatLevel (LocalVariables|GetDestructionTraceFullHalfHeight)) (ScalableFloat|GetValueatLevel (LocalVariables|GetDestructionTraceFullRadius)))))
            (bind _tracehitresults (DestructionTrace|DoActualDestructionTrace (Variables|GetTraceType) (Variables|GetPlayerPawn) (Utilities|Operators|Add (Transformation|GetActorLocation (Variables|GetPlayerPawn)) (Math|Vector|MakeVector 0.0 0.0 (LocalVariables|GetMaxClipHeight))) (+ (Utilities|Operators|Add (Transformation|GetActorLocation (Variables|GetPlayerPawn)) (Math|Vector|MakeVector 0.0 0.0 (LocalVariables|GetMaxClipHeight))) (Utilities|Operators|Multiply (Math|Vector|Normalize (Variables|GetTraceDirection)) (Variables|GetTraceDistance) 2.0)) (Variables|GetTraceActorsToIgnore) false (Variables|GetPlayerPawn))) ; Do a Destruction Trace that is offset upwards by the max clip height.
            (for _array_element _tracehitresults
              (if (and (Collision|BreakHitResult _array_element) (== (.HitActor (Collision|BreakHitResult _array_element)) _hitactor))
                (return false)))
            (return true)))
        (:else)))
    (:then_1
      (return false))))

(fn HandleDestructibleBuildingHits
  (params (PlayerPawn :type Fort%20Player%20Pawn%20Object%20Reference) (RamState :type Enum_RocketRamState%20Enum :default "NewEnumerator0") (Hits :type Array%20of%20Hit%20Result%20Structures))
  (bind _hits (Variables|GetHits))
  (bind _hitbuilding (LocalVariables|GetHitBuilding))
  (bind _cameracollisionignoretagname (LocalVariables|GetCameraCollisionIgnoreTagName))
  (Utilities|FlowControl|Sequence
    (:then_0
      (bind _playerpawn (Variables|GetPlayerPawn)
        (:then
          (for _array_element _hits
            (bind _output_get (LocalVariables|SetCurrentHit _array_element))
            (bind _asbuilding_actor (Utilities|Casting|CastToBuildingActor (.HitActor (Collision|BreakHitResult _output_get))))
            (LocalVariables|SetHitBuilding _asbuilding_actor)
            (Class|RocketRamSharedLibrary|RamLog (Variables|GetPlayerPawn) (Utilities|Text|FormatText "NSLOCTEXT(\"[5F0C5B7A54E665784329CE12CADAC24E]\", \"253AAEE442FEEE0BF569B480B6313C67\", \"Destruction - Destroyed Building {0}.\")" (LocalVariables|GetHitBuilding)) false true)
            (Utilities|FlowControl|Sequence
              (:then_0
                ;[1 0.5451 0] Ignore Collision With Building And Ignore In Next Destruction Trace
                (Collision|IgnoreActorwhenMoving (Variables|/Script/Engine.Character|GetCapsuleComponent (Variables|GetPlayerPawn)) _hitbuilding true)
                (Utilities|Array|Add (LocalVariables|GetIgnoredActors) _hitbuilding))
              (:then_1
                ;[0.474 0.3012 0.09189] Apply Damage To Building
                (GameplayEffects|ApplyGameplayEffectToTarget (Ability|GetFortAbilitySystemComponent (Variables|GetPlayerPawn)) (LocalVariables|GetGE_DestructionTraceDamage_Class) (Ability|GetAbilitySystemComponent (LocalVariables|GetHitBuilding)) 0.0 (GameplayEffects|MakeFortEffectContext (Ability|GetFortAbilitySystemComponent (Variables|GetPlayerPawn)) (Variables|GetPlayerPawn) (Variables|GetPlayerPawn) (Variables|GetPlayerPawn) 0))
                (LocalVariables|SetDidDamageSomething true))))
          ;[1 0 0.8982] Marks hit buildings (i.e. buildings that are predicted to be destroyed soon) to be ignored by the camera.
          ; This makes camera movement smoother because it won't try to pullback behind walls that no longer exists.
          ;
          ; The caller of the function is responsible for keeping track of the list of these actors and calling ClearCameraCollisionActorTags(() with that list after the ram rocket finishes its movement.
          (if (not (Networking|IsDedicatedServer _playerpawn))
            (for _array_element_2 _hits
              (bind _asbuilding_actor_2 (Utilities|Casting|CastToBuildingActor (.HitActor (Collision|BreakHitResult _array_element_2))))
              (if (not (Actor|ActorHasTag _asbuilding_actor_2 _cameracollisionignoretagname))
                (Class|BTechArtGlobalLibrary|AddActorTag _asbuilding_actor_2 _cameracollisionignoretagname _asbuilding_actor_2)
                (Utilities|Array|Add (LocalVariables|GetCameraIgnoredActors_) _asbuilding_actor_2)))))
        (:else)))
    (:then_1
      ;[1 0.4792 0.4792] Return data
      (return (LocalVariables|GetIgnoredActors) (LocalVariables|GetCameraIgnoredActors_) (LocalVariables|GetDidDamageSomething)))))

(fn HandleDestructionTraceHits
  (params (PlayerPawn :type Fort%20Player%20Pawn%20Object%20Reference) (RamState :type Enum_RocketRamState%20Enum :default "NewEnumerator0") (PawnHits :type Array%20of%20Hit%20Result%20Structures) (VehicleHits :type Array%20of%20Hit%20Result%20Structures) (DestructibleBuildingHits :type Array%20of%20Hit%20Result%20Structures))
  (Variables|GetPlayerPawn
    (:then
      (Utilities|Array|Clear (LocalVariables|GetIgnoredActors))
      (Utilities|FlowControl|Sequence
        (:then_0
          (bind (_actorstoignore _damaged_a_pawn) (DestructionTrace|HitHandling|HandlePawnHits (Variables|GetPlayerPawn) (Variables|GetRamState) (Variables|GetPawnHits) (Variables|GetPlayerPawn)))
          (Utilities|Array|AppendArray (LocalVariables|GetIgnoredActors) _actorstoignore)
          (LocalVariables|SetDamaged%20A%20Pawn_ _damaged_a_pawn))
        (:then_1
          (bind (_actorstoignore_2 _damaged_a_vehicle) (DestructionTrace|HitHandling|HandleVehicleHits (Variables|GetPlayerPawn) (Variables|GetRamState) (Variables|GetVehicleHits) (Variables|GetPlayerPawn)))
          (Utilities|Array|AppendArray (LocalVariables|GetIgnoredActors) _actorstoignore_2)
          (LocalVariables|SetDamaged%20A%20Vehicle_ _damaged_a_vehicle))
        (:then_2
          (bind (_actorstoignore_3 _cameraignoredactors _damagedsomething) (DestructionTrace|HitHandling|HandleDestructibleBuildingHits (Variables|GetPlayerPawn) (Variables|GetRamState) (Variables|GetDestructibleBuildingHits) (Variables|GetPlayerPawn)))
          (Utilities|Array|AppendArray (LocalVariables|GetIgnoredActors) _actorstoignore_3)
          (Utilities|Array|AppendArray (LocalVariables|GetActorsForCameraToIgnore_) _cameraignoredactors)
          (LocalVariables|SetDamaged%20A%20Build_ _damagedsomething))
        (:then_3
          (return (LocalVariables|GetIgnoredActors) (LocalVariables|GetActorsForCameraToIgnore_) (LocalVariables|GetDamaged%20A%20Build_) (LocalVariables|GetDamaged%20A%20Vehicle_) (LocalVariables|GetDamaged%20A%20Pawn_)))))
    (:else)))

(fn HandlePawnHits
  (params (PlayerPawn :type Fort%20Player%20Pawn%20Object%20Reference) (RamState :type Enum_RocketRamState%20Enum :default "NewEnumerator0") (Hits :type Array%20of%20Hit%20Result%20Structures))
  (Utilities|FlowControl|Sequence
    (:then_0
      (Variables|GetPlayerPawn
        (:then
          (for _array_element (Variables|GetHits)
            (bind _asfort_pawn (Utilities|Casting|CastToFortPawn (.HitActor (Collision|BreakHitResult _array_element))))
            (LocalVariables|SetHitPawn _asfort_pawn)
            (Class|RocketRamSharedLibrary|RamLog (Variables|GetPlayerPawn) (Utilities|Text|FormatText "NSLOCTEXT(\"[5F0C5B7A54E665784329CE12CADAC24E]\", \"6BF8900747882CEA4605A0A6D4D73E57\", \"Destruction - Hit Pawn {0}.\")" (LocalVariables|GetHitPawn)) false true)
            (switch Utilities|FlowControl|Switch|SwitchonEnum_RocketRamState (Variables|GetRamState)
              (:NewEnumerator1
                _)
              (:NewEnumerator2
                _)
              (:NewEnumerator5
                _)
              (:NewEnumerator7
                _))
            (Utilities|FlowControl|Sequence
              (:then_0
                ;[1 1 1] Apply Knockback To Pawn
                (Utilities|LaunchPawns (Variables|GetPlayerPawn) (LocalVariables|GetHitPawn) (* (+ (Math|Vector|Normalize (Transformation|GetVelocity (Variables|GetPlayerPawn))) "0.000000,0.000000,1.000000") (ScalableFloat|GetValueatLevel (LocalVariables|GetDestructionTraceKnockbackMagnitudeFromDirectHit))) true true (ScalableFloat|GetValueasBool (LocalVariables|GetGrantFallImmunityToKnockbackVictims)) true false true true true)
                (GameplayEffects|ApplyGameplayEffectToSelf :self (Ability|GetFortAbilitySystemComponent (LocalVariables|GetHitPawn)) :GameplayEffectClass "/RocketRamGameplay/Items/RocketRam/Effects/GE_VehicleLaunchImmunity_RocketRam.GE_VehicleLaunchImmunity_RocketRam_C" :Level 0.0))
              (:then_1
                ;[1 1 1] Apply Damage To Pawn
                (GameplayEffects|ApplyGameplayEffectToTarget (Ability|GetFortAbilitySystemComponent (Variables|GetPlayerPawn)) (LocalVariables|GetGE_DestructionTraceDamage_Class) (Ability|GetAbilitySystemComponent (LocalVariables|GetHitPawn)) 0.0 (GameplayEffects|MakeFortEffectContext (Ability|GetFortAbilitySystemComponent (Variables|GetPlayerPawn)) (Variables|GetPlayerPawn) (Variables|GetPlayerPawn) (Variables|GetPlayerPawn) 0))
                (LocalVariables|SetDamaged%20A%20Pawn_ true))
              (:then_2
                ;[1 1 1] Ignore Pawn In The Next Destruction Trace
                (Utilities|Array|Add (LocalVariables|GetIgnoredActors) (LocalVariables|GetHitPawn))))))
        (:else)))
    (:then_1
      (return (LocalVariables|GetIgnoredActors) (LocalVariables|GetDamaged%20A%20Pawn_)))))

(fn HandleVehicleHits
  (params (PlayerPawn :type Fort%20Player%20Pawn%20Object%20Reference) (RamState :type Enum_RocketRamState%20Enum :default "NewEnumerator0") (Hits :type Array%20of%20Hit%20Result%20Structures))
  (bind _playerpawn (Variables|GetPlayerPawn))
  (bind _returnvalue (Transformation|GetVelocity _playerpawn))
  (bind _sf_destructiontraceknockbackmagnitudefromdirecthitvehicles (LocalVariables|GetSF_DestructionTraceKnockbackMagnitudeFromDirectHitVehicles))
  (bind _returnvalue_2 (ScalableFloat|GetValueatLevel _sf_destructiontraceknockbackmagnitudefromdirecthitvehicles))
  (bind _returnvalue_3 (Math|Vector|Normalize _returnvalue))
  (bind _returnvalue_4 (+ _returnvalue_3 "0.000000,0.000000,0.300000"))
  (bind _returnvalue_5 (* _returnvalue_4 _returnvalue_2))
  (bind _playerpawn_2 (Variables|GetPlayerPawn))
  (bind _hitvehicle (LocalVariables|GetHitVehicle))
  (Utilities|FlowControl|Sequence
    (:then_0
      (Variables|GetPlayerPawn
        (:then
          (for _array_element (Variables|GetHits)
            (bind _asfort_athena_vehicle (Utilities|Casting|CastToFortAthenaVehicle (.HitActor (Collision|BreakHitResult _array_element))))
            (LocalVariables|SetHitVehicle _asfort_athena_vehicle)
            (Class|RocketRamSharedLibrary|RamLog (Variables|GetPlayerPawn) (Utilities|Text|FormatText "NSLOCTEXT(\"[5F0C5B7A54E665784329CE12CADAC24E]\", \"A34C4BAC446497F70171C4BD1EC0531D\", \"Destruction - Hit Vehicle {0}.\")" (LocalVariables|GetHitVehicle)) false true)
            (switch Utilities|FlowControl|Switch|SwitchonEnum_RocketRamState (Variables|GetRamState)
              (:NewEnumerator1
                _)
              (:NewEnumerator2
                _)
              (:NewEnumerator5
                _)
              (:NewEnumerator7
                _))
            (Utilities|FlowControl|Sequence
              (:then_0
                ;[0 0.7675 1] Apply Knockback To Vehicle
                (Audio|BroadcastAbilityImpactatLocation _playerpawn_2 0 _hitvehicle _returnvalue_5 777.0 "None" (Class|AthenaFunctionLibrary|RowtoBool (LocalVariables|GetSF_MakeVehicleImpulseAdditive)))
                ;;[0.9967 1 0] Log the vehicle impulse so we have extra information if the launching vehicles too far bug comes back.
                (Class|RocketRamSharedLibrary|RamLog _playerpawn_2 (Utilities|Text|FormatText "NSLOCTEXT(\"[5F0C5B7A54E665784329CE12CADAC24E]\", \"C2154086473840B1D9EA9681DD38B8C0\", \"HandleVehicleHits() vehicle: [{vehicle_name}]impulse: [{impulse}]\")" (Vehicle|GetDisplayName _hitvehicle :!class /Script/FortniteGame.FortAthenaVehicle) (Utilities|String|ToString(Vector) _returnvalue_5)) false false _playerpawn_2))
              (:then_1
                ;[1 1 1] Apply Damage To Vehicle
                (GameplayEffects|ApplyGameplayEffectToTarget (Ability|GetFortAbilitySystemComponent (Variables|GetPlayerPawn)) (LocalVariables|GetGE_DestructionTraceDamage_Class) (Ability|GetAbilitySystemComponent (LocalVariables|GetHitVehicle)) 0.0 (GameplayEffects|MakeFortEffectContext (Ability|GetFortAbilitySystemComponent (Variables|GetPlayerPawn)) (Variables|GetPlayerPawn) (Variables|GetPlayerPawn) (Variables|GetPlayerPawn) 0))
                (LocalVariables|SetDamaged%20A%20Vehicle_ true))
              (:then_2
                ;[1 1 1] Ignore Vehicle In The Next Destruction Trace
                (Utilities|Array|Add (LocalVariables|GetIgnoredActors) (LocalVariables|GetHitVehicle))))))
        (:else)))
    (:then_1
      (return (LocalVariables|GetIgnoredActors) (LocalVariables|GetDamaged%20A%20Vehicle_)))))

(fn HasPlayerHitFreeFallPeak
  (params (PlayerPawn :type Fort%20Player%20Pawn%20Object%20Reference))
  (bind (_returnvalue_x _returnvalue_y _returnvalue_z) (Math|Vector|BreakVector (Pawn|Components|CharacterMovement|GetLastUpdateVelocity (Variables|/Script/Engine.Character|GetCharacterMovement PlayerPawn))))
  (return (<= _returnvalue_z (ScalableFloat|GetValueatLevel (LocalVariables|GetSF_ZVelocityThresholdForFinalImpactAOE)))))

(fn ShouldBuildingBlockDuringFreefall
  (params (PlayerPawn :type Fort%20Player%20Pawn%20Object%20Reference) (BuildingHit :type Hit%20Result%20Structure) (TraceDirection :type Vector))
  (bind (_bblockinghit _binitialoverlap _time _distance _location _impactpoint _normal _impactnormal _physmat _hitactor) (Collision|BreakHitResult (Variables|GetBuildingHit)))
  (Utilities|FlowControl|Sequence
    (:then_0
      (Variables|GetPlayerPawn
        (:then
          (bind _asbuilding_actor (Utilities|Casting|CastToBuildingActor _hitactor))
          (if (GameplayTags|HasAnyMatchingGameplayTags _asbuilding_actor (LocalVariables|GetFreeFallBlockingBuildingTags))
            (bind _result (AOE|ShouldDoAOEImpact (Variables|GetPlayerPawn) (Variables|GetTraceDirection) (Variables|GetBuildingHit) (Variables|GetPlayerPawn) :!class /RocketRamGameplay/Items/RocketRam/RocketRam_AOEImpactLibrary.RocketRam_AOEImpactLibrary_C))
            (if _result
              (return true true))
            (else
              (return false true))))
        (:else)))
    (:then_1
      (return false false))))

(fn ShouldHitBlockDestructionTrace
  (params (PlayerPawn :type Fort%20Player%20Pawn%20Object%20Reference) (RamState :type Enum_RocketRamState%20Enum :default "NewEnumerator0") (TraceType :type Enum_DestructionTraceType%20Enum :default "NewEnumerator4") (TraceDirection :type Vector) (TraceDistance :type Float%20%28double-precision%29) (TraceActorsToIgnore :type Array%20of%20Actor%20Object%20References) (HitResult :type Hit%20Result%20Structure) (TimeBeenStopped :type Float%20%28double-precision%29))
  (bind (_bblockinghit _binitialoverlap _time _distance _location _impactpoint _normal _impactnormal _physmat _hitactor) (Collision|BreakHitResult HitResult))
  (bind _hitactor_2 (LocalVariables|GetHitActor))
  (if (ScalableFloat|GetValueasBool (LocalVariables|GetEverythingShouldBlock))
    (return true)
    (elif (ScalableFloat|GetValueasBool (LocalVariables|GetNothingShouldBlock))
      (return false)
      (else
        (bind _output_get (LocalVariables|SetHitActor _hitactor))
        (Utilities|IsValid _output_get
          (:Is%20Valid
            (Utilities|FlowControl|Sequence
              (:then_0
                (Utilities|Casting|CastToFortPawn (LocalVariables|GetHitActor)
                  (:then
                    _)
                  (:CastFailed
                    (Utilities|Casting|CastToFortAthenaVehicle (LocalVariables|GetHitActor)
                      (:then
                        _)
                      (:CastFailed
                        (bind _asbuilding_actor (Utilities|Casting|CastToBuildingActor (LocalVariables|GetHitActor))
                          (:then
                            (if (DestructionTrace|CanDestroyBuildingActor (Variables|GetPlayerPawn) _asbuilding_actor (Variables|GetPlayerPawn))
                              ;[0.8045 0.2471 1] Destructible Buildings
                              (switch Utilities|FlowControl|Switch|SwitchonEnum_RocketRamState (Variables|GetRamState)
                                (:NewEnumerator1
                                  _)
                                (:NewEnumerator2
                                  (if (DestructionTrace|HasPlayerHitFreeFallPeak (Variables|GetPlayerPawn) (Variables|GetPlayerPawn)) :!label _label
                                    (return (DestructionTrace|ShouldBuildingBlockDuringFreefall (Variables|GetPlayerPawn) (Variables|GetHitResult) (Variables|GetTraceDirection) (Variables|GetPlayerPawn)))
                                    (else
                                      _)))
                                (:NewEnumerator5
                                  (goto _label)))
                              (return false)
                              (elif (GameplayTags|HasAnyMatchingGameplayTags _asbuilding_actor (LocalVariables|GetIndestructibleBuildingsToIgnoreCollision))
                                (return false))))
                          (:CastFailed
                            _))
                        ;[0.2577 0.2174 1] If set to never block from terrain, never block from these actors, always skim past them.
                        (if (ScalableFloat|GetValueasBool (LocalVariables|GetNeverBlockActiveThrustAgainstTerrain))
                          (Utilities|Casting|CastToLandscapeSplineActor _hitactor_2
                            (:then
                              _)
                            (:CastFailed
                              (Utilities|Casting|CastToLandscapeSplineMeshesActor _hitactor_2
                                (:then
                                  _)
                                (:CastFailed
                                  (switch Utilities|FlowControl|Switch|SwitchonEnum_RocketRamState (Variables|GetRamState) :!label _label_2
                                    (:NewEnumerator1
                                      (if (ScalableFloat|GetValueasBool (LocalVariables|GetNeverBlockActiveThrustAgainstTerrain))
                                        (Utilities|Casting|CastToLandscapeProxy (LocalVariables|GetHitActor)
                                          (:then
                                            _)
                                          (:CastFailed
                                            (if (not (and (Editor|InEditororPIE) (Utilities|String|Contains (Utilities|GetDisplayName (LocalVariables|GetHitActor)) "BuilderGridPlane")))
                                              (return (> (Variables|GetTimeBeenStopped) (ScalableFloat|GetValueatLevel (LocalVariables|GetDestructionTraceMinTimeBlockedInActiveThrustToEndRide))))))) ; Ignore landscape during the active thrust phase. \nEverything else should only block if there's enough \nforce to perform an AOE.
                                        (return false))
                                      (return (> (Variables|GetTimeBeenStopped) (ScalableFloat|GetValueatLevel (LocalVariables|GetDestructionTraceMinTimeBlockedInActiveThrustToEndRide)))))
                                    (:NewEnumerator2
                                      _)
                                    (:NewEnumerator5
                                      _))
                                  (bind _result (AOE|ShouldDoAOEImpact (Variables|GetPlayerPawn) (Variables|GetTraceDirection) (Variables|GetHitResult) (Variables|GetPlayerPawn) :!class /RocketRamGameplay/Items/RocketRam/RocketRam_AOEImpactLibrary.RocketRam_AOEImpactLibrary_C))
                                  (if _result
                                    (bind _bisfloorclip (DestructionTrace|DoFloorClipCheck (Variables|GetPlayerPawn) (Variables|GetHitResult) (Variables|GetTraceType) (Variables|GetTraceDirection) (Variables|GetTraceDistance) (Variables|GetTraceActorsToIgnore) (Variables|GetPlayerPawn))) ; This check excludes actors that are extremely low \nto the ground that the player could just glide right over.
                                    (return (not _bisfloorclip))
                                    (else
                                      (return false)))))))
                          (return false))
                        (goto _label_2)))))
                (switch Utilities|FlowControl|Switch|SwitchonEnum_RocketRamState (Variables|GetRamState)
                  (:NewEnumerator1
                    _)
                  (:NewEnumerator2
                    _)
                  (:NewEnumerator5
                    _))
                (return false))
              (:then_1
                (return true))))
          (:Is%20Not%20Valid
            (return false)))))))
