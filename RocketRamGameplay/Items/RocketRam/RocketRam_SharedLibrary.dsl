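; Shared function library for the Rocket Ram item, providing gravity math, an angle-from-horizontal helper and logging.
; CalculateNextGravity grows gravity over time, clamped between a floor remapped from horizontal speed and deviation from launch direction, and a configured maximum.
; Tuning comes from DataClass_RocketRam scalable values, while CalculateBaseGravity is a stub returning zero.
; RamLog writes warnings or verbose messages to the shipping log, tagged Server or Client, with optional debug printing.
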(fn CalculateNextGravity
  (params (CurrentGravity :type Float%20%28double-precision%29) (CurrentVelocity :type Vector) (OriginalLaunchDirection :type Vector) (BaseGravity :type Float%20%28double-precision%29) (DeltaTime :type Float%20%28double-precision%29))
  (return (Math|Float|Clamp(Float) (+ CurrentGravity (* (ScalableFloat|GetValueatLevel (Variables|/RocketRamGameplay/Items/RocketRam/DataClass_RocketRam.DataClass_RocketRam_C|GetGravity_BaseIncreaseSpeed (LocalVariables|GetConfigData))) DeltaTime)) (Math|Float|Clamp(Float) (+ (Utilities|RemapFloat (Math|Vector|VectorLengthXY CurrentVelocity) (Math|Vector2D|MakeVector2D (ScalableFloat|GetValueatLevel (Variables|/RocketRamGameplay/Items/RocketRam/DataClass_RocketRam.DataClass_RocketRam_C|GetGravity_SpeedForLeastGravity (LocalVariables|GetConfigData))) (ScalableFloat|GetValueatLevel (Variables|/RocketRamGameplay/Items/RocketRam/DataClass_RocketRam.DataClass_RocketRam_C|GetGravity_SpeedForMostGravity (LocalVariables|GetConfigData)))) (Math|Vector2D|MakeVector2D 0.0 (ScalableFloat|GetValueatLevel (Variables|/RocketRamGameplay/Items/RocketRam/DataClass_RocketRam.DataClass_RocketRam_C|GetGravity_SpeedInfluenceOnGravity (LocalVariables|GetConfigData)))) true) (Utilities|RemapFloat (Math|Trig|Acos(Degrees) (Math|Float|Absolute(Float) (Math|Vector|DotProduct (Math|Vector|Normalize CurrentVelocity) OriginalLaunchDirection))) (Math|Vector2D|MakeVector2D (ScalableFloat|GetValueatLevel (Variables|/RocketRamGameplay/Items/RocketRam/DataClass_RocketRam.DataClass_RocketRam_C|GetGravity_DirectionDifferenceFromLaunchDirForLeastGravity (LocalVariables|GetConfigData))) (ScalableFloat|GetValueatLevel (Variables|/RocketRamGameplay/Items/RocketRam/DataClass_RocketRam.DataClass_RocketRam_C|GetGravity_DirectionDifferenceFromLaunchDirForMostGravity (LocalVariables|GetConfigData)))) (Math|Vector2D|MakeVector2D BaseGravity (ScalableFloat|GetValueatLevel (Variables|/RocketRamGameplay/Items/RocketRam/DataClass_RocketRam.DataClass_RocketRam_C|GetGravity_DirectionInfluenceOnGravity (LocalVariables|GetConfigData)))) true)) (ScalableFloat|GetValueatLevel (Variables|/RocketRamGameplay/Items/RocketRam/DataClass_RocketRam.DataClass_RocketRam_C|GetGravity_Min (LocalVariables|GetConfigData))) (ScalableFloat|GetValueatLevel (Variables|/RocketRamGameplay/Items/RocketRam/DataClass_RocketRam.DataClass_RocketRam_C|GetGravity_Max (LocalVariables|GetConfigData)))) (ScalableFloat|GetValueatLevel (Variables|/RocketRamGameplay/Items/RocketRam/DataClass_RocketRam.DataClass_RocketRam_C|GetGravity_Max (LocalVariables|GetConfigData))))))

(fn CalculateBaseGravity
  (params (LaunchDirection :type Vector))
  (return 0.0))

(fn GetClosenessToFlat
  (params (Dir :type Vector))
  (return (Math|Float|Absolute(Float) (- 9e+01 (Math|Trig|Acos(Degrees) (Math|Vector|DotProduct Dir "0.000000,0.000000,1.000000")))) (- 9e+01 (Math|Trig|Acos(Degrees) (Math|Vector|DotProduct Dir "0.000000,0.000000,1.000000")))))

(fn RamLog
  (params (PlayerPawn :type Fort%20Player%20Pawn%20Object%20Reference) (LogText :type Text) (bWarning :type Boolean) (bDebug :type Boolean))
  (bind _logtext (Variables|GetLogText))
  (if (or bWarning (Class|AthenaFunctionLibrary|RowtoBool (Variables|/RocketRamGameplay/Items/RocketRam/DataClass_RocketRam.DataClass_RocketRam_C|GetDoVerboseLogging (LocalVariables|GetConfigData))))
    (bind _playerpawn (Variables|GetPlayerPawn)
      (:then
        (if (not (Variables|GetbDebug))
          (if (LocalVariables|GetbEnableDebugLogs)
            (Development|PrintString 0 (Utilities|String|ToString(Text) (Utilities|Text|FormatText "NSLOCTEXT(\"[4BADC4C08662FA9CA0AE41DBB09551A8]\", \"421C04444A138FC6CCB036818A7F483F\", \"Rocket Ram Log [{0}]: {1}\")" (select (Networking|HasAuthority _playerpawn) "Server" "Client") _logtext)) true true (Math|Color|SelectColor "(R=1.000000,G=0.000000,B=0.000000,A=1.000000)" "(R=0.000000,G=0.032413,B=1.000000,A=1.000000)" (Variables|GetbWarning)))))
        (Fortnite|FortShippingLog 0 (Utilities|String|ToString(Text) (Utilities|Text|FormatText "NSLOCTEXT(\"[4BADC4C08662FA9CA0AE41DBB09551A8]\", \"421C04444A138FC6CCB036818A7F483F\", \"Rocket Ram Log [{0}]: {1}\")" (select (Networking|HasAuthority _playerpawn) "Server" "Client") _logtext)) (Variables|GetbWarning)))
      (:else
        (Fortnite|FortShippingLog 0 (Utilities|String|Append (Utilities|String|ToString(Text) LogText) " |  Also, the player pawn was null when this triggered") (Variables|GetbWarning))))))
