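# Sound cue for the Rocket Ram's charge-up blast, layered by distance into close, medium and distant parts.
# The close layer mixes a weighted random blast with an air layer; medium and distant layers reuse the same three medium waves, pitched up and down respectively.
# Each layer has its own attenuation and sound class override before a descending-volume mix.
# Weapon gunfire concurrency sets are assigned, alongside override values of one instance per owner.
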SoundCue /RocketRamGameplay/Items/RocketRam/Sounds/RocketRam_ChargeUp_Blast_Cue
root=Mixer_1
nodes [
 {
  id=Attenuation_0
  children [Modulator_0]
  properties {attenuationSettings=&Fort_Weapon_Projectile_Distant_Mono_Athena_Att}
 }
 {
  id=Attenuation_3
  children [Modulator_1]
  properties {attenuationSettings=/Game/Sounds/Attenuation/Fort_Weapon_Projectile_Close_Athena_Att_Scoped}
 }
 {
  id=Attenuation_4
  children [Modulator_9]
  properties {attenuationSettings=&Fort_Weapon_Projectile_Medium_Mono_Athena_Att}
 }
 {id=Mixer_1, children [SoundClass_2,SoundClass_3,SoundClass_0], properties {inputVolume [1,0.6,0.3]}}
 {id=Mixer_2, children [Random_3,WavePlayer_0], properties {inputVolume [1,0.7]}}
 {id=Modulator_0, children [Random_0], properties {pitchMax=0.8, pitchMin=0.9, volumeMax=2, volumeMin=2}}
 {id=Modulator_1, children [Mixer_2]}
 {id=Modulator_9, children [Random_6], properties {pitchMax=1.2, pitchMin=1.1, volumeMax=2, volumeMin=2}}
 {id=Random_0, children [WavePlayer_2,WavePlayer_3,WavePlayer_4], properties {weights [1,1,1]}}
 {id=Random_3, children [WavePlayer_6,WavePlayer_7,WavePlayer_8,WavePlayer_9], properties {weights [1,1,2,2]}}
 {id=Random_6, children [WavePlayer_23, WavePlayer_24, WavePlayer_25], properties {weights [1,1,1]}}
 {
  id=SoundClass_0
  children [Attenuation_0]
  properties {soundClassOverride=&Fort_WeaponRanged_Fire_Distant}
 }
 {
  id=SoundClass_2
  children [Attenuation_3]
  properties {soundClassOverride=&Fort_WeaponRanged_Fire_3P_PSM}
 }
 {
  id=SoundClass_3
  children [Attenuation_4]
  properties {soundClassOverride=&Fort_WeaponRanged_Fire_Distant}
 }
 {id=WavePlayer_0, properties {soundWaveAssetPtr=&/RocketRam_ChargeUp_Blast_Layer_Air_01}}
 {id=WavePlayer_2, properties {soundWaveAssetPtr=&/RocketRam_ChargeUp_Blast_Medium_02}}
 {id=WavePlayer_23, properties {soundWaveAssetPtr=&RocketRam_ChargeUp_Blast_Medium_02}}
 {id=WavePlayer_24, properties {soundWaveAssetPtr=&/RocketRam_ChargeUp_Blast_Medium_03}}
 {id=WavePlayer_25, properties {soundWaveAssetPtr=&/RocketRam_ChargeUp_Blast_Medium_04}}
 {id=WavePlayer_3, properties {soundWaveAssetPtr=&RocketRam_ChargeUp_Blast_Medium_03}}
 {id=WavePlayer_4, properties {soundWaveAssetPtr=&RocketRam_ChargeUp_Blast_Medium_04}}
 {id=WavePlayer_6, properties {soundWaveAssetPtr=&/RocketRam_ChargeUp_Blast_01}}
 {id=WavePlayer_7, properties {soundWaveAssetPtr=&/RocketRam_ChargeUp_Blast_02}}
 {id=WavePlayer_8, properties {soundWaveAssetPtr=&/RocketRam_ChargeUp_Blast_03}}
 {id=WavePlayer_9, properties {soundWaveAssetPtr=&/RocketRam_ChargeUp_Blast_04}}
]
properties {
 maxDistance=27500
 volumeMultiplier=0.8
 concurrencyOverrides {bLimitToOwner=true, maxCount=1}
 concurrencySet [&Weapon_Gunfire_Global_Conc, &Weapon_Gunfire_Local_Conc]
}
