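# Sound cue for the Rocket Ram's charge-up, mixing close, medium and distant layers, each with its own attenuation and sound class.
# The close layer blends random charge-up and AOE explosion waves with a delayed flareup-plus-air layer.
# Medium and distant layers both pick from medium charge-up variants, pitched up for medium and down for distant.
# Global and local weapon gunfire concurrency sets govern playback.
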SoundCue /RocketRamGameplay/Items/RocketRam/Sounds/RocketRam_ChargeUp_Cue
root=Mixer_1
nodes [
 {
  id=Attenuation_0
  children [Modulator_0]
  properties {attenuationSettings=&Fort_Weapon_Projectile_Distant_Mono_Athena_Att}
 }
 {
  id=Attenuation_3
  children [Modulator_8]
  properties {attenuationSettings=/Game/Sounds/Attenuation/Fort_Weapon_Projectile_Close_Athena_Att_Scoped}
 }
 {
  id=Attenuation_4
  children [Modulator_9]
  properties {attenuationSettings=&Fort_Weapon_Projectile_Medium_Mono_Athena_Att}
 }
 {id=Delay_0, children [Mixer_2], properties {delayMax=0.8, delayMin=0.8}}
 {id=Mixer_0, children [Random_2,Delay_0,Random_1], properties {inputVolume [1.3,0.8,1]}}
 {id=Mixer_1, children [SoundClass_2,SoundClass_3,SoundClass_0], properties {inputVolume [1,0.6,0.4]}}
 {id=Mixer_2, children [Random_0,WavePlayer_5], properties {inputVolume [1,0.7]}}
 {id=Modulator_0, children [Random_3], properties {pitchMax=0.95, pitchMin=0.85, volumeMax=2, volumeMin=2}}
 {id=Modulator_8, children [Mixer_0], properties {volumeMax=1.1, volumeMin=1.1}}
 {id=Modulator_9, children [Random_4], properties {pitchMax=1.2, pitchMin=1.1, volumeMax=2, volumeMin=2}}
 {id=Random_0, children [WavePlayer_0,WavePlayer_1,WavePlayer_2], properties {weights [1,1,1]}}
 {id=Random_1, children [WavePlayer_6,WavePlayer_7,WavePlayer_8], properties {weights [1,1,1]}}
 {id=Random_2, children [WavePlayer_10, WavePlayer_11, WavePlayer_12, WavePlayer_13], properties {weights [1,1,1,1]}}
 {id=Random_3, children [WavePlayer_3, WavePlayer_4, WavePlayer_9, WavePlayer_18], properties {weights [1,1,1,1]}}
 {id=Random_4, children [WavePlayer_14, WavePlayer_15, WavePlayer_16, WavePlayer_17], properties {weights [1,1,1,1]}}
 {
  id=SoundClass_0
  children [Attenuation_0]
  properties {soundClassOverride=&Fort_WeaponRanged_Fire_Distant}
 }
 {
  id=SoundClass_2
  children [Attenuation_3]
  properties {soundClassOverride=&Fort_WeaponRanged_Fire_3P_PSM}
 }
 {
  id=SoundClass_3
  children [Attenuation_4]
  properties {soundClassOverride=&Fort_WeaponRanged_Fire_Distant}
 }
 {id=WavePlayer_0, properties {soundWaveAssetPtr=&/RocketRam_ChargeUp_Layer_Flareup_01}}
 {id=WavePlayer_1, properties {soundWaveAssetPtr=&/RocketRam_ChargeUp_Layer_Flareup_02}}
 {id=WavePlayer_10, properties {soundWaveAssetPtr=&/RocketRam_ChargeUp_01}}
 {id=WavePlayer_11, properties {soundWaveAssetPtr=&/RocketRam_ChargeUp_02}}
 {id=WavePlayer_12, properties {soundWaveAssetPtr=&/RocketRam_ChargeUp_03}}
 {id=WavePlayer_13, properties {soundWaveAssetPtr=&/RocketRam_ChargeUp_04}}
 {id=WavePlayer_14, properties {soundWaveAssetPtr=&/RocketRam_ChargeUp_Medium_01}}
 {id=WavePlayer_15, properties {soundWaveAssetPtr=&/RocketRam_ChargeUp_Medium_02}}
 {id=WavePlayer_16, properties {soundWaveAssetPtr=&/RocketRam_ChargeUp_Medium_03}}
 {id=WavePlayer_17, properties {soundWaveAssetPtr=&/RocketRam_ChargeUp_Medium_04}}
 {id=WavePlayer_18, properties {soundWaveAssetPtr=&RocketRam_ChargeUp_Medium_04}}
 {id=WavePlayer_2, properties {soundWaveAssetPtr=&/RocketRam_ChargeUp_Layer_Flareup_03}}
 {id=WavePlayer_3, properties {soundWaveAssetPtr=&RocketRam_ChargeUp_Medium_01}}
 {id=WavePlayer_4, properties {soundWaveAssetPtr=&RocketRam_ChargeUp_Medium_02}}
 {id=WavePlayer_5, properties {soundWaveAssetPtr=&/RocketRam_ChargeUp_Layer_Air_01}}
 {id=WavePlayer_6, properties {soundWaveAssetPtr=&/RocketRam_ChargeUp_AOE_Explosion_01}}
 {id=WavePlayer_7, properties {soundWaveAssetPtr=&/RocketRam_ChargeUp_AOE_Explosion_02}}
 {id=WavePlayer_8, properties {soundWaveAssetPtr=&/RocketRam_ChargeUp_AOE_Explosion_03}}
 {id=WavePlayer_9, properties {soundWaveAssetPtr=&RocketRam_ChargeUp_Medium_03}}
]
properties {
 maxDistance=27500
 volumeMultiplier=0.5
 concurrencyOverrides {bLimitToOwner=true, maxCount=1}
 concurrencySet [&Weapon_Gunfire_Global_Conc, &Weapon_Gunfire_Local_Conc]
}
