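# RocketRam_Drop2_Cue is a SoundCue that plays a drop sound for the Rocket Ram item.
# A Random node picks among three RocketRam_Drop waves with equal weight, feeding a Modulator that slightly varies pitch and volume on each play.
# It uses the weapon-drop attenuation, the Loot_Collect sound class, a reduced volume and the Weapon_Drop_StopOldest concurrency set to limit overlapping drops.
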SoundCue /RocketRamGameplay/Items/RocketRam/Sounds/RocketRam_Drop2_Cue
root=Modulator_0
nodes [
 {id=Modulator_0, children [Random_1], properties {pitchMax=1.1, pitchMin=0.9, volumeMax=1.1, volumeMin=0.9}}
 {id=Random_1, children [WavePlayer_1,WavePlayer_3,WavePlayer_4], properties {weights [1,1,1]}}
 {id=WavePlayer_1, properties {soundWaveAssetPtr=&/RocketRam_Drop_01}}
 {id=WavePlayer_3, properties {soundWaveAssetPtr=&/RocketRam_Drop_02}}
 {id=WavePlayer_4, properties {soundWaveAssetPtr=&/RocketRam_Drop_03}}
]
properties {
 attenuationSettings=&Fort_Weapon_Drop_Att
 maxDistance=3000
 soundClassObject=&Fort_Loot_Collect
 volumeMultiplier=0.65
 concurrencyOverrides {resolutionRule=PreventNew}
 concurrencySet [&Weapon_Drop_StopOldest]
}
