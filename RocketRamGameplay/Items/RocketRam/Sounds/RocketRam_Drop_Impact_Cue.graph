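# Sound cue for the impact sound when the Rocket Ram item is dropped.
# A Random node picks one of three equally weighted drop-impact waves and feeds a Modulator at the root for variation.
# It plays at reduced volume through the loot-collect sound class, using shared weapon-drop attenuation and a weapon-drop stop-oldest concurrency set.
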SoundCue /RocketRamGameplay/Items/RocketRam/Sounds/RocketRam_Drop_Impact_Cue
root=Modulator_1
nodes [
 {id=Modulator_1, children [Random_0]}
 {id=Random_0, children [WavePlayer_0,WavePlayer_1,WavePlayer_2], properties {weights [1,1,1]}}
 {id=WavePlayer_0, properties {soundWaveAssetPtr=&/RocketRam_DropImpact_01}}
 {id=WavePlayer_1, properties {soundWaveAssetPtr=&/RocketRam_DropImpact_02}}
 {id=WavePlayer_2, properties {soundWaveAssetPtr=&/RocketRam_DropImpact_03}}
]
properties {
 attenuationSettings=&Fort_Weapon_Drop_Att
 maxDistance=3000
 soundClassObject=&Fort_Loot_Collect
 volumeMultiplier=0.65
 concurrencyOverrides {resolutionRule=PreventNew}
 concurrencySet [&Weapon_Drop_StopOldest]
}
