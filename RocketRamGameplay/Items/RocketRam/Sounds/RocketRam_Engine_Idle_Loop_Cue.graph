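# Engine idle sound cue for the RocketRam item, pairing a startup one-shot with a looping engine bed.
# A Mixer layers one of three randomly picked start waves over the looping wave, each branch passing through a Modulator.
# An Enveloper holds the loop's pitch and volume high at first, then eases them down over several seconds.
# It uses the melee weapon sound class, dedicated attenuation and a one-instance-per-owner concurrency limit.
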SoundCue /RocketRamGameplay/Items/RocketRam/Sounds/RocketRam_Engine_Idle_Loop_Cue
root=Mixer_1
nodes [
 {
  id=Enveloper_0
  children [Modulator_0]
  properties {
   loopEnd=10
   pitchCurve {editorCurveData {keys [<at 0 1.3>, <at 5.9958 1.3>, {time=13}]}}
   volumeCurve {editorCurveData {keys [<at 0.0168 1.9812>,<at 3.7447 1.9812>,<at 6.4039 1.7354>,<at 10.1567 1>]}}
  }
 }
 {id=Mixer_1, children [Enveloper_0,Modulator_1], properties {inputVolume [1,1.12]}}
 {id=Modulator_0, children [WavePlayer_2]}
 {id=Modulator_1, children [Random_1]}
 {id=Random_1, children [WavePlayer_0,WavePlayer_3,WavePlayer_4], properties {weights [1,1,1]}}
 {id=WavePlayer_0, properties {soundWaveAssetPtr=&/RocketRam_Engine_Start_01}}
 {id=WavePlayer_2, properties {bLooping=true, soundWaveAssetPtr=&/RocketRam_Engine_Loop_02}}
 {id=WavePlayer_3, properties {soundWaveAssetPtr=&/RocketRam_Engine_Start_02}}
 {id=WavePlayer_4, properties {soundWaveAssetPtr=&/RocketRam_Engine_Start_03}}
]
properties {
 attenuationSettings=/RocketRamGameplay/Items/RocketRam/Sounds/Settings/RocketRam_Engine_Loop_Att
 bOverrideConcurrency=true
 maxDistance=1500
 soundClassObject=&Fort_WeaponMelee
 volumeMultiplier=0.15
 concurrencyOverrides {bLimitToOwner=true, maxCount=1}
}
