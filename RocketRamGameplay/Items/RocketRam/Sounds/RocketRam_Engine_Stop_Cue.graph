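# Sound cue for the Rocket Ram item's engine stop sound, classed as a melee weapon sound.
# A Random node picks between two equally weighted stop recordings and feeds a Modulator for pitch and volume variation.
# It borrows the engine loop's attenuation settings and plays at reduced volume.
# A concurrency override limits it to one instance per owner.
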SoundCue /RocketRamGameplay/Items/RocketRam/Sounds/RocketRam_Engine_Stop_Cue
root=Modulator_0
nodes [
 {id=Modulator_0, children [Random_0]}
 {id=Random_0, children [WavePlayer_1,WavePlayer_2], properties {weights [1,1]}}
 {id=WavePlayer_1, properties {soundWaveAssetPtr=&/RocketRam_Engine_Stop_01}}
 {id=WavePlayer_2, properties {soundWaveAssetPtr=&/RocketRam_Engine_Stop_02}}
]
properties {
 attenuationSettings=/RocketRamGameplay/Items/RocketRam/Sounds/Settings/RocketRam_Engine_Loop_Att
 bOverrideConcurrency=true
 maxDistance=1500
 soundClassObject=&Fort_WeaponMelee
 volumeMultiplier=0.17
 concurrencyOverrides {bLimitToOwner=true, maxCount=1}
}
