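# Sound cue for the Rocket Ram equip sound, classed as a melee weapon sound.
# After a short fixed delay, it plays one of three equally weighted equip waves through a Modulator for pitch and volume variation.
# The cue lowers volume, raises pitch slightly, overrides concurrency to cap simultaneous plays, and borrows the Lobster melee weapon equip attenuation.
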SoundCue /RocketRamGameplay/Items/RocketRam/Sounds/RocketRam_Equip_Cue
root=Delay_0
nodes [
 {id=Delay_0, children [Modulator_0], properties {delayMax=0.1, delayMin=0.1}}
 {id=Modulator_0, children [Random_0]}
 {id=Random_0, children [WavePlayer_2,WavePlayer_3,WavePlayer_4], properties {weights [1,1,1]}}
 {id=WavePlayer_2, properties {soundWaveAssetPtr=&/RocketRam_Equip_01}}
 {id=WavePlayer_3, properties {soundWaveAssetPtr=&/RocketRam_Equip_02}}
 {id=WavePlayer_4, properties {soundWaveAssetPtr=&/RocketRam_Equip_03}}
]
properties {
 attenuationSettings=/Game/Sounds/Attenuation/Fort_Weapon_Melee_Lobster_Equip_Att
 bOverrideConcurrency=true
 maxDistance=4000
 pitchMultiplier=1.05
 soundClassObject=&Fort_WeaponMelee
 volumeMultiplier=0.4
 concurrencyOverrides {maxCount=4}
}
