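# Sound cue for the Rocket Ram's firing sound, layered into close, medium and distant versions.
# Each layer picks one of six waves at random, then runs through a modulator, its own projectile attenuation and a sound class override.
# Medium and distant layers share the distant-fire class and a sub-mixer, which the root mixer blends with the louder third-person close layer.
# Global and local weapon gunfire concurrency sets are assigned to the cue.
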SoundCue /RocketRamGameplay/Items/RocketRam/Sounds/RocketRam_Fire_Cue
root=Mixer_1
nodes [
 {
  id=Attenuation_3
  children [Modulator_1]
  properties {attenuationSettings=/Game/Sounds/Attenuation/Fort_Weapon_Projectile_Close_Athena_Att_Scoped}
 }
 {
  id=Attenuation_4
  children [Modulator_2]
  properties {attenuationSettings=&Fort_Weapon_Projectile_Medium_Mono_Athena_Att}
 }
 {
  id=Attenuation_5
  children [Modulator_3]
  properties {attenuationSettings=&Fort_Weapon_Projectile_Distant_Mono_Athena_Att}
 }
 {id=Mixer_1, children [SoundClass_2,Mixer_4], properties {inputVolume [1.2,0.75]}}
 {id=Mixer_4, children [SoundClass_3,SoundClass_4], properties {inputVolume [0.8,0.6]}}
 {id=Modulator_1, children [Random_2]}
 {id=Modulator_2, children [Random_4]}
 {id=Modulator_3, children [Random_5]}
 {
  id=Random_2
  children [WavePlayer_10, WavePlayer_11, WavePlayer_12, WavePlayer_13, WavePlayer_14, WavePlayer_15]
  properties {weights [1,1,1,1,1,1]}
 }
 {
  id=Random_4
  children [WavePlayer_16, WavePlayer_17, WavePlayer_18, WavePlayer_19, WavePlayer_20, WavePlayer_21]
  properties {weights [1,1,1,1,1,1]}
 }
 {
  id=Random_5
  children [WavePlayer_22, WavePlayer_23, WavePlayer_24, WavePlayer_25, WavePlayer_26, WavePlayer_27]
  properties {weights [1,1,1,1,1,1]}
 }
 {
  id=SoundClass_2
  children [Attenuation_3]
  properties {soundClassOverride=&Fort_WeaponRanged_Fire_3P_PSM}
 }
 {
  id=SoundClass_3
  children [Attenuation_4]
  properties {soundClassOverride=&Fort_WeaponRanged_Fire_Distant}
 }
 {
  id=SoundClass_4
  children [Attenuation_5]
  properties {soundClassOverride=&Fort_WeaponRanged_Fire_Distant}
 }
 {id=WavePlayer_10, properties {soundWaveAssetPtr=&/RocketRam_Fire_Close_01}}
 {id=WavePlayer_11, properties {soundWaveAssetPtr=&/RocketRam_Fire_Close_02}}
 {id=WavePlayer_12, properties {soundWaveAssetPtr=&/RocketRam_Fire_Close_03}}
 {id=WavePlayer_13, properties {soundWaveAssetPtr=&/RocketRam_Fire_Close_04}}
 {id=WavePlayer_14, properties {soundWaveAssetPtr=&/RocketRam_Fire_Close_05}}
 {id=WavePlayer_15, properties {soundWaveAssetPtr=&/RocketRam_Fire_Close_06}}
 {id=WavePlayer_16, properties {soundWaveAssetPtr=&/RocketRam_Fire_Medium_01}}
 {id=WavePlayer_17, properties {soundWaveAssetPtr=&/RocketRam_Fire_Medium_02}}
 {id=WavePlayer_18, properties {soundWaveAssetPtr=&/RocketRam_Fire_Medium_03}}
 {id=WavePlayer_19, properties {soundWaveAssetPtr=&/RocketRam_Fire_Medium_04}}
 {id=WavePlayer_20, properties {soundWaveAssetPtr=&/RocketRam_Fire_Medium_05}}
 {id=WavePlayer_21, properties {soundWaveAssetPtr=&/RocketRam_Fire_Medium_06}}
 {id=WavePlayer_22, properties {soundWaveAssetPtr=&/RocketRam_Fire_Distant_01}}
 {id=WavePlayer_23, properties {soundWaveAssetPtr=&/RocketRam_Fire_Distant_02}}
 {id=WavePlayer_24, properties {soundWaveAssetPtr=&/RocketRam_Fire_Distant_03}}
 {id=WavePlayer_25, properties {soundWaveAssetPtr=&/RocketRam_Fire_Distant_04}}
 {id=WavePlayer_26, properties {soundWaveAssetPtr=&/RocketRam_Fire_Distant_05}}
 {id=WavePlayer_27, properties {soundWaveAssetPtr=&/RocketRam_Fire_Distant_06}}
]
properties {
 maxDistance=27500
 volumeMultiplier=1.4
 concurrencyOverrides {bLimitToOwner=true, maxCount=1}
 concurrencySet [&Weapon_Gunfire_Global_Conc, &Weapon_Gunfire_Local_Conc]
}
