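# Sound cue for the Rocket Ram's in-air rocket fire loop, played at reduced volume in the melee weapon sound class.
# A random node picks one of three equally weighted rocket wave variations and feeds it through a modulator at default settings.
# Spatialization uses the Rocket Ram in-air attenuation asset.
# A concurrency override allows one instance per owner.
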SoundCue /RocketRamGameplay/Items/RocketRam/Sounds/RocketRam_Fire_InAirLoop_Rocket_Cue
root=Modulator_0
nodes [
 {id=Modulator_0, children [Random_0]}
 {id=Random_0, children [WavePlayer_1,WavePlayer_2,WavePlayer_3], properties {weights [1,1,1]}}
 {id=WavePlayer_1, properties {soundWaveAssetPtr=&/RocketRam_Fire_InAirLoop_Rocket_01}}
 {id=WavePlayer_2, properties {soundWaveAssetPtr=&/RocketRam_Fire_InAirLoop_Rocket_02}}
 {id=WavePlayer_3, properties {soundWaveAssetPtr=&/RocketRam_Fire_InAirLoop_Rocket_03}}
]
properties {
 attenuationSettings=/RocketRamGameplay/Items/RocketRam/Sounds/Settings/RocketRam_InAir_Att
 bOverrideConcurrency=true
 maxDistance=8500
 soundClassObject=&Fort_WeaponMelee
 volumeMultiplier=0.3
 concurrencyOverrides {bLimitToOwner=true, maxCount=1}
}
