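# SoundCue for the Rocket Ram's downward wind loop while in the air, going by its name, with in-air attenuation and the melee weapon sound class.
# A mixer layers one randomly chosen looping wind wave with one randomly chosen looping flapping wave.
# The mix runs through a modulator and an enveloper that ramps volume up from silence, behind a fixed one-second start delay.
# A concurrency override allows one instance per owner.
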SoundCue /RocketRamGameplay/Items/RocketRam/Sounds/RocketRam_Fire_InAirLoop_Wind_Down_Cue
root=Delay_0
nodes [
 {id=Delay_0, children [Enveloper_0], properties {delayMax=1, delayMin=1}}
 {
  id=Enveloper_0
  children [Modulator_0]
  properties {pitchCurve {editorCurveData {keys [{time=0}]}}, volumeCurve {editorCurveData {keys [<at 0 0>,<at 1 1>]}}}
 }
 {id=Mixer_0, children [Random_0,Random_1], properties {inputVolume [1,1]}}
 {id=Modulator_0, children [Mixer_0]}
 {id=Random_0, children [WavePlayer_1,WavePlayer_2], properties {weights [1,1]}}
 {id=Random_1, children [WavePlayer_0,WavePlayer_4], properties {weights [1,1]}}
 {
  id=WavePlayer_0
  properties {bLooping=true, soundWaveAssetPtr=&/RocketRam_Fire_InAirLoop_Wind_Down_Flapping_01}
 }
 {
  id=WavePlayer_1
  properties {bLooping=true, soundWaveAssetPtr=&/RocketRam_Fire_InAirLoop_Wind_Down_01}
 }
 {
  id=WavePlayer_2
  properties {bLooping=true, soundWaveAssetPtr=&/RocketRam_Fire_InAirLoop_Wind_Down_02}
 }
 {
  id=WavePlayer_4
  properties {bLooping=true, soundWaveAssetPtr=&/RocketRam_Fire_InAirLoop_Wind_Down_Flapping_02}
 }
]
properties {
 attenuationSettings=/RocketRamGameplay/Items/RocketRam/Sounds/Settings/RocketRam_InAir_Att
 bOverrideConcurrency=true
 maxDistance=8500
 soundClassObject=&Fort_WeaponMelee
 volumeMultiplier=0.3
 concurrencyOverrides {bLimitToOwner=true, maxCount=1}
}
