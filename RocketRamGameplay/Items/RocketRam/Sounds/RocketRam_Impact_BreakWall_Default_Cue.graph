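# Sound cue for the Rocket Ram breaking through a wall on impact, played in the melee weapon sound class.
# A Random node picks one of five equally weighted impact waves and feeds a Modulator that slightly varies pitch and volume.
# The cue lowers overall pitch and volume, borrows the Rocket Ram fire attenuation, and overrides concurrency to cap simultaneous plays per owner.
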SoundCue /RocketRamGameplay/Items/RocketRam/Sounds/RocketRam_Impact_BreakWall_Default_Cue
root=Modulator_3
nodes [
 {id=Modulator_3, children [Random_4], properties {pitchMax=1.1, pitchMin=0.9, volumeMax=1.1, volumeMin=0.9}}
 {id=Random_4, children [WavePlayer_25, WavePlayer_26, WavePlayer_27, WavePlayer_28, WavePlayer_29], properties {weights [1,1,1,1,1]}}
 {id=WavePlayer_25, properties {soundWaveAssetPtr=&/RocketRam_Impact_BreakWall_Default_01}}
 {id=WavePlayer_26, properties {soundWaveAssetPtr=&/RocketRam_Impact_BreakWall_Default_02}}
 {id=WavePlayer_27, properties {soundWaveAssetPtr=&/RocketRam_Impact_BreakWall_Default_03}}
 {id=WavePlayer_28, properties {soundWaveAssetPtr=&/RocketRam_Impact_BreakWall_Default_04}}
 {id=WavePlayer_29, properties {soundWaveAssetPtr=&/RocketRam_Impact_BreakWall_Default_05}}
]
properties {
 attenuationSettings=/RocketRamGameplay/Items/RocketRam/Sounds/Settings/RocketRam_Fire_Att
 bOverrideConcurrency=true
 maxDistance=12000
 pitchMultiplier=0.8
 soundClassObject=&Fort_WeaponMelee
 volumeMultiplier=0.4
 concurrencyOverrides {bLimitToOwner=true, maxCount=2}
}
