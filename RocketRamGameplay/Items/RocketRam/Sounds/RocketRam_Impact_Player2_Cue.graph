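# Sound cue for the Rocket Ram's player impact sound, in the melee weapon sound class.
# A random node picks among six equally weighted impact waves, then a modulator slightly varies pitch and volume on each play.
# It plays at reduced volume, uses the Rocket Ram fire attenuation settings, and overrides concurrency to allow one instance per owner.
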SoundCue /RocketRamGameplay/Items/RocketRam/Sounds/RocketRam_Impact_Player2_Cue
root=Modulator_3
nodes [
 {id=Modulator_3, children [Random_1], properties {pitchMax=1.1, pitchMin=0.9, volumeMax=1.1, volumeMin=0.9}}
 {
  id=Random_1
  children [WavePlayer_6, WavePlayer_7, WavePlayer_8, WavePlayer_9, WavePlayer_10, WavePlayer_11]
  properties {weights [1,1,1,1,1,1]}
 }
 {id=WavePlayer_10, properties {soundWaveAssetPtr=&/RocketRam_Impact_Player_05}}
 {id=WavePlayer_11, properties {soundWaveAssetPtr=&/RocketRam_Impact_Player_06}}
 {id=WavePlayer_6, properties {soundWaveAssetPtr=&/RocketRam_Impact_Player_01}}
 {id=WavePlayer_7, properties {soundWaveAssetPtr=&/RocketRam_Impact_Player_02}}
 {id=WavePlayer_8, properties {soundWaveAssetPtr=&/RocketRam_Impact_Player_03}}
 {id=WavePlayer_9, properties {soundWaveAssetPtr=&/RocketRam_Impact_Player_04}}
]
properties {
 attenuationSettings=/RocketRamGameplay/Items/RocketRam/Sounds/Settings/RocketRam_Fire_Att
 bOverrideConcurrency=true
 maxDistance=12000
 soundClassObject=&Fort_WeaponMelee
 volumeMultiplier=0.5
 concurrencyOverrides {bLimitToOwner=true, maxCount=1}
}
