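# RocketRam_Pickup2_Cue is a sound cue for picking up the Rocket Ram item.
# A Random node picks one of two pickup waves with equal weight and passes it through a Modulator at the root.
# The cue uses the weapon pickup attenuation and loot-collect sound class at reduced volume.
# A concurrency override allows only one instance at a time and blocks new plays while it sounds.
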SoundCue /RocketRamGameplay/Items/RocketRam/Sounds/RocketRam_Pickup2_Cue
root=Modulator_1
nodes [
 {id=Modulator_1, children [Random_1]}
 {id=Random_1, children [WavePlayer_2,WavePlayer_4], properties {weights [1,1]}}
 {id=WavePlayer_2, properties {soundWaveAssetPtr=&/RocketRam_Pickup_01}}
 {id=WavePlayer_4, properties {soundWaveAssetPtr=&/RocketRam_Pickup_02}}
]
properties {
 attenuationSettings=&Fort_Weapon_Pickup_Att
 bOverrideConcurrency=true
 maxDistance=3500
 soundClassObject=&Fort_Loot_Collect
 volumeMultiplier=0.65
 concurrencyOverrides {maxCount=1, resolutionRule=PreventNew}
}
