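# Sound cue for the Rocket Ram item's rocket dive, routed to the melee weapon sound class with in-air attenuation.
# A mixer layers randomly picked whoosh, whistle and engine waves, with an extra modulator on the whistle branch.
# The mix passes through a delay node and a root modulator.
# A concurrency override limits playback to one instance per owner.
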SoundCue /RocketRamGameplay/Items/RocketRam/Sounds/RocketRam_Rocket_Dive_Cue
root=Modulator_0
nodes [
 {id=Delay_2, children [Mixer_1]}
 {id=Mixer_1, children [Random_3,Modulator_1,Random_2], properties {inputVolume [0.4,0.4,0.3]}}
 {id=Modulator_0, children [Delay_2]}
 {id=Modulator_1, children [Random_1]}
 {id=Random_1, children [WavePlayer_3,WavePlayer_4,WavePlayer_5,WavePlayer_6], properties {weights [1,1,1,1]}}
 {id=Random_2, children [WavePlayer_7,WavePlayer_8,WavePlayer_9], properties {weights [1,1,1]}}
 {id=Random_3, children [WavePlayer_10, WavePlayer_11, WavePlayer_12], properties {weights [1,1,1]}}
 {id=WavePlayer_10, properties {soundWaveAssetPtr=&/RocketRam_Rocket_Dive_Whoosh_01}}
 {id=WavePlayer_11, properties {soundWaveAssetPtr=&/RocketRam_Rocket_Dive_Whoosh_02}}
 {id=WavePlayer_12, properties {soundWaveAssetPtr=&/RocketRam_Rocket_Dive_Whoosh_03}}
 {id=WavePlayer_3, properties {soundWaveAssetPtr=&/RocketRam_Rocket_Dive_Whistle_01}}
 {id=WavePlayer_4, properties {soundWaveAssetPtr=&/RocketRam_Rocket_Dive_Whistle_02}}
 {id=WavePlayer_5, properties {soundWaveAssetPtr=&/RocketRam_Rocket_Dive_Whistle_03}}
 {id=WavePlayer_6, properties {soundWaveAssetPtr=&/RocketRam_Rocket_Dive_Whistle_04}}
 {id=WavePlayer_7, properties {soundWaveAssetPtr=&/RocketRam_Rocket_Dive_Engine_01}}
 {id=WavePlayer_8, properties {soundWaveAssetPtr=&/RocketRam_Rocket_Dive_Engine_03}}
 {id=WavePlayer_9, properties {soundWaveAssetPtr=&/RocketRam_Rocket_Dive_Engine_02}}
]
properties {
 attenuationSettings=/RocketRamGameplay/Items/RocketRam/Sounds/Settings/RocketRam_InAir_Att
 bOverrideConcurrency=true
 maxDistance=8500
 soundClassObject=&Fort_WeaponMelee
 volumeMultiplier=0.4
 concurrencyOverrides {bLimitToOwner=true, maxCount=1}
}
