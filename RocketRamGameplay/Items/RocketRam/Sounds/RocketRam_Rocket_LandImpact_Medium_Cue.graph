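# Sound cue for the medium land impact of the Rocket Ram rocket, played as a layered one-shot.
# A mixer combines a randomly picked impact wave with a randomly picked debris wave, which is pitched down and slightly delayed to trail the hit.
# It uses the Rocket Ram fire attenuation and melee weapon sound class, with a concurrency override allowing one instance per owner.
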SoundCue /RocketRamGameplay/Items/RocketRam/Sounds/RocketRam_Rocket_LandImpact_Medium_Cue
root=Mixer_0
nodes [
 {id=Delay_0, children [Modulator_0], properties {delayMax=0.2, delayMin=0.2}}
 {id=Mixer_0, children [Modulator_2,Delay_0], properties {inputVolume [1,1]}}
 {id=Modulator_0, children [Random_0], properties {pitchMax=0.9, pitchMin=0.8}}
 {id=Modulator_2, children [Random_3]}
 {id=Random_0, children [WavePlayer_1, WavePlayer_9, WavePlayer_10], properties {weights [1,1,1]}}
 {id=Random_3, children [WavePlayer_0,WavePlayer_3,WavePlayer_5,WavePlayer_6,WavePlayer_7,WavePlayer_8], properties {weights [1,1,1,1,1,1]}}
 {id=WavePlayer_0, properties {soundWaveAssetPtr=&/RocketRam_Rocket_LandImpact_Medium_01}}
 {id=WavePlayer_1, properties {soundWaveAssetPtr=&/RocketRam_Rocket_LandImpact_Debris_01}}
 {id=WavePlayer_10, properties {soundWaveAssetPtr=&/RocketRam_Rocket_LandImpact_Debris_05}}
 {id=WavePlayer_3, properties {soundWaveAssetPtr=&/RocketRam_Rocket_LandImpact_Medium_02}}
 {id=WavePlayer_5, properties {soundWaveAssetPtr=&/RocketRam_Rocket_LandImpact_Medium_03}}
 {id=WavePlayer_6, properties {soundWaveAssetPtr=&/RocketRam_Rocket_LandImpact_Medium_04}}
 {id=WavePlayer_7, properties {soundWaveAssetPtr=&/RocketRam_Rocket_LandImpact_Medium_05}}
 {id=WavePlayer_8, properties {soundWaveAssetPtr=&/RocketRam_Rocket_LandImpact_Medium_06}}
 {id=WavePlayer_9, properties {soundWaveAssetPtr=&/RocketRam_Rocket_LandImpact_Debris_04}}
]
properties {
 attenuationSettings=/RocketRamGameplay/Items/RocketRam/Sounds/Settings/RocketRam_Fire_Att
 bOverrideConcurrency=true
 maxDistance=12000
 soundClassObject=&Fort_WeaponMelee
 volumeMultiplier=0.5
 concurrencyOverrides {bLimitToOwner=true, maxCount=1}
}
