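# Sound cue for the RocketRam melee weapon's rocket raise, spatialized with the weapon's engine-loop attenuation and limited to one instance per owner.
# A mixer combines two staggered, randomized raise layers (A and B variants) with a later, pitched-up equip sound.
# As saved, both raise layers have zero mixer volume and the cue's overall volume multiplier is zero, so it plays silent.
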SoundCue /RocketRamGameplay/Items/RocketRam/Sounds/RocketRam_Rocket_Raise_Cue
root=Modulator_3
nodes [
 {id=Delay_0, children [Random_2], properties {delayMax=0.4, delayMin=0.4}}
 {id=Delay_1, children [WavePlayer_12], properties {delayMax=0.6, delayMin=0.6}}
 {id=Delay_3, children [Random_1], properties {delayMax=0.1, delayMin=0.1}}
 {id=Mixer_0, children [Delay_3,Delay_0,Modulator_0], properties {inputVolume [0,0,0.4]}}
 {id=Modulator_0, children [Delay_1], properties {pitchMax=1.4, pitchMin=1.4}}
 {id=Modulator_3, children [Mixer_0]}
 {id=Random_1, children [WavePlayer_3,WavePlayer_4,WavePlayer_5], properties {weights [1,1,1]}}
 {id=Random_2, children [WavePlayer_6,WavePlayer_7,WavePlayer_8], properties {weights [1,1,1]}}
 {id=WavePlayer_12, properties {soundWaveAssetPtr=&/RocketRam_Equip_01}}
 {id=WavePlayer_3, properties {soundWaveAssetPtr=&/RocketRam_Rocket_Raise_01_A}}
 {id=WavePlayer_4, properties {soundWaveAssetPtr=&/RocketRam_Rocket_Raise_02_A}}
 {id=WavePlayer_5, properties {soundWaveAssetPtr=&/RocketRam_Rocket_Raise_03_A}}
 {id=WavePlayer_6, properties {soundWaveAssetPtr=&/RocketRam_Rocket_Raise_01_B}}
 {id=WavePlayer_7, properties {soundWaveAssetPtr=&/RocketRam_Rocket_Raise_02_B}}
 {id=WavePlayer_8, properties {soundWaveAssetPtr=&/RocketRam_Rocket_Raise_03_B}}
]
properties {
 attenuationSettings=/RocketRamGameplay/Items/RocketRam/Sounds/Settings/RocketRam_Engine_Loop_Att
 bOverrideConcurrency=true
 maxDistance=1500
 soundClassObject=&Fort_WeaponMelee
 volumeMultiplier=0
 concurrencyOverrides {bLimitToOwner=true, maxCount=1}
}
