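# Sound cue for the sputter of the Rocket Ram item's rocket.
# A Random node picks between two equally weighted sputter waves and feeds a root Modulator left at its default settings.
# It plays at reduced volume in the melee weapon sound class, using the Rocket Ram in-air attenuation asset.
# A concurrency override limits it to one instance per owner.
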SoundCue /RocketRamGameplay/Items/RocketRam/Sounds/RocketRam_Rocket_Sputter_Cue
root=Modulator_0
nodes [
 {id=Modulator_0, children [Random_1]}
 {id=Random_1, children [WavePlayer_4,WavePlayer_5], properties {weights [1,1]}}
 {id=WavePlayer_4, properties {soundWaveAssetPtr=&/RocketRam_Rocket_Sputter_01}}
 {id=WavePlayer_5, properties {soundWaveAssetPtr=&/RocketRam_Rocket_Sputter_03}}
]
properties {
 attenuationSettings=/RocketRamGameplay/Items/RocketRam/Sounds/Settings/RocketRam_InAir_Att
 bOverrideConcurrency=true
 maxDistance=8500
 soundClassObject=&Fort_WeaponMelee
 volumeMultiplier=0.4
 concurrencyOverrides {bLimitToOwner=true, maxCount=1}
}
