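# Sound cue for the Rocket Ram item, played by name as a timeout "can't fire" feedback sound.
# A single wave player feeds the root directly, with no randomization or modulation, at a reduced volume in the Fort_WeaponMelee sound class.
# Its own concurrency override allows only one instance per owner.
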SoundCue /RocketRamGameplay/Items/RocketRam/Sounds/RocketRam_Timeout_CantFire_Cue
root=WavePlayer_7
nodes [{id=WavePlayer_7, properties {soundWaveAssetPtr=&/RocketRam_Timeout_CantFire_02}}]
properties {
 bOverrideConcurrency=true
 maxDistance=2097152
 soundClassObject=&Fort_WeaponMelee
 volumeMultiplier=0.35
 concurrencyOverrides {bLimitToOwner=true, maxCount=1}
}
