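# Sound cue for the Rocket Ram item's timeout stop sound, in the melee weapon sound class.
# A random node picks one of three equally weighted stop variations, played slightly pitched up at reduced volume.
# It borrows the Rocket Ram engine loop attenuation settings.
# A concurrency override allows only one instance at a time per owner.
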SoundCue /RocketRamGameplay/Items/RocketRam/Sounds/RocketRam_Timeout_Stop_Cue
root=Random_0
nodes [
 {id=Random_0, children [WavePlayer_0,WavePlayer_2,WavePlayer_3], properties {weights [1,1,1]}}
 {id=WavePlayer_0, properties {soundWaveAssetPtr=&/RocketRam_Timeout_Stop_01}}
 {id=WavePlayer_2, properties {soundWaveAssetPtr=&/RocketRam_Timeout_Stop_02}}
 {id=WavePlayer_3, properties {soundWaveAssetPtr=&/RocketRam_Timeout_Stop_03}}
]
properties {
 attenuationSettings=/RocketRamGameplay/Items/RocketRam/Sounds/Settings/RocketRam_Engine_Loop_Att
 bOverrideConcurrency=true
 maxDistance=1500
 pitchMultiplier=1.1
 soundClassObject=&Fort_WeaponMelee
 volumeMultiplier=0.3
 concurrencyOverrides {bLimitToOwner=true, maxCount=1}
}
