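# Sound cue named for the Rocket Ram item's timeout tick, with no sound nodes or root, so it has no audio to play as saved.
# It still carries playback settings: very low volume, the melee weapon sound class and attenuation borrowed from the Rocket Ram engine loop.
# A concurrency override caps it at one instance per owner.
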SoundCue /RocketRamGameplay/Items/RocketRam/Sounds/RocketRam_Timeout_Tick_Cue
root=null
nodes []
properties {
 attenuationSettings=/RocketRamGameplay/Items/RocketRam/Sounds/Settings/RocketRam_Engine_Loop_Att
 bOverrideConcurrency=true
 maxDistance=1500
 soundClassObject=&Fort_WeaponMelee
 volumeMultiplier=0.05
 concurrencyOverrides {bLimitToOwner=true, maxCount=1}
}
